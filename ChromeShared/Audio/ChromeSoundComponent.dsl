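; Actor component that applies or removes Chrome audio on players, vehicles, wildlife, chests and gameplay buildings in Fortnite.
; Setup casts the actor to pick a handler that swaps sounds, source effect chains, sound libraries and mix modifiers.
; Looping timers drive blob-player and vehicle loops, feeding speed parameters.
; The blob loop also sets team affiliation, shows sound indicators, and interpolates a listener-height parameter for hostile pawns.
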(event EventBeginPlay)

(event EventTick
  (params (DeltaSeconds :type Float%20%28single-precision%29)))

(event Chrome|EventSetupChromeAudio
  (params (Actor :type Actor%20Object%20Reference) (bAdd :type Boolean) (ChromeType :type ELacklusterChromeType%20Enum :default "Chrome"))
  (Data|SetChromeAudioActor Actor bAdd (Utilities|Select :Index ChromeType :Chrome "NewEnumerator1" :ChromePhaseable "NewEnumerator12" :ChromePhaseableMasked "NewEnumerator13" :ChromeGeneric "NewEnumerator14" :ChromePetrified "NewEnumerator15")))

(fn AudioVisualizerOverrrideIcon
  (params (Activate :type Boolean))
  (bind _icon_override (Variables|Visualizer|GetIcon%20Override))
  (bind _playerpawn (Variables|Player|GetPlayerPawn))
  (bind _soundindicatorcomponent (Variables|/Script/FortniteGame.FortPawn|GetSoundIndicatorComponent _playerpawn))
  (if Activate
    (SoundIndicator|AddSoundIndicatorIconOverride _soundindicatorcomponent _icon_override)
    (else
      (SoundIndicator|RemoveSoundIndicatorIconOverride _soundindicatorcomponent _icon_override))))

(fn BlobListenerPerspective
  (bind _playerpawn (Variables|Player|GetPlayerPawn))
  (bind _returnvalue (Transformation|GetActorLocation _playerpawn))
  (bind _z (.z _returnvalue))
  (bind _playerpawn_2 (Variables|Player|GetPlayerPawn))
  ;[1 1 1] Check for Blob pawn vs local controller. If blob is local don't do any more checks. If pawn is hostile then check for listener.
  (if (Pawn|IsLocallyViewed _playerpawn_2)
    _
    (elif (Utilities|Enum|Equal(Enum) (Team|GetTeamAffiliationforActortoLocalPlayer _playerpawn_2 true true) "Hostile")
      ;[1 1 1] Get closest listen pawner to the blob pawn. Check above or below and then evualate audio param.
      (bind _listenerposition (Audio|GetClosestListenerLocation (Transformation|GetActorLocation (Variables|Player|GetPlayerPawn)) 8e+03 true))
      (Utilities|FlowControl|Sequence
        (:then_0
          (if (Math|Float|NearlyEqual(Float) (.z _listenerposition) _z 256.0)
            (Variables|Player|SetListenerPositionParam 0.0)
            (elif (> (.z _listenerposition) _z)
              (Variables|Player|SetListenerPositionParam -1.0)
              (elif (< (.z _listenerposition) _z)
                (Variables|Player|SetListenerPositionParam 1.0)))))
        (:then_1
          (bind _output_get (Variables|Player|SetCurrentListenerValue (Math|Interpolation|FInterpTo (Variables|Player|GetCurrentListenerValue) (Variables|Player|GetListenerPositionParam) (Utilities|Time|GetWorldDeltaSeconds) 1e+01)))
          (Audio|Parameter|SetFloatParameter (Variables|Default|GetChromeAudioLoop) "BlobVerticality" _output_get :!class /Script/AudioExtensions.AudioParameterControllerInterface))))))

(fn BlobStateMovement
  (bind _chromesound (Variables|Data|GetChromeSound))
  (Variables|Player|GetPlayerPawn
    (:then
      (Utilities|IsValid _chromesound
        (:Is%20Valid
          (CallFunction|SpawnChromeAudioAttach _chromesound (Variables|Default|GetAttach%20to%20Component) (Variables|Data|GetComponentLocation))
          (bind _chromeaudioloop (Variables|Default|GetChromeAudioLoop)
            (:then
              (bind _output_get (Variables|Player|SetPlayerBlobVelocitySpeed (Math|Interpolation|FInterpTo (Variables|Player|GetPlayerBlobVelocitySpeed) (Math|Float|SelectFloat (Math|Float|MapRangeClamped (Math|Vector|VectorLength (Transformation|GetVelocity (Variables|Player|GetPlayerPawn))) 0.0 9e+02 0.0 1.0) 0.0 (not (or (Utilities|Enum|Equal(Enum) (Variables|/Script/Engine.CharacterMovementComponent|GetMovementMode (Variables|/Script/Engine.Character|GetCharacterMovement (Variables|Player|GetPlayerPawn))) "MOVE_Flying") (Utilities|Enum|Equal(Enum) (Variables|/Script/Engine.CharacterMovementComponent|GetMovementMode (Variables|/Script/Engine.Character|GetCharacterMovement (Variables|Player|GetPlayerPawn))) "MOVE_Falling")))) (Utilities|Time|GetWorldDeltaSeconds) 13.0)))
              (Audio|Parameter|SetFloatParameter _chromeaudioloop "BlobMovement" _output_get :!class /Script/AudioExtensions.AudioParameterControllerInterface)
              (Player|SetAudioCompTeamAffiliation _chromeaudioloop)
              (Player|BlobListenerPerspective)
              (if (> (Variables|Player|GetPlayerBlobVelocitySpeed) 0)
                (SoundIndicator|ShowSoundIndicator (Variables|/Script/FortniteGame.FortPawn|GetSoundIndicatorComponent (Variables|Player|GetPlayerPawn)) (Transformation|GetActorLocation (Variables|Player|GetPlayerPawn)) "/ChromeShared/Audio/Player/Chrome_Blob_Footstep_Jog_Cue.Chrome_Blob_Footstep_Jog_Cue" (ScalableFloat|GetValueatLevel (Variables|Default|GetBlobAudioVisualizerDistanceMultiplier)) 0)))
            (:else)))))
    (:else)))

(fn BuildingGameplayActorDeath
  (params (Damage :type Float%20%28single-precision%29) (DamageTags :type Gameplay%20Tag%20Container%20Structure%20%28by%20ref%29) (Momentum :type Vector) (HitInfo :type Hit%20Result%20Structure%20%28by%20ref%29) (InstigatedBy :type Fort%20Pawn%20Object%20Reference) (DamageCauser :type Actor%20Object%20Reference) (EffectContext :type Gameplay%20Effect%20Context%20Handle%20Structure))
  (Audio|PlaySoundatLocation "/ChromeShared/Audio/Building/Chrome_Build_Wall_Destroy_Cue.Chrome_Build_Wall_Destroy_Cue" (Collision|GetActorBounds (Variables|Building|GetBuilding%20Gameplay%20Actor) false true))
  (return))

(fn ChromeBuilding
  (params (TargetActor :type Actor%20Object%20Reference)))

(fn ChromeBuildingGameplayActor
  (bind _building_gameplay_actor (Variables|Building|GetBuilding%20Gameplay%20Actor)
    (:then
      (Variables|/Script/FortniteGame.BuildingActor|SetPlayHitSound "/ChromeShared/Audio/Building/Chrome_Object_Wobble_Cue.Chrome_Object_Wobble_Cue" _building_gameplay_actor)
      (EventDispatchers|BindEventtoOnDeathPlayEffectsDelegate _building_gameplay_actor (EventDispatchers|CreateEvent :!event Building%20Gameplay%20Actor%20Death))
      (Audio|PlaySoundatLocation "/ChromeShared/Audio/Building/Chrome_Appliocation_Wall_Cue.Chrome_Appliocation_Wall_Cue" (Collision|GetActorBounds _building_gameplay_actor false true)))
    (:else)))

(fn ChromeChest
  (bind _tieredchestparent (Variables|Chest|GetTieredChestParent))
  (bind _chest_ambient_sound (Variables|/Game/Building/ActorBlueprints/Containers/Tiered_Chest_Athena_Parent.Tiered_Chest_Athena_Parent_C|GetChest_Ambient_Sound _tieredchestparent))
  (bind _regularchest (Variables|Chest|GetRegularChest))
  (bind _tieredchestparent_2 (Variables|Chest|GetTieredChestParent))
  (bind _chest_ambient_sound_2 (Variables|/Game/Building/ActorBlueprints/Containers/Tiered_Chest_Athena_Parent.Tiered_Chest_Athena_Parent_C|GetChest_Ambient_Sound _tieredchestparent_2))
  (bind _regularchest_2 (Variables|Chest|GetRegularChest))
  (bind _chest_ambient_sound_3 (Variables|/Game/Building/ActorBlueprints/Containers/Tiered_Chest_Athena.Tiered_Chest_Athena_C|GetChest_Ambient_Sound _regularchest_2))
  (bind _chest_ambient_sound_4 (Variables|/Game/Building/ActorBlueprints/Containers/Tiered_Chest_Athena.Tiered_Chest_Athena_C|GetChest_Ambient_Sound _regularchest))
  (Utilities|FlowControl|Sequence
    (:then_0
      (bind _regularchest_3 (Variables|Chest|GetRegularChest)
        (:then
          (if (Variables|/Game/Building/ActorBlueprints/Containers/Tiered_Chest_Athena.Tiered_Chest_Athena_C|GetbOpened _regularchest_3)
            (if (Audio|Components|Audio|IsPlaying _chest_ambient_sound_3)
              (Audio|Components|Audio|FadeOut _chest_ambient_sound_3 0.5 0.0 "Linear" :!class /Script/Engine.AudioComponent))
            (else
              (Variables|/Game/Building/ActorBlueprints/Containers/Tiered_Chest_Athena.Tiered_Chest_Athena_C|SetChestOpenSourceEffectOverride (Variables|Chest|GetChromeChestSourceEffectChain) _regularchest)
              (Variables|/Script/Engine.AudioComponent|SetSourceEffectChain (Variables|Chest|GetChromeChestSourceEffectChain) _chest_ambient_sound_4)
              (Chests|PlayChestAudioLoop _chest_ambient_sound_4)
              (Chests|SetChestOverrides (Variables|Chest|GetRegularChest)))))
        (:else)))
    (:then_1
      (bind _tieredchestparent_3 (Variables|Chest|GetTieredChestParent)
        (:then
          (if (Variables|/Game/Building/ActorBlueprints/Containers/Tiered_Chest_Athena_Parent.Tiered_Chest_Athena_Parent_C|GetbOpened _tieredchestparent_3)
            (if (Audio|Components|Audio|IsPlaying _chest_ambient_sound)
              (Audio|Components|Audio|FadeOut _chest_ambient_sound 0.5 0.0 "Linear" :!class /Script/Engine.AudioComponent))
            (else
              (Variables|/Game/Building/ActorBlueprints/Containers/Tiered_Chest_Athena_Parent.Tiered_Chest_Athena_Parent_C|SetChestOpenSourceEffectOverride (Variables|Chest|GetChromeChestSourceEffectChain) _tieredchestparent_2)
              (Variables|/Script/Engine.AudioComponent|SetSourceEffectChain (Variables|Chest|GetChromeChestSourceEffectChain) _chest_ambient_sound_2)
              (Chests|PlayChestAudioLoop _chest_ambient_sound_2)
              (Chests|SetChestOverrides (Variables|Chest|GetTieredChestParent)))))
        (:else)))))

(fn ChromePlayer
  (Variables|Player|GetPlayerPawn
    (:then
      (Player|UpdateChromeBaseState)
      (Player|UpdateChromeBlobState)
      (Player|UpdatePlayerEmoteEffects))
    (:else)))

(fn ChromeVehicle
  (if (not (Variables|Building|GetVehicleChromeApplied))
    (Variables|Building|SetVehicleChromeApplied true)
    ;[1 1 1] Apply Chrome audio to vehicle
    (bind _vehicle (Variables|Vehicle|GetVehicle)
      (:then
        (Audio|SpawnSoundAttached "/ChromeShared/Audio/Vehicles/Chrome_Veh_Transition_In_Cue.Chrome_Veh_Transition_In_Cue" (Variables|/Script/Engine.Actor|GetRootComponent _vehicle) "None" (Variables|Vehicle|GetVehicleSoundLocation))
        (Data|SetVariableData (Variables|Data|GetAssetType) "/ChromeShared/Audio/Vehicles/Chrome_Veh_Ambient_Loop_Cue.Chrome_Veh_Ambient_Loop_Cue" (Variables|/Script/Engine.Actor|GetRootComponent _vehicle) (Variables|Vehicle|GetVehicleSoundLocation))
        ;[1 1 1] Setup bindings for vehicle when player interacts.
        (Vehicle|BindEventtoOnPawnEnteredVehicleatSeat (Variables|Vehicle|GetVehicle) (EventDispatchers|CreateEvent :!event Pawn%20Enter%20Vehicle))
        (Vehicle|BindEventtoOnPawnExitedVehiclefromSeat (Variables|Vehicle|GetVehicle) (EventDispatchers|CreateEvent :!event Pawn%20Exit%20Vehicle))
        (bind _vehicle_2 (Variables|Vehicle|GetVehicle)
          (:then
            (Game|BindEventtoOnEndPlay _vehicle_2 (EventDispatchers|CreateEvent :!event End%20VehicleTimer)))
          (:else)))
      (:else))))

(fn ChromeWildlife
  (if (Variables|Default|GetApplyChrome)
    (Variables|/Game/Athena/AI/Pawns/NPC/Parents/NPC_Pawn_Wildlife_Parent.NPC_Pawn_Wildlife_Parent_C|SetVocalSourceEffectChainOverride (Variables|Wildlife|GetWildlifeSourceEffectChain) (Variables|Wildlife|GetWildlife))
    (else
      (Variables|/Game/Athena/AI/Pawns/NPC/Parents/NPC_Pawn_Wildlife_Parent.NPC_Pawn_Wildlife_Parent_C|SetVocalSourceEffectChainOverride 0 (Variables|Wildlife|GetWildlife)))))

(fn EndVehicleTimer
  (params (Actor :type Actor%20Object%20Reference) (EndPlayReason :type EEndPlayReason%20Enum :default "Destroyed"))
  (Utilities|Time|ClearandInvalidateTimerbyHandle (Variables|Data|GetHandleAudioLoop))
  (bind _chromeaudioloop (Variables|Default|GetChromeAudioLoop)
    (:then
      (Audio|Components|Audio|FadeOut _chromeaudioloop 1.0 0.0 "Linear" :!class /Script/Engine.AudioComponent))
    (:else)))

(fn OnPawnDied
  (params (Actor :type Actor%20Object%20Reference) (EndPlayReason :type EEndPlayReason%20Enum :default "Destroyed"))
  (Utilities|FlowControl|Sequence
    (:then_0
      ;[1 1 1] Clear Blob state on death
      (if (Variables|Player|GetIsChromeBlobPlayer)
        (Player|RemoveChromeBlobAudio)))
    (:then_1
      ;[1 1 1] Clear Mix on Death
      (Player|SetPlayerMixState "/ChromeShared/Audio/Player/Chrome_Player_MixMod.Chrome_Player_MixMod") ; Remove Base Chrome Mix
    )))

(fn PawnEnterVehicle
  (params (Vehicle :type Fort%20Vehicle%20Interface%20Interface%20%28by%20ref%29) (PlayerPawn :type Fort%20Pawn%20Object%20Reference) (SeatIndex :type Integer))
  ;[1 1 1] When driver enters seat start audio loop update param.
  (if (Vehicle|HasDriver Vehicle)
    (bind _returnvalue (Utilities|Time|SetTimerbyEvent (EventDispatchers|CreateEvent :!event Vehicle%20Movement%20Audio) 0.2 true))
    (Variables|Data|SetHandleAudioLoop _returnvalue)))

(fn PawnExitVehicle
  (params (Vehicle :type Fort%20Vehicle%20Interface%20Interface%20%28by%20ref%29) (PlayerPawn :type Fort%20Pawn%20Object%20Reference) (SeatIndex :type Integer))
  ;[1 1 1] When driver leaves seat stop audio loop and timers.
  (if (not (Vehicle|HasDriver Vehicle))
    (Utilities|FlowControl|Sequence
      (:then_0
        (Utilities|Time|ClearandInvalidateTimerbyHandle (Variables|Data|GetHandleAudioLoop)))
      (:then_1
        (bind _chromeaudioloop (Variables|Default|GetChromeAudioLoop)
          (:then
            (Audio|Components|Audio|FadeOut _chromeaudioloop 1.5 0.0 "Linear" :!class /Script/Engine.AudioComponent))
          (:else))))))

(fn PlayChestAudioLoop
  (params (AudioComponent :type Audio%20Component%20Object%20Reference))
  (Utilities|IsValid AudioComponent
    (:Is%20Valid
      (if (Audio|Components|Audio|IsPlaying AudioComponent)
        (Audio|Components|Audio|Play AudioComponent 0.0)))))

(fn RemoveChromeBlobAudio
  (bind _returnvalue (WorldSubsystems|GetSoundLibrarySubsystem))
  (bind _bloblib (Variables|Player|GetBlobLib))
  (Utilities|FlowControl|Sequence
    (:then_0
      ;[1 1 1] Set player footsteps bank to default
      (Variables|/Script/FortniteGame.FortPawn|SetFootstepBank "/Game/Athena/Sounds/Character/Footsteps/DefaultFootstepBankAthena.DefaultFootstepBankAthena" (Variables|Player|GetPlayerPawn))
      (Utilities|IsValid _returnvalue
        (:Is%20Valid
          (SoundLibrary|RemoveLibrary _returnvalue (Variables|Player|GetPlayerPawn) _bloblib :!class /Script/SoundLibrary.SoundLibrarySubsystem)
          (Development|PrintString (Utilities|String|Append "Remove: " (Utilities|GetDisplayName _bloblib))))))
    (:then_1
      ;[1 1 1] Exit blob state audio. Stop looping sound
      (bind _returnvalue_2 (Audio|SpawnSoundAttached (Variables|Player|GetExitBlobStateSound) (Variables|/Script/Engine.Actor|GetRootComponent (Variables|Player|GetPlayerPawn)) "None" "0, 0,50.000000"))
      (Player|SetAudioCompTeamAffiliation _returnvalue_2)
      (Utilities|Time|ClearandInvalidateTimerbyHandle (Variables|Player|GetHandleBlobStateAudio))
      (bind _chromeaudioloop (Variables|Default|GetChromeAudioLoop)
        (:then
          (Audio|Components|Audio|FadeOut _chromeaudioloop 1.0 0.0 "Linear" :!class /Script/Engine.AudioComponent))
        (:else)))
    (:then_2
      ;[1 1 1] Reset blob mix state
      (Variables|Player|GetPlayerChromeBlobMix
        (:then
          (Player|SetPlayerMixState "/ChromeShared/Audio/Player/Chrome_Player_Blob_MixMod.Chrome_Player_Blob_MixMod") ; Remove Blob Mix
          (Variables|Player|SetPlayerChromeBlobMix 0)
          (Visualizer|AudioVisualizerOverrrideIcon false))
        (:else)))))

(fn SetActorType
  (params (ActorType :type Actor%20Object%20Reference))
  (Variables|Data|SetActorType ActorType)
  (bind _asfort_player_pawn (Utilities|Casting|CastToFortPlayerPawn (Variables|Data|GetActorType))
    (:then
      (Variables|Player|SetPlayerPawn _asfort_player_pawn)
      (Variables|Data|SetAssetType "NewEnumerator1")
      (Player|ChromePlayer))
    (:CastFailed
      (bind _asfort_athena_vehicle (Utilities|Casting|CastToFortAthenaVehicle (Variables|Data|GetActorType))
        (:then
          (Variables|Vehicle|SetVehicle _asfort_athena_vehicle)
          (Variables|Data|SetAssetType "NewEnumerator0")
          (Vehicle|ChromeVehicle))
        (:CastFailed
          (bind _asnpc_pawn_wildlife_parent (Utilities|Casting|CastToNPC_Pawn_Wildlife_Parent (Variables|Data|GetActorType))
            (:then
              (Variables|Wildlife|SetWildlife _asnpc_pawn_wildlife_parent)
              (Variables|Data|SetAssetType "NewEnumerator4")
              (Wildlife|ChromeWildlife))
            (:CastFailed
              (bind _astiered_chest_athena (Utilities|Casting|CastToTiered_Chest_Athena (Variables|Data|GetActorType))
                (:then
                  (Variables|Chest|SetRegularChest _astiered_chest_athena))
                (:CastFailed
                  (bind _astiered_chest_athena_parent (Utilities|Casting|CastToTiered_Chest_Athena_Parent (Variables|Data|GetActorType))
                    (:then
                      (Variables|Chest|SetTieredChestParent _astiered_chest_athena_parent) ; For rare and special chests
                      (Variables|Data|SetAssetType "NewEnumerator3")
                      (Chests|ChromeChest))
                    (:CastFailed
                      (bind _asbuilding_gameplay_actor (Utilities|Casting|CastToBuildingGameplayActor (Variables|Data|GetActorType))
                        (:then
                          (Variables|Building|SetBuilding%20Gameplay%20Actor _asbuilding_gameplay_actor)
                          (Variables|Data|SetAssetType "NewEnumerator2")
                          (Building|ChromeBuildingGameplayActor))
                        (:CastFailed
                          (Variables|Data|SetAssetType "NewEnumerator6")
                          (Components|DestroyComponent)))))
                  (return (Variables|Data|GetAssetType)))))))))))

(fn SetAudioCompTeamAffiliation
  (params (AudioComponent :type Audio%20Component%20Object%20Reference))
  ;[1 1 1] Set player affiliation
  ; 0 = Local Player
  ; 1 = Non-Hostile
  ; 2 = Hostiles
  (Audio|Parameter|SetIntegerParameter AudioComponent "PlayerAffliation" (Math|Integer|SelectInt 0 (Utilities|Select :Index (Team|GetTeamAffiliationforActortoLocalPlayer (Variables|Player|GetPlayerPawn) true true) :Friendly 1 :Neutral 1 :Hostile 2) (Pawn|IsLocallyViewed (Variables|Player|GetPlayerPawn))) :!class /Script/Engine.AudioComponent))

(fn SetChestOverrides
  (params (ChestActor :type Actor%20Object%20Reference))
  (bind _asbuilding_static_mesh (Utilities|Casting|CastToBuildingSMActor ChestActor))
  (Variables|/Script/FortniteGame.BuildingActor|SetPlayHitSound "/ChromeShared/Audio/Building/Chrome_Object_Wobble_Cue.Chrome_Object_Wobble_Cue" _asbuilding_static_mesh)
  (bind _asbuilding_prop (Utilities|Casting|CastToBuildingProp _asbuilding_static_mesh))
  (Variables|/Script/FortniteGame.BuildingSMActor|SetDeathSound "/ChromeShared/Audio/Building/Chrome_Build_Prop_Destroy_Cue.Chrome_Build_Prop_Destroy_Cue" _asbuilding_prop)
  (Audio|PlaySoundatLocation "/ChromeShared/Audio/Building/Chrome_Appliocation_Prop_Cue.Chrome_Appliocation_Prop_Cue" (Collision|GetActorBounds _asbuilding_prop false true)))

(fn SetChromeAudioActor
  (params (Actor :type Actor%20Object%20Reference) (Add :type Boolean) (ChromeType :type B_Lackluster_ChromeTypes%20Enum :default "NewEnumerator1"))
  (Variables|Default|SetApplyChrome Add)
  (Variables|Default|SetChromeType ChromeType)
  (Default|SetActorType Actor))

(fn SetPlayChromeLibrary
  (params (Add%20Lib :type Boolean))
  (bind _returnvalue (WorldSubsystems|GetSoundLibrarySubsystem))
  (bind _returnvalue_2 (WorldSubsystems|GetSoundLibrarySubsystem))
  (bind _playerpawn (Variables|Player|GetPlayerPawn))
  (bind _baselib (Variables|Player|GetBaseLib))
  (if Add%20Lib
    (Utilities|IsValid _returnvalue_2
      (:Is%20Valid
        (SoundLibrary|RemoveLibrary _returnvalue_2 _playerpawn _baselib :!class /Script/SoundLibrary.SoundLibrarySubsystem)
        (SoundLibrary|AddLibrary _returnvalue_2 _playerpawn _baselib :!class /Script/SoundLibrary.SoundLibrarySubsystem)))
    (else
      (Utilities|IsValid _returnvalue
        (:Is%20Valid
          (SoundLibrary|RemoveLibrary _returnvalue (Variables|Player|GetPlayerPawn) (Variables|Player|GetBaseLib) :!class /Script/SoundLibrary.SoundLibrarySubsystem))))))

(fn SetPlayerMixState
  (params (InSoundMixModifier :type Sound%20Mix%20Object%20Reference) (IsActivated :type Boolean))
  (bind _insoundmixmodifier (Variables|GetInSoundMixModifier))
  (if (Pawn|IsLocallyViewed (Variables|Player|GetPlayerPawn))
    (if IsActivated
      (Audio|PushSoundMixModifier _insoundmixmodifier)
      (else
        (Audio|PopSoundMixModifier _insoundmixmodifier)))
    (return _insoundmixmodifier)))

(fn SetVariableData
  (params (AssetType :type ChromeAudioAssetType%20Enum :default "NewEnumerator6") (Chrome%20Sound :type Sound%20Base%20Object%20Reference) (Attach%20to%20Component :type Scene%20Component%20Object%20Reference) (ComponentLocation :type Vector))
  (Variables|Data|SetAssetType AssetType)
  (Variables|Data|SetChromeSound Chrome%20Sound)
  (Variables|Default|SetAttach%20to%20Component Attach%20to%20Component)
  (Variables|Data|SetComponentLocation ComponentLocation))

(fn SpawnChromeAudioAttach
  (params (ChromeSound :type Sound%20Base%20Object%20Reference) (AttachToComponent :type Scene%20Component%20Object%20Reference) (Location :type Vector))
  (Variables|Default|GetChromeAudioLoop
    (:then)
    (:else
      (Utilities|IsValid ChromeSound
        (:Is%20Valid
          (Utilities|IsValid AttachToComponent
            (:Is%20Valid
              (bind _returnvalue (Audio|SpawnSoundAttached ChromeSound AttachToComponent "None" Location))
              (bind _output_get (Variables|Default|SetChromeAudioLoop _returnvalue))
              (Audio|Components|Audio|FadeIn _output_get 0.25 1.0 0.0 "Linear" :!class /Script/Engine.AudioComponent)
              (Player|SetAudioCompTeamAffiliation _output_get))))))))

(fn StartBlobStateAudio
  (bind _returnvalue (WorldSubsystems|GetSoundLibrarySubsystem))
  (bind _playerpawn (Variables|Player|GetPlayerPawn))
  (bind _bloblib (Variables|Player|GetBlobLib))
  (bind _rootcomponent (Variables|/Script/Engine.Actor|GetRootComponent _playerpawn))
  (bind _playerpawn_2 (Variables|Player|GetPlayerPawn))
  (Variables|Player|GetPlayerPawn
    (:then
      (Utilities|FlowControl|Sequence
        (:then_0
          ;[1 1 1] Bind on pawn actions, set footstep bank for jump and land
          (Game|BindEventtoOnEndPlay _playerpawn_2 (EventDispatchers|CreateEvent :!event On%20Pawn%20Died))
          (Variables|/Script/FortniteGame.FortPawn|SetFootstepBank "/ChromeShared/Audio/Player/ChromeBlobFootstepBank.ChromeBlobFootstepBank" _playerpawn_2)
          (Utilities|IsValid _returnvalue
            (:Is%20Valid
              (SoundLibrary|AddLibrary _returnvalue (Variables|Player|GetPlayerPawn) _bloblib :!class /Script/SoundLibrary.SoundLibrarySubsystem)
              (Development|PrintString (Utilities|String|Append "Add: " (Utilities|GetDisplayName _bloblib))))))
        (:then_1
          ;[1 1 1] Set Looping audio for movement
          (bind _returnvalue_2 (Audio|SpawnSoundAttached (Variables|Player|GetEnterBlobStateSound) _rootcomponent "None" "0, 0,50.000000"))
          (Player|SetAudioCompTeamAffiliation _returnvalue_2)
          (Data|SetVariableData (Variables|Data|GetAssetType) "/ChromeShared/Audio/Player/Chrome_Blob_Movement_Loop_Cue.Chrome_Blob_Movement_Loop_Cue" _rootcomponent "0, 0,50.000000")
          (bind _returnvalue_3 (Utilities|Time|SetTimerbyEvent (EventDispatchers|CreateEvent :!event Blob%20State%20Movement) 0.1 true))
          (Variables|Player|SetHandleBlobStateAudio _returnvalue_3))
        (:then_2
          ;[1 1 1] Set mix state
          (Variables|Player|GetPlayerChromeBlobMix
            (:then)
            (:else
              (bind _outmix (Player|SetPlayerMixState "/ChromeShared/Audio/Player/Chrome_Player_Blob_MixMod.Chrome_Player_Blob_MixMod" true)) ; Add Blob Mix
              (Variables|Player|SetPlayerChromeBlobMix _outmix)
              (Visualizer|AudioVisualizerOverrrideIcon true))))))
    (:else)))

(fn StartBuildingHeal
  (params (BuildingSMActor :type Building%20Static%20Mesh%20Object%20Reference))
  (Audio|PlaySoundatLocation "/ChromeShared/Audio/Building/Chrome_Build_Wall_Heal_Cue.Chrome_Build_Wall_Heal_Cue" (Collision|GetActorBounds BuildingSMActor false true)))

(fn StopBuildingAudio
  (params (BuildingActor :type Actor%20Object%20Reference)))

(fn UpdateChromeBaseState
  (if (Variables|Player|GetIsChromeBasePlayer)
    (if (Variables|Default|GetApplyChrome)
      (Variables|Player|GetPlayerChromeMix
        (:then
          _)
        (:else
          (bind _outmix (Player|SetPlayerMixState "/ChromeShared/Audio/Player/Chrome_Player_MixMod.Chrome_Player_MixMod" true)) ; Add Base Chrome Mix 
          (Variables|Player|SetPlayerChromeMix _outmix)))
      (bind _returnvalue (Audio|SpawnSoundAttached "/ChromeShared/Audio/Player/Chrome_Player_Status_Activate_Cue.Chrome_Player_Status_Activate_Cue" (Variables|/Script/Engine.Actor|GetRootComponent (Variables|Player|GetPlayerPawn))))
      (bind _output_get (Variables|Default|SetPlayerChromeActivateAudio _returnvalue))
      (Player|SetAudioCompTeamAffiliation _output_get)
      (Player|SetPlayChromeLibrary true)
      (else
        (Variables|Player|GetPlayerChromeMix
          (:then
            (Player|SetPlayerMixState "/ChromeShared/Audio/Player/Chrome_Player_MixMod.Chrome_Player_MixMod") ; Remove Base Chrome Mix
            (Variables|Player|SetPlayerChromeMix 0))
          (:else
            _))
        (bind _playerchromeactivateaudio (Variables|Default|GetPlayerChromeActivateAudio)
          (:then
            (Audio|Components|Audio|FadeOut _playerchromeactivateaudio 0.5 0.0 "Linear" :!class /Script/Engine.AudioComponent))
          (:else
            _))
        (Player|SetPlayChromeLibrary false)))))

(fn UpdateChromeBlobState
  (if (Variables|Player|GetIsChromeBlobPlayer)
    (if (Variables|Default|GetApplyChrome)
      (Player|StartBlobStateAudio)
      (else
        (Player|RemoveChromeBlobAudio)))))

(fn UpdatePlayerEmoteEffects
  (bind _playerpawn (Variables|Player|GetPlayerPawn))
  (if (Variables|Player|GetIsChromeBasePlayer)
    (if (Variables|Default|GetApplyChrome)
      (Variables|/Script/FortniteGame.FortPawn|SetInGameEmoteSoundEffectSoundPresetChain (Variables|Player|GetEmoteSourceEffectChain) _playerpawn)
      (Development|PrintString "Emote DSP Applied" true true "(R=1.000000,G=0.056408,B=0.000000,A=1.000000)" 1e+01)
      (else
        (Variables|/Script/FortniteGame.FortPawn|SetInGameEmoteSoundEffectSoundPresetChain 0 _playerpawn)
        (Development|PrintString "Emote DSP Removed" true true "(R=1.000000,G=0.056408,B=0.000000,A=1.000000)" 1e+01)))))

(fn UpdateVehicleChromeParameter
  ;[1 1 1] Get vehicle and update chrome audio loop param based on speed of vehicle.
  (bind _vehicle (Variables|Vehicle|GetVehicle)
    (:then
      (bind _chromeaudioloop (Variables|Default|GetChromeAudioLoop)
        (:then
          (Audio|Parameter|SetFloatParameter _chromeaudioloop "VehicleSpeed" (Math|Float|MapRangeClamped (Vehicle|GetTotalSpeedKmh _vehicle) 0.0 (Class|FortVehicleInterface|GetHUDTopSpeed _vehicle) 0.0 1.0) :!class /Script/AudioExtensions.AudioParameterControllerInterface))
        (:else)))
    (:else)))

(fn VehicleMovementAudio
  (bind _chromesound (Variables|Data|GetChromeSound))
  ;[1 1 1] Gets valid movement and plays audio accordingly. Is on event timer to update based on looping value.
  (bind _vehicle (Variables|Vehicle|GetVehicle)
    (:then
      (Vehicle|UpdateVehicleChromeParameter)
      (if (Vehicle|HasDriver _vehicle)
        (Utilities|IsValid _chromesound
          (:Is%20Valid
            (CallFunction|SpawnChromeAudioAttach _chromesound (Variables|Default|GetAttach%20to%20Component) (Variables|Data|GetComponentLocation))))
        (else
          (bind _chromeaudioloop (Variables|Default|GetChromeAudioLoop)
            (:then
              (Audio|Components|Audio|FadeOut _chromeaudioloop 1.5 0.0 "Linear" :!class /Script/Engine.AudioComponent))
            (:else)))))
    (:else)))
