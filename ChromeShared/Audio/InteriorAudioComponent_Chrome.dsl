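; Interior audio component that drives ambient audio for nearby chromed building pieces.
; After each scan update, it collects unique locations of building-grid actors with the Chrome status tag or ApplyChromeComponent, then starts a random-interval timer and adds a chrome ambient tag.
; Each timer tick plays a random one-shot at a random location, recursively pruning picks beyond listener range.
; When no locations remain, it clears the timer and removes the tag.
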(event EventBP_PostUpdate
  (params (InResults :type Interior%20Audio%20Scan%20Results%20Structure%20%28by%20ref%29))
  (bind (_viewtarget _viewtargettransform _buildinggrid) (Utilities|Struct|BreakInteriorAudioScanResults InResults))
  (bind _values (Utilities|Map|Values (Utilities|Struct|BreakInteriorAudioScanResultsBuildingGrid _buildinggrid)))
  (for _array_element _values
    (if (Utilities|Array|ContainsItem (Variables|/Script/Engine.Actor|GetTags (Utilities|Struct|BreakInteriorAudioScanDirectionResults _array_element)) "Gameplay.Status.Chrome")
      (Utilities|Array|AddUnique (Variables|Default|GetOneshotLocations) (Transformation|GetActorLocation (Utilities|Struct|BreakInteriorAudioScanDirectionResults _array_element)))
      (else
        (Utilities|IsValid (Actor|GetComponentbyClass (Utilities|Struct|BreakInteriorAudioScanDirectionResults _array_element) "/Script/ChromeSharedRuntime.ApplyChromeComponent")
          (:Is%20Valid
            (Utilities|Array|AddUnique (Variables|Default|GetOneshotLocations) (Transformation|GetActorLocation (Utilities|Struct|BreakInteriorAudioScanDirectionResults _array_element))))))))
  (if (and (> (Utilities|Array|Length (Variables|Default|GetOneshotLocations)) 0) (not (Utilities|Time|IsTimerActivebyHandle (Variables|Default|GetAmbientOneshotTimerHandle))))
    (bind _returnvalue (Utilities|Time|SetTimerbyEvent (EventDispatchers|CreateEvent :!event TryPlayOneshot) (CallFunction|GetRandomTime)))
    (Variables|Default|SetAmbientOneshotTimerHandle _returnvalue)
    (AmbientAudio|AddGameplayTag (WorldSubsystems|GetAmbientAudioSubsystem) (Variables|Default|GetChromeAmbientTag))))

(fn GetRandomTime
  (bind (_randomtimerange_x _randomtimerange_y) (Math|Vector2D|BreakVector2D (Variables|GetRandomTimeRange_X)))
  (return (Math|Random|RandomFloatinRange _randomtimerange_x _randomtimerange_y)))

(fn TryPlayOneshot
  (bind _oneshotlocations (Variables|Default|GetOneshotLocations))
  (bind (_outitem _outindex) (Utilities|Array|RandomArrayItem _oneshotlocations))
  (if (> (Utilities|Array|Length (Variables|Default|GetOneshotLocations)) 0)
    (LocalVariables|SetRandomLoc _outitem)
    (LocalVariables|SetRandomLocIndex _outindex)
    (LocalVariables|SetRandomSound (Utilities|Array|RandomArrayItem (Variables|Default|GetOneshotSounds)))
    (bind (_listenerposition _returnvalue) (Audio|GetClosestListenerLocation (LocalVariables|GetRandomLoc) (Variables|/Script/Engine.SoundBase|GetMaxDistance (LocalVariables|GetRandomSound)) true)) ; Attempt to flush any out-of-range sounds
    (if _returnvalue
      (Audio|PlaySoundatLocation (LocalVariables|GetRandomSound) (LocalVariables|GetRandomLoc))
      (bind _returnvalue_2 (Utilities|Time|SetTimerbyEvent (EventDispatchers|CreateEvent :!event TryPlayOneshot) (CallFunction|GetRandomTime)))
      (Variables|Default|SetAmbientOneshotTimerHandle _returnvalue_2)
      (else
        (Utilities|Array|RemoveIndex (Variables|Default|GetOneshotLocations) (LocalVariables|GetRandomLocIndex))
        (CallFunction|TryPlayOneshot)))
    (else
      (Utilities|Time|ClearandInvalidateTimerbyHandle (Variables|Default|GetAmbientOneshotTimerHandle))
      (AmbientAudio|RemoveGameplayTag (WorldSubsystems|GetAmbientAudioSubsystem) (Variables|Default|GetChromeAmbientTag)))))
