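; Function library for the Lackluster chrome effect, swapping actor materials and audio to chrome versions.
; ApplyChromeMaterial classifies actors by tags or class casts, harvests textures through per-type parameter sets and feeds them into override materials.
; ApplyChrometoActor runs this only for cosmetics, caches material data in ChromeSubsystem and sets chrome sounds.
; Several functions mirror nativized versions, selected by UseNativeFunctions; type-specific hooks handle containers, vehicles and shadow proxies.
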(fn ActorSupportChromeSoundComponent
  (params (Actor :type Actor%20Object%20Reference))
  (bind _returnvalue (WorldSubsystems|GetChromeSubsystem_C 0))
  (bind _buildingchromeapplied (Variables|/ChromeShared/ChromeSubsystem.ChromeSubsystem_C|GetBuildingChromeApplied _returnvalue))
  (bind _asbuilding_actor (Utilities|Casting|CastToBuildingActor Actor)
    (:then
      (if (not (.ReturnValue (Utilities|Map|Find _buildingchromeapplied _asbuilding_actor)))
        (Utilities|Map|Add _buildingchromeapplied _asbuilding_actor "true")
        (if (not (CallFunction|AreWeOverridingAudio _asbuilding_actor _asbuilding_actor))
          (CallFunction|StopBuildingAudio _asbuilding_actor _asbuilding_actor)
          ;[1 1 1] Set building actors audio without creating any additional sound components.
          (bind _asbuilding_static_mesh (Utilities|Casting|CastToBuildingSMActor _asbuilding_actor)
            (:then
              (Variables|/Script/FortniteGame.BuildingActor|SetPlayHitSound "/ChromeShared/Audio/Building/Chrome_Object_Wobble_Cue.Chrome_Object_Wobble_Cue" _asbuilding_static_mesh)
              (bind _asbuilding_prop (Utilities|Casting|CastToBuildingProp _asbuilding_static_mesh)
                (:then
                  (Variables|/Script/FortniteGame.BuildingSMActor|SetDeathSound "/ChromeShared/Audio/Building/Chrome_Build_Prop_Destroy_Cue.Chrome_Build_Prop_Destroy_Cue" _asbuilding_prop)
                  (Audio|PlaySoundatLocation 0 "/ChromeShared/Audio/Building/Chrome_Appliocation_Prop_Cue.Chrome_Appliocation_Prop_Cue" (Collision|GetActorBounds _asbuilding_prop false true)))
                (:CastFailed
                  (Variables|/Script/FortniteGame.BuildingSMActor|SetDeathSound "/ChromeShared/Audio/Building/Chrome_Build_Wall_Destroy_Cue.Chrome_Build_Wall_Destroy_Cue" _asbuilding_static_mesh)
                  (Audio|PlaySoundatLocation 0 "/ChromeShared/Audio/Building/Chrome_Appliocation_Wall_Cue.Chrome_Appliocation_Wall_Cue" (Collision|GetActorBounds _asbuilding_actor false true))))
              (return false))
            (:CastFailed
              _))
          (else
            _))
        (return true)))
    (:CastFailed
      _)))

(fn ApplyChrometoActor
  (params (Actor :type Actor%20Object%20Reference%20%28by%20ref%29) (ActorType :type B_Lackluster_ActorTypes%20Enum%20%28by%20ref%29 :default "NewEnumerator8") (ChromeType :type B_Lackluster_ChromeTypes%20Enum%20%28by%20ref%29 :default "NewEnumerator1") (ChromeMaterialStructs :type Array%20of%20B%20Lackluster%20Chrome%20Material%20Struct%20Structures) (OriginalMaterialStructs :type Array%20of%20B%20Lackluster%20Original%20Material%20Struct%20Structures) (NativeActorType :type ELacklusterActorType%20Enum%20%28by%20ref%29 :default "Undefined") (NativeChromeType :type ELacklusterChromeType%20Enum%20%28by%20ref%29 :default "Chrome") (NativeChromeMaterialStructs :type Array%20of%20Lackluster%20Material%20Data%20Structures) (NativeOriginalMaterialStructs :type Array%20of%20Lackluster%20Original%20Material%20Data%20Structures) (ContainerMaterialScalarParamMap :type Map%20of%20Names%20to%20Float%20%28double-precision%29s) (ContainerMaterialVectorParamMap :type Map%20of%20Names%20to%20Linear%20Color%20Structures))
  (bind (_applychromeclasspath _chromeaudioclasspath _materialstoignore _setparameternames _classestoactortypes _chromematerialoverrides _actortotextures _tagquerytochrometypeoverrides _tagtoactortypes _genericchromematerial _noanimchestmaterialtags _chromeanimchestmaterial _chromenoanimchestmaterial _buildingchromeapplied _buildingsmactorhitsound _buildingsmactorpropdeathsound _buildingsmactornonpropdeathsound _buildingsmactorpropchromedsound _buildingsmactornonpropchromedsound _tieredchestclass _tieredchestparentclass _chromesoundcomponents _shadowproxymeshcomponentclass _materialsandtypes) (Utilities|Struct|BreakLacklusterData (Variables|/Script/ChromeSharedRuntime.FortChromeSubsystem|GetLacklusterData (WorldSubsystems|GetChromeSubsystem_C 0))))
  (bind _returnvalue (WorldSubsystems|GetChromeSubsystem_C 0))
  (bind _chrometype (Variables|GetChromeType))
  (Networking|CanExecuteCosmeticEvents
    (:True
      (Utilities|IsValid Actor
        (:Is%20Valid
          (if (Lackluster|UseNativeFunctions)
            (Lackluster|NativeApplyChromeMaterial Actor NativeChromeMaterialStructs NativeOriginalMaterialStructs NativeChromeType NativeActorType)
            (else
              (CallFunction|ApplyChromeMaterial Actor ChromeMaterialStructs OriginalMaterialStructs ChromeType ActorType)))
          (if (CallFunction|NeedstoSaveMaterialsandTypes Actor ActorType Actor)
            (Utilities|Map|Add (Variables|/ChromeShared/ChromeSubsystem.ChromeSubsystem_C|GetActorMaterialsAndTypes _returnvalue) Actor (Utilities|Struct|MakeBLacklusterMaterialsandTypes (Variables|GetActorType) (Variables|GetChromeType) (Variables|GetChromeMaterialStructs) (Variables|GetOriginalMaterialStructs) (Variables|GetNativeActorType) (Variables|GetNativeChromeType) (Variables|GetNativeChromeMaterialStructs) (Variables|GetNativeOriginalMaterialStructs) (Variables|GetContainerMaterialScalarParamMap) (Variables|GetContainerMaterialVectorParamMap)))
            (Utilities|Map|Add _materialsandtypes Actor (Utilities|Struct|MakeLacklusterMaterialsandTypes (Variables|GetNativeActorType) (Variables|GetNativeChromeType) (Variables|GetNativeChromeMaterialStructs) (Variables|GetNativeOriginalMaterialStructs) (Variables|GetContainerMaterialScalarParamMap) (Variables|GetContainerMaterialVectorParamMap))))
          (if (Lackluster|UseNativeFunctions)
            (if (Lackluster|NativeDoesActorSupportChromeSoundComponent Actor)
              (Lackluster|NativeSetChromeAudio Actor true (Utilities|Select :Index _chrometype :NewEnumerator1 "Chrome" :NewEnumerator12 "ChromePhaseable" :NewEnumerator13 "ChromePhaseableMasked" :NewEnumerator14 "ChromeGeneric" :NewEnumerator15 "ChromePetrified")))
            (elif (CallFunction|ActorSupportChromeSoundComponent Actor Actor)
              (CallFunction|SetChromeAudio Actor true _chrometype Actor))))))))

(fn ApplyChromeMaterial
  (params (Actor :type Actor%20Object%20Reference) (ChromeMaterialStructIn :type Array%20of%20B%20Lackluster%20Chrome%20Material%20Struct%20Structures) (OriginalMaterialStructIn :type Array%20of%20B%20Lackluster%20Original%20Material%20Struct%20Structures) (ChromeType :type B_Lackluster_ChromeTypes%20Enum :default "NewEnumerator1") (OverrideActorType :type B_Lackluster_ActorTypes%20Enum :default "NewEnumerator8"))
  ;[1 0 0] ATTENTION: This function has been nativized. Currently both the native and the blueprint versions are in use. If you need to make changes here, talk to Brian Zaugg so we can make sure the corresponding changes are made to the native version.
  (LocalVariables|SetChromeTypeLocal ChromeType)
  (Utilities|IsValid Actor
    (:Is%20Valid
      (bind _output_get (LocalVariables|SetCurrentActor Actor))
      (if (Actor|ActorHasTag _output_get "Gameplay.Status.NeverChrome")
        _
        (elif (Utilities|Array|IsEmpty ChromeMaterialStructIn)
          (if (Utilities|Enum|Equal(Enum) OverrideActorType "NewEnumerator8")
            (if (Utilities|Enum|Equal(Enum) (Variables|GetChromeType) "NewEnumerator14")
              (LocalVariables|SetActorType "NewEnumerator17")
              (else
                (bind (_found _actortype) (CallFunction|CasttoSpecificClasses _output_get (LocalVariables|GetChromeTypeLocal)))
                (LocalVariables|SetActorType _actortype)
                (LocalVariables|SetChromeTypeLocal (LocalVariables|GetChromeTypeLocal))))
            (else
              (LocalVariables|SetActorType (Variables|GetOverrideActorType))))
          (if (Chrome|IsActorChromable (LocalVariables|GetCurrentActor))
            (bind (_chromematerialstruct _originalmaterialstruct) (CallFunction|FindTexturesfromComponents (LocalVariables|GetActorType) (LocalVariables|GetCurrentActor)))
            (CallFunction|SetChromeMaterials _chromematerialstruct (LocalVariables|GetChromeTypeLocal))
            (Variables|SetBy-RefVar (Variables|GetChromeMaterialStructIn) _chromematerialstruct)
            (Variables|SetBy-RefVar (Variables|GetOriginalMaterialStructIn) _originalmaterialstruct)
            (bind _asbuilding_static_mesh (Utilities|Casting|CastToBuildingSMActor (LocalVariables|GetCurrentActor)))
            (Variables|/Script/FortniteGame.BuildingSMActor|SetbNeedsMIDsForCreative true _asbuilding_static_mesh)
            (return))
          (else
            ;;[1 1 1] Already populated
            (CallFunction|SetChromeMaterials (Variables|GetChromeMaterialStructIn) (LocalVariables|GetChromeTypeLocal))
            (return)))))))

(fn AreWeOverridingAudio
  (params (Actor :type Actor%20Object%20Reference))
  ;[1 1 1] Set this so we can update the chest audio properly.
  (Utilities|Casting|CastToTiered_Chest_Athena Actor
    (:then
      _)
    (:CastFailed
      (Utilities|Casting|CastToTiered_Chest_Athena_Parent Actor
        (:then
          _)
        (:CastFailed
          (return false)))))
  (return true))

(fn CastToParentClasses
  (params (ActorObjectRef :type Actor%20Object%20Reference))
  ;[1 0 0] ATTENTION: This function has been nativized. Currently both the native and the blueprint versions are in use. If you need to make changes here, talk to Brian Zaugg so we can make sure the corresponding changes are made to the native version.
  (Utilities|Casting|CastToBuildingRoof ActorObjectRef
    (:then
      (LocalVariables|SetActorType "NewEnumerator0"))
    (:CastFailed
      (Utilities|Casting|CastToBuildingWall (Variables|GetActorObjectRef)
        (:then
          (LocalVariables|SetActorType "NewEnumerator1"))
        (:CastFailed
          (Utilities|Casting|CastToBuildingCorner (Variables|GetActorObjectRef)
            (:then
              (LocalVariables|SetActorType "NewEnumerator3"))
            (:CastFailed
              (Utilities|Casting|CastToBuildingPropWall (Variables|GetActorObjectRef)
                (:then
                  (LocalVariables|SetActorType "NewEnumerator2"))
                (:CastFailed
                  (Utilities|Casting|CastToBuildingFloor (Variables|GetActorObjectRef)
                    (:then
                      (LocalVariables|SetActorType "NewEnumerator4"))
                    (:CastFailed
                      (Utilities|Casting|CastToBuildingProp (Variables|GetActorObjectRef)
                        (:then
                          (LocalVariables|SetActorType "NewEnumerator5"))
                        (:CastFailed
                          (Utilities|Casting|CastToBuildingDeco (Variables|GetActorObjectRef)
                            (:then
                              (LocalVariables|SetActorType "NewEnumerator6"))
                            (:CastFailed
                              (Utilities|Casting|CastToBuildingPillar (Variables|GetActorObjectRef)
                                (:then
                                  (LocalVariables|SetActorType "NewEnumerator7"))
                                (:CastFailed
                                  (Utilities|Casting|CastToStaticMeshActor (Variables|GetActorObjectRef)
                                    (:then
                                      (LocalVariables|SetActorType "NewEnumerator10"))
                                    (:CastFailed
                                      (Utilities|Casting|CastToFortAIPawn (Variables|GetActorObjectRef)
                                        (:then
                                          (LocalVariables|SetActorType "NewEnumerator18"))
                                        (:CastFailed
                                          (Utilities|Casting|CastToFortDagwoodVehicle (Variables|GetActorObjectRef)
                                            (:then
                                              (LocalVariables|SetActorType "NewEnumerator21"))
                                            (:CastFailed
                                              (Utilities|Casting|CastToFortAthenaFerretVehicle (Variables|GetActorObjectRef)
                                                (:then
                                                  (LocalVariables|SetActorType "NewEnumerator22"))
                                                (:CastFailed
                                                  (Utilities|Casting|CastToFortHoagieVehicle (Variables|GetActorObjectRef)
                                                    (:then
                                                      (LocalVariables|SetActorType "NewEnumerator22"))
                                                    (:CastFailed
                                                      (Utilities|Casting|CastToFortMeatballVehicle (Variables|GetActorObjectRef)
                                                        (:then
                                                          (LocalVariables|SetActorType "NewEnumerator22"))
                                                        (:CastFailed
                                                          (Utilities|Casting|CastToFortAthenaSKPushCannon (Variables|GetActorObjectRef)
                                                            (:then
                                                              (LocalVariables|SetActorType "NewEnumerator22"))
                                                            (:CastFailed
                                                              (Utilities|Casting|CastToBuildingActor (Variables|GetActorObjectRef)
                                                                (:then
                                                                  (LocalVariables|SetActorType "NewEnumerator27"))
                                                                (:CastFailed
                                                                  ;;[1 0 0.002752] Failed, fallback to generic chrome
                                                                  (bind _output_get (LocalVariables|SetActorType "NewEnumerator17"))
                                                                  (return false _output_get)))))))))))))))))))))))))))))))))
  (return true (LocalVariables|GetActorType)))

(fn CastToSpecificClasses
  (params (ActorObjectRef :type Actor%20Object%20Reference) (ChromeType :type B_Lackluster_ChromeTypes%20Enum :default "NewEnumerator1"))
  (bind _tagtoactortypes (LocalVariables|GetTagToActorTypes))
  ;[1 0 0] ATTENTION: This function has been nativized. Currently both the native and the blueprint versions are in use. If you need to make changes here, talk to Brian Zaugg so we can make sure the corresponding changes are made to the native version.
  (Utilities|IsValid ActorObjectRef
    (:Is%20Valid
      (LocalVariables|SetChromeTypeLocal ChromeType)
      (Utilities|FlowControl|Sequence
        (:then_0
          ;;[1 1 1] Override logic for ensuring the asset is petrified.
          (bind _asgameplay_tag_asset_interface (Utilities|Casting|CastToGameplayTagAssetInterface (Variables|GetActorObjectRef)))
          (if (and (GameplayTags|HasAnyMatchingGameplayTags _asgameplay_tag_asset_interface (LocalVariables|GetPetrifyTags)) (not (GameplayTags|HasAnyMatchingGameplayTags _asgameplay_tag_asset_interface (LocalVariables|GetDoNotPetrifyTags))))
            (LocalVariables|SetChromeTypeLocal "NewEnumerator15")))
        (:then_1
          ;;[1 1 1] Try and find actor tags
          (for _array_element (Variables|/Script/Engine.Actor|GetTags (Variables|GetActorObjectRef))
            (if (.ReturnValue (Utilities|Map|Find _tagtoactortypes _array_element))
              (return true (Utilities|Map|Find _tagtoactortypes _array_element) (LocalVariables|GetChromeTypeLocal))))
          (bind (_found _foundactortype) (CallFunction|CasttoParentClasses (Variables|GetActorObjectRef)))
          (if _found
            (return true _foundactortype (LocalVariables|GetChromeTypeLocal))
            (else
              _)))))
    (:Is%20Not%20Valid
      _))
  (return false "NewEnumerator8" (LocalVariables|GetChromeTypeLocal)))

(fn CleanupOwnerTypeSpecificFunctionality
  (params (Actor :type Actor%20Object%20Reference) (ActorType :type B_Lackluster_ActorTypes%20Enum :default "NewEnumerator8"))
  (Networking|CanExecuteCosmeticEvents
    (:True
      (switch Utilities|FlowControl|Switch|SwitchonB_Lackluster_ActorTypes ActorType
        (:NewEnumerator13
          _)
        (:NewEnumerator14
          _))
      (CallFunction|CleanupOwnerTypeSpecificFunctionalityContainer 0 Actor))))

(fn CleanupOwnerTypeSpecificFunctionality_Container
  (params (Actor :type Actor%20Object%20Reference))
  (Networking|CanExecuteCosmeticEvents
    (:True
      (Utilities|Casting|CastToBuildingContainer Actor))))

(fn DisableCurieFire
  (params (Actor :type Actor%20Object%20Reference))
  (Utilities|IsValid Actor
    (:Is%20Valid
      (Curie|DisableElementAttachmentonActor Actor "(TagName=\"Curie.Element.Fire\")")
      (Curie|DisableElementInteractiononActor Actor "(TagName=\"Curie.Element.Fire\")"))))

(fn DisableProxyShadowWPO
  (params (Actor :type Actor%20Object%20Reference%20%28by%20ref%29))
  (for _array_element (Actor|GetComponentsByClass Actor "/Game/Blueprints/Components/ShadowProxy/ShadowProxyMeshComponent.ShadowProxyMeshComponent_C")
    (LocalVariables|SetCurrentShadowProxyMesh _array_element)
    (for _index (range (Rendering|Material|GetNumMaterials _array_element))
      (bind _returnvalue (Rendering|Material|CreateDynamicMaterialInstance (LocalVariables|GetCurrentShadowProxyMesh) _index (Rendering|Material|GetMaterial (LocalVariables|GetCurrentShadowProxyMesh) _index) "None" :!class /Script/Engine.PrimitiveComponent))
      (Rendering|Material|SetScalarParameterValue _returnvalue "EnableWPO" 0.0 :!class /Script/Engine.MaterialInstanceDynamic))))

(fn ExcludeFromChroming
  (params (SceneComponent :type Scene%20Component%20Object%20Reference))
  ;[1 0 0] ATTENTION: This function has been nativized. Currently both the native and the blueprint versions are in use. If you need to make changes here, talk to Brian Zaugg so we can make sure the corresponding changes are made to the native version.
  (return (Math|Boolean|ORBoolean (Variables|/Script/Engine.ActorComponent|GetbIsEditorOnly SceneComponent) (not (Variables|/Script/Engine.SceneComponent|GetbVisible SceneComponent)) (Utilities|String|Contains (Utilities|GetDisplayName SceneComponent) "Widget"))))

(fn ExecuteOwnerTypeFunctionality_PostChrome_Container
  (params (Actor :type Actor%20Object%20Reference) (ContainerTags :type Gameplay%20Tag%20Container%20Structure) (ContainerMaterialScalarParamMap :type Map%20of%20Names%20to%20Float%20%28double-precision%29s) (ContainerMaterialVectorParamMap :type Map%20of%20Names%20to%20Linear%20Color%20Structures))
  (bind _containermaterialscalarparammap (Variables|GetContainerMaterialScalarParamMap))
  (bind _containermaterialvectorparammap (Variables|GetContainerMaterialVectorParamMap))
  (Networking|CanExecuteCosmeticEvents
    (:True
      (bind _asbuilding_container (Utilities|Casting|CastToBuildingContainer Actor)
        (:then
          (Class|ChromeSubsystem|BindChromeEvents (WorldSubsystems|GetChromeSubsystem_C 0) _asbuilding_container)
          (if (GameplayTags|HasAnyMatchingGameplayTags _asbuilding_container ContainerTags)
            (bind _staticmeshcomponent (Variables|/Script/FortniteGame.BuildingSMActor|GetStaticMeshComponent _asbuilding_container)
              (:then
                (bind _asmaterial_instance_dynamic (Utilities|Casting|CastToMaterialInstanceDynamic (Rendering|Material|GetMaterial _staticmeshcomponent 0)))
                (Utilities|FlowControl|Sequence
                  (:then_0
                    (bind _keys (Utilities|Map|Keys _containermaterialscalarparammap))
                    (for _array_element _keys
                      (if (.ReturnValue (Utilities|Map|Find _containermaterialscalarparammap _array_element))
                        (Rendering|Material|SetScalarParameterValue _asmaterial_instance_dynamic _array_element (Utilities|Map|Find _containermaterialscalarparammap _array_element) :!class /Script/Engine.MaterialInstanceDynamic))))
                  (:then_1
                    (bind _keys_2 (Utilities|Map|Keys _containermaterialvectorparammap))
                    (for _array_element_2 _keys_2
                      (if (.ReturnValue (Utilities|Map|Find _containermaterialvectorparammap _array_element_2))
                        (Rendering|Material|SetVectorParameterValue _asmaterial_instance_dynamic _array_element_2 (Utilities|Map|Find _containermaterialvectorparammap _array_element_2) :!class /Script/Engine.MaterialInstanceDynamic))))))
              (:else))))
        (:CastFailed
          ;[1 0 0] Warning: Owner is not of BuildingContainer class.
          (Fortnite|FortShippingLog Actor (Utilities|String|Append "ExecuteOwnerTypeFunctionalityPostChromeContainer: Owner actor " (Utilities|GetDisplayName Actor) " is NOT of type BuildingContainer.") true))))))

(fn ExecuteOwnerTypeFunctionality_PostChrome_Vehicle
  (params (Actor :type Actor%20Object%20Reference))
  ;[0.1094 0.02231 0.08453] ExecuteOwnerTypeFunctionality_PostChrome_Vehicle: Setup delegate on the vehicle so that its state changes can be handled.
  (Networking|CanExecuteCosmeticEvents
    (:True
      (Class|ChromeSubsystem|BindChromeEvents (WorldSubsystems|GetChromeSubsystem_C 0) Actor))))

(fn ExecuteOwnerTypeFunctionality_PreChrome_Container
  (params (ContainerTags :type Gameplay%20Tag%20Container%20Structure) (ContainerMaterialScalarParamMap :type Map%20of%20Names%20to%20Float%20%28double-precision%29s) (ContainerMaterialVectorParamMap :type Map%20of%20Names%20to%20Linear%20Color%20Structures) (Actor :type Actor%20Object%20Reference%20%28by%20ref%29))
  (bind _containermaterialscalarparammap (Variables|GetContainerMaterialScalarParamMap))
  (bind _containermaterialvectorparammap (Variables|GetContainerMaterialVectorParamMap))
  (Networking|CanExecuteCosmeticEvents
    (:True
      (bind _asbuilding_container (Utilities|Casting|CastToBuildingContainer Actor))
      (if (GameplayTags|HasAnyMatchingGameplayTags _asbuilding_container ContainerTags)
        (bind _staticmeshcomponent (Variables|/Script/FortniteGame.BuildingSMActor|GetStaticMeshComponent _asbuilding_container)
          (:then
            (bind _asmaterial_instance_dynamic (Utilities|Casting|CastToMaterialInstanceDynamic (Rendering|Material|GetMaterial _staticmeshcomponent 0))
              (:then
                (LocalVariables|SetMID _asmaterial_instance_dynamic))
              (:CastFailed
                ;[1 1 1] If we can't find a MID, make a temporary one so we can harvest scalar/vector parameters
                (bind _asmaterial_instance_constant (Utilities|Casting|CastToMaterialInstanceConstant (Rendering|Material|GetMaterial _staticmeshcomponent 0)))
                (bind _returnvalue (Rendering|Material|CreateDynamicMaterialInstance (Variables|GetActor) _asmaterial_instance_constant "None" "Transient" :!class /Script/Engine.KismetMaterialLibrary))
                (LocalVariables|SetMID _returnvalue)))
            (bind _mid (LocalVariables|GetMID)
              (:then
                (Utilities|FlowControl|Sequence
                  (:then_0
                    (bind _keys (Utilities|Map|Keys _containermaterialscalarparammap))
                    (for _array_element _keys
                      (bind _returnvalue_2 (Rendering|Material|GetScalarParameterValue _mid _array_element :!class /Script/Engine.MaterialInstanceDynamic))
                      (Utilities|Map|Add _containermaterialscalarparammap _array_element _returnvalue_2)))
                  (:then_1
                    (bind _keys_2 (Utilities|Map|Keys _containermaterialvectorparammap))
                    (for _array_element_2 _keys_2
                      (bind _returnvalue_3 (Rendering|Material|GetVectorParameterValue _mid _array_element_2 :!class /Script/Engine.MaterialInstanceDynamic))
                      (Utilities|Map|Add _containermaterialvectorparammap _array_element_2 _returnvalue_3)))))
              (:else)))
          (:else))))))

(fn ExecuteOwnerTypeSpecificFunctionality_PostChrome
  (params (Actor :type Actor%20Object%20Reference) (ActorType :type B_Lackluster_ActorTypes%20Enum :default "NewEnumerator8") (ContainerTags :type Gameplay%20Tag%20Container%20Structure) (ContainerMaterialScalarParamMap :type Map%20of%20Names%20to%20Float%20%28double-precision%29s) (ContainerMaterialVectorParamMap :type Map%20of%20Names%20to%20Linear%20Color%20Structures))
  (Networking|CanExecuteCosmeticEvents
    (:True
      (switch Utilities|FlowControl|Switch|SwitchonB_Lackluster_ActorTypes ActorType
        (:NewEnumerator9
          (CallFunction|DisableProxyShadowWPO Actor))
        (:NewEnumerator13
          _)
        (:NewEnumerator14
          _)
        (:NewEnumerator21
          (CallFunction|ExecuteOwnerTypeFunctionalityPostChromeVehicle (Variables|GetActor) (Variables|GetActor)))
        (:NewEnumerator22
          _))
      (CallFunction|ExecuteOwnerTypeFunctionalityPostChromeContainer Actor ContainerTags ContainerMaterialScalarParamMap ContainerMaterialVectorParamMap Actor))))

(fn ExecuteOwnerTypeSpecificFunctionality_PreChrome
  (params (Actor :type Actor%20Object%20Reference) (ActorType :type B_Lackluster_ActorTypes%20Enum :default "NewEnumerator8") (ContainerTags :type Gameplay%20Tag%20Container%20Structure) (ContainerMaterialScalarParamMap :type Map%20of%20Names%20to%20Float%20%28double-precision%29s) (ContainerMaterialVectorParamMap :type Map%20of%20Names%20to%20Linear%20Color%20Structures))
  (Networking|CanExecuteCosmeticEvents
    (:True
      (switch Utilities|FlowControl|Switch|SwitchonB_Lackluster_ActorTypes ActorType
        (:NewEnumerator13
          _)
        (:NewEnumerator14
          _))
      (CallFunction|ExecuteOwnerTypeFunctionalityPreChromeContainer ContainerTags ContainerMaterialScalarParamMap ContainerMaterialVectorParamMap Actor Actor))))

(fn FindGetSetParameters
  (params (IsMaskedMaterial :type Boolean) (ActorType :type B_Lackluster_ActorTypes%20Enum :default "NewEnumerator8") (MeshComponent :type Mesh%20Component%20Object%20Reference))
  ;[1 0 0] ATTENTION: This function has been nativized. Currently both the native and the blueprint versions are in use. If you need to make changes here, talk to Brian Zaugg so we can make sure the corresponding changes are made to the native version.
  (return (Utilities|Select :Index ActorType :NewEnumerator17 (LocalVariables|GetGetSet_Generic) :NewEnumerator0 (LocalVariables|GetGetSet_Base) :NewEnumerator1 (LocalVariables|GetGetSet_Base) :NewEnumerator2 (LocalVariables|GetGetSet_Base) :NewEnumerator3 (LocalVariables|GetGetSet_Base) :NewEnumerator4 (LocalVariables|GetGetSet_Base) :NewEnumerator5 (LocalVariables|GetGetSet_Base) :NewEnumerator6 (LocalVariables|GetGetSet_Base) :NewEnumerator7 (LocalVariables|GetGetSet_Base) :NewEnumerator9 (select (Variables|GetIsMaskedMaterial) (LocalVariables|GetGetSet_ParentTreeApollo) (LocalVariables|GetGetSet_ParentTreeApollo_Trunk)) :NewEnumerator10 (LocalVariables|GetGetSet_Base) :NewEnumerator11 (LocalVariables|GetGetSet_ParentTreeApollo) :NewEnumerator12 (LocalVariables|GetGetSet_PropPlantShrubParent) :NewEnumerator13 (LocalVariables|GetGetSet_Base) :NewEnumerator14 (LocalVariables|GetGetSet_Base) :NewEnumerator15 (LocalVariables|GetGetSet_FenceChainlinkParent) :NewEnumerator16 (LocalVariables|GetGetSet_ParentTreeApollo) :NewEnumerator18 (LocalVariables|GetGetSet_AI) :NewEnumerator19 (LocalVariables|GetGetSet_AI_Grandma) :NewEnumerator20 (LocalVariables|GetGetSet_Cornfield) :NewEnumerator21 (select (CallFunction|GetChromeChonkerTires MeshComponent MeshComponent) (LocalVariables|GetGetSet_ValetChonkers) (LocalVariables|GetGetSet_Valet)) :NewEnumerator22 (LocalVariables|GetGetSet_Vehicle) :NewEnumerator23 (LocalVariables|GetGetSet_AI) :NewEnumerator24 (LocalVariables|GetGetSet_Base) :NewEnumerator25 (LocalVariables|GetGetSet_Prop_FortressBaseMaterial) :NewEnumerator26 (LocalVariables|GetGetSet_PropTwoSidedMasked) :NewEnumerator28 (LocalVariables|GetGetSet_Jungle) :NewEnumerator29 (LocalVariables|GetGetSet_Base) :NewEnumerator30 (LocalVariables|GetGetSet_HeraldsFortress))))

(fn FindSetParameterName
  (params (TextureGroup :type B_Lackluster_TextureGroup%20Enum :default "NewEnumerator1"))
  ;[1 0 0] ATTENTION: This function has been nativized. Currently both the native and the blueprint versions are in use. If you need to make changes here, talk to Brian Zaugg so we can make sure the corresponding changes are made to the native version.
  (return (Utilities|Select :Index TextureGroup :NewEnumerator1 "Normals" :NewEnumerator2 "Normals_Texture_2" :NewEnumerator3 "Normals_Texture_3" :NewEnumerator4 "Normals_Texture_4" :NewEnumerator0 "Diffuse" :NewEnumerator5 "Diffuse_Texture_2" :NewEnumerator6 "Diffuse_Texture_3" :NewEnumerator7 "Diffuse_Texture_4" :NewEnumerator8 "OpacityMask" :NewEnumerator9 "Emissive" :NewEnumerator10 "Specular Mask" :NewEnumerator11 "Gmap/Emissive/Lights" :NewEnumerator12 "VP_Color" :NewEnumerator13 "VP_RGBMask" :NewEnumerator14 "VP_Normal" :NewEnumerator15 "None")))

(fn FindTexturesFromComponents
  (params (ActorType :type B_Lackluster_ActorTypes%20Enum :default "NewEnumerator8") (Actor :type Actor%20Object%20Reference))
  ;[1 0 0] ATTENTION: This function has been nativized. Currently both the native and the blueprint versions are in use. If you need to make changes here, talk to Brian Zaugg so we can make sure the corresponding changes are made to the native version.
  (bind _result (CallFunction|GetChromeVisualComponent Actor false))
  (Utilities|IsValid _result
    (:Is%20Valid
      (if (Utilities|Array|IsValidIndex (Variables|/ChromeShared/Blueprints/B_Lackluster_ApplyChrome.B_Lackluster_ApplyChrome_C|GetPredefinedMeshComponents _result))
        (LocalVariables|SetMeshComponents (Variables|/ChromeShared/Blueprints/B_Lackluster_ApplyChrome.B_Lackluster_ApplyChrome_C|GetPredefinedMeshComponents _result))
        (else
          (LocalVariables|SetMeshComponents (Actor|GetComponentsByClass Actor "/Script/Engine.MeshComponent"))))
      ;;[1 1 1] Loop through all mesh components
      (for _array_element (LocalVariables|GetMeshComponents)
        (bind _output_get (LocalVariables|SetCurrentMeshComponent _array_element))
        (if (not (CallFunction|ExcludefromChroming _output_get))
          ;;[1 1 1] Loop through all materials in mesh component
          (for _index (range (Rendering|Material|GetNumMaterials (LocalVariables|GetCurrentMeshComponent)))
            (bind _output_get_2 (LocalVariables|SetCurrentMaterialIndexLocal _index))
            (LocalVariables|SetCurrentMaterialLocal (Rendering|Material|GetMaterial (LocalVariables|GetCurrentMeshComponent) _output_get_2))
            (Utilities|Array|AddUnique (LocalVariables|GetOriginalMaterialStructLocal) (Utilities|Struct|MakeBLacklusterOriginalMaterialStruct (LocalVariables|GetCurrentMeshComponent) (LocalVariables|GetCurrentMaterialLocal) (LocalVariables|GetCurrentMaterialIndexLocal) (Variables|GetActor)))
            (Utilities|IsValid (LocalVariables|GetCurrentMaterialLocal)
              (:Is%20Valid
                (bind _asmaterial_instance_dynamic (Utilities|Casting|CastToMaterialInstanceDynamic (LocalVariables|GetCurrentMaterialLocal))
                  (:then
                    ;;[1 1 1] MID
                    (bind _output_get_3 (LocalVariables|SetMID _asmaterial_instance_dynamic)))
                  (:CastFailed
                    ;;[1 1 1] MIC
                    (bind _asmaterial_instance_constant (Utilities|Casting|CastToMaterialInstanceConstant (LocalVariables|GetCurrentMaterialLocal))
                      (:then
                        (bind _output_get_4 (LocalVariables|SetMIC _asmaterial_instance_constant))
                        (if (not (CallFunction|ShouldIgnoreMaterial _output_get_4))
                          (LocalVariables|SetIsMaskedMaterial (CallFunction|IsMaskedBlendMode (LocalVariables|GetMIC)))
                          (LocalVariables|SetIsMID false)))
                      (:CastFailed
                        (bind _asmaterial (Utilities|Casting|CastToMaterial (LocalVariables|GetCurrentMaterialLocal)))
                        (bind _returnvalue (Rendering|Material|CreateDynamicMaterialInstance 0 _asmaterial "None" "Transient" :!class /Script/Engine.KismetMaterialLibrary))
                        (LocalVariables|SetMID _returnvalue)
                        (if (not (CallFunction|ShouldIgnoreMaterial _output_get_3))
                          (LocalVariables|SetIsMaskedMaterial (CallFunction|IsMaskedBlendMode (LocalVariables|GetMID)))
                          (LocalVariables|SetIsMID true))))
                    ;;[1 1 1] Find textures
                    (for _array_element_2 (CallFunction|FindGetSetParameters (LocalVariables|GetIsMaskedMaterial) (Variables|GetActorType) (LocalVariables|GetCurrentMeshComponent))
                      (LocalVariables|SetCurrentTextureGroup (.TextureGroup_13_AD1EEF4544A3B0F1A2FBFE86C14C52A5 (Utilities|Struct|BreakBLacklusterTextureSet _array_element_2)))
                      (LocalVariables|SetGetParameterNameLocal (Utilities|Struct|BreakBLacklusterTextureSet _array_element_2))
                      (LocalVariables|SetSetParameterNameLocal (CallFunction|FindSetParameterName (.TextureGroup_13_AD1EEF4544A3B0F1A2FBFE86C14C52A5 (Utilities|Struct|BreakBLacklusterTextureSet _array_element_2))))
                      (if (LocalVariables|GetIsMID)
                        (bind _returnvalue_2 (Rendering|Material|GetTextureParameterValue (LocalVariables|GetMID) (LocalVariables|GetGetParameterNameLocal) :!class /Script/Engine.MaterialInstanceDynamic))
                        (Utilities|IsValid _returnvalue_2
                          (:Is%20Valid
                            (Utilities|Map|Add (LocalVariables|GetSetTexture) (LocalVariables|GetSetParameterNameLocal) _returnvalue_2)))
                        (else
                          (bind _returnvalue_3 (Rendering|Material|GetTextureParameterValue (LocalVariables|GetMIC) (LocalVariables|GetGetParameterNameLocal) :!class /Script/Engine.MaterialInstanceConstant))
                          (Utilities|IsValid _returnvalue_3
                            (:Is%20Valid
                              (Utilities|Map|Add (LocalVariables|GetSetTexture) (LocalVariables|GetSetParameterNameLocal) _returnvalue_3))))))
                    (Utilities|Array|AddUnique (LocalVariables|GetChromeMaterialStructLocal) (Utilities|Struct|MakeBLacklusterChromeMaterialStruct (LocalVariables|GetCurrentMeshComponent) (LocalVariables|GetSetTexture) (LocalVariables|GetCurrentMaterialIndexLocal) (LocalVariables|GetIsMaskedMaterial) (Variables|GetActorType) (Variables|GetActor))))))))))
      (return (LocalVariables|GetChromeMaterialStructLocal) (LocalVariables|GetOriginalMaterialStructLocal)))
    (:Is%20Not%20Valid
      _)))

(fn GetChromeActorType
  (params (InActor :type Actor%20Object%20Reference))
  ;[1 0 0] ATTENTION: This function has been nativized. Currently both the native and the blueprint versions are in use. If you need to make changes here, talk to Brian Zaugg so we can make sure the corresponding changes are made to the native version.
  (bind (_found _actortype _chrometypeout) (CallFunction|CasttoSpecificClasses InActor "NewEnumerator1" InActor))
  (return _actortype _chrometypeout))

(fn GetChromeChonkerTires
  (params (MeshCompoenet :type Mesh%20Component%20Object%20Reference))
  ;[1 0 0] ATTENTION: This function has been nativized. Currently both the native and the blueprint versions are in use. If you need to make changes here, talk to Brian Zaugg so we can make sure the corresponding changes are made to the native version.
  (bind _asstatic_mesh_component (Utilities|Casting|CastToStaticMeshComponent MeshCompoenet))
  (return (Utilities|String|EqualExactly(String) (Utilities|GetDisplayName (Variables|/Script/Engine.StaticMeshComponent|GetStaticMesh _asstatic_mesh_component)) "Off_Road_Tires")))

(fn GetChromeMaterialOverride
  (params (IsMasked :type Boolean) (ActorType :type B_Lackluster_ActorTypes%20Enum :default "NewEnumerator8") (Actor :type Actor%20Object%20Reference) (ChromeType :type B_Lackluster_ChromeTypes%20Enum :default "NewEnumerator1") (MaterialIndex :type Integer))
  (bind _materialindex (Variables|GetMaterialIndex))
  ;[1 0 0] ATTENTION: This function has been nativized. Currently both the native and the blueprint versions are in use. If you need to make changes here, talk to Brian Zaugg so we can make sure the corresponding changes are made to the native version.
  (Utilities|FlowControl|Sequence
    (:then_0
      ;;[1 1 1] Handle Impostor materials
      (if (Utilities|Enum|Equal(Enum) (Variables|GetActorType) "NewEnumerator9")
        (LocalVariables|SetIsImpostorMaterial false)
        (bind _array_element (Utilities|Array|ForEachLoopwithBreak (Actor|GetComponentsByClass (Variables|GetActor) "/Script/Engine.StaticMeshComponent"))
          (:LoopBody
            (Utilities|IsValid (Rendering|Material|GetMaterial _array_element _materialindex)
              (:Is%20Valid
                (bind _returnvalue (Rendering|Material|GetBaseMaterial (Rendering|Material|GetMaterial _array_element _materialindex)))
                (Utilities|IsValid _returnvalue
                  (:Is%20Valid
                    (if (or (== _returnvalue "/Game/ImpostorBaker/Materials/M_ImposterBump01_SimpleGrid.M_ImposterBump01_SimpleGrid") (== _returnvalue "/Game/TechArt/Seasonal/LackLuster/Materials/M_ImposterBump01_Lackluster.M_ImposterBump01_Lackluster"))
                      (LocalVariables|SetIsImpostorMaterial true)
                      (bind _returnvalue_2 (Rendering|Material|CreateDynamicMaterialInstance _array_element (Variables|GetMaterialIndex) (Rendering|Material|GetMaterial _array_element _materialindex) "None" :!class /Script/Engine.PrimitiveComponent))
                      (LocalVariables|SetImpostorMaterial _returnvalue_2))))))))))
    (:then_1
      (return (Utilities|Select :Index (Variables|GetActorType) :NewEnumerator8 (LocalVariables|GetChromeGeneric) :NewEnumerator17 (LocalVariables|GetChromeGeneric) :NewEnumerator0 (Utilities|Select :Index (Variables|GetChromeType) :NewEnumerator1 (LocalVariables|GetChrome_4Layer) :NewEnumerator12 (LocalVariables|GetChromeMaterialPhaseable) :NewEnumerator13 (LocalVariables|GetChromeMaterialPhased) :NewEnumerator14 (LocalVariables|GetChromeGeneric) :NewEnumerator15 (LocalVariables|GetChrome_4Layer_PropPetrified)) :NewEnumerator1 (Utilities|Select :Index (Variables|GetChromeType) :NewEnumerator1 (LocalVariables|GetChrome_4Layer) :NewEnumerator12 (LocalVariables|GetChromeMaterialPhaseable) :NewEnumerator13 (LocalVariables|GetChromeMaterialPhased) :NewEnumerator14 (LocalVariables|GetChromeGeneric) :NewEnumerator15 (LocalVariables|GetChrome_4Layer_PropPetrified)) :NewEnumerator2 (Utilities|Select :Index (Variables|GetChromeType) :NewEnumerator1 (LocalVariables|GetChrome_4Layer) :NewEnumerator12 (LocalVariables|GetChromeMaterialPhaseable) :NewEnumerator13 (LocalVariables|GetChromeMaterialPhased) :NewEnumerator14 (LocalVariables|GetChromeGeneric) :NewEnumerator15 (LocalVariables|GetChrome_4Layer_PropPetrified)) :NewEnumerator3 (Utilities|Select :Index (Variables|GetChromeType) :NewEnumerator1 (LocalVariables|GetChrome_4Layer) :NewEnumerator12 (LocalVariables|GetChromeMaterialPhaseable) :NewEnumerator13 (LocalVariables|GetChromeMaterialPhased) :NewEnumerator14 (LocalVariables|GetChromeGeneric) :NewEnumerator15 (LocalVariables|GetChrome_4Layer_PropPetrified)) :NewEnumerator4 (Utilities|Select :Index (Variables|GetChromeType) :NewEnumerator1 (LocalVariables|GetChrome_4Layer) :NewEnumerator12 (LocalVariables|GetChromeMaterialPhaseable) :NewEnumerator13 (LocalVariables|GetChromeMaterialPhased) :NewEnumerator14 (LocalVariables|GetChromeGeneric) :NewEnumerator15 (LocalVariables|GetChrome_4Layer_PropPetrified)) :NewEnumerator5 (Utilities|Select :Index (Variables|GetChromeType) :NewEnumerator1 (LocalVariables|GetChrome_4Layer) :NewEnumerator12 (LocalVariables|GetChromeMaterialPhaseable) :NewEnumerator13 (LocalVariables|GetChromeMaterialPhased) :NewEnumerator14 (LocalVariables|GetChromeGeneric) :NewEnumerator15 (LocalVariables|GetChrome_4Layer_PropPetrified)) :NewEnumerator6 (Utilities|Select :Index (Variables|GetChromeType) :NewEnumerator1 (LocalVariables|GetChrome_4Layer) :NewEnumerator12 (LocalVariables|GetChromeMaterialPhaseable) :NewEnumerator13 (LocalVariables|GetChromeMaterialPhased) :NewEnumerator14 (LocalVariables|GetChromeGeneric) :NewEnumerator15 (LocalVariables|GetChrome_4Layer_PropPetrified)) :NewEnumerator7 (Utilities|Select :Index (Variables|GetChromeType) :NewEnumerator1 (LocalVariables|GetChrome_4Layer) :NewEnumerator12 (LocalVariables|GetChromeMaterialPhaseable) :NewEnumerator13 (LocalVariables|GetChromeMaterialPhased) :NewEnumerator14 (LocalVariables|GetChromeGeneric) :NewEnumerator15 (LocalVariables|GetChrome_4Layer_PropPetrified)) :NewEnumerator9 (select (LocalVariables|GetIsImpostorMaterial) (LocalVariables|GetChrome%20Material%20Parent%20Tree%20Apollo%20Impostor) (select (Variables|GetIsMasked) (LocalVariables|GetChromeMaterial_ParentTreeApollo_Masked) (LocalVariables|GetChromeMaterial_ParentTreeApollo))) :NewEnumerator10 (LocalVariables|GetChromeMaterialStaticMesh) :NewEnumerator11 (select (LocalVariables|GetIsImpostorMaterial) (LocalVariables|GetChrome%20Material%20Parent%20Tree%20Apollo%20Impostor) (select (Variables|GetIsMasked) (LocalVariables|GetChromeMaterial_ParentTreeApollo_Masked) (LocalVariables|GetChromeMaterial_ParentTreeApollo))) :NewEnumerator12 (LocalVariables|GetChromeMaterial_PropPlantShrubParent) :NewEnumerator13 (CallFunction|GetChromeTieredChestMaterial (Variables|GetActor) (Variables|GetActor)) :NewEnumerator14 (CallFunction|GetChromeTieredChestMaterial (Variables|GetActor) (Variables|GetActor)) :NewEnumerator15 (LocalVariables|GetChromeMaterialMasked) :NewEnumerator16 (LocalVariables|GetChromeCornfield) :NewEnumerator18 (LocalVariables|GetChromeAI) :NewEnumerator19 (LocalVariables|GetChromeAI_Grandma) :NewEnumerator20 (LocalVariables|GetChromeCornfield) :NewEnumerator21 (LocalVariables|GetChromeValet) :NewEnumerator22 (LocalVariables|GetChromeVehicle) :NewEnumerator23 (LocalVariables|GetChromeAI_Avian) :NewEnumerator24 (LocalVariables|GetChrome_4Layer_PropPetrified_TwoSided) :NewEnumerator25 (LocalVariables|GetChrome_Prop_Masked) :NewEnumerator26 (LocalVariables|GetChrome_4Layer_PropPetrified_TwoSided_Masked) :NewEnumerator27 (Utilities|Select :Index (Variables|GetChromeType) :NewEnumerator1 (LocalVariables|GetChrome_4Layer) :NewEnumerator12 (LocalVariables|GetChromeMaterialPhaseable) :NewEnumerator13 (LocalVariables|GetChromeMaterialPhased) :NewEnumerator14 (LocalVariables|GetChromeGeneric) :NewEnumerator15 (LocalVariables|GetChrome_4Layer_PropPetrified)) :NewEnumerator28 (LocalVariables|GetChromeMaterial_Jungle) :NewEnumerator29 (LocalVariables|GetChromeMaterial_InteractableProp) :NewEnumerator30 (LocalVariables|GetChromeMaterial_HeraldsFortress)) (LocalVariables|GetIsImpostorMaterial) (LocalVariables|GetImpostorMaterial)))))

(fn GetChromeSoundComponent
  (params (Actor :type Actor%20Object%20Reference) (CreateComponentIfNotFound :type Boolean :default true))
  (bind _returnvalue (Actor|GetComponentbyClass Actor "/ChromeShared/Audio/ChromeSoundComponent.ChromeSoundComponent_C"))
  (Utilities|IsValid Actor
    (:Is%20Valid
      (Utilities|IsValid _returnvalue
        (:Is%20Valid
          (return _returnvalue))
        (:Is%20Not%20Valid
          (if (Variables|GetCreateComponentIfNotFound)
            (bind _returnvalue_2 (Game|AddComponentbyClass (Variables|GetActor) 0 "/ChromeShared/Audio/ChromeSoundComponent.ChromeSoundComponent_C"))
            (return _returnvalue_2)
            (else
              (return 0))))))
    (:Is%20Not%20Valid
      (return 0))))

(fn GetChromeTieredChestMaterial
  (params (InActor :type Actor%20Object%20Reference))
  ;[1 0 0] ATTENTION: This function has been nativized. Currently both the native and the blueprint versions are in use. If you need to make changes here, talk to Brian Zaugg so we can make sure the corresponding changes are made to the native version.
  (bind _asbuilding_container (Utilities|Casting|CastToBuildingContainer InActor)
    (:then
      (return (select (GameplayTags|HasAnyMatchingGameplayTags _asbuilding_container (GameplayTags|MakeLiteralGameplayTagContainer "(GameplayTags=((TagName=\"Building.Type.Container.TreasureChest.Rare\"),(TagName=\"Building.Type.Container.Toolbox\"),(TagName=\"Building.Type.Container.CashRegister\")))")) (LocalVariables|GetChromeNoAnimChestMaterial) (LocalVariables|GetChromeAnimChestMaterial))))
    (:CastFailed
      (return (LocalVariables|GetGenericChromeMaterial)))))

(fn GetChromeVisualComponent
  (params (Actor :type Actor%20Object%20Reference) (CreateComponentIfNotFound :type Boolean :default true))
  (bind _returnvalue (Actor|GetComponentbyClass Actor "/ChromeShared/Blueprints/B_Lackluster_ApplyChrome.B_Lackluster_ApplyChrome_C"))
  (bind _actor (Variables|GetActor))
  ;[1 0 0] ATTENTION: This function has been nativized. Currently both the native and the blueprint versions are in use. If you need to make changes here, talk to Brian Zaugg so we can make sure the corresponding changes are made to the native version.
  (Utilities|IsValid Actor
    (:Is%20Valid
      (Utilities|IsValid _returnvalue
        (:Is%20Valid
          (return _returnvalue))
        (:Is%20Not%20Valid
          (if (Variables|GetCreateComponentIfNotFound)
            (Utilities|Casting|CastToStaticMeshActor _actor
              (:then
                _)
              (:CastFailed
                (bind _returnvalue_2 (Game|AddComponentbyClass _actor 0 "/ChromeShared/Blueprints/B_Lackluster_ApplyChrome.B_Lackluster_ApplyChrome_C"))
                (return _returnvalue_2))))
          (return 0))))
    (:Is%20Not%20Valid
      (return 0))))

(fn Initialize
  (bind _returnvalue (WorldSubsystems|GetChromeSubsystem_C 0))
  (bind _lacklusterdata (Variables|/Script/ChromeSharedRuntime.FortChromeSubsystem|GetLacklusterData _returnvalue))
  (bind (_applychromeclasspath _chromeaudioclasspath _materialstoignore _setparameternames _classestoactortypes _chromematerialoverrides _actortotextures) (Utilities|Struct|BreakLacklusterData _lacklusterdata))
  (bind _returnvalue_2 (WorldSubsystems|GetChromeSubsystem_C 0))
  (bind _lacklusterdata_2 (Variables|/Script/ChromeSharedRuntime.FortChromeSubsystem|GetLacklusterData _returnvalue_2))
  (bind (_applychromeclasspath_2 _chromeaudioclasspath_2 _materialstoignore_2 _setparameternames_2 _classestoactortypes_2 _chromematerialoverrides_2) (Utilities|Struct|BreakLacklusterData _lacklusterdata_2))
  (bind _chrome_4layer_proppetrified (LocalVariables|GetChrome_4Layer_PropPetrified))
  (bind _chromegeneric (LocalVariables|GetChromeGeneric))
  (bind _chromematerialphased (LocalVariables|GetChromeMaterialPhased))
  (bind _chromematerialphaseable (LocalVariables|GetChromeMaterialPhaseable))
  (bind _map (Utilities|Map|MakeMap "Chrome" _chromematerialphaseable "ChromePhaseable" _chromematerialphaseable "ChromePhaseableMasked" _chromematerialphased "ChromeGeneric" _chromegeneric "ChromePetrified" _chrome_4layer_proppetrified))
  (bind _chrome_4layer_proppetrified_2 (LocalVariables|GetChrome_4Layer_PropPetrified))
  (bind _chromegeneric_2 (LocalVariables|GetChromeGeneric))
  (bind _chromematerialphased_2 (LocalVariables|GetChromeMaterialPhased))
  (bind _chromematerialphaseable_2 (LocalVariables|GetChromeMaterialPhaseable))
  (bind _chrome_4layer (LocalVariables|GetChrome_4Layer))
  (bind _map_2 (Utilities|Map|MakeMap "Chrome" _chrome_4layer "ChromePhaseable" _chromematerialphaseable_2 "ChromePhaseableMasked" _chromematerialphased_2 "ChromeGeneric" _chromegeneric_2 "ChromePetrified" _chrome_4layer_proppetrified_2))
  (bind _getset_base (LocalVariables|GetGetSet_Base))
  (bind _chrometexturesetarray (Utilities|Struct|MakeChromeTextureSetArray _getset_base))
  (bind _getset_parenttreeapollo (LocalVariables|GetGetSet_ParentTreeApollo))
  (bind _chrometexturesetarray_2 (Utilities|Struct|MakeChromeTextureSetArray _getset_parenttreeapollo))
  (bind _chromegeneric_3 (LocalVariables|GetChromeGeneric))
  (bind _chrometypetomaterial (Utilities|Struct|MakeChromeTypetoMaterial _chromegeneric_3))
  (bind _chrometypetomaterial_2 (Utilities|Struct|MakeChromeTypetoMaterial 0 _map))
  (bind _chromematerial_parenttreeapollo_masked (LocalVariables|GetChromeMaterial_ParentTreeApollo_Masked))
  (bind _chrometypetomaterial_3 (Utilities|Struct|MakeChromeTypetoMaterial _chromematerial_parenttreeapollo_masked))
  (bind _chromematerial_parenttreeapollo (LocalVariables|GetChromeMaterial_ParentTreeApollo))
  (bind _chrometypetomaterial_4 (Utilities|Struct|MakeChromeTypetoMaterial _chromematerial_parenttreeapollo))
  (bind _chromecornfield (LocalVariables|GetChromeCornfield))
  (bind _chrometypetomaterial_5 (Utilities|Struct|MakeChromeTypetoMaterial _chromecornfield))
  (bind _chromevalet (LocalVariables|GetChromeValet))
  (bind _chrometypetomaterial_6 (Utilities|Struct|MakeChromeTypetoMaterial _chromevalet))
  (bind _chrometypetomaterial_7 (Utilities|Struct|MakeChromeTypetoMaterial 0 _map_2))
  (bind _getset_ai (LocalVariables|GetGetSet_AI))
  (bind _chrometexturesetarray_3 (Utilities|Struct|MakeChromeTextureSetArray _getset_ai))
  (Utilities|IsValid (WorldSubsystems|GetChromeSubsystem_C 0)
    (:Is%20Valid
      (Utilities|FlowControl|Sequence
        (:then_0
          (Utilities|Map|Add _actortotextures "GenericChrome" _chrometexturesetarray)
          (Utilities|Map|Add _actortotextures "BuildingRoof" _chrometexturesetarray)
          (Utilities|Map|Add _actortotextures "BuildingWall" _chrometexturesetarray)
          (Utilities|Map|Add _actortotextures "BuildingPropWall" _chrometexturesetarray)
          (Utilities|Map|Add _actortotextures "BuildingCorner" _chrometexturesetarray)
          (Utilities|Map|Add _actortotextures "BuildingFloor" _chrometexturesetarray)
          (Utilities|Map|Add _actortotextures "BuildingProp" _chrometexturesetarray)
          (Utilities|Map|Add _actortotextures "BuildingDeco" _chrometexturesetarray)
          (Utilities|Map|Add _actortotextures "BuildingPillar" _chrometexturesetarray)
          (Utilities|Map|Add _actortotextures "ParentTreeApollo" _chrometexturesetarray_2)
          (Utilities|Map|Add _actortotextures "ParentTreeApolloTrunk" (Utilities|Struct|MakeChromeTextureSetArray (LocalVariables|GetGetSet_ParentTreeApollo_Trunk)))
          (Utilities|Map|Add _actortotextures "StaticMeshActor" _chrometexturesetarray)
          (Utilities|Map|Add _actortotextures "ArtemisHedgeParentMasked" _chrometexturesetarray_2)
          (Utilities|Map|Add _actortotextures "ArtemisHedgeParent" _chrometexturesetarray_2)
          (Utilities|Map|Add _actortotextures "PropPlantShrubParent" (Utilities|Struct|MakeChromeTextureSetArray (LocalVariables|GetGetSet_PropPlantShrubParent)))
          (Utilities|Map|Add _actortotextures "TieredChestAthena" _chrometexturesetarray)
          (Utilities|Map|Add _actortotextures "TieredAmmoAthena" _chrometexturesetarray)
          (Utilities|Map|Add _actortotextures "FenceChainlinkParent" (Utilities|Struct|MakeChromeTextureSetArray (LocalVariables|GetGetSet_FenceChainlinkParent)))
          (Utilities|Map|Add _actortotextures "BigBush" _chrometexturesetarray_2)
          (Utilities|Map|Add _actortotextures "AI" _chrometexturesetarray_3)
          (Utilities|Map|Add _actortotextures "AI_Grandma" (Utilities|Struct|MakeChromeTextureSetArray (LocalVariables|GetGetSet_AI_Grandma)))
          (Utilities|Map|Add _actortotextures "Cornfield" (Utilities|Struct|MakeChromeTextureSetArray (LocalVariables|GetGetSet_Cornfield)))
          (Utilities|Map|Add _actortotextures "Valet" (Utilities|Struct|MakeChromeTextureSetArray (LocalVariables|GetGetSet_Valet)))
          (Utilities|Map|Add _actortotextures "ValetChonkers" (Utilities|Struct|MakeChromeTextureSetArray (LocalVariables|GetGetSet_ValetChonkers)))
          (Utilities|Map|Add _actortotextures "Vehicle" (Utilities|Struct|MakeChromeTextureSetArray (LocalVariables|GetGetSet_Vehicle)))
          (Utilities|Map|Add _actortotextures "AI_Avian" _chrometexturesetarray_3)
          (Utilities|Map|Add _actortotextures "BuildingProp_TwoSided" _chrometexturesetarray)
          (Utilities|Map|Add _actortotextures "Prop_FortressBaseMaterial" (Utilities|Struct|MakeChromeTextureSetArray (LocalVariables|GetGetSet_Prop_FortressBaseMaterial)))
          (Utilities|Map|Add _actortotextures "BuildingProp_TwoSided_Masked" (Utilities|Struct|MakeChromeTextureSetArray (LocalVariables|GetGetSet_PropTwoSidedMasked)))
          (Utilities|Map|Add _actortotextures "Jungle" (Utilities|Struct|MakeChromeTextureSetArray (LocalVariables|GetGetSet_Jungle)))
          (Utilities|Map|Add _actortotextures "InteractableProp" _chrometexturesetarray)
          (Utilities|Map|Add _actortotextures "HeraldsFortress" (Utilities|Struct|MakeChromeTextureSetArray (LocalVariables|GetGetSet_HeraldsFortress)))
          (Utilities|Map|Add _actortotextures "IceCreamTruck" (Utilities|Struct|MakeChromeTextureSetArray (LocalVariables|GetGetSet_IceCreamTruck)))
          (Utilities|Map|Add _actortotextures "ChromeDirectionalLaunchpad" (Utilities|Struct|MakeChromeTextureSetArray (LocalVariables|GetGetSet_LowGravLauncher))))
        (:then_1
          (Utilities|Map|Add _chromematerialoverrides_2 "Undefined" _chrometypetomaterial)
          (Utilities|Map|Add _chromematerialoverrides_2 "GenericChrome" _chrometypetomaterial)
          (Utilities|Map|Add _chromematerialoverrides_2 "BuildingRoof" _chrometypetomaterial_2)
          (Utilities|Map|Add _chromematerialoverrides_2 "BuildingWall" _chrometypetomaterial_2)
          (Utilities|Map|Add _chromematerialoverrides_2 "BuildingPropWall" _chrometypetomaterial_7)
          (Utilities|Map|Add _chromematerialoverrides_2 "BuildingCorner" _chrometypetomaterial_2)
          (Utilities|Map|Add _chromematerialoverrides_2 "BuildingFloor" _chrometypetomaterial_7)
          (Utilities|Map|Add _chromematerialoverrides_2 "BuildingProp" _chrometypetomaterial_7)
          (Utilities|Map|Add _chromematerialoverrides_2 "BuildingDeco" _chrometypetomaterial_7)
          (Utilities|Map|Add _chromematerialoverrides_2 "BuildingPillar" _chrometypetomaterial_7)
          (Utilities|Map|Add _chromematerialoverrides_2 "ParentTreeApollo" _chrometypetomaterial_3)
          (Utilities|Map|Add _chromematerialoverrides_2 "ParentTreeApolloTrunk" _chrometypetomaterial_4)
          (Utilities|Map|Add _chromematerialoverrides_2 "StaticMeshActor" (Utilities|Struct|MakeChromeTypetoMaterial (LocalVariables|GetChromeMaterialStaticMesh)))
          (Utilities|Map|Add _chromematerialoverrides_2 "ArtemisHedgeParent" _chrometypetomaterial_4)
          (Utilities|Map|Add _chromematerialoverrides_2 "ArtemisHedgeParentMasked" _chrometypetomaterial_3)
          (Utilities|Map|Add _chromematerialoverrides_2 "PropPlantShrubParent" (Utilities|Struct|MakeChromeTypetoMaterial (LocalVariables|GetChromeMaterial_PropPlantShrubParent)))
          (Utilities|Map|Add _chromematerialoverrides_2 "FenceChainlinkParent" (Utilities|Struct|MakeChromeTypetoMaterial (LocalVariables|GetChromeMaterialMasked)))
          (Utilities|Map|Add _chromematerialoverrides_2 "BigBush" _chrometypetomaterial_5)
          (Utilities|Map|Add _chromematerialoverrides_2 "AI" (Utilities|Struct|MakeChromeTypetoMaterial (LocalVariables|GetChromeAI)))
          (Utilities|Map|Add _chromematerialoverrides_2 "AI_Grandma" (Utilities|Struct|MakeChromeTypetoMaterial (LocalVariables|GetChromeAI_Grandma)))
          (Utilities|Map|Add _chromematerialoverrides_2 "Cornfield" _chrometypetomaterial_5)
          (Utilities|Map|Add _chromematerialoverrides_2 "Valet" _chrometypetomaterial_6)
          (Utilities|Map|Add _chromematerialoverrides_2 "ValetChonkers" _chrometypetomaterial_6)
          (Utilities|Map|Add _chromematerialoverrides_2 "Vehicle" (Utilities|Struct|MakeChromeTypetoMaterial (LocalVariables|GetChromeVehicle)))
          (Utilities|Map|Add _chromematerialoverrides_2 "AI_Avian" (Utilities|Struct|MakeChromeTypetoMaterial (LocalVariables|GetChromeAI_Avian)))
          (Utilities|Map|Add _chromematerialoverrides_2 "BuildingProp_TwoSided" (Utilities|Struct|MakeChromeTypetoMaterial (LocalVariables|GetChrome_4Layer_PropPetrified_TwoSided)))
          (Utilities|Map|Add _chromematerialoverrides_2 "Prop_FortressBaseMaterial" (Utilities|Struct|MakeChromeTypetoMaterial (LocalVariables|GetChrome_Prop_Masked)))
          (Utilities|Map|Add _chromematerialoverrides_2 "BuildingProp_TwoSided_Masked" (Utilities|Struct|MakeChromeTypetoMaterial (LocalVariables|GetChrome_4Layer_PropPetrified_TwoSided_Masked)))
          (Utilities|Map|Add _chromematerialoverrides_2 "BuildingActor" _chrometypetomaterial_7)
          (Utilities|Map|Add _chromematerialoverrides_2 "Jungle" (Utilities|Struct|MakeChromeTypetoMaterial 0 (Utilities|Map|MakeMap "Chrome" (LocalVariables|GetChromeMaterial_Jungle) "ChromePhaseable" (LocalVariables|GetChromeMaterialJunglePhaseable) "ChromePhaseableMasked" (LocalVariables|GetChromeMaterialJunglePhaseableMasked) "ChromeGeneric" (LocalVariables|GetChromeGeneric) "ChromePetrified" (LocalVariables|GetChrome_4Layer_PropPetrified))))
          (Utilities|Map|Add _chromematerialoverrides_2 "InteractableProp" (Utilities|Struct|MakeChromeTypetoMaterial (LocalVariables|GetChromeMaterial_InteractableProp)))
          (Utilities|Map|Add _chromematerialoverrides_2 "HeraldsFortress" (Utilities|Struct|MakeChromeTypetoMaterial 0 (Utilities|Map|MakeMap "Chrome" (LocalVariables|GetChromeMaterialHeraldsFortress) "ChromePhaseable" (LocalVariables|GetChromeMaterialHeraldsFortress) "ChromePhaseableMasked" (LocalVariables|GetChromeMaterialHeraldsFortressPhased) "ChromeGeneric" (LocalVariables|GetChromeMaterialHeraldsFortress) "ChromePetrified" (LocalVariables|GetChromeMaterialHeraldsFortress))))
          (Utilities|Map|Add _chromematerialoverrides_2 "IceCreamTruck" (Utilities|Struct|MakeChromeTypetoMaterial (LocalVariables|GetChrome_4Layer_PropPetrified)))
          (Utilities|Map|Add _chromematerialoverrides_2 "ChromeDirectionalLaunchpad" (Utilities|Struct|MakeChromeTypetoMaterial (LocalVariables|GetChrome_LowGravLauncher))))
        (:then_2
          (Utilities|Struct|SetmembersinLacklusterData (Variables|/Script/ChromeSharedRuntime.FortChromeSubsystem|GetLacklusterData (WorldSubsystems|GetChromeSubsystem_C 0)) true))))
    (:Is%20Not%20Valid
      (Fortnite|FortShippingLog 0 "Lackluster function library not initialized. Probably failed to get ChromeSubsystem from blueprint. Not using nativized chrome application." true))))

(fn IsChromeParentAlreadyApplied
  (params (Material%20Interface :type Material%20Instance%20Object%20Reference))
  (return (Utilities|String|Contains (Utilities|GetDisplayName (Variables|/Script/Engine.MaterialInstance|GetParent Material%20Interface)) "LackLuster")))

(fn IsMaskedBlendMode
  (params (MaterialInterface :type Material%20Interface%20Object%20Reference))
  ;[1 0 0] ATTENTION: This function has been nativized. Currently both the native and the blueprint versions are in use. If you need to make changes here, talk to Brian Zaugg so we can make sure the corresponding changes are made to the native version.
  (return (Utilities|Enum|Equal(Enum) (Rendering|Material|GetBlendMode MaterialInterface) "BLEND_Masked")))

(fn NeedsToSaveMaterialsAndTypes
  (params (Actor :type Actor%20Object%20Reference) (ActorType :type B_Lackluster_ActorTypes%20Enum :default "NewEnumerator8"))
  (return true))

(fn PythonApplyChromeToActor
  (params (Actor :type Actor%20Object%20Reference))
  ;[1 0.08333 0.08333] Changing inputs to this function will break the python interaction so you must update TechArt\MenuTools\Seasonal\S22\S22_AddContextMenu_Chrome.py
  (CallFunction|ApplyChromeMaterial Actor (Utilities|Array|MakeArray) (Utilities|Array|MakeArray)))

(fn SetChromeAudio
  (params (Actor :type Actor%20Object%20Reference) (Add :type Boolean) (ChromeType :type B_Lackluster_ChromeTypes%20Enum :default "NewEnumerator1"))
  (bind _results (CallFunction|GetChromeSoundComponent Actor true Actor))
  (Utilities|IsValid Actor
    (:Is%20Valid
      (Utilities|IsValid _results
        (:Is%20Valid
          (Utilities|Map|Add (Variables|/ChromeShared/ChromeSubsystem.ChromeSubsystem_C|GetChromeSoundComponents (WorldSubsystems|GetChromeSubsystem_C 0)) Actor _results)
          (Data|SetChromeAudioActor _results Actor Add ChromeType))))))

(fn SetChromeMaterials
  (params (Array :type Array%20of%20B%20Lackluster%20Chrome%20Material%20Struct%20Structures) (ChromeType :type B_Lackluster_ChromeTypes%20Enum :default "NewEnumerator1"))
  (bind _chrometype (Variables|GetChromeType))
  (bind _tireindex (LocalVariables|GetTireIndex))
  ;[1 0 0] ATTENTION: This function has been nativized. Currently both the native and the blueprint versions are in use. If you need to make changes here, talk to Brian Zaugg so we can make sure the corresponding changes are made to the native version.
  (LocalVariables|SetTireIndex 0)
  (for _array_element Array
    (LocalVariables|SetIsImpostor (.IsImpostorMaterial_out (CallFunction|GetChromeMaterialOverride (.isMaskedMaterial_14_721683D94AF354148DC31E8E004EE8FD (Utilities|Struct|BreakBLacklusterChromeMaterialStruct _array_element)) (.ActorType_22_FAC314444D4A41D79AF195BB97E513AA (Utilities|Struct|BreakBLacklusterChromeMaterialStruct _array_element)) (.Actor_25_F6FA105746972B378314899F7244AD76 (Utilities|Struct|BreakBLacklusterChromeMaterialStruct _array_element)) _chrometype (.MaterialIndex_11_AC01E8624312F64FBFF6A6B08AE7BA7C (Utilities|Struct|BreakBLacklusterChromeMaterialStruct _array_element)) (.Actor_25_F6FA105746972B378314899F7244AD76 (Utilities|Struct|BreakBLacklusterChromeMaterialStruct _array_element)))))
    (LocalVariables|SetOriginalImpostorMaterial (.ImpostorMaterial_out (CallFunction|GetChromeMaterialOverride (.isMaskedMaterial_14_721683D94AF354148DC31E8E004EE8FD (Utilities|Struct|BreakBLacklusterChromeMaterialStruct _array_element)) (.ActorType_22_FAC314444D4A41D79AF195BB97E513AA (Utilities|Struct|BreakBLacklusterChromeMaterialStruct _array_element)) (.Actor_25_F6FA105746972B378314899F7244AD76 (Utilities|Struct|BreakBLacklusterChromeMaterialStruct _array_element)) _chrometype (.MaterialIndex_11_AC01E8624312F64FBFF6A6B08AE7BA7C (Utilities|Struct|BreakBLacklusterChromeMaterialStruct _array_element)) (.Actor_25_F6FA105746972B378314899F7244AD76 (Utilities|Struct|BreakBLacklusterChromeMaterialStruct _array_element)))))
    (bind _returnvalue (Chrome|GetSharedChromeMID (.Actor_25_F6FA105746972B378314899F7244AD76 (Utilities|Struct|BreakBLacklusterChromeMaterialStruct _array_element)) (CallFunction|GetChromeMaterialOverride (.isMaskedMaterial_14_721683D94AF354148DC31E8E004EE8FD (Utilities|Struct|BreakBLacklusterChromeMaterialStruct _array_element)) (.ActorType_22_FAC314444D4A41D79AF195BB97E513AA (Utilities|Struct|BreakBLacklusterChromeMaterialStruct _array_element)) (.Actor_25_F6FA105746972B378314899F7244AD76 (Utilities|Struct|BreakBLacklusterChromeMaterialStruct _array_element)) _chrometype (.MaterialIndex_11_AC01E8624312F64FBFF6A6B08AE7BA7C (Utilities|Struct|BreakBLacklusterChromeMaterialStruct _array_element)) (.Actor_25_F6FA105746972B378314899F7244AD76 (Utilities|Struct|BreakBLacklusterChromeMaterialStruct _array_element))) (.SetTexture_18_8A367EA4464B531450E451ACF9B8B387 (Utilities|Struct|BreakBLacklusterChromeMaterialStruct _array_element))))
    (Rendering|Material|SetMaterial (Utilities|Struct|BreakBLacklusterChromeMaterialStruct _array_element) (.MaterialIndex_11_AC01E8624312F64FBFF6A6B08AE7BA7C (Utilities|Struct|BreakBLacklusterChromeMaterialStruct _array_element)) _returnvalue)
    (Utilities|FlowControl|Sequence
      (:then_0
        (if (CallFunction|GetChromeChonkerTires (Utilities|Struct|BreakBLacklusterChromeMaterialStruct _array_element) (Utilities|Struct|BreakBLacklusterChromeMaterialStruct _array_element))
          (bind _asfort_vehicle_interface (Utilities|Casting|CastToFortVehicleInterface (.Actor_25_F6FA105746972B378314899F7244AD76 (Utilities|Struct|BreakBLacklusterChromeMaterialStruct _array_element))))
          (Vehicle|SetAttachedWheelDataMaterialInstance _asfort_vehicle_interface _tireindex _returnvalue)
          (Math|Integer|IncrementInt _tireindex)))
      (:then_1
        (if (LocalVariables|GetIsImpostor)
          (Rendering|Material|CopyMaterialInstanceParameters _returnvalue (LocalVariables|GetOriginalImpostorMaterial) true) ; Impostor mat instances are all unique WPO Parameters
        )))))

(fn SetupChromeOnActor
  (params (ActorType :type B_Lackluster_ActorTypes%20Enum%20%28by%20ref%29 :default "NewEnumerator8") (ChromeType :type B_Lackluster_ChromeTypes%20Enum%20%28by%20ref%29 :default "NewEnumerator1") (ChromeMaterialStuct :type Array%20of%20B%20Lackluster%20Chrome%20Material%20Struct%20Structures) (OriginalMaterialStruct :type Array%20of%20B%20Lackluster%20Original%20Material%20Struct%20Structures) (NativeActorType :type ELacklusterActorType%20Enum%20%28by%20ref%29 :default "Undefined") (NativeChromeType :type ELacklusterChromeType%20Enum%20%28by%20ref%29 :default "Chrome") (NativeChromeMaterialStruct :type Array%20of%20Lackluster%20Material%20Data%20Structures) (NativeOriginalMaterialStruct :type Array%20of%20Lackluster%20Original%20Material%20Data%20Structures) (Actor :type Actor%20Object%20Reference) (ContainerMaterialScalarParamMap :type Map%20of%20Names%20to%20Float%20%28double-precision%29s) (ContainerMaterialVectorParamMap :type Map%20of%20Names%20to%20Linear%20Color%20Structures))
  (bind _returnvalue (WorldSubsystems|GetChromeSubsystem_C 0))
  (bind _containertags (Variables|/ChromeShared/ChromeSubsystem.ChromeSubsystem_C|GetContainerTags _returnvalue))
  (CallFunction|DisableCurieFire Actor Actor)
  (Utilities|FlowControl|Sequence
    (:then_0
      (bind (_actortype _chrometype) (CallFunction|GetChromeActorType Actor Actor))
      (Variables|SetBy-RefVar ActorType _actortype)
      (Variables|SetBy-RefVar ChromeType _chrometype))
    (:then_1
      (bind (_chrometypeout _actortypeout) (Lackluster|NativeGetChromeActorType Actor))
      (Variables|SetBy-RefVar NativeActorType _actortypeout)
      (Variables|SetBy-RefVar NativeChromeType _chrometypeout))
    (:then_2
      (CallFunction|ExecuteOwnerTypeSpecificFunctionalityPreChrome Actor ActorType _containertags ContainerMaterialScalarParamMap ContainerMaterialVectorParamMap Actor)
      (Default|ApplyChromeToActor Actor ActorType ChromeType ChromeMaterialStuct OriginalMaterialStruct NativeActorType NativeChromeType NativeChromeMaterialStruct NativeOriginalMaterialStruct ContainerMaterialScalarParamMap ContainerMaterialVectorParamMap Actor)
      (CallFunction|ExecuteOwnerTypeSpecificFunctionalityPostChrome Actor ActorType _containertags ContainerMaterialScalarParamMap ContainerMaterialVectorParamMap Actor))))

(fn ShouldIgnoreMaterial
  (params (MaterialInterface :type Material%20Interface%20Object%20Reference))
  ;[1 0 0] ATTENTION: This function has been nativized. Currently both the native and the blueprint versions are in use. If you need to make changes here, talk to Brian Zaugg so we can make sure the corresponding changes are made to the native version.
  (bind _asmaterial_instance_dynamic (Utilities|Casting|CastToMaterialInstanceDynamic MaterialInterface)
    (:then
      (bind _returnvalue (Rendering|Material|GetBaseMaterial _asmaterial_instance_dynamic))
      (LocalVariables|SetParentMaterial _returnvalue)
      (if (Utilities|Array|ContainsItem (LocalVariables|GetMaterialsToIgnore) (Utilities|GetDisplayName (LocalVariables|GetParentMaterial)))
        (return true)
        (else
          _)))
    (:CastFailed
      (bind _asmaterial_instance_constant (Utilities|Casting|CastToMaterialInstanceConstant MaterialInterface)
        (:then
          (bind _returnvalue_2 (Rendering|Material|GetBaseMaterial _asmaterial_instance_constant))
          (LocalVariables|SetParentMaterial _returnvalue_2))
        (:CastFailed
          (bind _asmaterial (Utilities|Casting|CastToMaterial MaterialInterface)
            (:then
              (LocalVariables|SetParentMaterial _asmaterial))
            (:CastFailed
              _))))))
  (return (Math|Boolean|ORBoolean (Utilities|Enum|Equal(Enum) (Rendering|Material|GetBlendMode MaterialInterface) "BLEND_Translucent") (Utilities|Enum|Equal(Enum) (Rendering|Material|GetBlendMode MaterialInterface) "BLEND_Additive") (Utilities|Enum|Equal(Enum) (Rendering|Material|GetBlendMode MaterialInterface) "BLEND_Modulate") (Utilities|Enum|Equal(Enum) (Rendering|Material|GetBlendMode MaterialInterface) "BLEND_AlphaComposite"))))

(fn StopBuildingAudio
  (params (BuildingActor :type Building%20Actor%20Object%20Reference))
  (for _array_element (Actor|GetComponentsByClass BuildingActor "/Script/Engine.AudioComponent")
    (Audio|Components|Audio|FadeOut _array_element 0.5 0.0 "Linear" :!class /Script/Engine.AudioComponent)))
