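; Actor component that gives its owner the Lackluster chrome look and audio, and can restore the original materials.
; BeginPlay disables Curie fire on the owner and, where cosmetic events can run, initializes chrome materials once playlist data is ready.
; Setup and ApplyChrome prefer the Blueprint ChromeSubsystem's per-actor cache, then the native LacklusterData map, before falling back.
; EndPlay runs owner-type cleanup and sets a flag that blocks late setup.
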(event EventBeginPlay
  (bind _returnvalue (Components|GetOwner))
  ;[0.007784 0.1406 0.04891] BeginPlay: Apply chrome cosmetic to the owner and perform owner actor type specific functionality.
  (|CallParentFunction)
  (Utilities|FlowControl|Sequence
    (:then_0
      (Utilities|IsValid _returnvalue
        (:Is%20Valid
          ;;[1 0.501 0.3707] Handle fire immunity
          (Class|BLackLusterFunctionLibrary|DisableCurieFire _returnvalue)
          (if (> (Utilities|Array|Length (Actor|GetComponentsByClass _returnvalue (Utilities|GetClass self))) 1)
            (Fortnite|FortShippingLog self (Utilities|String|Append "BeginPlay: Owner " (Utilities|GetDisplayName _returnvalue) " has another ApplyChrome component already") true)))))
    (:then_1
      (Networking|CanExecuteCosmeticEvents
        (:True
          (Class|FortAsyncActionCurrentPlaylistReady|PlaylistDataReadyAsync
            (:OnReady
              (if (not (Variables|Default|GetbChromeMaterialsInitialized))
                (CallFunction|InitChromeMaterials)))))))))

(event Custom|InitChromeMaterials
  (bind _returnvalue (Components|GetOwner))
  (bind _returnvalue_2 (WorldSubsystems|GetChromeSubsystem_C))
  (bind _actormaterialsandtypes (Variables|/ChromeShared/ChromeSubsystem.ChromeSubsystem_C|GetActorMaterialsAndTypes _returnvalue_2))
  (bind (_value _returnvalue_3) (Utilities|Map|Find _actormaterialsandtypes _returnvalue))
  (bind (_actortype_2_5b99d33c44fcfc26b361ee99b2bc0a4a _chrometype_5_68195aba4bf5bd23330635bc8ce95a46 _chromematerialstructs_9_a596659c4eb279eb1d6a5288d978425c _originalmaterialstructs_12_567c73ae46c386f2a6d4c9b35c9d5864 _nativeactortype_22_cfdbf9b1473c12413cfd3b947db2e001 _nativechrometype_23_860f664148debfe8fc0f50a5e7fefb6f _nativechromematerialstructs_21_eb1373e044d8f3e92f5f7c97e850fe26 _nativeoriginalmaterialstructs_29_3b92abd442db398f00adb889972847b8 _containermaterialscalarparammap_34_88450d0546bb2f29ce9612ba6aaa1b20 _containermaterialvectorparammap_37_3436013e48082fbf8e5aaba2b0bd7901) (Utilities|Struct|BreakBLacklusterMaterialsandTypes _value))
  (bind _lacklusterdata (Variables|/Script/ChromeSharedRuntime.FortChromeSubsystem|GetLacklusterData _returnvalue_2))
  (bind (_applychromeclasspath _chromeaudioclasspath _materialstoignore _setparameternames _classestoactortypes _chromematerialoverrides _actortotextures _tagquerytochrometypeoverrides _tagtoactortypes _genericchromematerial _noanimchestmaterialtags _chromeanimchestmaterial _chromenoanimchestmaterial _buildingchromeapplied _buildingsmactorhitsound _buildingsmactorpropdeathsound _buildingsmactornonpropdeathsound _buildingsmactorpropchromedsound _buildingsmactornonpropchromedsound _tieredchestclass _tieredchestparentclass _chromesoundcomponents _shadowproxymeshcomponentclass _materialsandtypes) (Utilities|Struct|BreakLacklusterData _lacklusterdata))
  (bind (_value_2 _returnvalue_4) (Utilities|Map|Find _materialsandtypes _returnvalue))
  (bind (_actortype _chrometype _chromematerialdata _originalmaterialdata _containermatscalarparammap _containermatvectorparammap) (Utilities|Struct|BreakLacklusterMaterialsandTypes _value_2))
  (bind _returnvalue_5 (Components|GetOwner))
  ;[0.007784 0.1406 0.04891] BeginPlay: Apply chrome cosmetic to the owner and perform owner actor type specific functionality.
  (Variables|Default|SetbChromeMaterialsInitialized true)
  (if (Variables|Default|GetbHadEndPlay)
    _
    (elif _returnvalue_3
      (Class|BLackLusterFunctionLibrary|SetupChromeonActor _actortype_2_5b99d33c44fcfc26b361ee99b2bc0a4a _chrometype_5_68195aba4bf5bd23330635bc8ce95a46 _chromematerialstructs_9_a596659c4eb279eb1d6a5288d978425c _originalmaterialstructs_12_567c73ae46c386f2a6d4c9b35c9d5864 _nativeactortype_22_cfdbf9b1473c12413cfd3b947db2e001 _nativechrometype_23_860f664148debfe8fc0f50a5e7fefb6f _nativechromematerialstructs_21_eb1373e044d8f3e92f5f7c97e850fe26 _nativeoriginalmaterialstructs_29_3b92abd442db398f00adb889972847b8 _returnvalue_5 _containermaterialscalarparammap_34_88450d0546bb2f29ce9612ba6aaa1b20 _containermaterialvectorparammap_37_3436013e48082fbf8e5aaba2b0bd7901)
      (elif _returnvalue_4
        (Lackluster|NativeSetupChromeonActor _returnvalue_5 _actortype _chrometype _chromematerialdata _originalmaterialdata _containermatscalarparammap _containermatvectorparammap)
        (elif (Lackluster|UseNativeFunctions)
          (Chrome|NativeSetupChromeonActor (WorldSubsystems|GetChromeSubsystem_C) (Components|GetOwner))
          (else
            (Class|BLackLusterFunctionLibrary|SetupChromeonActor (Variables|Default|GetOwnerActorType) (Variables|Default|GetChromeType) (Variables|Default|GetChromeMaterialStruct) (Variables|Default|GetOriginalMaterialStruct) (Variables|Default|GetNativeActorType) (Variables|Default|GetNativeChromeType) (Variables|Default|GetNativeChromeMaterialStructs) (Variables|Default|GetNativeOriginalMaterialStructs) (Components|GetOwner) (Variables|Container|GetContainerMaterialScalarParamMap) (Variables|Container|GetContainerMaterialVectorParamMap))))))))

(event Custom|ApplyChrome
  (params (ChromeType :type B_Lackluster_ChromeTypes%20Enum :default "NewEnumerator1") (NativeChromeType :type ELacklusterChromeType%20Enum :default "Chrome"))
  (bind _returnvalue (Components|GetOwner))
  (bind _returnvalue_2 (WorldSubsystems|GetChromeSubsystem_C))
  (bind _actormaterialsandtypes (Variables|/ChromeShared/ChromeSubsystem.ChromeSubsystem_C|GetActorMaterialsAndTypes _returnvalue_2))
  (bind (_value _returnvalue_3) (Utilities|Map|Find _actormaterialsandtypes _returnvalue))
  (bind (_actortype_2_5b99d33c44fcfc26b361ee99b2bc0a4a _chrometype_5_68195aba4bf5bd23330635bc8ce95a46 _chromematerialstructs_9_a596659c4eb279eb1d6a5288d978425c _originalmaterialstructs_12_567c73ae46c386f2a6d4c9b35c9d5864 _nativeactortype_22_cfdbf9b1473c12413cfd3b947db2e001 _nativechrometype_23_860f664148debfe8fc0f50a5e7fefb6f _nativechromematerialstructs_21_eb1373e044d8f3e92f5f7c97e850fe26 _nativeoriginalmaterialstructs_29_3b92abd442db398f00adb889972847b8 _containermaterialscalarparammap_34_88450d0546bb2f29ce9612ba6aaa1b20 _containermaterialvectorparammap_37_3436013e48082fbf8e5aaba2b0bd7901) (Utilities|Struct|BreakBLacklusterMaterialsandTypes _value))
  (bind _lacklusterdata (Variables|/Script/ChromeSharedRuntime.FortChromeSubsystem|GetLacklusterData _returnvalue_2))
  (bind (_applychromeclasspath _chromeaudioclasspath _materialstoignore _setparameternames _classestoactortypes _chromematerialoverrides _actortotextures _tagquerytochrometypeoverrides _tagtoactortypes _genericchromematerial _noanimchestmaterialtags _chromeanimchestmaterial _chromenoanimchestmaterial _buildingchromeapplied _buildingsmactorhitsound _buildingsmactorpropdeathsound _buildingsmactornonpropdeathsound _buildingsmactorpropchromedsound _buildingsmactornonpropchromedsound _tieredchestclass _tieredchestparentclass _chromesoundcomponents _shadowproxymeshcomponentclass _materialsandtypes) (Utilities|Struct|BreakLacklusterData _lacklusterdata))
  (bind (_value_2 _returnvalue_4) (Utilities|Map|Find _materialsandtypes _returnvalue))
  (bind (_actortype _chrometype _chromematerialdata _originalmaterialdata _containermatscalarparammap _containermatvectorparammap) (Utilities|Struct|BreakLacklusterMaterialsandTypes _value_2))
  ;[0.1354 0.03408 0.1206] ApplyChrome: Applies chrome material and audio to the owner.
  (if (not (Variables|Default|GetbChromeMaterialsInitialized))
    (CallFunction|InitChromeMaterials))
  (if _returnvalue_3
    (Default|ApplyChromeToActor (Components|GetOwner) _actortype_2_5b99d33c44fcfc26b361ee99b2bc0a4a ChromeType _chromematerialstructs_9_a596659c4eb279eb1d6a5288d978425c _originalmaterialstructs_12_567c73ae46c386f2a6d4c9b35c9d5864 _nativeactortype_22_cfdbf9b1473c12413cfd3b947db2e001 NativeChromeType _nativechromematerialstructs_21_eb1373e044d8f3e92f5f7c97e850fe26 _nativeoriginalmaterialstructs_29_3b92abd442db398f00adb889972847b8 _containermaterialscalarparammap_34_88450d0546bb2f29ce9612ba6aaa1b20 _containermaterialvectorparammap_37_3436013e48082fbf8e5aaba2b0bd7901)
    (elif _returnvalue_4
      (Lackluster|NativeApplyChrometoActor (Components|GetOwner) _actortype NativeChromeType _chromematerialdata _originalmaterialdata _containermatscalarparammap _containermatvectorparammap)
      (else
        (Default|ApplyChromeToActor (Components|GetOwner) (Variables|Default|GetOwnerActorType) ChromeType (Variables|Default|GetChromeMaterialStruct) (Variables|Default|GetOriginalMaterialStruct) (Variables|Default|GetNativeActorType) NativeChromeType (Variables|Default|GetNativeChromeMaterialStructs) (Variables|Default|GetNativeOriginalMaterialStructs) (Variables|Container|GetContainerMaterialScalarParamMap) (Variables|Container|GetContainerMaterialVectorParamMap))))))

(event EventEndPlay
  (params (EndPlayReason :type EEndPlayReason%20Enum :default "Destroyed"))
  ;[0.05729 0.0353 0.004978] EndPlay: Run clean up logic based on the owner type.
  (|CallParentFunction EndPlayReason)
  (Networking|CanExecuteCosmeticEvents
    (:True
      (Class|BLackLusterFunctionLibrary|CleanupOwnerTypeSpecificFunctionality (Components|GetOwner) (Variables|Default|GetOwnerActorType))
      (Variables|Default|SetbHadEndPlay true))))

(event Custom|RestoreMaterials
  (CallFunction|RestoreOriginalMaterials))

(event Custom|DestroyMe
  (Components|DestroyComponent))

(fn GetOwnerActorType
  ;[0.02332 0.07696 0.09896] Returns the owner actor type.
  (return (Variables|Default|GetOwnerActorType)))

(fn RestoreOriginalMaterials
  (bind _returnvalue (Components|GetOwner))
  (bind (_value _returnvalue_2) (Utilities|Map|Find (Variables|/ChromeShared/ChromeSubsystem.ChromeSubsystem_C|GetChromeSoundComponents (WorldSubsystems|GetChromeSubsystem_C)) _returnvalue))
  (if (and (Utilities|IsValid _value :!class /Script/Engine.KismetSystemLibrary) _returnvalue_2)
    (Class|BLackLusterFunctionLibrary|SetChromeAudio _returnvalue false (Variables|Default|GetChromeType))
    (else
      _))
  (for _array_element (Variables|Default|GetNativeOriginalMaterialStructs)
    (bind (_array_element_meshcomponent _array_element_material _array_element_materialindex) (Utilities|Struct|BreakLacklusterOriginalMaterialData _array_element))
    (Utilities|IsValid _array_element_material
      (:Is%20Valid
        (Utilities|IsValid _array_element_meshcomponent
          (:Is%20Valid
            (Rendering|Material|SetMaterial _array_element_meshcomponent _array_element_materialindex _array_element_material))))))
  (bind _asbuilding_static_mesh (Utilities|Casting|CastToBuildingSMActor (Components|GetOwner)))
  (Variables|/Script/FortniteGame.BuildingSMActor|SetbNeedsMIDsForCreative false _asbuilding_static_mesh))
