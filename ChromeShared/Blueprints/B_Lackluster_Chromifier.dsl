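; Chrome tagging tool actor that adds or removes the chrome visual actor tag on actors level-wide or inside its mesh volume.
; Volume queries combine a component overlap on the mesh with Chrome subsystem results for overlapping BuildingFoundations.
; A preview event applies chrome materials to tagged actors in the volume through the Lackluster function library.
; Helper functions, uncalled here, add chrome components to non-static-mesh actors and editor-only chrome gameplay tags to building actors.
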(event Custom|RemoveChromeTagOnAllActorsInLevel
  (bind _outactors (Actor|GetAllActorsOfClass "/Script/Engine.Actor"))
  (CallFunction|RemoveActorTagonActors _outactors))

(event Custom|ApplyChromeTagToAllActorsInLevel
  (bind _outactors (Actor|GetAllActorsOfClass "/Script/Engine.Actor"))
  (CallFunction|ApplyActorTagtoActors _outactors))

(event Custom|ApplyChromeTagToActorsInVolume
  (bind _actors (CallFunction|GetOverlappingActors))
  (CallFunction|ApplyActorTagtoActors _actors))

(event Custom|RemoveChromeTagOnActorsInVolume
  (bind _actors (CallFunction|GetOverlappingActors))
  (CallFunction|RemoveActorTagonActors _actors))

(event Custom|PreviewChromeOnTaggedActorsInVolume
  (bind _emptyoriginalmaterialstructin (Variables|Default|GetEmptyOriginalMaterialStructIn))
  (bind _empty_chrome_material_struct_in (Variables|Default|GetEmpty%20Chrome%20Material%20Struct%20In))
  (bind _actors (CallFunction|GetOverlappingActors))
  (for _array_element _actors
    (if (Utilities|Array|ContainsItem (Variables|/Script/Engine.Actor|GetTags _array_element) (Variables|Default|GetChromeVisualActorTag))
      (Class|BLackLusterFunctionLibrary|ApplyChromeMaterial _array_element _empty_chrome_material_struct_in _emptyoriginalmaterialstructin)
      (Utilities|Array|Clear _empty_chrome_material_struct_in)
      (Utilities|Array|Clear _emptyoriginalmaterialstructin))))

(fn AddChromeComponentToNonSMActors
  (params (Actors :type Array%20of%20Actor%20Object%20References))
  (for _array_element Actors
    (Utilities|Casting|CastToBuildingSMActor _array_element
      (:CastFailed
        (Utilities|Casting|CastToStaticMeshActor _array_element
          (:CastFailed
            (Game|AddComponentbyClass _array_element "/ChromeShared/Blueprints/B_Lackluster_ApplyChrome.B_Lackluster_ApplyChrome_C")))))))

(fn ApplyActorTagToActors
  (params (Actors :type Array%20of%20Actor%20Object%20References))
  (for _array_element Actors
    (bind _output_get (LocalVariables|SetExistingActorTags (Variables|/Script/Engine.Actor|GetTags _array_element)))
    (Utilities|Array|AddUnique _output_get (Variables|Default|GetChromeVisualActorTag))
    (Variables|/Script/Engine.Actor|SetTags (LocalVariables|GetExistingActorTags) _array_element)))

(fn ApplyEditorOnlyTagToActors
  (params (Actors :type Array%20of%20Actor%20Object%20References))
  (for _array_element Actors
    (bind _asbuilding_static_mesh (Utilities|Casting|CastToBuildingSMActor _array_element))
    (Tags|AddEditorOnlyInstanceGameplayTags _asbuilding_static_mesh (GameplayTags|MakeLiteralGameplayTagContainer "(GameplayTags=((TagName=\"Gameplay.Status.Chrome\")))"))))

(fn GetOverlapping_Actors
  (bind _outactors (Collision|ComponentOverlapActors (Variables|Default|GetStaticMesh) (Transformation|GetWorldTransform (Variables|Default|GetStaticMesh)) (Utilities|Array|MakeArray "ObjectTypeQuery1" "ObjectTypeQuery2") "/Script/Engine.Actor" (Utilities|Array|MakeArray self)))
  (Chrome|GetOverlappingActorsOfClass (WorldSubsystems|GetChromeSubsystem_C) self "/Script/FortniteGame.BuildingFoundation" (LocalVariables|GetOverlappingActors))
  (Utilities|Array|AppendArray (LocalVariables|GetOverlappingActors) _outactors)
  (return (LocalVariables|GetOverlappingActors)))

(fn RemoveActorTagOnActors
  (params (Actors :type Array%20of%20Actor%20Object%20References))
  (for _array_element Actors
    (bind _output_get (LocalVariables|SetExistingActorTags (Variables|/Script/Engine.Actor|GetTags _array_element)))
    (bind _returnvalue (Utilities|Array|RemoveItem _output_get (Variables|Default|GetChromeVisualActorTag)))
    (if _returnvalue
      (Variables|/Script/Engine.Actor|SetTags (LocalVariables|GetExistingActorTags) _array_element))))

(fn ConstructionScript)
