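# Player animation layer Blueprint that plays the change into a Chrome blob, filling only the final pose override of the MovementMode layer group.
# A state machine enters through a conduit and picks jump/fall first, then locomotion above a 2D speed threshold, then idle as fallback.
# Jump and locomotion play blendspaces driven by movement direction, while idle plays an in-place entry sequence.
# The other layers and event graph hold no logic.
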AnimBlueprint /ChromeShared/Blueprints/ChromeBlobLayerAnimBP
target_skeleton=&Fortnite_M_Avg_Player_Skeleton
graphs [
 AnimGraph {nodes [Output Pose() {layerGroup=None}]}
 EventGraph {nodes [
   TryGetPawnOwner()
   Event BlueprintUpdateAnimation()
  ]}
 MovementMode_AimOffset {nodes [
   Input Pose() {name=InAimOffsetsPose}
   Input Pose() {name=InSourcePosePreAimOffsets}
   Output Pose() {layerGroup=MovementMode}
  ]}
 MovementMode_BaseLocomotion {nodes [
   Input Pose() {name=InPoseBaseLocomotion}
   Output Pose() {layerGroup=MovementMode}
  ]}
 MovementMode_FinalPoseOverride {nodes [
   Input Pose() {name=InFinalPose}
   AnimGraphNode_StateMachine_0 = ChangeIntoBlob() {is=title AnimGraphNode_StateMachine, bAllowConduitEntryStates=true}
   Output Pose(Result=@AnimGraphNode_StateMachine_0) {layerGroup=MovementMode}
  ]}
 ChangeIntoBlob {
  entry=Conduit
  path=MovementMode_FinalPoseOverride.AnimGraphNode_StateMachine_0.ChangeIntoBlob
  ids=id
  nodes [
   AnimStateConduitNode_0 = Conduit()
   AnimStateNode_0 = FromIdle() {is=title AnimStateNode}
   AnimStateNode_1 = FromLoco() {is=title AnimStateNode}
   AnimStateNode_2 = FromJump() {is=title AnimStateNode}
   AnimStateTransitionNode_0 = AnimStateTransitionNode() {logicType=TLT_Inertialization, priorityOrder=100}
   AnimStateTransitionNode_1 = AnimStateTransitionNode() {crossfadeDuration=0.3, logicType=TLT_Inertialization, priorityOrder=2}
   AnimStateTransitionNode_2 = AnimStateTransitionNode()
  ]
  links [
   {
    dst=FromIdle
    rule=MovementMode_FinalPoseOverride.AnimGraphNode_StateMachine_0.ChangeIntoBlob.AnimStateTransitionNode_0.Transition
    src=Conduit
    via=AnimStateTransitionNode_0
   }
   {
    dst=FromJump
    rule=MovementMode_FinalPoseOverride.AnimGraphNode_StateMachine_0.ChangeIntoBlob.AnimStateTransitionNode_2.Transition
    src=Conduit
    via=AnimStateTransitionNode_2
   }
   {
    dst=FromLoco
    rule=MovementMode_FinalPoseOverride.AnimGraphNode_StateMachine_0.ChangeIntoBlob.AnimStateTransitionNode_1.Transition
    src=Conduit
    via=AnimStateTransitionNode_1
   }
  ]
 }
 Conduit {
  path=MovementMode_FinalPoseOverride.AnimGraphNode_StateMachine_0.ChangeIntoBlob.AnimStateConduitNode_0.Conduit
  nodes [Result(bCanEnterTransition=true)]
 }
 FromIdle {
  path=MovementMode_FinalPoseOverride.AnimGraphNode_StateMachine_0.ChangeIntoBlob.AnimStateNode_0.FromIdle
  nodes [Output Animation Pose(Result=Sequence Player 'Chrome_Blob_player_entry_InPlace'() {groupName=IntoBlob, method=SyncGroup, sequence=/ChromeShared/Gameplay/PlayerAnimation/Chrome_Blob_player_entry_InPlace})]
 }
 FromLoco {
  path=MovementMode_FinalPoseOverride.AnimGraphNode_StateMachine_0.ChangeIntoBlob.AnimStateNode_1.FromLoco
  nodes [Output Animation Pose(Result=Blendspace Player 'Player_Loco_Blob_Into_BS'(X=Get PawnMovementDirectionAzimuth() -> (PawnMovementDirectionAzimuth:float), Y=0) {blendSpace=/ChromeShared/Gameplay/PlayerAnimation/Player_Loco_Blob_Into_BS})]
 }
 FromJump {
  path=MovementMode_FinalPoseOverride.AnimGraphNode_StateMachine_0.ChangeIntoBlob.AnimStateNode_2.FromJump
  nodes [Output Animation Pose(Result=Blendspace Player 'Player_Jump_Blob_Into_BS'(X=Get PawnMovementDirectionAzimuth(), Y=0) {blendSpace=/ChromeShared/Gameplay/PlayerAnimation/Player_Jump_Blob_Into_BS})]
 }
 Transition {
  path=MovementMode_FinalPoseOverride.AnimGraphNode_StateMachine_0.ChangeIntoBlob.AnimStateTransitionNode_0.Transition
  nodes [Result(bCanEnterTransition=true)]
 }
 Transition {
  path=MovementMode_FinalPoseOverride.AnimGraphNode_StateMachine_0.ChangeIntoBlob.AnimStateTransitionNode_1.Transition
  nodes [Result(bCanEnterTransition="float >= float"(A=Get PawnSpeed2D(), B=250))]
 }
 Transition {
  path=MovementMode_FinalPoseOverride.AnimGraphNode_StateMachine_0.ChangeIntoBlob.AnimStateTransitionNode_2.Transition
  nodes [Result(bCanEnterTransition=OR Boolean(A=Get bIsJumping(), B=Get bIsFalling()))]
 }
 MovementMode_FullBodyOverride {nodes [
   Input Pose() {name=InputPoseFullBody}
   Output Pose() {layerGroup=MovementMode}
  ]}
 MovementMode_InAirOverride {nodes [
   Input Pose() {name=InSource_Pose}
   Input Pose() {name=InAirPose}
   Output Pose() {layerGroup=MovementMode}
  ]}
 MovementMode_LocomotionAdditives {nodes [
   Input Pose() {name=InLocomotionAdditivesPose}
   Output Pose() {layerGroup=MovementMode}
  ]}
 MovementMode_LowerBodyOverride {nodes [
   Input Pose() {name=InLowerBodyPose}
   Input Pose() {name=InSourcePose}
   Output Pose() {layerGroup=MovementMode}
  ]}
 MovementMode_OffScreen {nodes [
   Input Pose() {name=InPoseOffscreen}
   Output Pose() {layerGroup=MovementMode}
  ]}
 MovementMode_SourcePose {nodes [
   Input Pose() {name=InSourcePoseBase}
   Output Pose() {layerGroup=MovementMode}
  ]}
]
variables []
