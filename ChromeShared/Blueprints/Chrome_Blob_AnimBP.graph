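# Animation Blueprint for the Chrome Blob gadget, animating the blob's locomotion, jumping, dashing, wall phasing, swimming and ziplining.
# A conduit-routed locomotion state machine selects states from bool variables, with nested machines for locomotion start, phasing, swim and zipline.
# An additive machine layers jump and landing poses on top, followed by DefaultSlot, root-only slope warping using movement-component floor info, and inertialization.
# Blendspaces take speed, lean, aim pitch and zipline input.
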AnimBlueprint /ChromeShared/Blueprints/Chrome_Blob_AnimBP
target_skeleton=/ChromeShared/Meshes/Chrome_Blob_Gadget_Skeleton
graphs [
 AnimGraph {nodes [
   AnimGraphNode_StateMachine_0 = LocomotionStateMachine() {is=title AnimGraphNode_StateMachine, bAllowConduitEntryStates=true}
   AnimGraphNode_StateMachine_1 = AdditiveStateMachine() {is=title AnimGraphNode_StateMachine, bAllowConduitEntryStates=true}
   n1 = Apply Additive(Base=@AnimGraphNode_StateMachine_0, Additive=@AnimGraphNode_StateMachine_1) {alphaBoolBlend {blendOption=Linear}}
   n2 = Quadruped Slope Warping!(ComponentPose:Component Space Pose Link Structure=Local To Component(LocalPose=Slot 'DefaultSlot'(Source=@n1)), Alpha:float=1) -> (Pose:Component Space Pose Link Structure) {
    is=title FortAnimGraphNode_QuadrupedSlopeWarping
    bOnlyAffectSkeletonRootBone=true
    bUseMovementComponentFloorInfoInsteadOfTraces=true
    alphaBoolBlend {blendOption=Linear}
    armsFeetDefinitions [
     {numBonesInLimb=2, fKFootBone {boneName=root}, iKFootBone {boneName=root}}
     {numBonesInLimb=2, fKFootBone {boneName=root}, iKFootBone {boneName=root}}
    ]
    armsIKFootRootBone {boneName=root}
    armsRootBone {boneName=root}
    feetDefinitions [{numBonesInLimb=2, fKFootBone {boneName=root}, iKFootBone {boneName=root}}]
    iKFootRootBone {boneName=root}
    pelvisBone {boneName=root}
    skeletonRootBone {boneName=root}
   }
   Output Pose(Result=Inertialization!(Source:Pose Link Structure=Component To Local(ComponentPose=@n2)) -> (Pose:Pose Link Structure) {is=title AnimGraphNode_Inertialization}) {layerGroup=None}
  ]}
 LocomotionStateMachine {
  entry=Conduit
  path=AnimGraph.AnimGraphNode_StateMachine_0.LocomotionStateMachine
  ids=id
  nodes [
   IsDash() {is=title AnimStateAliasNode, also {aliases [JumpLoop,JumpStart]}}
   AnimStateConduitNode_0 = Conduit()
   AnimStateNode_0 = Idle() {
    stateEntered {guid=#1, notifyName=EnterIdle}
    stateLeft {guid=#2, notifyName=ExitIdle}
   }
   AnimStateNode_10 = Locomotion() {is=title AnimStateNode}
   AnimStateNode_2 = LocomotionStop() {is=title AnimStateNode, stateEntered {guid=#3, notifyName=EnterStop}}
   AnimStateNode_3 = JumpStart() {is=title AnimStateNode, stateEntered {guid=#4, notifyName=EnterJump}}
   AnimStateNode_4 = JumpLoop() {
    is=title AnimStateNode
    stateEntered {guid=#5, notifyName=EnterJumpLoop}
    stateFullyBlended {guid=#6}
    stateLeft {guid=#7}
   }
   AnimStateNode_5 = Dash() {is=title AnimStateNode}
   AnimStateNode_6 = Phasing() {is=title AnimStateNode, stateEntered {guid=#8, notifyName=EnterPhasing}}
   AnimStateNode_7 = Swim() {
    is=title AnimStateNode
    stateEntered {guid=#9, notifyName=EnterSwim}
    stateLeft {guid=#10, notifyName=ExitSwim}
   }
   AnimStateNode_8 = ZipLine() {is=title AnimStateNode}
   AnimStateTransitionNode_0 = AnimStateTransitionNode() {logicType=TLT_Inertialization, priorityOrder=100}
   AnimStateTransitionNode_1 = AnimStateTransitionNode()
   AnimStateTransitionNode_10 = AnimStateTransitionNode() {logicType=TLT_Inertialization, priorityOrder=10}
   AnimStateTransitionNode_11 = AnimStateTransitionNode() {priorityOrder=6}
   AnimStateTransitionNode_12 = AnimStateTransitionNode()
   AnimStateTransitionNode_13 = AnimStateTransitionNode()
   AnimStateTransitionNode_14 = AnimStateTransitionNode() {crossfadeDuration=0.4, logicType=TLT_Inertialization, priorityOrder=7}
   AnimStateTransitionNode_15 = AnimStateTransitionNode()
   AnimStateTransitionNode_16 = AnimStateTransitionNode() {bAutomaticRuleBasedOnSequencePlayerInState=true}
   AnimStateTransitionNode_17 = AnimStateTransitionNode() {crossfadeDuration=0.35, logicType=TLT_Inertialization}
   AnimStateTransitionNode_18 = AnimStateTransitionNode()
   AnimStateTransitionNode_19 = AnimStateTransitionNode() {logicType=TLT_Inertialization, priorityOrder=4}
   AnimStateTransitionNode_2 = AnimStateTransitionNode()
   AnimStateTransitionNode_21 = AnimStateTransitionNode() {
    logicType=TLT_Inertialization
    transitionEnd {guid=#11}
    transitionInterrupt {guid=#12}
    transitionStart {guid=#13}
   }
   AnimStateTransitionNode_22 = AnimStateTransitionNode()
   AnimStateTransitionNode_3 = AnimStateTransitionNode() {crossfadeDuration=0.1, logicType=TLT_Inertialization, priorityOrder=2}
   AnimStateTransitionNode_4 = AnimStateTransitionNode() {logicType=TLT_Inertialization}
   AnimStateTransitionNode_5 = AnimStateTransitionNode() {bAutomaticRuleBasedOnSequencePlayerInState=true}
   AnimStateTransitionNode_6 = AnimStateTransitionNode() {crossfadeDuration=0.5, logicType=TLT_Inertialization, priorityOrder=100}
   AnimStateTransitionNode_7 = AnimStateTransitionNode() {crossfadeDuration=0.5, logicType=TLT_Inertialization, priorityOrder=3}
   AnimStateTransitionNode_8 = AnimStateTransitionNode() {
    transitionEnd {guid=#14}
    transitionInterrupt {guid=#15}
    transitionStart {guid=#16}
   }
   AnimStateTransitionNode_9 = AnimStateTransitionNode()
  ]
  links [
   {
    dst=Idle
    rule=AnimGraph.AnimGraphNode_StateMachine_0.LocomotionStateMachine.AnimStateTransitionNode_0.Transition
    src=Conduit
    via=AnimStateTransitionNode_0
   }
   {
    dst=JumpLoop
    rule=AnimGraph.AnimGraphNode_StateMachine_0.LocomotionStateMachine.AnimStateTransitionNode_7.Transition
    src=Conduit
    via=AnimStateTransitionNode_7
   }
   {
    dst=JumpStart
    rule=AnimGraph.AnimGraphNode_StateMachine_0.LocomotionStateMachine.AnimStateTransitionNode_3.Transition
    src=Conduit
    via=AnimStateTransitionNode_3
   }
   {
    dst=Locomotion
    rule=AnimGraph.AnimGraphNode_StateMachine_0.LocomotionStateMachine.AnimStateTransitionNode_19.Transition
    src=Conduit
    via=AnimStateTransitionNode_19
   }
   {
    dst=Phasing
    rule=AnimGraph.AnimGraphNode_StateMachine_0.LocomotionStateMachine.AnimStateTransitionNode_10.Transition
    src=Conduit
    via=AnimStateTransitionNode_10
   }
   {
    dst=Swim
    rule=AnimGraph.AnimGraphNode_StateMachine_0.LocomotionStateMachine.AnimStateTransitionNode_14.Transition
    src=Conduit
    via=AnimStateTransitionNode_14
   }
   {
    dst=ZipLine
    rule=AnimGraph.AnimGraphNode_StateMachine_0.LocomotionStateMachine.AnimStateTransitionNode_11.Transition
    src=Conduit
    via=AnimStateTransitionNode_11
   }
   {
    dst=Conduit
    rule=AnimGraph.AnimGraphNode_StateMachine_0.LocomotionStateMachine.AnimStateTransitionNode_12.Transition
    src=Dash
    via=AnimStateTransitionNode_12
   }
   {
    dst=Conduit
    rule=AnimGraph.AnimGraphNode_StateMachine_0.LocomotionStateMachine.AnimStateTransitionNode_1.Transition
    src=Idle
    via=AnimStateTransitionNode_1
   }
   {
    dst=Dash
    rule=AnimGraph.AnimGraphNode_StateMachine_0.LocomotionStateMachine.AnimStateTransitionNode_9.Transition
    src=IsDash
    via=AnimStateTransitionNode_9
   }
   {
    dst=Conduit
    rule=AnimGraph.AnimGraphNode_StateMachine_0.LocomotionStateMachine.AnimStateTransitionNode_8.Transition
    src=JumpLoop
    via=AnimStateTransitionNode_8
   }
   {
    dst=Conduit
    rule=AnimGraph.AnimGraphNode_StateMachine_0.LocomotionStateMachine.AnimStateTransitionNode_2.Transition
    src=JumpStart
    via=AnimStateTransitionNode_2
   }
   {
    dst=JumpLoop
    rule=AnimGraph.AnimGraphNode_StateMachine_0.LocomotionStateMachine.AnimStateTransitionNode_6.Transition
    src=JumpStart
    via=AnimStateTransitionNode_6
   }
   {
    dst=Conduit
    rule=AnimGraph.AnimGraphNode_StateMachine_0.LocomotionStateMachine.AnimStateTransitionNode_22.Transition
    src=Locomotion
    via=AnimStateTransitionNode_22
   }
   {
    dst=LocomotionStop
    rule=AnimGraph.AnimGraphNode_StateMachine_0.LocomotionStateMachine.AnimStateTransitionNode_21.Transition
    src=Locomotion
    via=AnimStateTransitionNode_21
   }
   {
    dst=Conduit
    rule=AnimGraph.AnimGraphNode_StateMachine_0.LocomotionStateMachine.AnimStateTransitionNode_18.Transition
    src=LocomotionStop
    via=AnimStateTransitionNode_18
   }
   {
    dst=Conduit
    rule=AnimGraph.AnimGraphNode_StateMachine_0.LocomotionStateMachine.AnimStateTransitionNode_5.Transition
    src=LocomotionStop
    via=AnimStateTransitionNode_5
   }
   {
    dst=Locomotion
    rule=AnimGraph.AnimGraphNode_StateMachine_0.LocomotionStateMachine.AnimStateTransitionNode_4.Transition
    src=LocomotionStop
    via=AnimStateTransitionNode_4
   }
   {
    dst=Conduit
    rule=AnimGraph.AnimGraphNode_StateMachine_0.LocomotionStateMachine.AnimStateTransitionNode_16.Transition
    src=Phasing
    via=AnimStateTransitionNode_16
   }
   {
    dst=JumpLoop
    rule=AnimGraph.AnimGraphNode_StateMachine_0.LocomotionStateMachine.AnimStateTransitionNode_17.Transition
    src=Phasing
    via=AnimStateTransitionNode_17
   }
   {
    dst=Conduit
    rule=AnimGraph.AnimGraphNode_StateMachine_0.LocomotionStateMachine.AnimStateTransitionNode_15.Transition
    src=Swim
    via=AnimStateTransitionNode_15
   }
   {
    dst=Conduit
    rule=AnimGraph.AnimGraphNode_StateMachine_0.LocomotionStateMachine.AnimStateTransitionNode_13.Transition
    src=ZipLine
    via=AnimStateTransitionNode_13
   }
  ]
 }
 Conduit {path=AnimGraph.AnimGraphNode_StateMachine_0.LocomotionStateMachine.AnimStateConduitNode_0.Conduit, nodes [Result(bCanEnterTransition=true)]}
 Idle {
  path=AnimGraph.AnimGraphNode_StateMachine_0.LocomotionStateMachine.AnimStateNode_0.Idle
  nodes [
   Blendspace Player 'Blob_Idle_BS'(X=Get PawnSpeed2D(), Y=0) {blendSpace=/ChromeShared/Gameplay/ANimation/Blob_Idle_BS, groupName=Locomotion, method=SyncGroup}
   Output Animation Pose(Result=Sequence Player 'blob_idle_loop_Chrome_Blob_Gadget'() {sequence=/ChromeShared/Gameplay/ANimation/blob_idle_loop_Chrome_Blob_Gadget})
  ]
 }
 Locomotion {
  path=AnimGraph.AnimGraphNode_StateMachine_0.LocomotionStateMachine.AnimStateNode_10.Locomotion
  nodes [
   AnimGraphNode_StateMachine_0 = LocomotionStateMachine() {bAllowConduitEntryStates=true}
   Output Animation Pose(Result=@AnimGraphNode_StateMachine_0)
  ]
 }
 LocomotionStateMachine {
  entry=Conduit
  path=AnimGraph.AnimGraphNode_StateMachine_0.LocomotionStateMachine.AnimStateNode_10.Locomotion.AnimGraphNode_StateMachine_0.LocomotionStateMachine
  ids=id
  nodes [
   AnimStateConduitNode_0 = Conduit()
   AnimStateNode_1 = Locomotion() {
    stateEntered {guid=#17}
    stateFullyBlended {guid=#18}
    stateLeft {guid=#19}
   }
   AnimStateNode_9 = LocomotionStart() {
    is=title AnimStateNode
    stateEntered {guid=#20, notifyName=EnterLocoStart}
    stateFullyBlended {guid=#21}
    stateLeft {guid=#22, notifyName=ExitLocoStart}
   }
   AnimStateTransitionNode_2 = AnimStateTransitionNode() {logicType=TLT_Inertialization}
   AnimStateTransitionNode_20 = AnimStateTransitionNode() {bAutomaticRuleBasedOnSequencePlayerInState=true, crossfadeDuration=0.3, logicType=TLT_Inertialization, priorityOrder=2}
   AnimStateTransitionNode_3 = AnimStateTransitionNode() {crossfadeDuration=0.4, logicType=TLT_Inertialization, priorityOrder=100}
  ]
  links [
   {
    dst=Locomotion
    rule=AnimGraph.AnimGraphNode_StateMachine_0.LocomotionStateMachine.AnimStateNode_10.Locomotion.AnimGraphNode_StateMachine_0.LocomotionStateMachine.AnimStateTransitionNode_3.Transition
    src=Conduit
    via=AnimStateTransitionNode_3
   }
   {
    dst=LocomotionStart
    rule=AnimGraph.AnimGraphNode_StateMachine_0.LocomotionStateMachine.AnimStateNode_10.Locomotion.AnimGraphNode_StateMachine_0.LocomotionStateMachine.AnimStateTransitionNode_2.Transition
    src=Conduit
    via=AnimStateTransitionNode_2
   }
   {
    dst=Locomotion
    rule=AnimGraph.AnimGraphNode_StateMachine_0.LocomotionStateMachine.AnimStateNode_10.Locomotion.AnimGraphNode_StateMachine_0.LocomotionStateMachine.AnimStateTransitionNode_20.AnimationTransitionGraph_0
    src=LocomotionStart
    via=AnimStateTransitionNode_20
   }
  ]
 }
 Conduit {
  path=AnimGraph.AnimGraphNode_StateMachine_0.LocomotionStateMachine.AnimStateNode_10.Locomotion.AnimGraphNode_StateMachine_0.LocomotionStateMachine.AnimStateConduitNode_0.Conduit
  nodes [Result(bCanEnterTransition=true)]
 }
 Locomotion {
  path=AnimGraph.AnimGraphNode_StateMachine_0.LocomotionStateMachine.AnimStateNode_10.Locomotion.AnimGraphNode_StateMachine_0.LocomotionStateMachine.AnimStateNode_1.Locomotion
  nodes [
   n1 = Get PawnSpeed2D()
   n2 = Get LeanAngle() -> (LeanAngle:float)
   n3 = Get SprintPlayRate() -> (SprintPlayRate:double)
   n4 = Get PawnSpeed2D()
   Output Animation Pose(Result=Blend Poses by bool(
     BlendPose_0=Blendspace Player 'Blob_Sprint_Loco_BS'!(X=@n2, Y=0, PlayRate:float=@n3) -> (Pose) {tag=Sprint, blendSpace=/ChromeShared/Gameplay/ANimation/Blob_Sprint_Loco_BS, groupName=Locomotion, method=SyncGroup}
     BlendPose_1=Blendspace Player 'Blob_Jog_Loco_BS'(X=@n1, Y=@n2) {blendSpace=/ChromeShared/Gameplay/ANimation/Locomotion/Blob_Jog_Loco_BS, groupName=Locomotion, method=SyncGroup}
     BlendTime_0=0.4
     BlendTime_1=0.4
     bActiveValue="float >= float"(A=@n4, B=625)
    ) {blendType=Linear, transitionType=Inertialization})
  ]
 }
 LocomotionStart {
  path=AnimGraph.AnimGraphNode_StateMachine_0.LocomotionStateMachine.AnimStateNode_10.Locomotion.AnimGraphNode_StateMachine_0.LocomotionStateMachine.AnimStateNode_9.LocomotionStart
  nodes [Output Animation Pose(Result=Blendspace Player 'Blob_Idle_To_Start_BS'(X=Get LocoStartDirection() -> (LocoStartDirection:double), Y=Get PawnSpeed2D()) {bLoop=false, blendSpace=/ChromeShared/Gameplay/ANimation/Blob_Idle_To_Start_BS})]
 }
 Transition {
  path=AnimGraph.AnimGraphNode_StateMachine_0.LocomotionStateMachine.AnimStateNode_10.Locomotion.AnimGraphNode_StateMachine_0.LocomotionStateMachine.AnimStateTransitionNode_2.Transition
  nodes [Result(bCanEnterTransition=Get EnterIdle() -> (EnterIdle:bool))]
 }
 AnimationTransitionGraph_0 {
  path=AnimGraph.AnimGraphNode_StateMachine_0.LocomotionStateMachine.AnimStateNode_10.Locomotion.AnimGraphNode_StateMachine_0.LocomotionStateMachine.AnimStateTransitionNode_20.AnimationTransitionGraph_0
  nodes [Result(bCanEnterTransition="float < float"(A="Time Remaining (ratio) (Blob_Idle_To_Start_BS)"!() -> (ReturnValue:float) {is=title K2Node_AnimGetter}, B=0.3))]
 }
 Transition {
  path=AnimGraph.AnimGraphNode_StateMachine_0.LocomotionStateMachine.AnimStateNode_10.Locomotion.AnimGraphNode_StateMachine_0.LocomotionStateMachine.AnimStateTransitionNode_3.Transition
  nodes [Result(bCanEnterTransition=true)]
 }
 LocomotionStop {
  path=AnimGraph.AnimGraphNode_StateMachine_0.LocomotionStateMachine.AnimStateNode_2.LocomotionStop
  nodes [Output Animation Pose(Result=Sequence Player 'Blob_Stop'() {sequence=/ChromeShared/Gameplay/ANimation/Blob_Stop})]
 }
 JumpStart {
  path=AnimGraph.AnimGraphNode_StateMachine_0.LocomotionStateMachine.AnimStateNode_3.JumpStart
  nodes [Output Animation Pose(Result=Blendspace Player 'Blob_Jump_Start_BS'(X=Get PawnSpeed2D(), Y=0) {blendSpace=/ChromeShared/Gameplay/ANimation/Blob_Jump_Start_BS})]
 }
 JumpLoop {
  path=AnimGraph.AnimGraphNode_StateMachine_0.LocomotionStateMachine.AnimStateNode_4.JumpLoop
  nodes [Output Animation Pose(Result=Blendspace Player 'Blob_Jump_Loop_BS'(X=Get PawnSpeed2D(), Y=0) {blendSpace=/ChromeShared/Gameplay/ANimation/Blob_Jump_Loop_BS})]
 }
 Dash {
  path=AnimGraph.AnimGraphNode_StateMachine_0.LocomotionStateMachine.AnimStateNode_5.Dash
  nodes [
   n1 = Get AimPitch()
   Output Animation Pose(Result=Apply Additive(
     Base=Sequence Player 'blob_dash'() {bLoopAnimation=false, sequence=/ChromeShared/Gameplay/ANimation/Dash/blob_dash}
     Additive=Blendspace Player 'Blob_Dash_Direction_BS'(X=@n1, Y=0) {blendSpace=/ChromeShared/Gameplay/ANimation/Dash/Blob_Dash_Direction_BS}
     Alpha=Get DashPitchAlpha() -> (DashPitchAlpha:float)
    ) {alphaBoolBlend {blendOption=Linear}})
  ]
 }
 Phasing {
  path=AnimGraph.AnimGraphNode_StateMachine_0.LocomotionStateMachine.AnimStateNode_6.Phasing
  nodes [
   AnimGraphNode_StateMachine_0 = PhaseStateMachine() {is=title AnimGraphNode_StateMachine, bAllowConduitEntryStates=true}
   Output Animation Pose(Result=@AnimGraphNode_StateMachine_0)
  ]
 }
 PhaseStateMachine {
  entry=Conduit
  path=AnimGraph.AnimGraphNode_StateMachine_0.LocomotionStateMachine.AnimStateNode_6.Phasing.AnimGraphNode_StateMachine_0.PhaseStateMachine
  ids=id
  nodes [
   AnimStateConduitNode_0 = Conduit()
   AnimStateNode_0 = PhaseOnGround() {is=title AnimStateNode}
   AnimStateNode_1 = PhaseInAir() {is=title AnimStateNode}
   AnimStateTransitionNode_0 = AnimStateTransitionNode() {logicType=TLT_Inertialization, priorityOrder=100}
   AnimStateTransitionNode_1 = AnimStateTransitionNode() {logicType=TLT_Inertialization}
  ]
  links [
   {
    dst=PhaseInAir
    rule=AnimGraph.AnimGraphNode_StateMachine_0.LocomotionStateMachine.AnimStateNode_6.Phasing.AnimGraphNode_StateMachine_0.PhaseStateMachine.AnimStateTransitionNode_1.Transition
    src=Conduit
    via=AnimStateTransitionNode_1
   }
   {
    dst=PhaseOnGround
    rule=AnimGraph.AnimGraphNode_StateMachine_0.LocomotionStateMachine.AnimStateNode_6.Phasing.AnimGraphNode_StateMachine_0.PhaseStateMachine.AnimStateTransitionNode_0.Transition
    src=Conduit
    via=AnimStateTransitionNode_0
   }
  ]
 }
 Conduit {
  path=AnimGraph.AnimGraphNode_StateMachine_0.LocomotionStateMachine.AnimStateNode_6.Phasing.AnimGraphNode_StateMachine_0.PhaseStateMachine.AnimStateConduitNode_0.Conduit
  nodes [Result(bCanEnterTransition=true)]
 }
 PhaseOnGround {
  path=AnimGraph.AnimGraphNode_StateMachine_0.LocomotionStateMachine.AnimStateNode_6.Phasing.AnimGraphNode_StateMachine_0.PhaseStateMachine.AnimStateNode_0.PhaseOnGround
  nodes [Output Animation Pose(Result=Sequence Player 'wall_phase_Chrome_Blob_Gadget'() {bLoopAnimation=false, playRate=0.6, sequence=/ChromeShared/Gameplay/ANimation/wall_phase_Chrome_Blob_Gadget})]
 }
 PhaseInAir {
  path=AnimGraph.AnimGraphNode_StateMachine_0.LocomotionStateMachine.AnimStateNode_6.Phasing.AnimGraphNode_StateMachine_0.PhaseStateMachine.AnimStateNode_1.PhaseInAir
  nodes [Output Animation Pose(Result=Sequence Player 'blob_dash_phase_end'() {bLoopAnimation=false, sequence=/ChromeShared/Gameplay/ANimation/blob_dash_phase_end})]
 }
 Transition {
  path=AnimGraph.AnimGraphNode_StateMachine_0.LocomotionStateMachine.AnimStateNode_6.Phasing.AnimGraphNode_StateMachine_0.PhaseStateMachine.AnimStateTransitionNode_0.Transition
  nodes [Result(bCanEnterTransition=true)]
 }
 Transition {
  path=AnimGraph.AnimGraphNode_StateMachine_0.LocomotionStateMachine.AnimStateNode_6.Phasing.AnimGraphNode_StateMachine_0.PhaseStateMachine.AnimStateTransitionNode_1.Transition
  nodes [Result(bCanEnterTransition=Get IsPhaseInAir() -> (IsPhaseInAir:bool))]
 }
 Swim {
  path=AnimGraph.AnimGraphNode_StateMachine_0.LocomotionStateMachine.AnimStateNode_7.Swim
  nodes [
   AnimGraphNode_StateMachine_0 = SwimStateMachine() {is=title AnimGraphNode_StateMachine, bAllowConduitEntryStates=true}
   Output Animation Pose(Result=@AnimGraphNode_StateMachine_0)
  ]
 }
 SwimStateMachine {
  entry=Conduit
  path=AnimGraph.AnimGraphNode_StateMachine_0.LocomotionStateMachine.AnimStateNode_7.Swim.AnimGraphNode_StateMachine_0.SwimStateMachine
  ids=id
  nodes [
   Alias() {is=title AnimStateAliasNode, also {aliases [SwimDash, SwimJumpIdle, SwimJumpIdleLoop, SwimJumpLand, SwimJumpLoco, SwimJumpLoop, SwimLoco]}}
   AnimStateConduitNode_0 = Conduit()
   AnimStateNode_0 = SwimDash() {is=title AnimStateNode}
   AnimStateNode_1 = SwimJumpLoco() {is=title AnimStateNode}
   AnimStateNode_2 = SwimLoco() {is=title AnimStateNode}
   AnimStateNode_3 = SwimJumpLoop() {is=title AnimStateNode}
   AnimStateNode_4 = SwimJumpLand() {is=title AnimStateNode}
   AnimStateNode_5 = SwimJumpIdle() {is=title AnimStateNode}
   AnimStateNode_6 = SwimJumpIdleLoop() {is=title AnimStateNode}
   AnimStateTransitionNode_0 = AnimStateTransitionNode() {crossfadeDuration=0.25, logicType=TLT_Inertialization, priorityOrder=4}
   AnimStateTransitionNode_1 = AnimStateTransitionNode()
   AnimStateTransitionNode_10 = AnimStateTransitionNode() {bAutomaticRuleBasedOnSequencePlayerInState=true}
   AnimStateTransitionNode_11 = AnimStateTransitionNode() {crossfadeDuration=0.1, logicType=TLT_Inertialization, priorityOrder=4}
   AnimStateTransitionNode_12 = AnimStateTransitionNode() {bAutomaticRuleBasedOnSequencePlayerInState=true, crossfadeDuration=0.3, logicType=TLT_Inertialization}
   AnimStateTransitionNode_2 = AnimStateTransitionNode()
   AnimStateTransitionNode_3 = AnimStateTransitionNode() {crossfadeDuration=0.4, logicType=TLT_Inertialization, priorityOrder=100}
   AnimStateTransitionNode_4 = AnimStateTransitionNode()
   AnimStateTransitionNode_5 = AnimStateTransitionNode() {crossfadeDuration=0.4, logicType=TLT_Inertialization, priorityOrder=3}
   AnimStateTransitionNode_6 = AnimStateTransitionNode() {crossfadeDuration=0.5, logicType=TLT_Inertialization, priorityOrder=2}
   AnimStateTransitionNode_7 = AnimStateTransitionNode() {bAutomaticRuleBasedOnSequencePlayerInState=true, crossfadeDuration=0.4, logicType=TLT_Inertialization}
   AnimStateTransitionNode_8 = AnimStateTransitionNode()
   AnimStateTransitionNode_9 = AnimStateTransitionNode() {crossfadeDuration=0.3, logicType=TLT_Inertialization}
  ]
  links [
   {
    dst=SwimDash
    rule=AnimGraph.AnimGraphNode_StateMachine_0.LocomotionStateMachine.AnimStateNode_7.Swim.AnimGraphNode_StateMachine_0.SwimStateMachine.AnimStateTransitionNode_0.Transition
    src=Alias
    via=AnimStateTransitionNode_0
   }
   {
    dst=SwimJumpIdle
    rule=AnimGraph.AnimGraphNode_StateMachine_0.LocomotionStateMachine.AnimStateNode_7.Swim.AnimGraphNode_StateMachine_0.SwimStateMachine.AnimStateTransitionNode_11.Transition
    src=Conduit
    via=AnimStateTransitionNode_11
   }
   {
    dst=SwimJumpIdleLoop
    rule=AnimGraph.AnimGraphNode_StateMachine_0.LocomotionStateMachine.AnimStateNode_7.Swim.AnimGraphNode_StateMachine_0.SwimStateMachine.AnimStateTransitionNode_5.Transition
    src=Conduit
    via=AnimStateTransitionNode_5
   }
   {
    dst=SwimJumpLoco
    rule=AnimGraph.AnimGraphNode_StateMachine_0.LocomotionStateMachine.AnimStateNode_7.Swim.AnimGraphNode_StateMachine_0.SwimStateMachine.AnimStateTransitionNode_1.Transition
    src=Conduit
    via=AnimStateTransitionNode_1
   }
   {
    dst=SwimJumpLoop
    rule=AnimGraph.AnimGraphNode_StateMachine_0.LocomotionStateMachine.AnimStateNode_7.Swim.AnimGraphNode_StateMachine_0.SwimStateMachine.AnimStateTransitionNode_6.Transition
    src=Conduit
    via=AnimStateTransitionNode_6
   }
   {
    dst=SwimLoco
    rule=AnimGraph.AnimGraphNode_StateMachine_0.LocomotionStateMachine.AnimStateNode_7.Swim.AnimGraphNode_StateMachine_0.SwimStateMachine.AnimStateTransitionNode_3.Transition
    src=Conduit
    via=AnimStateTransitionNode_3
   }
   {
    dst=Conduit
    rule=AnimGraph.AnimGraphNode_StateMachine_0.LocomotionStateMachine.AnimStateNode_7.Swim.AnimGraphNode_StateMachine_0.SwimStateMachine.AnimStateTransitionNode_2.Transition
    src=SwimDash
    via=AnimStateTransitionNode_2
   }
   {
    dst=SwimJumpIdleLoop
    rule=AnimGraph.AnimGraphNode_StateMachine_0.LocomotionStateMachine.AnimStateNode_7.Swim.AnimGraphNode_StateMachine_0.SwimStateMachine.AnimStateTransitionNode_12.Transition
    src=SwimJumpIdle
    via=AnimStateTransitionNode_12
   }
   {
    dst=Conduit
    rule=AnimGraph.AnimGraphNode_StateMachine_0.LocomotionStateMachine.AnimStateNode_7.Swim.AnimGraphNode_StateMachine_0.SwimStateMachine.AnimStateTransitionNode_8.Transition
    src=SwimJumpIdleLoop
    via=AnimStateTransitionNode_8
   }
   {
    dst=Conduit
    rule=AnimGraph.AnimGraphNode_StateMachine_0.LocomotionStateMachine.AnimStateNode_7.Swim.AnimGraphNode_StateMachine_0.SwimStateMachine.AnimStateTransitionNode_10.Transition
    src=SwimJumpLand
    via=AnimStateTransitionNode_10
   }
   {
    dst=SwimJumpLoop
    rule=AnimGraph.AnimGraphNode_StateMachine_0.LocomotionStateMachine.AnimStateNode_7.Swim.AnimGraphNode_StateMachine_0.SwimStateMachine.AnimStateTransitionNode_7.Transition
    src=SwimJumpLoco
    via=AnimStateTransitionNode_7
   }
   {
    dst=SwimJumpLand
    rule=AnimGraph.AnimGraphNode_StateMachine_0.LocomotionStateMachine.AnimStateNode_7.Swim.AnimGraphNode_StateMachine_0.SwimStateMachine.AnimStateTransitionNode_9.Transition
    src=SwimJumpLoop
    via=AnimStateTransitionNode_9
   }
   {
    dst=Conduit
    rule=AnimGraph.AnimGraphNode_StateMachine_0.LocomotionStateMachine.AnimStateNode_7.Swim.AnimGraphNode_StateMachine_0.SwimStateMachine.AnimStateTransitionNode_4.Transition
    src=SwimLoco
    via=AnimStateTransitionNode_4
   }
  ]
 }
 Conduit {
  path=AnimGraph.AnimGraphNode_StateMachine_0.LocomotionStateMachine.AnimStateNode_7.Swim.AnimGraphNode_StateMachine_0.SwimStateMachine.AnimStateConduitNode_0.Conduit
  nodes [Result(bCanEnterTransition=true)]
 }
 SwimDash {
  path=AnimGraph.AnimGraphNode_StateMachine_0.LocomotionStateMachine.AnimStateNode_7.Swim.AnimGraphNode_StateMachine_0.SwimStateMachine.AnimStateNode_0.SwimDash
  nodes [
   n1 = Get AimPitch()
   Output Animation Pose(Result=Apply Additive(
     Base=Sequence Player 'blob_dash'() {bLoopAnimation=false, sequence=&blob_dash}
     Additive=Blendspace Player 'Blob_Dash_Direction_BS'(X=@n1, Y=0) {blendSpace=&Blob_Dash_Direction_BS}
     Alpha=Get DashPitchAlpha()
    ) {alphaBoolBlend {blendOption=Linear}})
  ]
 }
 SwimJumpLoco {
  path=AnimGraph.AnimGraphNode_StateMachine_0.LocomotionStateMachine.AnimStateNode_7.Swim.AnimGraphNode_StateMachine_0.SwimStateMachine.AnimStateNode_1.SwimJumpLoco
  nodes [Output Animation Pose(Result=Sequence Player 'Blob_Swim_Jump_Into'() {sequence=/ChromeShared/Gameplay/ANimation/Blob_Swim_Jump_Into})]
 }
 SwimLoco {
  path=AnimGraph.AnimGraphNode_StateMachine_0.LocomotionStateMachine.AnimStateNode_7.Swim.AnimGraphNode_StateMachine_0.SwimStateMachine.AnimStateNode_2.SwimLoco
  nodes [
   n1 = Get PawnSpeed2D()
   n2 = Get LeanAngle()
   Output Animation Pose(Result=Blend Poses by bool(
     BlendPose_0=Blendspace Player 'Blob_Swim_Loco_BS'(X=@n1, Y=@n2) {blendSpace=/ChromeShared/Gameplay/ANimation/Blob_Swim_Loco_BS}
     BlendPose_1=Sequence Player 'blob_swim_idle_Chrome_Blob_Gadget'() {sequence=/ChromeShared/Gameplay/ANimation/blob_swim_idle_Chrome_Blob_Gadget}
     BlendTime_0=0.8
     BlendTime_1=0.5
     bActiveValue=Get IsLoco() -> (IsLoco:bool)
    ) {blendType=Linear, transitionType=Inertialization})
  ]
 }
 SwimJumpLoop {
  path=AnimGraph.AnimGraphNode_StateMachine_0.LocomotionStateMachine.AnimStateNode_7.Swim.AnimGraphNode_StateMachine_0.SwimStateMachine.AnimStateNode_3.SwimJumpLoop
  nodes [Output Animation Pose(Result=Sequence Player 'Blob_Swim_Jump_Loop'() {sequence=/ChromeShared/Gameplay/ANimation/Blob_Swim_Jump_Loop})]
 }
 SwimJumpLand {
  path=AnimGraph.AnimGraphNode_StateMachine_0.LocomotionStateMachine.AnimStateNode_7.Swim.AnimGraphNode_StateMachine_0.SwimStateMachine.AnimStateNode_4.SwimJumpLand
  nodes [Output Animation Pose(Result=Sequence Player 'Blob_Swim_Jump_Land'() {sequence=/ChromeShared/Gameplay/ANimation/Blob_Swim_Jump_Land})]
 }
 SwimJumpIdle {
  path=AnimGraph.AnimGraphNode_StateMachine_0.LocomotionStateMachine.AnimStateNode_7.Swim.AnimGraphNode_StateMachine_0.SwimStateMachine.AnimStateNode_5.SwimJumpIdle
  nodes [Output Animation Pose(Result=Sequence Player 'blob_jump_into_Chrome_Blob_Gadget'() {bLoopAnimation=false, sequence=/ChromeShared/Gameplay/ANimation/blob_jump_into_Chrome_Blob_Gadget})]
 }
 SwimJumpIdleLoop {
  path=AnimGraph.AnimGraphNode_StateMachine_0.LocomotionStateMachine.AnimStateNode_7.Swim.AnimGraphNode_StateMachine_0.SwimStateMachine.AnimStateNode_6.SwimJumpIdleLoop
  nodes [Output Animation Pose(Result=Sequence Player 'blob_jump_Loop_Chrome_Blob_Gadget'() {sequence=/ChromeShared/Gameplay/ANimation/blob_jump_Loop_Chrome_Blob_Gadget})]
 }
 Transition {
  path=AnimGraph.AnimGraphNode_StateMachine_0.LocomotionStateMachine.AnimStateNode_7.Swim.AnimGraphNode_StateMachine_0.SwimStateMachine.AnimStateTransitionNode_0.Transition
  nodes [Result(bCanEnterTransition=Get IsDashing() -> (IsDashing:bool))]
 }
 Transition {
  path=AnimGraph.AnimGraphNode_StateMachine_0.LocomotionStateMachine.AnimStateNode_7.Swim.AnimGraphNode_StateMachine_0.SwimStateMachine.AnimStateTransitionNode_1.Transition
  nodes [Result(bCanEnterTransition=Get IsSwimJumpLoco() -> (IsSwimJumpLoco:bool))]
 }
 Transition {
  path=AnimGraph.AnimGraphNode_StateMachine_0.LocomotionStateMachine.AnimStateNode_7.Swim.AnimGraphNode_StateMachine_0.SwimStateMachine.AnimStateTransitionNode_10.Transition
  nodes [Result()]
 }
 Transition {
  path=AnimGraph.AnimGraphNode_StateMachine_0.LocomotionStateMachine.AnimStateNode_7.Swim.AnimGraphNode_StateMachine_0.SwimStateMachine.AnimStateTransitionNode_11.Transition
  nodes [Result(bCanEnterTransition=Get IsSwimJumpIdle() -> (IsSwimJumpIdle:bool))]
 }
 Transition {
  path=AnimGraph.AnimGraphNode_StateMachine_0.LocomotionStateMachine.AnimStateNode_7.Swim.AnimGraphNode_StateMachine_0.SwimStateMachine.AnimStateTransitionNode_12.Transition
  nodes [Result()]
 }
 Transition {
  path=AnimGraph.AnimGraphNode_StateMachine_0.LocomotionStateMachine.AnimStateNode_7.Swim.AnimGraphNode_StateMachine_0.SwimStateMachine.AnimStateTransitionNode_2.Transition
  nodes [Result(bCanEnterTransition=NOT Boolean(A=Get IsDashing()))]
 }
 Transition {
  path=AnimGraph.AnimGraphNode_StateMachine_0.LocomotionStateMachine.AnimStateNode_7.Swim.AnimGraphNode_StateMachine_0.SwimStateMachine.AnimStateTransitionNode_3.Transition
  nodes [Result(bCanEnterTransition=true)]
 }
 Transition {
  path=AnimGraph.AnimGraphNode_StateMachine_0.LocomotionStateMachine.AnimStateNode_7.Swim.AnimGraphNode_StateMachine_0.SwimStateMachine.AnimStateTransitionNode_4.Transition
  nodes [Result(bCanEnterTransition=true)]
 }
 Transition {
  path=AnimGraph.AnimGraphNode_StateMachine_0.LocomotionStateMachine.AnimStateNode_7.Swim.AnimGraphNode_StateMachine_0.SwimStateMachine.AnimStateTransitionNode_5.Transition
  nodes [Result(bCanEnterTransition=Get IsSwimFallIdle() -> (IsSwimFallIdle:bool))]
 }
 Transition {
  path=AnimGraph.AnimGraphNode_StateMachine_0.LocomotionStateMachine.AnimStateNode_7.Swim.AnimGraphNode_StateMachine_0.SwimStateMachine.AnimStateTransitionNode_6.Transition
  nodes [Result(bCanEnterTransition=Get IsSwimFallLoco() -> (IsSwimFallLoco:bool))]
 }
 Transition {
  path=AnimGraph.AnimGraphNode_StateMachine_0.LocomotionStateMachine.AnimStateNode_7.Swim.AnimGraphNode_StateMachine_0.SwimStateMachine.AnimStateTransitionNode_7.Transition
  nodes [Result()]
 }
 Transition {
  path=AnimGraph.AnimGraphNode_StateMachine_0.LocomotionStateMachine.AnimStateNode_7.Swim.AnimGraphNode_StateMachine_0.SwimStateMachine.AnimStateTransitionNode_8.Transition
  nodes [Result(bCanEnterTransition=Get IsJumpLand() -> (IsJumpLand:bool))]
 }
 Transition {
  path=AnimGraph.AnimGraphNode_StateMachine_0.LocomotionStateMachine.AnimStateNode_7.Swim.AnimGraphNode_StateMachine_0.SwimStateMachine.AnimStateTransitionNode_9.Transition
  nodes [
   Result(bCanEnterTransition=Get IsJumpLand())
   AND Boolean(A=NOT Boolean(A=Get bIsFalling()), B=NOT Boolean(A=Get bIsJumping()))
  ]
 }
 ZipLine {
  path=AnimGraph.AnimGraphNode_StateMachine_0.LocomotionStateMachine.AnimStateNode_8.ZipLine
  nodes [
   AnimGraphNode_StateMachine_0 = ZiplineStatemachine() {is=title AnimGraphNode_StateMachine}
   Output Animation Pose(Result=@AnimGraphNode_StateMachine_0)
  ]
 }
 ZiplineStatemachine {
  entry=ZipEnter
  path=AnimGraph.AnimGraphNode_StateMachine_0.LocomotionStateMachine.AnimStateNode_8.ZipLine.AnimGraphNode_StateMachine_0.ZiplineStatemachine
  ids=id
  nodes [
   AnimStateNode_0 = ZipEnter() {is=title AnimStateNode}
   AnimStateNode_1 = ZipLoco() {is=title AnimStateNode}
   AnimStateTransitionNode_1 = AnimStateTransitionNode() {bAutomaticRuleBasedOnSequencePlayerInState=true, crossfadeDuration=0.3, logicType=TLT_Inertialization}
  ]
  links [{
    dst=ZipLoco
    rule=AnimGraph.AnimGraphNode_StateMachine_0.LocomotionStateMachine.AnimStateNode_8.ZipLine.AnimGraphNode_StateMachine_0.ZiplineStatemachine.AnimStateTransitionNode_1.Transition
    src=ZipEnter
    via=AnimStateTransitionNode_1
   }]
 }
 ZipEnter {
  path=AnimGraph.AnimGraphNode_StateMachine_0.LocomotionStateMachine.AnimStateNode_8.ZipLine.AnimGraphNode_StateMachine_0.ZiplineStatemachine.AnimStateNode_0.ZipEnter
  nodes [Output Animation Pose(Result=Sequence Player 'blob_zipline_loop_entry_Chrome_Blob_Gadget'() {playRateBasis=1.3, sequence=/ChromeShared/Gameplay/ANimation/blob_zipline_loop_entry_Chrome_Blob_Gadget})]
 }
 ZipLoco {
  path=AnimGraph.AnimGraphNode_StateMachine_0.LocomotionStateMachine.AnimStateNode_8.ZipLine.AnimGraphNode_StateMachine_0.ZiplineStatemachine.AnimStateNode_1.ZipLoco
  nodes [
   n1 = Get Zipline Input Lean Yaw() -> (Zipline Input Lean Yaw:float)
   n2 = Get PawnSpeed2D()
   Output Animation Pose(Result=Apply Additive(
     Base=Blendspace Player 'Blob_Zipline_BS'!(X=@n2, Y=0, PlayRate:float=1) -> (Pose) {blendSpace=/ChromeShared/Gameplay/ANimation/Blob_Zipline_BS}
     Additive=Blendspace Player 'Blob_Zipline_Lean_BS'(X=@n1, Y=@n2) {blendSpace=/ChromeShared/Gameplay/ANimation/Blob_Zipline_Lean_BS}
    ) {alphaBoolBlend {blendOption=Linear}})
  ]
 }
 Transition {
  path=AnimGraph.AnimGraphNode_StateMachine_0.LocomotionStateMachine.AnimStateNode_8.ZipLine.AnimGraphNode_StateMachine_0.ZiplineStatemachine.AnimStateTransitionNode_1.Transition
  nodes [Result()]
 }
 Transition {
  path=AnimGraph.AnimGraphNode_StateMachine_0.LocomotionStateMachine.AnimStateTransitionNode_0.Transition
  nodes [Result(bCanEnterTransition=true)]
 }
 Transition {
  path=AnimGraph.AnimGraphNode_StateMachine_0.LocomotionStateMachine.AnimStateTransitionNode_1.Transition
  nodes [Result(bCanEnterTransition=true)]
 }
 Transition {
  path=AnimGraph.AnimGraphNode_StateMachine_0.LocomotionStateMachine.AnimStateTransitionNode_10.Transition
  nodes [Result(bCanEnterTransition=Get IsPhasing() -> (IsPhasing:bool))]
 }
 Transition {
  path=AnimGraph.AnimGraphNode_StateMachine_0.LocomotionStateMachine.AnimStateTransitionNode_11.Transition
  nodes [Result(bCanEnterTransition=Get IsZiplining() -> (IsZiplining:bool))]
 }
 Transition {
  path=AnimGraph.AnimGraphNode_StateMachine_0.LocomotionStateMachine.AnimStateTransitionNode_12.Transition
  nodes [Result(bCanEnterTransition=NOT Boolean(A=Get IsDashing()))]
 }
 Transition {
  path=AnimGraph.AnimGraphNode_StateMachine_0.LocomotionStateMachine.AnimStateTransitionNode_13.Transition
  nodes [Result(bCanEnterTransition=NOT Boolean(A=Get IsZiplining()))]
 }
 Transition {
  path=AnimGraph.AnimGraphNode_StateMachine_0.LocomotionStateMachine.AnimStateTransitionNode_14.Transition
  nodes [Result(bCanEnterTransition=Get IsBlobSwim() -> (IsBlobSwim:bool))]
 }
 Transition {
  path=AnimGraph.AnimGraphNode_StateMachine_0.LocomotionStateMachine.AnimStateTransitionNode_15.Transition
  nodes [Result(bCanEnterTransition=true)]
 }
 Transition {
  path=AnimGraph.AnimGraphNode_StateMachine_0.LocomotionStateMachine.AnimStateTransitionNode_16.Transition
  nodes [Result(bCanEnterTransition=NOT Boolean(A=Get IsPhasing()))]
 }
 Transition {
  path=AnimGraph.AnimGraphNode_StateMachine_0.LocomotionStateMachine.AnimStateTransitionNode_17.Transition
  nodes [Result(bCanEnterTransition=Get IsPhaseFall() -> (IsPhaseFall:bool))]
 }
 Transition {
  path=AnimGraph.AnimGraphNode_StateMachine_0.LocomotionStateMachine.AnimStateTransitionNode_18.Transition
  nodes [Result(bCanEnterTransition=Get IsLocomotionExit() -> (IsLocomotionExit:bool))]
 }
 Transition {
  path=AnimGraph.AnimGraphNode_StateMachine_0.LocomotionStateMachine.AnimStateTransitionNode_19.Transition
  nodes [Result(bCanEnterTransition=Get IsLocomotionEnter() -> (IsLocomotionEnter:bool))]
 }
 Transition {
  path=AnimGraph.AnimGraphNode_StateMachine_0.LocomotionStateMachine.AnimStateTransitionNode_2.Transition
  nodes [Result(bCanEnterTransition=Get IsGroundJumpExit() -> (IsGroundJumpExit:bool))]
 }
 Transition {
  path=AnimGraph.AnimGraphNode_StateMachine_0.LocomotionStateMachine.AnimStateTransitionNode_21.Transition
  nodes [Result(bCanEnterTransition=NOT Boolean(A=Get IsLoco()))]
 }
 Transition {
  path=AnimGraph.AnimGraphNode_StateMachine_0.LocomotionStateMachine.AnimStateTransitionNode_22.Transition
  nodes [
   Result(bCanEnterTransition=Get IsLocomotionExit())
   OR Boolean!(A=Get IsZiplining(), B=Get IsSwimming() -> (IsSwimming:bool), C:bool=Get bIsJumping(), D:bool=Get bIsFalling(), E:bool) -> (ReturnValue)
  ]
 }
 Transition {
  path=AnimGraph.AnimGraphNode_StateMachine_0.LocomotionStateMachine.AnimStateTransitionNode_3.Transition
  nodes [Result(bCanEnterTransition=Get IsGroundJump() -> (IsGroundJump:bool))]
 }
 Transition {
  path=AnimGraph.AnimGraphNode_StateMachine_0.LocomotionStateMachine.AnimStateTransitionNode_4.Transition
  nodes [Result(bCanEnterTransition=Get IsLoco())]
 }
 Transition {
  path=AnimGraph.AnimGraphNode_StateMachine_0.LocomotionStateMachine.AnimStateTransitionNode_5.Transition
  nodes [Result(bCanEnterTransition=true)]
 }
 Transition {
  path=AnimGraph.AnimGraphNode_StateMachine_0.LocomotionStateMachine.AnimStateTransitionNode_6.Transition
  nodes [Result(bCanEnterTransition="float < float"(A="Time Remaining (ratio) (Blob_Jump_Start_BS)"!() -> (ReturnValue:float) {is=title K2Node_AnimGetter}, B=0.1))]
 }
 Transition {
  path=AnimGraph.AnimGraphNode_StateMachine_0.LocomotionStateMachine.AnimStateTransitionNode_7.Transition
  nodes [Result(bCanEnterTransition=Get IsGroundFall() -> (IsGroundFall:bool))]
 }
 Transition {
  path=AnimGraph.AnimGraphNode_StateMachine_0.LocomotionStateMachine.AnimStateTransitionNode_8.Transition
  nodes [Result(bCanEnterTransition=Get IsGroundJumpExit())]
 }
 Transition {
  path=AnimGraph.AnimGraphNode_StateMachine_0.LocomotionStateMachine.AnimStateTransitionNode_9.Transition
  nodes [Result(bCanEnterTransition=Get IsDashing())]
 }
 AdditiveStateMachine {
  entry=Conduit
  path=AnimGraph.AnimGraphNode_StateMachine_1.AdditiveStateMachine
  ids=id
  nodes [
   AnimStateConduitNode_0 = Conduit()
   AnimStateNode_0 = DefaultState() {is=title AnimStateNode}
   AnimStateNode_1 = Jump() {is=title AnimStateNode}
   AnimStateNode_3 = JumpLand() {is=title AnimStateNode}
   AnimStateTransitionNode_0 = AnimStateTransitionNode() {logicType=TLT_Inertialization, priorityOrder=100}
   AnimStateTransitionNode_1 = AnimStateTransitionNode()
   AnimStateTransitionNode_2 = AnimStateTransitionNode() {logicType=TLT_Inertialization}
   AnimStateTransitionNode_4 = AnimStateTransitionNode() {crossfadeDuration=0.1, logicType=TLT_Inertialization}
   AnimStateTransitionNode_7 = AnimStateTransitionNode() {bAutomaticRuleBasedOnSequencePlayerInState=true}
  ]
  links [
   {
    dst=DefaultState
    rule=AnimGraph.AnimGraphNode_StateMachine_1.AdditiveStateMachine.AnimStateTransitionNode_0.Transition
    src=Conduit
    via=AnimStateTransitionNode_0
   }
   {
    dst=Jump
    rule=AnimGraph.AnimGraphNode_StateMachine_1.AdditiveStateMachine.AnimStateTransitionNode_2.Transition
    src=Conduit
    via=AnimStateTransitionNode_2
   }
   {
    dst=Conduit
    rule=AnimGraph.AnimGraphNode_StateMachine_1.AdditiveStateMachine.AnimStateTransitionNode_1.Transition
    src=DefaultState
    via=AnimStateTransitionNode_1
   }
   {
    dst=JumpLand
    rule=AnimGraph.AnimGraphNode_StateMachine_1.AdditiveStateMachine.AnimStateTransitionNode_4.Transition
    src=Jump
    via=AnimStateTransitionNode_4
   }
   {
    dst=Conduit
    rule=AnimGraph.AnimGraphNode_StateMachine_1.AdditiveStateMachine.AnimStateTransitionNode_7.Transition
    src=JumpLand
    via=AnimStateTransitionNode_7
   }
  ]
 }
 Conduit {path=AnimGraph.AnimGraphNode_StateMachine_1.AdditiveStateMachine.AnimStateConduitNode_0.Conduit, nodes [Result(bCanEnterTransition=true)]}
 DefaultState {
  path=AnimGraph.AnimGraphNode_StateMachine_1.AdditiveStateMachine.AnimStateNode_0.DefaultState
  nodes [Output Animation Pose(Result=Additive Identity Pose())]
 }
 Jump {
  path=AnimGraph.AnimGraphNode_StateMachine_1.AdditiveStateMachine.AnimStateNode_1.Jump
  nodes [Output Animation Pose(Result=Additive Identity Pose())]
 }
 JumpLand {
  path=AnimGraph.AnimGraphNode_StateMachine_1.AdditiveStateMachine.AnimStateNode_3.JumpLand
  nodes [Output Animation Pose(Result="Sequence Player (additive) 'jump_LandPredict_and_Land_Chrome_Blob_Gadget'"() {sequence=/ChromeShared/Gameplay/ANimation/jump_LandPredict_and_Land_Chrome_Blob_Gadget})]
 }
 Transition {
  path=AnimGraph.AnimGraphNode_StateMachine_1.AdditiveStateMachine.AnimStateTransitionNode_0.Transition
  nodes [Result(bCanEnterTransition=true)]
 }
 Transition {
  path=AnimGraph.AnimGraphNode_StateMachine_1.AdditiveStateMachine.AnimStateTransitionNode_1.Transition
  nodes [Result(bCanEnterTransition=true)]
 }
 Transition {
  path=AnimGraph.AnimGraphNode_StateMachine_1.AdditiveStateMachine.AnimStateTransitionNode_2.Transition
  nodes [
   n1 = Get IsPhasing()
   n2 = Get bIsFalling()
   n3 = Get bIsJumping()
   Result(bCanEnterTransition=AND Boolean(A=OR Boolean!(A=@n2, B=@n3) -> (ReturnValue), B=NOT Boolean(A=@n1)))
  ]
 }
 Transition {
  path=AnimGraph.AnimGraphNode_StateMachine_1.AdditiveStateMachine.AnimStateTransitionNode_4.Transition
  nodes [
   Result(bCanEnterTransition=Get IsJumpLand())
   AND Boolean(A=NOT Boolean(A=Get bIsFalling()), B=NOT Boolean(A=Get bIsJumping()))
  ]
 }
 Transition {path=AnimGraph.AnimGraphNode_StateMachine_1.AdditiveStateMachine.AnimStateTransitionNode_7.Transition, nodes [Result()]}
 EventGraph {nodes [
   Event BlueprintUpdateAnimation()
   > n1 = Cast To FortPlayerPawn(Object=TryGetPawnOwner())
   > Set IsSwimming(IsSwimming:bool=IsSurfaceSwimming!(self:Fort Pawn Object Reference=@n1) -> (ReturnValue:bool)) -> (Output_Get:bool)
   > n2 = IsZiplining!(execute, self:Fort Player Pawn Object Reference=@n1) -> (then, ReturnValue:bool)
   > Set IsZiplining(IsZiplining:bool=@n2) -> (Output_Get:bool)
   > Set IsDashing(IsDashing:bool=HasMatchingGameplayTag!(self:Gameplay Tag Asset Interface Interface=@n1, TagToCheck:Gameplay Tag Structure=(TagName="Gameplay.Status.ChromePlayer.Dashing")) -> (ReturnValue:bool)) -> (Output_Get:bool)
   > Set IsPhasing(IsPhasing:bool=HasMatchingGameplayTag(self=@n1, TagToCheck=(TagName="Granted.PhaseableBuildGameplay.PhaseableBuild.Player.Phase"))) -> (Output_Get:bool)
   > n3 = Sequence!(execute) -> (then_0, then_1, then_2:Exec, then_3:Exec, then_4:Exec)
   > Set Zipline Input Lean Yaw(Zipline Input Lean Yaw:float=Get!() -> (
     ZiplineInput_bIsZiplining:bool=~
     ZiplineInput_bIsOnAscender:bool=~
     ZiplineInput_bShouldPlayPivotTransition:bool=~
     ZiplineInput_LeanYaw:float=~
     ZiplineInput_PivotBlendDelayRemaining:float=~
     ZiplineInput_PivotBlendDelay:float=~
     ZiplineInput_PivotCardinalDirection:EFortCardinalDirection Enum
     ZiplineInput_WorldVelocityLastTick:Vector=~
    ).ZiplineInput_LeanYaw) -> (Output_Get:float)
   > Set SprintPlayRate(SprintPlayRate:double=float / float(A=Get PawnSpeed2D(), B=860)) -> (Output_Get:double)
   > Set DashPitchAlpha(DashPitchAlpha:float=GetCurveValue(CurveName=DashPitch)) -> (Output_Get:float)
   > n4 = Branch(Condition="float > float"(A=Get PawnSpeed2D()))
   > Set IsLoco(IsLoco:bool=true) -> (Output_Get:bool)
   Set IsLoco(execute=@n4.else, IsLoco=false)
   n5 = Branch(execute=@n3.then_1, Condition="float > float"(A=Get PawnSpeed2D(), B=15))
   > Set LocoStartDirection(LocoStartDirection:double=Get LocoStartDirection()) -> (Output_Get:double)
   Set LocoStartDirection(execute=@n5.else, LocoStartDirection=Get LocalVelocityAngle() -> (LocalVelocityAngle:float))
   n6 = Get EnterSwim() -> (EnterSwim:bool)
   n7 = Get IsPhasing()
   Set IsGroundJump(execute=@n3.then_3, IsGroundJump:bool=AND Boolean!(A=Get bIsJumping(), B=NOT Boolean(A=@n6), C:bool=NOT Boolean(A=@n7)) -> (ReturnValue)) -> (Output_Get:bool)
   > Set IsBlobSwim(IsBlobSwim:bool=OR Boolean(A=Get IsSwimming(), B=Get IsSwimIdleJump() -> (IsSwimIdleJump:bool))) -> (Output_Get:bool)
   n8 = NOT Boolean(A=Get bIsFalling())
   n9 = NOT Boolean(A=Get bIsJumping())
   > Set IsGroundJumpExit(IsGroundJumpExit:bool=OR Boolean(A=AND Boolean!(A=@n8, B=@n9) -> (ReturnValue), B=Get IsPhasing())) -> (Output_Get:bool)
   n10 = Get EnterSwim()
   n11 = Get IsPhasing()
   > Set IsGroundFall(IsGroundFall:bool=AND Boolean!(A=Get bIsFalling(), B=NOT Boolean(A=@n10), C:bool=NOT Boolean(A=@n11)) -> (ReturnValue)) -> (Output_Get:bool)
   > Set IsLocomotionExit(IsLocomotionExit:bool=OR Boolean!(
     A=Get bIsJumping()
     B=Get bIsFalling()
     C:bool=Get IsDash() -> (IsDash:bool)
     D:bool=Get IsPhasing()
     E:bool=Get IsSwimming()
     F:bool=Get IsZiplining()
    ) -> (ReturnValue)) -> (Output_Get:bool)
   n12 = Get bIsJumping()
   n13 = Get IsZiplining()
   n14 = Get IsSwimming()
   n15 = Get bIsFalling()
   n16 = Get IsDashing()
   > Set IsLocomotionEnter(IsLocomotionEnter:bool=AND Boolean!(
     A=Get IsLoco()
     B=NOT Boolean(A=@n13)
     C=NOT Boolean(A=@n14)
     D:bool=NOT Boolean(A=@n12)
     E:bool=NOT Boolean(A=@n15)
     F:bool=NOT Boolean(A=@n16)
    ) -> (ReturnValue)) -> (Output_Get:bool)
   n17 = Get bIsFalling()
   n18 = Get bIsJumping()
   > Set IsJumpLand(IsJumpLand:bool=AND Boolean!(A=NOT Boolean(A=@n17), B=NOT Boolean(A=@n18)) -> (ReturnValue)) -> (Output_Get:bool)
   n19 = Get PawnSpeed2D()
   > Set IsSwimJumpLoco(IsSwimJumpLoco:bool=AND Boolean(A=Get bIsJumping(), B="float >= float"(A=@n19, B=50))) -> (Output_Get:bool)
   n20 = Get PawnSpeed2D()
   > Set IsSwimFallLoco(IsSwimFallLoco:bool=AND Boolean(A=Get bIsFalling(), B="float >= float"(A=@n20, B=50))) -> (Output_Get:bool)
   n21 = Get PawnSpeed2D()
   > Set IsSwimJumpIdle(IsSwimJumpIdle:bool=AND Boolean(A=Get bIsJumping(), B="float < float"(A=@n21, B=50))) -> (Output_Get:bool)
   n22 = Get PawnSpeed2D()
   > Set IsSwimFallIdle(IsSwimFallIdle:bool=AND Boolean(A=Get bIsFalling(), B="float < float"(A=@n22, B=50))) -> (Output_Get:bool)
   n23 = GetCurveValue(CurveName=Speed)
   Set IsPhaseFall(execute=@n3.then_2, IsPhaseFall:bool=AND Boolean(A="float > float"(A=@n23, B=0.95), B=Get bIsFalling())) -> (Output_Get:bool)
   n24 = Get bIsJumping()
   n25 = Get bIsFalling()
   > Set IsSwimIdleJump(IsSwimIdleJump:bool=AND Boolean(A=OR Boolean!(A=@n24, B=@n25) -> (ReturnValue), B=Get EnterSwim())) -> (Output_Get:bool)
   n26 = Get bIsJumping()
   n27 = Get bIsFalling()
   > Branch(Condition=AND Boolean(A=NOT Boolean(A=@n26), B=NOT Boolean(A=@n27)))
   > Set IsPhaseInAir(IsPhaseInAir:bool=false) -> (Output_Get:bool)
   AnimNotify_EnterJump() {is=title K2Node_Event}
   > Set IsPhaseInAir(IsPhaseInAir=true)
   AnimNotify_EnterIdle() {is=title K2Node_Event}
   > Set EnterIdle(EnterIdle:bool=true) -> (Output_Get:bool)
   AnimNotify_EnterJumpLoop() {is=title K2Node_Event}
   > Set IsPhaseInAir(IsPhaseInAir=true)
   AnimNotify_EnterLocoStart() {is=title K2Node_Event}
   > Set EnterIdle(EnterIdle=false)
   AnimNotify_EnterPhasing() {is=title K2Node_Event}
   > Set IsPhaseInAir(IsPhaseInAir=false)
   AnimNotify_EnterSwim() {is=title K2Node_Event}
   > Set EnterSwim(EnterSwim:bool=true) -> (Output_Get:bool)
   AnimNotify_ExitSwim() {is=title K2Node_Event}
   > Set EnterSwim(EnterSwim=false)
   AnimNotify_EnterStop() {is=title K2Node_Event}
   > Set EnterIdle(EnterIdle=true)
  ]}
]
variables [
 Character Movement {type=CharacterMovementComponent, default=null}
 DashPitchAlpha {type=number}
 EnterIdle {type=boolean}
 EnterSwim {type=boolean}
 IsBlobSwim {type=boolean}
 IsDash {type=boolean}
 IsDashing {type=boolean}
 IsGroundFall {type=boolean}
 IsGroundJump {type=boolean}
 IsGroundJumpExit {type=boolean}
 IsJumpLand {type=boolean}
 IsLoco {type=boolean}
 IsLocomotionEnter {type=boolean}
 IsLocomotionExit {type=boolean}
 IsPhaseFall {type=boolean}
 IsPhaseInAir {type=boolean}
 IsPhasing {type=boolean}
 IsSwimFallIdle {type=boolean}
 IsSwimFallLoco {type=boolean}
 IsSwimIdleJump {type=boolean}
 IsSwimJumpIdle {type=boolean}
 IsSwimJumpLoco {type=boolean}
 IsSwimming {type=boolean}
 IsZiplining {type=boolean}
 LocoStartDirection {type=number}
 NewVar {type=boolean}
 SprintPlayRate {type=number}
 Zipline Input Lean Yaw {type=number, default=5}
]
