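# Translucent two-sided lit material, named for a LackLuster "Apply" overlap volume, that draws diagonal stripes with highlighted edges.
# World-aligned UVs are rotated a fixed 45 degrees, then fed through a rounded sine to alternate between low and high stripe parameters.
# A UV-based SphereMask marks edges, reading only U because Max takes R twice.
# That mask blends both the stripe value and the stripe-to-edge color parameters.
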Material /ChromeShared/Blueprints/M_LackLuster_Apply_OverlapVolume
graph {
 properties {
  blendMode=BLEND_Translucent
  pixelDepthOffsetMode=PDOM_Legacy
  translucencyLightingMode=TLM_VolumetricPerVertexDirectional
  twoSided=true
  usedShadingModels=Default Lit
 }
 nodes [
  n1 = StaticBool()
  n2 = ComponentMask(None=Rotator(
    Coordinate=MaterialFunctionCall!(UseLocalSpace=@n1, WorldPosition, NormalOverride) {materialFunction=/Game/Environments/Apollo/Sets/Skyfire/Materials/MaterialFunctions/MF_WorldAlignedUVs_HardEdge}.Result
    Time=Constant() {r=0.7853}
   ) {speed=1}) {r=true}
  n3 = ConstantBiasScale(None=Sine(None=Divide(A=@n2) {constB=20}))
  n4 = TextureCoordinate()
  n5 = MaterialFunctionCall(Float2=Abs(None=ConstantBiasScale(None=@n4) {bias=-0.5, scale=2})) {materialFunction=&BreakOutFloat2Components}
  n6 = Constant() {r=0.01}
  n7 = SphereMask(A=Max(A=@n5.R, B=@n5.R), B=OneMinus(None=@n6), Radius=@n6)
  LinearInterpolate(
   A=LinearInterpolate(
    A=ScalarParameter() {defaultValue=0.07, parameterName=StripeLow}
    B=ScalarParameter() {defaultValue=0.15, parameterName=StripeHigh}
    Alpha=Round(None=@n3)
   )
   Alpha=@n7
  )
  LinearInterpolate(
   A=VectorParameter() {parameterName=BaseColorStripes, defaultValue=<rgba 0.9132 0.567 0.141 0>}.RGB
   B=VectorParameter() {parameterName=BaseColorEdges, defaultValue=<rgba 0.9132 0.1839 0.8297 0>}.RGB
   Alpha=@n7
  )
 ]
}
