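; Chrome subsystem Blueprint that sets up, applies and restores chrome materials on actors, notably containers and vehicles.
; Setup resets material caches before calling the Lackluster function library, while restore uses native per-actor map data.
; Cosmetic-only handlers redo setup on container mesh or upgrade changes, reapply chrome on searched-state changes, and chrome newly added offroad tires.
; Container handlers support both Blueprint and native data paths; UnbindChromeEvents removes these delegates, but BindChromeEvents is empty.
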(event Chrome|EventApplyChromeToActor
  (params (Actor :type Actor%20Object%20Reference))
  (Class|BLackLusterFunctionLibrary|PythonApplyChrometoActor Actor))

(event Chrome|EventInitLacklusterFunctionLibrary
  (Class|BLackLusterFunctionLibrary|Initialize))

(event Chrome|EventSetupChromeOnActor
  (params (Actor :type Actor%20Object%20Reference))
  (bind _containermaterialscalarparammap (Variables|Default|GetContainerMaterialScalarParamMap))
  (bind _containermaterialvectorparammap (Variables|Default|GetContainerMaterialVectorParamMap))
  (Utilities|IsValid Actor
    (:Is%20Valid
      (Utilities|FlowControl|Sequence
        (:then_0
          (Utilities|Array|Clear (Variables|Default|GetChromeMaterialStructs))
          (Utilities|Array|Clear (Variables|Default|GetOriginalMaterialStructs))
          (Utilities|Array|Clear (Variables|Default|GetNativeChromeMaterialStructs))
          (Utilities|Array|Clear (Variables|Default|GetNativeOriginalMaterialStructs))
          (Utilities|FlowControl|Sequence
            (:then_0
              (bind _keys (Utilities|Map|Keys _containermaterialscalarparammap))
              (for _array_element _keys
                (Utilities|Map|Add _containermaterialscalarparammap _array_element)))
            (:then_1
              (bind _keys_2 (Utilities|Map|Keys _containermaterialvectorparammap))
              (for _array_element_2 _keys_2
                (Utilities|Map|Add _containermaterialvectorparammap _array_element_2 "(R=0.000000,G=0.000000,B=0.000000,A=0.000000)")))))
        (:then_1
          (Class|BLackLusterFunctionLibrary|DisableCurieFire 0)
          (Class|BLackLusterFunctionLibrary|SetupChromeonActor (Variables|Default|GetActorType) (Variables|Default|GetChromeType) (Variables|Default|GetChromeMaterialStructs) (Variables|Default|GetOriginalMaterialStructs) (Variables|Default|GetNativeActorType) (Variables|Default|GetNativeChromeType) (Variables|Default|GetNativeChromeMaterialStructs) (Variables|Default|GetNativeOriginalMaterialStructs) Actor (Variables|Default|GetContainerMaterialScalarParamMap) (Variables|Default|GetContainerMaterialVectorParamMap)))))))

(event Chrome|EventRestoreActorOriginalMaterials
  (params (Actor :type Actor%20Object%20Reference))
  (bind _lacklusterdata (Variables|Lackluster|GetLacklusterData))
  (bind (_applychromeclasspath _chromeaudioclasspath _materialstoignore _setparameternames _classestoactortypes _chromematerialoverrides _actortotextures _tagquerytochrometypeoverrides _tagtoactortypes _genericchromematerial _noanimchestmaterialtags _chromeanimchestmaterial _chromenoanimchestmaterial _buildingchromeapplied _buildingsmactorhitsound _buildingsmactorpropdeathsound _buildingsmactornonpropdeathsound _buildingsmactorpropchromedsound _buildingsmactornonpropchromedsound _tieredchestclass _tieredchestparentclass _chromesoundcomponents _shadowproxymeshcomponentclass _materialsandtypes) (Utilities|Struct|BreakLacklusterData _lacklusterdata))
  (bind (_value _returnvalue) (Utilities|Map|Find _materialsandtypes Actor))
  (bind (_actortype _chrometype _chromematerialdata _originalmaterialdata) (Utilities|Struct|BreakLacklusterMaterialsandTypes _value))
  (Utilities|IsValid Actor
    (:Is%20Valid
      (if _returnvalue
        (Lackluster|NativeRestoreOriginalMaterials Actor _originalmaterialdata)))))

(fn BindChromeEvents
  (params (Actor :type Actor%20Object%20Reference)))

(fn HandleContainerMeshSetChanged
  (params (BuildingSMActor :type Building%20Static%20Mesh%20Object%20Reference) (NewMeshSet :type Mesh%20Set%20Structure%20%28by%20ref%29))
  (bind _lacklusterdata (Variables|Lackluster|GetLacklusterData))
  (bind (_applychromeclasspath _chromeaudioclasspath _materialstoignore _setparameternames _classestoactortypes _chromematerialoverrides _actortotextures _tagquerytochrometypeoverrides _tagtoactortypes _genericchromematerial _noanimchestmaterialtags _chromeanimchestmaterial _chromenoanimchestmaterial _buildingchromeapplied _buildingsmactorhitsound _buildingsmactorpropdeathsound _buildingsmactornonpropdeathsound _buildingsmactorpropchromedsound _buildingsmactornonpropchromedsound _tieredchestclass _tieredchestparentclass _chromesoundcomponents _shadowproxymeshcomponentclass _materialsandtypes) (Utilities|Struct|BreakLacklusterData _lacklusterdata))
  (bind (_value _returnvalue) (Utilities|Map|Find _materialsandtypes))
  (bind _actormaterialsandtypes (Variables|Default|GetActorMaterialsAndTypes))
  ;[0.1927 0.0706 0.1264] HandleContainerMeshSetChanged: Used to handle the scenario where the mesh and material of a container is changed when it upgrades to a better version.
  ; For e.g., regular chest to rare chest and small ammo box to large ammo box.
  ; In these cases, treat the actor as a completely new actor and remove it from the ActorMaterialsAndTypes map and setup chrome on the actor once again.
  (Networking|CanExecuteCosmeticEvents
    (:True
      (bind _asbuilding_container (Utilities|Casting|CastToBuildingContainer BuildingSMActor))
      (if (.ReturnValue (Utilities|Map|Find _actormaterialsandtypes _asbuilding_container))
        (Utilities|Map|Remove _actormaterialsandtypes _asbuilding_container)
        (Chrome|SetupChromeOnActor _asbuilding_container)
        (return)
        (elif _returnvalue
          (Utilities|Map|Remove _materialsandtypes _asbuilding_container)
          (Chrome|NativeSetupChromeonActor _asbuilding_container)
          (return)
          (else
            _))))))

(fn HandleContainerRandomUpgradeApplied
  (params (BuildingContainer :type Building%20Container%20Object%20Reference) (UpgradeIndex :type Integer) (UpgradeLootTierGroup :type Name%20%28by%20ref%29))
  (bind _lacklusterdata (Variables|Lackluster|GetLacklusterData))
  (bind (_applychromeclasspath _chromeaudioclasspath _materialstoignore _setparameternames _classestoactortypes _chromematerialoverrides _actortotextures _tagquerytochrometypeoverrides _tagtoactortypes _genericchromematerial _noanimchestmaterialtags _chromeanimchestmaterial _chromenoanimchestmaterial _buildingchromeapplied _buildingsmactorhitsound _buildingsmactorpropdeathsound _buildingsmactornonpropdeathsound _buildingsmactorpropchromedsound _buildingsmactornonpropchromedsound _tieredchestclass _tieredchestparentclass _chromesoundcomponents _shadowproxymeshcomponentclass _materialsandtypes) (Utilities|Struct|BreakLacklusterData _lacklusterdata))
  (bind _actormaterialsandtypes (Variables|Default|GetActorMaterialsAndTypes))
  (bind (_value _returnvalue) (Utilities|Map|Find _actormaterialsandtypes BuildingContainer))
  (bind (_value_2 _returnvalue_2) (Utilities|Map|Find _materialsandtypes))
  ;[0.02919 0.2135 0.06475] HandleRandomUpdateApplied: Called when a building container upgrades.
  ; Setup chrome material on the applied container again.
  (Networking|CanExecuteCosmeticEvents
    (:True
      (Utilities|IsValid BuildingContainer
        (:Is%20Valid
          (if _returnvalue
            (Utilities|Map|Remove _actormaterialsandtypes BuildingContainer)
            (Chrome|SetupChromeOnActor BuildingContainer)
            (return)
            (elif _returnvalue_2
              (Utilities|Map|Remove _materialsandtypes BuildingContainer)
              (Chrome|NativeSetupChromeonActor BuildingContainer)
              (else
                _))))))))

(fn HandleContainerSearchedStateChanged
  (params (BuildingContainer :type Building%20Container%20Object%20Reference) (bSearched :type Boolean))
  (bind _actormaterialsandtypes (Variables|Default|GetActorMaterialsAndTypes))
  (bind (_value _returnvalue) (Utilities|Map|Find _actormaterialsandtypes BuildingContainer))
  (bind (_actortype_2_5b99d33c44fcfc26b361ee99b2bc0a4a _chrometype_5_68195aba4bf5bd23330635bc8ce95a46 _chromematerialstructs_9_a596659c4eb279eb1d6a5288d978425c _originalmaterialstructs_12_567c73ae46c386f2a6d4c9b35c9d5864 _nativeactortype_22_cfdbf9b1473c12413cfd3b947db2e001 _nativechrometype_23_860f664148debfe8fc0f50a5e7fefb6f _nativechromematerialstructs_21_eb1373e044d8f3e92f5f7c97e850fe26 _nativeoriginalmaterialstructs_29_3b92abd442db398f00adb889972847b8 _containermaterialscalarparammap_34_88450d0546bb2f29ce9612ba6aaa1b20 _containermaterialvectorparammap_37_3436013e48082fbf8e5aaba2b0bd7901) (Utilities|Struct|BreakBLacklusterMaterialsandTypes _value))
  (bind _lacklusterdata (Variables|Lackluster|GetLacklusterData))
  (bind (_applychromeclasspath _chromeaudioclasspath _materialstoignore _setparameternames _classestoactortypes _chromematerialoverrides _actortotextures _tagquerytochrometypeoverrides _tagtoactortypes _genericchromematerial _noanimchestmaterialtags _chromeanimchestmaterial _chromenoanimchestmaterial _buildingchromeapplied _buildingsmactorhitsound _buildingsmactorpropdeathsound _buildingsmactornonpropdeathsound _buildingsmactorpropchromedsound _buildingsmactornonpropchromedsound _tieredchestclass _tieredchestparentclass _chromesoundcomponents _shadowproxymeshcomponentclass _materialsandtypes) (Utilities|Struct|BreakLacklusterData _lacklusterdata))
  (bind (_value_2 _returnvalue_2) (Utilities|Map|Find _materialsandtypes BuildingContainer))
  (bind (_actortype _chrometype _chromematerialdata _originalmaterialdata _containermatscalarparammap _containermatvectorparammap) (Utilities|Struct|BreakLacklusterMaterialsandTypes _value_2))
  (bind _containertags (Variables|Default|GetContainerTags))
  (bind _containertags_2 (Variables|Default|GetContainerTags))
  (Networking|CanExecuteCosmeticEvents
    (:True
      (if (Lackluster|UseNativeFunctions)
        (if _returnvalue_2
          (Lackluster|NativeExecuteOwnerTypeFunctionalityPreChromeContainer BuildingContainer _containertags _containermaterialscalarparammap_34_88450d0546bb2f29ce9612ba6aaa1b20 _containermaterialvectorparammap_37_3436013e48082fbf8e5aaba2b0bd7901)
          (Lackluster|NativeApplyChromeMaterial BuildingContainer _chromematerialdata _originalmaterialdata _chrometype _actortype)
          (Lackluster|NativeExecuteOwnerTypeFunctionalityPostChromeContainer BuildingContainer _containertags_2 _containermatscalarparammap _containermatvectorparammap))
        (elif _returnvalue
          (Class|BLackLusterFunctionLibrary|ExecuteOwnerTypeFunctionalityPreChromeContainer _containertags _containermaterialscalarparammap_34_88450d0546bb2f29ce9612ba6aaa1b20 _containermaterialvectorparammap_37_3436013e48082fbf8e5aaba2b0bd7901 BuildingContainer)
          (Class|BLackLusterFunctionLibrary|ApplyChromeMaterial BuildingContainer _chromematerialstructs_9_a596659c4eb279eb1d6a5288d978425c _originalmaterialstructs_12_567c73ae46c386f2a6d4c9b35c9d5864 _chrometype_5_68195aba4bf5bd23330635bc8ce95a46 _actortype_2_5b99d33c44fcfc26b361ee99b2bc0a4a)
          (Class|BLackLusterFunctionLibrary|ExecuteOwnerTypeFunctionalityPostChromeContainer BuildingContainer _containertags_2 _containermaterialscalarparammap_34_88450d0546bb2f29ce9612ba6aaa1b20 _containermaterialvectorparammap_37_3436013e48082fbf8e5aaba2b0bd7901))))))

(fn HandleVehicleModUpdated
  (params (Vehicle :type Actor%20Object%20Reference) (ModTag :type Gameplay%20Tag%20Structure%20%28by%20ref%29))
  ;[0.02528 0.2188 0.1155] HandleVehicleModUpdated: When a vehicle gets an offroad tire added to it, apply the chrome material to the added offroad tire static meshes only.
  (Networking|CanExecuteCosmeticEvents
    (:True
      (if (== ModTag "(TagName=\"Vehicle.Mod.Tire.Offroad\")")
        (bind _chromematerialstructs (Lackluster|NativeFindTexturesfromComponents "Valet" Vehicle)) ; Find the chrome materail struct information for all the components in the actor. \nThis will now include the offroad tire components. 
        ;;[0.04498 0.1382 0.1667] Cache the material structs that are for the offroad tires only.
        (for _array_element _chromematerialstructs
          (bind _returnvalue (Lackluster|NativeGetChromeChonkerTires (Utilities|Struct|BreakLacklusterMaterialData _array_element)))
          (if _returnvalue
            (Utilities|Array|Add (LocalVariables|GetMaterialDataStructs) _array_element)))
        ;;[0.1927 0.04334 0.1459] Apply chrome material to the offroad tires.
        (Lackluster|NativeSetChromeMaterials (LocalVariables|GetMaterialDataStructs))))))

(fn UnbindChromeEvents
  (params (Actor :type Actor%20Object%20Reference))
  ;[0.2812 0.2073 0.06481] UnbindChromeEvents: Unbind delegates bound to actors when they were being chromed.
  ; See function BindChromeEvents().
  (Networking|CanExecuteCosmeticEvents
    (:True
      (bind _asbuilding_container (Utilities|Casting|CastToBuildingContainer Actor)
        (:then
          (Building|UnbindEventfromOnMeshSetChanged _asbuilding_container (EventDispatchers|CreateEvent :!event HandleContainerMeshSetChanged))
          (Container|UnbindEventfromOnContainerSearchedStateChangedClient _asbuilding_container (EventDispatchers|CreateEvent :!event HandleContainerSearchedStateChanged))
          (Container|UnbindEventfromOnContainerRandomUpgradeApplied _asbuilding_container (EventDispatchers|CreateEvent :!event HandleContainerRandomUpgradeApplied)))
        (:CastFailed
          (bind _asfort_athena_vehicle (Utilities|Casting|CastToFortAthenaVehicle Actor))
          (Vehicle|Mods|UnbindEventfromOnModAppliedonClientDelegate _asfort_athena_vehicle (EventDispatchers|CreateEvent :!event HandleVehicleModUpdated)))))))
