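# Additive, two-sided material for a spinning biplane propeller blur in a chrome variant, enabled for Niagara mesh particles.
# Time-driven rotation spins normal and fast blur textures, which crossfade by an eased timestamp from boost start and duration.
# The hue-shifted, desaturated result tints LackLuster base color.
# Its green channel, modulated by panning noise and a radial mask, drives opacity with view-angle fade and depth fade, bypassed on mobile.
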Material /ChromeShared/FX/Materials/M_Biplane_Propeller_Blur_Chrome
graph {
 properties {
  bUsedWithNiagaraMeshParticles=true
  blendMode=BLEND_Additive
  pixelDepthOffsetMode=PDOM_Legacy
  translucencyLightingMode=TLM_Surface
  twoSided=true
  usedShadingModels=Default Lit
 }
 nodes [
  n1 = Multiply(A=Time(), B=ScalarParameter() {defaultValue=3.5, parameterName=SpinSpeed})
  n2 = Rotator(Coordinate=TextureCoordinate() {coordinateIndex=1}, Time=@n1) {speed=-25}
  n3 = TextureSample(UVs=@n2) {texture=/Game/Athena/DrivableVehicles/FX/Plane/T_Biplane_Propeller_Blur}
  n4 = Multiply(A=TextureSample(UVs=@n2) {texture=/Game/Athena/DrivableVehicles/FX/Plane/T_Biplane_Propeller_Blur_Fast}.RGB) {constB=1.5}
  n5 = Subtract(A=Time() {period=5}, B=ScalarParameter() {parameterName=BoostStartTime})
  n6 = ScalarParameter() {defaultValue=3, group=Timestamp, parameterName=BoostDuration}
  n7 = MaterialFunctionCall!(In=Clamp(None=Divide(A=@n5, B=@n6))) {materialFunction=/Game/Effects/Fort_Effects/Materials/_Functions/MF_Smoothstep_EaseOut}
  n8 = Desaturation(
   None=MaterialFunctionCall(
    Hue Shift Percentage=ScalarParameter() {parameterName=HueShift}
    Texture=LinearInterpolate(A=@n3.RGB, B=@n4, Alpha=@n7.Result) {constAlpha=1}
   ) {materialFunction=&HueShift}.Result
   Fraction=ScalarParameter() {defaultValue=1, parameterName=Desaturate}
  )
  n9 = TextureCoordinate() {vTiling=8}
  n10 = Power(Base=SphereMask(A=TextureCoordinate() {coordinateIndex=1}, B=Constant() {r=0.5}) {attenuationRadius=0.5, hardnessPercent=0}) {constExponent=0.8}
  n11 = Multiply(
   A=TextureSample(UVs=Panner(Coordinate=@n9, Time=@n1) {speedX=-7}) {texture=&T_Smooth_tile}.G
   B=OneMinus(None=@n10)
  )
  n12 = Add(
   A=Multiply(A=ComponentMask(None=@n8) {g=true}, B=LinearInterpolate(Alpha=@n11) {constA=0.7, constB=1.5}) {constB=1.5}
   B=LinearInterpolate(Alpha=@n10) {constA=0.7, constB=0}
  )
  n13 = MaterialFunctionCall("FadeSensitivity (optional)"=ScalarParameter() {defaultValue=1.5, parameterName=AngleFade}) {materialFunction=&FadeBasedOnViewAngle}
  n14 = DepthFade() {fadeDistanceDefault=25}
  n15 = Constant() {r=1}
  Multiply(
   A=Multiply(A=Multiply(A=@n12, B=@n13.Result), B=ScalarParameter() {defaultValue=1, parameterName=Opacity})
   B=FeatureLevelSwitch(Default=@n14, ES2=@n15, ES3_1=@n15, SM4_REMOVED=@n14)
  )
  Multiply(
   A=BreakMaterialAttributes!(Attr=MaterialFunctionCall!(
     bBlendMaterialAttributes
     MA_In
     UVsOverride
     Override Normals
     BlendMaterialAttributesAlpha
     BaseColorOverride
     WorldPositionOverride
     VertexNormalOverride
     SourceNormalsStrengthOverride
    ) {materialFunction=/Game/TechArt/Seasonal/LackLuster/MaterialFunctions/MF_LackLuster_Base}.MA Out).BaseColor
   B=@n8
  )
 ]
}
