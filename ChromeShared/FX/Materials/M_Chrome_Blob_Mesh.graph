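# Default Lit chrome blob material for Niagara mesh and sprite particles, built on the shared LackLuster base function.
# Dynamic parameter X, scaled by Surface Normal Bias, skews world normals, converted to tangent space, as the base's normal override.
# The base's emissive output feeds an FX emissive-adjustment function.
# A displacement along vertex normals, scaled by the WPO parameter, blends radius-normalized distance from particle center with instance-space vertical normal direction.
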Material /ChromeShared/FX/Materials/M_Chrome_Blob_Mesh
graph {
 properties {bUsedWithNiagaraMeshParticles=true, bUsedWithNiagaraSprites=true, pixelDepthOffsetMode=PDOM_Legacy, usedShadingModels=Default Lit}
 nodes [
  Add()
  n1 = DynamicParameter() {parameterIndex=1, defaultValue=<rgba 1 0 0 0>, paramNames [X,Y,Z,Param4]}
  n2 = Multiply(
   A=MaterialFunctionCall(X=@n1.X, Y=@n1.X, Z=@n1.X) {materialFunction=&MakeFloat3}.Result
   B=ScalarParameter() {group=Normals, parameterName=Surface Normal Bias}
  )
  n3 = VertexNormalWS()
  n4 = Transform(None=Normalize(VectorInput=Add(A=@n2, B=@n3))) {transformSourceType=TRANSFORMSOURCE_World, transformType=TRANSFORM_Tangent}
  MaterialFunctionCall(Emissive=BreakMaterialAttributes!(Attr=MaterialFunctionCall!(
     bBlendMaterialAttributes
     MA_In
     UVsOverride
     Override Normals=@n4
     BlendMaterialAttributesAlpha
     BaseColorOverride
     WorldPositionOverride
     VertexNormalOverride=@n3
     SourceNormalsStrengthOverride
    ) {materialFunction=/Game/TechArt/Seasonal/LackLuster/MaterialFunctions/MF_LackLuster_Base}.MA Out).EmissiveColor) {materialFunction=&MF_FX_EmissiveAdjustments}
  n5 = DynamicParameter() {defaultValue=<rgba 1 1 0 1>, paramNames [Erode 1-0, NoiseTIle, WPO, Param4]}
  n6 = Distance(A=WorldPosition().XYZ, B=ParticlePositionWS())
  n7 = Multiply(A=@n5.Param4, B=ScalarParameter() {defaultValue=1, parameterName=MeshRadius})
  n8 = VertexNormalWS()
  n9 = LinearInterpolate(
   A=LinearInterpolate(Alpha=Divide(A=@n6, B=@n7)) {constA=-1}
   B=DotProduct(
    A=Transform(None=@n8) {transformSourceType=TRANSFORMSOURCE_World, transformType=TRANSFORM_Instance}
    B=Constant3Vector() {constant=<rgba 0 0 1 0>}
   )
  ) {constAlpha=0.3}
  n10 = LinearInterpolate(Alpha=Power(Base=Saturate(None=@n9))) {constA=-1}
  Multiply(A=@n5.WPO, B=Multiply(A=Multiply(B=@n10) {constA=1}, B=VertexNormalWS()))
 ]
}
