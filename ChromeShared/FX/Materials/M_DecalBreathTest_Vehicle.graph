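# Translucent, unlit deferred-decal test material for Niagara sprites and meshes, named as a vehicle breath effect.
# Separate branches build a particle-alpha sphere-mask fade, a radial ring animated by a dynamic parameter and masked by custom-depth difference, and hue-shifted sine emissive bands.
# Further branches derive normals from band height and compute camera-aligned virtual-plane coordinates.
# The graph shows no branch connected to a material output, so it appears unfinished.
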Material /ChromeShared/FX/Materials/M_DecalBreathTest_Vehicle
graph {
 properties {
  bTangentSpaceNormal=false
  bUsedWithNiagaraMeshParticles=true
  bUsedWithNiagaraSprites=true
  blendMode=BLEND_Translucent
  materialDomain=MD_DeferredDecal
  pixelDepthOffsetMode=PDOM_Legacy
  shadingModel=MSM_Unlit
  usedShadingModels=Unlit
 }
 nodes [
  n1 = ParticleColor()
  n2 = TextureSample(UVs=Panner(Coordinate=ParticleMacroUV()) {speedY=1}) {texture=/Game/Effects/Fort_Effects/Textures/Grad/gradient_healed}
  n3 = SphereMask(A=TextureCoordinate(), B=Constant() {r=0.5}) {attenuationRadius=0.5, hardnessPercent=0}
  DepthFade(Opacity=Multiply(A=Multiply(A=@n1.A, B=@n2.R), B=@n3))
  n4 = Divide(A=Subtract(A=SceneTexture() {sceneTextureId=PPI_CustomDepth}.Color, B=PixelDepth().R)) {constB=200}
  n5 = MaterialFunctionCall(Vector or UVs=ParticleMacroUV()) {materialFunction=&VectorToRadialValue}
  n6 = Frac(None=DynamicParameter().R)
  n7 = Multiply(A=@n6, B=Constant() {r=-120})
  n8 = Constant()
  n9 = MaterialFunctionCall(Float2=Add(A=@n5.Radial Coordinates, B=Add(A=@n7, B=@n8))) {materialFunction=&BreakOutFloat2Components}
  n10 = Multiply(A=@n9.G, B=ScalarParameter() {defaultValue=0.02, parameterName=Param_2})
  n11 = Saturate(None=MaterialFunctionCall(Float2=Add(A=@n10)) {materialFunction=&BreakOutFloat2Components}.G)
  n12 = OneMinus(None=@n11)
  n13 = Saturate(None=Multiply(A=Min(A=@n12, B=@n11)) {constB=2})
  n14 = Saturate(None=Multiply(
    A=MaterialFunctionCall(Float2=@n5.Radial Coordinates) {materialFunction=&BreakOutFloat2Components}.G
    B=Constant() {r=0.15}
   ))
  n15 = OneMinus(None=@n6)
  n16 = Constant() {r=1}
  LinearInterpolate(B=OneMinus(None=Saturate(None=@n4)), Alpha=Multiply(A=Multiply(A=@n13, B=@n14), B=Power(Base=@n15, Exp=@n16)))
  n17 = Multiply(A=@n9.G, B=ScalarParameter() {defaultValue=0.02, parameterName=Param_1})
  n18 = ScalarParameter() {defaultValue=-0.25}
  n19 = Add(A=Sine(None=Add(A=@n17, B=@n18) {constB=0.5}) {period=2})
  n20 = Multiply(A=OneMinus(None=Multiply(A=@n19) {constB=0.5}), B=Constant() {r=1})
  n21 = Power(Base=Saturate(None=OneMinus(None=@n20)), Exp=ScalarParameter() {defaultValue=2, parameterName=Power})
  n22 = Multiply(A=OneMinus(None=@n21), B=ScalarParameter() {defaultValue=0.25, parameterName=A})
  n23 = VectorParameter() {parameterName=C, defaultValue=<rgba 0.1 0.2333 0.5 1>}
  MaterialFunctionCall(Emissive=Multiply(
    A=@n21
    B=MaterialFunctionCall(Hue Shift Percentage=@n22, Texture=@n23.RGB) {materialFunction=&HueShift}.Result
   )) {materialFunction=&MF_FX_EmissiveAdjustments}
  MaterialFunctionCall(
   Normal=Transform(None=MaterialFunctionCall!(Height=@n20, Absolute World Position, World Space Vertex Normals) {materialFunction=/Engine/Functions/Engine_MaterialFunctions03/Procedurals/HeightToNormalSmooth}.WorldSpaceNormal) {transformSourceType=TRANSFORMSOURCE_World, transformType=TRANSFORM_Tangent}
   Flatness=Constant() {r=5}
  ) {materialFunction=&FlattenNormal}
  MaterialFunctionCall(
   Plane Normal Axis=Transform(None=Constant3Vector() {constant=<rgba 0 0 1 0>}) {transformSourceType=TRANSFORMSOURCE_Camera}
   Plane X Axis=Transform(None=Constant3Vector() {constant=<rgba 1 0 0 0>}) {transformSourceType=TRANSFORMSOURCE_Camera}
   Plane Center=Add(A=Constant3Vector() {constant=<rgba 0 0 35 1>}, B=ActorPositionWS())
  ) {materialFunction=&VirtualPlaneCoordinates}
  Multiply(A=Time(), B=Constant() {r=0.75})
  Multiply(A=@n3, B=ParticleColor().A)
 ]
}
