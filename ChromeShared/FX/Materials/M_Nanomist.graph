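# Translucent unlit Niagara sprite material for a nanomist effect: sparkly, holographic-tinted smoke from an 8x8 inky smoke flipbook.
# Screen-aligned high-frequency noise, offset by smoke, is sharpened, multiplied by smoke and an inverted per-cell sphere mask.
# That mask samples a holo curve-atlas gradient, hue-shifted over particle lifetime and added to a smoke-driven blue-to-mauve blend.
# The same mask times particle alpha forms the second output, apparently opacity.
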Material /ChromeShared/FX/Materials/M_Nanomist
graph {
 properties {bUsedWithNiagaraSprites=true, blendMode=BLEND_Translucent, pixelDepthOffsetMode=PDOM_Legacy, shadingModel=MSM_Unlit, usedShadingModels=Unlit}
 nodes [
  n1 = TextureSample() {samplerType=SAMPLERTYPE_LinearColor, texture=/Game/Effects/Fort_Effects/Textures/Smoke/T_Inky_Smoke_SubUV_8x8}
  n2 = Frac(None=Multiply(A=TextureCoordinate()) {constB=8})
  n3 = Constant() {r=0.5}
  n4 = Constant() {r=0.4}
  n5 = Constant() {r=0.25}
  n6 = MaterialFunctionCall!(TextureResolution=Constant2Vector() {g=86, r=86}) {materialFunction=/Engine/Functions/Engine_MaterialFunctions02/Texturing/ScreenAlignedPixelToPixelUVs}
  n7 = Multiply(A=@n1.R) {constB=0.5}
  n8 = Power(Base=TextureSample(UVs=Add(A=@n6.Result, B=@n7)) {samplerType=SAMPLERTYPE_LinearGrayscale, texture=/Engine/EngineMaterials/Good64x64TilingNoiseHighFreq}.R) {constExponent=3}
  n9 = Multiply(A=OneMinus(None=SphereMask(A=@n2, B=@n3, Radius=@n4, Hardness=@n5)), B=Multiply(A=Multiply(A=@n8) {constB=2}, B=@n1.R))
  n10 = ParticleRelativeTime()
  Add(
   A=LinearInterpolate(A=Constant3Vector() {constant=<rgba 0.3912 0.49 0.8333 0>}, B=Constant3Vector() {constant=<rgba 0.526 0.2712 0.4208 0>}, Alpha=@n1.R)
   B=MaterialFunctionCall(
    Hue Shift Percentage=LinearInterpolate(Alpha=@n10)
    Texture=CurveAtlasRowParameter(CurveTime=@n9) {atlas=&AthenaCurveAtlas01, curve=/Game/Athena/LinearColorCurves/Curve_C_Holo, defaultValue=250}
   ) {materialFunction=&HueShift}.Result
  )
  Multiply(A=@n9, B=ParticleColor().A)
 ]
}
