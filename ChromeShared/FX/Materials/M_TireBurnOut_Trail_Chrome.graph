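# Masked, two-sided lit material for tire burnout trails on beam trails and Niagara ribbons.
# An inverted, power-sharpened burnout texture on UV1 is multiplied by a parabolic UV0 edge fade and particle alpha to form a mask.
# The graph also holds world-aligned water-ripple normal functions and a LackLuster base function whose emissive feeds FX emissive adjustments.
# No connections to the material outputs appear in the graph.
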Material /ChromeShared/FX/Materials/M_TireBurnOut_Trail_Chrome
graph {
 properties {
  bUsedWithBeamTrails=true
  bUsedWithNiagaraRibbons=true
  blendMode=BLEND_Masked
  pixelDepthOffsetMode=PDOM_Legacy
  translucencyPass=MTP_BeforeDOF
  twoSided=true
  usedShadingModels=Default Lit
 }
 nodes [
  n1 = TextureObject() {samplerType=SAMPLERTYPE_Normal, texture=/Game/Effects/Fort_Effects/Gameplay/Water/Textures/T_WaterRipples_01}
  n2 = VectorParameter() {parameterName=Texture Size, defaultValue=<rgba 100 100 100 0>}
  MaterialFunctionCall(TextureObject=@n1, TextureSize=@n2.RGB) {materialFunction=&WorldAlignedNormal}
  MaterialFunctionCall!(
   TextureObject=@n1
   TextureSize=@n2.RGB
   WorldPosition
   Export Float 4
   In=AppendVector(A=ScalarParameter() {defaultValue=0.25, parameterName=Param_1}, B=ScalarParameter())
   World Space Normal
   ProjectionTransitionContrast
  ) {materialFunction=/Game/Athena/SafeZone/WorldAlignedTexture_ScalePan}
  MaterialFunctionCall(Emissive=BreakMaterialAttributes!(Attr=MaterialFunctionCall!(
     bBlendMaterialAttributes
     MA_In
     UVsOverride
     Override Normals
     BlendMaterialAttributesAlpha
     BaseColorOverride
     WorldPositionOverride
     VertexNormalOverride
     SourceNormalsStrengthOverride
    ) {materialFunction=/Game/TechArt/Seasonal/LackLuster/MaterialFunctions/MF_LackLuster_Base}.MA Out).EmissiveColor) {materialFunction=&MF_FX_EmissiveAdjustments}
  n3 = TextureCoordinate() {coordinateIndex=1}
  n4 = Power(
   Base=OneMinus(None=TextureSample(UVs=@n3) {texture=/Game/Effects/Fort_Effects/Textures/Masks/T_TireBurnOut}.RGB)
   Exp=ScalarParameter() {defaultValue=40, parameterName=sss}
  ) {constExponent=40}
  n5 = ComponentMask(None=TextureCoordinate()) {r=true}
  n6 = OneMinus(None=@n5)
  n7 = Power(Base=Multiply(A=Multiply(A=@n6, B=@n5), B=ScalarParameter() {defaultValue=100, parameterName=edgeWidthMultiply}) {constB=16})
  n8 = OneMinus(None=Multiply(A=Clamp(None=@n4), B=Clamp(None=@n7)))
  Multiply(A=ParticleColor().A, B=OneMinus(None=Clamp(None=@n8)))
  TextureCoordinate()
 ]
}
