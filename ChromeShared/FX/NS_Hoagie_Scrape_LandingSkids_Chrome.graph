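# Niagara scrape effect for the Hoagie's landing skids on chrome: short-lived glowing bubble meshes with sprites, flying drops, velocity-stretched splats and smoke.
# User.Speed drives drop and splat spawn rates, smoke burst count and sprite scaling, while normalized User.ThrustVelocity sets the direction of drops, splats and smoke.
# All four emitters set spawn probability by lerping 1 to 0 on User.PhysMatInt, so its default of 1 suppresses spawning.
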NiagaraSystem /ChromeShared/FX/NS_Hoagie_Scrape_LandingSkids_Chrome
dependencies {usedDataInterfaces [&NiagaraDataInterfaceCurve], usedDynamicInputs [], usedModules [&SystemState]}
emitters {
 Bubbles {
  emitterUpdateScript [
   &EmitterState()
   &SpawnBurst_Instantaneous(
    Spawn Count=&RandomRangeInt(Minimum=3, Maximum=5)
    Spawn Probability=&Lerp_Float(A=1, B=0, Alpha=&MakeFloatFromInt(Int=$User.PhysMatInt))
    Use Spawn Probability=true
    UseLoopCountLimit=false
   )
  ]
  particleSpawnScript [
   &InitializeParticle(
    Color=<rgba 1.3728 3.0599 5 1>
    Lifetime Max=0.21
    Lifetime Min=0.175
    Lifetime Mode=::NewEnumerator1
    Mesh Scale Mode=::NewEnumerator3
    Mesh Uniform Scale=1.25
    Sprite Size Mode=::NewEnumerator3
    Uniform Sprite Size=100
    UsePositionOffset=false
    Write Material Random=true
   )
   &ShapeLocation(
    Box Size=<25 25 200>
    MaskSpawnGroup=false
    Offset=&RandomRangeVector(Minimum=<-10 -10 0>, Maximum=<10 10 0>)
    Offset Mode=::NewEnumerator0
    Override Seed=false
    Shape Primitive=::NewEnumerator2
    Surface Only=false
    UseOwnerScale=true
   )
   &InitialMeshOrientation(
    Mesh Axis To Orient=<0 0 1>
    Mesh Orientation Mode=::NewEnumerator2
    Orientation Vector=$Particles.Velocity
    Rotation=&RandomRangeVector(Minimum=<0 0 0>, Maximum=<0 0 1>)
    Use Rotation Vector=true
   )
   &CameraOffset(Camera Offset Amount=50)
  ]
  particleUpdateScript [
   &ParticleState()
   &Drag(Drag=5, Use Linear Drag=true, Use Rotational Drag=false)
   &GravityForce(Gravity=&MakeVector(
     X=0
     Y=0
     Z=&FloatFromCurve(FloatCurve=DataInterface() {propertyValues {Curve {keys [<at 0.2 0>,<at 0.3 1>]}}}, Scale Curve=-1600)
    ))
   &SolveForcesAndVelocity(Clamp Velocity=false, Limit Acceleration=false)
   &ScaleColor(
    Scale RGB=&VectorFromFloat(Value=&FloatFromCurve(FloatCurve=DataInterface() {propertyValues {Curve {keys [
          {arriveTangent=171.9012, interpMode=RCIM_Cubic, time=0, value=0}
          {arriveTangent=29.4118, interpMode=RCIM_Cubic, leaveTangent=29.4118, time=0.1, value=2}
          {arriveTangent=-89.7088, interpMode=RCIM_Cubic, leaveTangent=-89.7088, tangentMode=RCTM_User, time=0.17, value=5}
          {
           arriveTangent=-7.3461
           arriveTangentWeight=0.0565
           interpMode=RCIM_Cubic
           leaveTangent=-7.3461
           leaveTangentWeight=0.1797
           tangentMode=RCTM_User
           time=0.25
           value=1
          }
          {interpMode=RCIM_Cubic, tangentMode=RCTM_User, time=0.7, value=0}
         ]}}}))
    ScaleA=false
    ScaleRGB=true
   )
   &ScaleMeshSize(Scale Factor=&VectorFromFloat(Value=&FloatFromCurve(FloatCurve=DataInterface() {propertyValues {Curve {keys [
          {arriveTangent=15.6842, interpMode=RCIM_Cubic, leaveTangent=15.6842, tangentMode=RCTM_User, time=0, value=0}
          {arriveTangent=0.5907, interpMode=RCIM_Cubic, leaveTangent=0.5907, tangentMode=RCTM_User, time=1, value=4}
         ]}}})))
   &DynamicMaterialParameters(
    Index 0 Param 1=&FloatFromCurve(FloatCurve=DataInterface() {propertyValues {Curve {keys [
         {arriveTangent=-1.7239, interpMode=RCIM_Cubic, leaveTangent=-1.7239, tangentMode=RCTM_User, time=0.2, value=1}
         {arriveTangent=-2.4141, interpMode=RCIM_Cubic, leaveTangent=-2.4141, tangentMode=RCTM_User, time=0.45, value=0.58}
         {arriveTangent=-0.1262, interpMode=RCIM_Cubic, leaveTangent=-0.1262, tangentMode=RCTM_User, time=1, value=0}
        ]}}})
    Index 0 Param 2=&RandomRangeFloat(Minimum=1.5, Maximum=2)
    Index 0 Param 3=&FloatFromCurve(
     FloatCurve=DataInterface() {propertyValues {Curve {keys [
         {arriveTangent=13.3756, interpMode=RCIM_Cubic, leaveTangent=13.3756, tangentMode=RCTM_User, time=0.4, value=0}
         <at 0.6 1 RCIM_Cubic>
         {arriveTangent=-4.0708, interpMode=RCIM_Cubic, leaveTangent=-4.0708, tangentMode=RCTM_User, time=1, value=0}
        ]}}}
     Scale Curve=20
    )
    Index 0 Param 4=&MakeFloatFromVector(Vector=$Particles.Scale, Channel=::NewEnumerator0)
    Index 3 Param 1=4
    Index 3 Param 2=0.4
    Module.Param0WriteEnabled=true
    Module.Param0WriteEnabled3=true
    Module.Param1WriteEnabled=true
    Module.Param1WriteEnabled3=true
    Module.Param2WriteEnabled=true
    Module.Param2WriteEnabled3=false
    Module.Param3WriteEnabled=true
    Module.Param3WriteEnabled3=false
    Write Parameter Index 3=true
   )
   &UpdateMeshOrientation(
    Rotation Rate=&FloatFromCurve(
     FloatCurve=DataInterface() {propertyValues {Curve {keys [
         {arriveTangent=10.9187, interpMode=RCIM_Cubic, time=0, value=1}
         {arriveTangent=-2.4372, interpMode=RCIM_Cubic, leaveTangent=-3.2804, tangentMode=RCTM_Break, time=0.4, value=0.5}
         <at 1 0>
        ]}}}
     Scale Curve=1.5
    )
    Rotation Vector=<0.1 0 1>
   )
   /SetVariables(Particles.GlowColor=&MakeLinearColorFromVectorAndFloat(
     "Vector (RGB)"=&MakeVectorFromLinearColorRGB(Color=<rgba 0.2448 0.6923 1 1>)
     "Float (Alpha)"=&FloatFromCurve(FloatCurve=DataInterface() {propertyValues {Curve {keys [
          {arriveTangent=18.2404, interpMode=RCIM_Cubic, leaveTangent=18.2404, tangentMode=RCTM_User, time=0, value=1}
          {arriveTangent=-1.2, interpMode=RCIM_Cubic, leaveTangent=-1.2, time=0.12, value=2}
          {
           arriveTangent=-4.84
           arriveTangentWeight=0.0565
           interpMode=RCIM_Cubic
           leaveTangent=-4.84
           leaveTangentWeight=0.1797
           tangentMode=RCTM_User
           time=0.25
           value=0.7
          }
          {arriveTangent=-6.1958, interpMode=RCIM_Cubic, leaveTangent=-6.1958, tangentMode=RCTM_User, time=0.4, value=0.9}
          {interpMode=RCIM_Cubic, tangentMode=RCTM_User, time=1, value=0}
         ]}}})
    )) {isSetParameters=true}
   &ScaleSpriteSize(
    Uniform Curve Scale=&MakeFloatFromVector(Vector=$Particles.Scale, Channel=::NewEnumerator0)
    Uniform Curve Sprite Scale=DataInterface() {propertyValues {Curve {keys [
        {arriveTangent=171.9012, interpMode=RCIM_Cubic, time=0.4, value=0.3}
        {
         arriveTangent=-3.2031
         arriveTangentWeight=0.0565
         interpMode=RCIM_Cubic
         leaveTangent=-3.2031
         leaveTangentWeight=0.1797
         tangentMode=RCTM_User
         time=0.5
         value=0.7
        }
        {arriveTangent=-0.9894, interpMode=RCIM_Cubic, leaveTangent=-0.9894, tangentMode=RCTM_User, time=0.6, value=0.5}
        {arriveTangent=-2.7772, interpMode=RCIM_Cubic, leaveTangent=-2.7772, tangentMode=RCTM_User, time=1, value=0}
       ]}}}
   )
  ]
  rendererClasses [&NiagaraMeshRendererProperties, &NiagaraSpriteRendererProperties]
  renderers [
   {rendererClass=&NiagaraMeshRendererProperties, rendererIndex=0}
   {rendererClass=&NiagaraSpriteRendererProperties, rendererIndex=1}
  ]
 }
 Chrome_Drops {
  emitterUpdateScript [
   &EmitterState()
   &SpawnRate(
    Spawn Probability=&Lerp_Float(A=1, B=0, Alpha=&MakeFloatFromInt(Int=$User.PhysMatInt))
    SpawnRate=&FloatFromCurve(
     FloatCurve=DataInterface() {propertyValues {Curve {keys [
         {arriveTangent=-0.0078, interpMode=RCIM_Cubic, leaveTangent=-0.0078, tangentMode=RCTM_User, time=25, value=0}
         <at 29.3003 0.6081>
         {arriveTangent=0.0541, interpMode=RCIM_Cubic, leaveTangent=0.0541, time=41.6721, value=0.9785}
         {arriveTangent=0.0445, interpMode=RCIM_Cubic, leaveTangent=0.0445, time=66.2601, value=2.6064}
         <at 89.4969 3.108 RCIM_Cubic>
        ]}}}
     CurveIndex=$User.Speed
     Scale Curve=75
    )
    Use Spawn Probability=true
   )
  ]
  particleSpawnScript [
   &InitializeParticle(
    Color=<rgba 1.3728 3.0599 5 1>
    Lifetime Max=1.5
    Lifetime Min=0.5
    Lifetime Mode=::NewEnumerator1
    Sprite Size Mode=::NewEnumerator3
    Uniform Sprite Size=&Lerp_Float(
     A=4
     B=13
     Alpha=&FloatFromCurve(
      FloatCurve=DataInterface() {propertyValues {Curve {keys [
          {interpMode=RCIM_Cubic, tangentMode=RCTM_User, time=0, value=0}
          {arriveTangent=1.2457, interpMode=RCIM_Cubic, leaveTangent=1.2457, tangentMode=RCTM_User, time=0.7348, value=0.4334}
          {interpMode=RCIM_Cubic, tangentMode=RCTM_User, time=1, value=1}
         ]}}}
      CurveIndex=&RandomRangeFloat(Minimum=0, Maximum=1)
     )
    )
    UsePositionOffset=false
    Write Material Random=true
   )
   &ShapeLocation(
    Box Size=<25 25 200>
    MaskSpawnGroup=false
    Offset=&RandomRangeVector(Minimum=<-10 -10 0>, Maximum=<10 10 0>)
    Offset Mode=::NewEnumerator0
    Override Seed=false
    Shape Primitive=::NewEnumerator2
    Surface Only=false
    UseOwnerScale=true
   )
   &AddVelocity(Velocity=&Add_Vector(
     A=&Multiply_Vector(
      A=&NormalizeVector(Vector To Normalize=$User.ThrustVelocity)
      B=&VectorFromFloat(Value=&RandomRangeFloat(Minimum=250, Maximum=1000))
     )
     B=&RandomRangeVector(Minimum=<0 0 0>, Maximum=<0 0 100>)
    ))
   &AddVelocity(
    Cone Angle=15
    Cone Axis=&RandomRangeVector(Minimum=<0 0 0>, Maximum=<15 15 15>)
    Module.UseConeFalloff=false
    Module.UseVelDistribution=false
    Override Seed=false
    Velocity Mode=::NewEnumerator2
    Velocity Speed=&RandomRangeFloat(Minimum=250, Maximum=500)
   ) {module=AddVelocity001}
   &CameraOffset(Camera Offset Amount=50)
  ]
  particleUpdateScript [
   &ParticleState()
   &GravityForce()
   &CurlNoiseForce(Noise Frequency=0.25, Noise Quality / Cost=::NewEnumerator0, Noise Strength=100, Pan Noise=true, Pan Noise Field=<25 0 0>)
   &SolveForcesAndVelocity(Clamp Velocity=false, Limit Acceleration=false)
   &DynamicMaterialParameters(
    Index 0 Param 1=&FloatFromCurve(
     FloatCurve=DataInterface() {propertyValues {Curve {keys [
         {arriveTangent=-5.8452, interpMode=RCIM_Cubic, time=0, value=1}
         {arriveTangentWeight=0.0434, interpMode=RCIM_Cubic, leaveTangentWeight=0.217, tangentMode=RCTM_User, time=0.2, value=0}
         {arriveTangent=3.4164, interpMode=RCIM_Cubic, leaveTangent=3.4164, tangentMode=RCTM_User, time=0.8, value=0.7}
         <at 1 0 RCIM_Cubic>
        ]}}}
     Scale Curve=1.5
    )
    Index 0 Param 2=10
    Index 0 Param 3=700
    Index 0 Param 4=5
    Module.Param0WriteEnabled=true
    Module.Param1WriteEnabled=true
    Module.Param2WriteEnabled=true
    Module.Param3WriteEnabled=true
   )
   &ScaleColor(
    Scale RGB=&VectorFromFloat(Value=&FloatFromCurve(FloatCurve=DataInterface() {propertyValues {Curve {keys [{arriveTangent=-4.4221, interpMode=RCIM_Cubic, time=0, value=1}, <at 0.3 0>]}}}))
    ScaleA=true
    ScaleRGB=true
   )
   &ScaleSpriteSize(Uniform Curve Sprite Scale=DataInterface() {propertyValues {Curve {keys [
        {arriveTangent=40.4327, interpMode=RCIM_Cubic, leaveTangent=40.4327, tangentMode=RCTM_User, time=0, value=0.7}
        {arriveTangent=-9.8929, interpMode=RCIM_Cubic, leaveTangent=-9.8929, tangentMode=RCTM_User, time=0.07, value=1.5}
        <at 0.16 0.8>
        {interpMode=RCIM_Cubic, tangentMode=RCTM_User, time=0.75, value=0.8}
        {arriveTangent=-4.4982, interpMode=RCIM_Cubic, leaveTangent=-4.4982, tangentMode=RCTM_User, time=1, value=0}
       ]}}})
   &ScaleSpriteSize(
    Uniform Curve Index=$User.Speed
    Uniform Curve Sprite Scale=DataInterface() {propertyValues {Curve {keys [
        {arriveTangent=0.4214, interpMode=RCIM_Cubic, time=25, value=0}
        {arriveTangent=0.0444, interpMode=RCIM_Cubic, leaveTangent=0.0444, time=40, value=2}
        <at 70 2 RCIM_Cubic>
        <at 90 2 RCIM_Cubic>
       ]}}}
   ) {module=ScaleSpriteSize001}
  ]
  rendererClasses [&NiagaraSpriteRendererProperties]
  renderers [{rendererClass=&NiagaraSpriteRendererProperties}]
 }
 Chrome_Splats_Velocity {
  emitterUpdateScript [
   &EmitterState()
   &SpawnRate(
    Spawn Probability=&Lerp_Float(A=1, B=0, Alpha=&MakeFloatFromInt(Int=$User.PhysMatInt))
    SpawnRate=&FloatFromCurve(
     FloatCurve=DataInterface() {propertyValues {Curve {keys [
         {arriveTangent=-0.0078, interpMode=RCIM_Cubic, leaveTangent=-0.0078, tangentMode=RCTM_User, time=25, value=0}
         <at 29.3003 0.6081>
         {arriveTangent=0.0541, interpMode=RCIM_Cubic, leaveTangent=0.0541, time=41.6721, value=0.9785}
         {arriveTangent=0.0445, interpMode=RCIM_Cubic, leaveTangent=0.0445, time=66.2601, value=2.6064}
         <at 89.4969 3.108 RCIM_Cubic>
        ]}}}
     CurveIndex=$User.Speed
     Scale Curve=25
    )
    Use Spawn Probability=true
   )
  ]
  particleSpawnScript [
   &InitializeParticle(
    Color=<rgba 1.3728 3.0599 5 1>
    Lifetime Max=1.1333
    Lifetime Min=0.333
    Lifetime Mode=::NewEnumerator1
    Sprite Size Max=<45 60>
    Sprite Size Min=<30 40>
    Sprite Size Mode=::NewEnumerator2
    UsePositionOffset=false
    Write Material Random=true
   )
   &ShapeLocation(
    Box Size=<25 25 200>
    MaskSpawnGroup=false
    Offset=&RandomRangeVector(Minimum=<-10 -10 0>, Maximum=<10 10 0>)
    Offset Mode=::NewEnumerator0
    Override Seed=false
    Shape Primitive=::NewEnumerator2
    Surface Only=false
    UseOwnerScale=true
   )
   &AddVelocity(Velocity=&Add_Vector(
     A=&Multiply_Vector(
      A=&NormalizeVector(Vector To Normalize=$User.ThrustVelocity)
      B=&VectorFromFloat(Value=&RandomRangeFloat(Minimum=150, Maximum=950))
     )
     B=&RandomRangeVector(Minimum=<0 0 0>, Maximum=<0 0 100>)
    ))
   &AddVelocity(
    Cone Angle=15
    Cone Axis=&RandomRangeVector(Minimum=<0 0 0>, Maximum=<15 15 15>)
    Module.UseConeFalloff=false
    Module.UseVelDistribution=false
    Override Seed=false
    Velocity Mode=::NewEnumerator2
    Velocity Speed=&RandomRangeFloat(Minimum=250, Maximum=500)
   ) {module=AddVelocity001}
   &CameraOffset(Camera Offset Amount=50)
   &SubUVAnimation(End Frame=3, SubUV Animation Mode:ENiagara_SubUVLookupMode=::NewEnumerator2)
  ]
  particleUpdateScript [
   &ParticleState()
   &GravityForce()
   &CurlNoiseForce(Noise Frequency=0.25, Noise Quality / Cost=::NewEnumerator0, Noise Strength=100, Pan Noise=true, Pan Noise Field=<25 0 0>)
   &SolveForcesAndVelocity(Clamp Velocity=false, Limit Acceleration=false)
   &DynamicMaterialParameters(
    Index 0 Param 1=&FloatFromCurve(
     FloatCurve=DataInterface() {propertyValues {Curve {keys [
         {arriveTangent=-5.8452, interpMode=RCIM_Cubic, time=0, value=1}
         {arriveTangentWeight=0.0434, interpMode=RCIM_Cubic, leaveTangentWeight=0.217, tangentMode=RCTM_User, time=0.2, value=0}
         {arriveTangent=3.4164, interpMode=RCIM_Cubic, leaveTangent=3.4164, tangentMode=RCTM_User, time=0.8, value=0.7}
         <at 1 0 RCIM_Cubic>
        ]}}}
     Scale Curve=1.5
    )
    Index 0 Param 2=10
    Index 0 Param 3=700
    Index 0 Param 4=5
    Module.Param0WriteEnabled=true
    Module.Param1WriteEnabled=true
    Module.Param2WriteEnabled=true
    Module.Param3WriteEnabled=true
   )
   &ScaleColor(
    Scale Alpha=&FloatFromCurve(FloatCurve=DataInterface() {propertyValues {Curve {keys [<at 0 1 RCIM_Cubic>,<at 1 0 RCIM_Cubic>]}}})
    Scale RGB=&VectorFromFloat(Value=&FloatFromCurve(FloatCurve=DataInterface() {propertyValues {Curve {keys [{arriveTangent=-9.5152, interpMode=RCIM_Cubic, leaveTangent=-9.5152, tangentMode=RCTM_User, time=0, value=1}, <at 0.3 0>]}}}))
    ScaleA=true
    ScaleRGB=true
   )
   &ScaleSpriteSize(Uniform Curve Sprite Scale=DataInterface() {propertyValues {Curve {keys [
        {arriveTangent=40.4327, interpMode=RCIM_Cubic, leaveTangent=40.4327, tangentMode=RCTM_User, time=0, value=0.7}
        {arriveTangent=-9.8929, interpMode=RCIM_Cubic, leaveTangent=-9.8929, tangentMode=RCTM_User, time=0.07, value=1.5}
        <at 0.16 0.8>
        {interpMode=RCIM_Cubic, tangentMode=RCTM_User, time=0.75, value=0.8}
        {arriveTangent=-4.4982, interpMode=RCIM_Cubic, leaveTangent=-4.4982, tangentMode=RCTM_User, time=1, value=0}
       ]}}})
   &ScaleSpriteSize(
    Uniform Curve Index=$User.Speed
    Uniform Curve Sprite Scale=DataInterface() {propertyValues {Curve {keys [
        {arriveTangent=0.4214, interpMode=RCIM_Cubic, time=25, value=0}
        {arriveTangent=0.0444, interpMode=RCIM_Cubic, leaveTangent=0.0444, time=40, value=2}
        <at 70 2 RCIM_Cubic>
        <at 90 2 RCIM_Cubic>
       ]}}}
   ) {module=ScaleSpriteSize001}
   &ScaleSpriteSizeBySpeed(Max Scale Factor=<1 2>, Min Scale Factor=<1 0.25>)
  ]
  rendererClasses [&NiagaraSpriteRendererProperties]
  renderers [{rendererClass=&NiagaraSpriteRendererProperties}]
 }
 Smoke {
  emitterUpdateScript [
   &EmitterState()
   &SpawnBurst_Instantaneous(
    Spawn Count=&UniformRangedInt(
     Minimum=&UniformRangedInt(Minimum=1, Maximum=2)
     Maximum=&MakeIntFromFloat(Float=&FloatFromCurve(
       FloatCurve=DataInterface() {propertyValues {Curve {keys [
           {arriveTangent=-0.0078, interpMode=RCIM_Cubic, leaveTangent=-0.0078, tangentMode=RCTM_User, time=65, value=0}
           {arriveTangent=0.1798, interpMode=RCIM_Cubic, leaveTangent=0.1798, time=65.896, value=1}
           <at 70.5614 1 RCIM_Cubic>
           <at 89.4969 1 RCIM_Cubic>
          ]}}}
       CurveIndex=$User.Speed
      ))
    )
    Spawn Probability=&Lerp_Float(A=1, B=0, Alpha=&MakeFloatFromInt(Int=$User.PhysMatInt))
    Use Spawn Probability=true
   )
  ]
  particleSpawnScript [
   /Niagara/Modules/Spawn/Initialization/InitializeParticle(
    Color=<rgba 0.6309 0.5896 1 0.1>
    Lifetime=&UniformRangedFloat(Minimum=0.2, Maximum=0.5)
    Module.Write Color=true
    Module.Write Lifetime=true
    Module.Write Mass=true
    Module.Write Material Random=true
    Module.Write Position=false
    Module.Write Scale=false
    Module.Write SpriteRotation=true
    Module.Write SpriteSize=true
    Position=&SimulationPosition()
    Sprite Rotation=&UniformRangedFloat(Minimum=0, Maximum=360)
    Sprite Size=&Vector2DFromFloat(Value=&UniformRangedFloat(Minimum=100, Maximum=250))
   )
   &BoxLocation(Box Size=<15 50 15>, MaskSpawnGroup=false, Module.Override Seed=false, Rotate Box=false, Surface Only=false)
   /Niagara/Modules/Spawn/Location/SphereLocation(Hemisphere X=true, MaskSpawnGroup=false, Module.Override Randomness=false, Sphere Radius=50, Surface Only=false)
   &AddVelocity(Velocity=&Multiply_Vector(
     A=$Output.SphereLocation.SphereNormal
     B=&VectorFromFloat(Value=&UniformRangedFloat(Minimum=0, Maximum=200))
    ))
   &AddVelocity(Velocity=&Multiply_Vector(
     A=&NormalizeVector(Vector To Normalize=$User.ThrustVelocity)
     B=&VectorFromFloat(Value=&UniformRangedFloat(Minimum=200, Maximum=300))
    )) {module=AddVelocity001}
   &ApplyInitialForces()
   &CameraOffset(Camera Offset Amount=50)
  ]
  particleUpdateScript [
   &ParticleState()
   &AccelerationForce()
   &Drag(Drag=2, Use Linear Drag=true, Use Rotational Drag=false)
   &SolveForcesAndVelocity(Clamp Velocity=false, Limit Acceleration=false)
   &ScaleColor(
    Scale Alpha=&FloatFromCurve(FloatCurve=DataInterface() {propertyValues {Curve {keys [
         {interpMode=RCIM_Cubic, tangentMode=RCTM_User, time=0, value=0}
         {arriveTangent=0.002, interpMode=RCIM_Cubic, leaveTangent=0.002, tangentMode=RCTM_User, time=0.25, value=1}
         {interpMode=RCIM_Cubic, tangentMode=RCTM_User, time=1.0011, value=0.002}
        ]}}})
    Scale RGB=&Lerp_Vector(A=<1 1 1>, B=<1 1 1>, Alpha=$Particles.NormalizedAge)
    ScaleA=true
    ScaleRGB=true
   )
   &DynamicMaterialParameters(
    Index 0 Param 1=&FloatFromCurve(FloatCurve=DataInterface() {propertyValues {Curve {keys [<at 0.38 1.0113 RCIM_Cubic>,<at 1.0011 0.002 RCIM_Cubic>]}}})
    Index 0 Param 4=1
    Module.Param0WriteEnabled=true
    Module.Param1WriteEnabled=true
    Module.Param2WriteEnabled=true
    Module.Param3WriteEnabled=true
   )
   &ScaleSpriteSize(Scale Factor=&Vector2DFromFloat(Value=&FloatFromCurve(
      FloatCurve=DataInterface() {propertyValues {Curve {keys [
          {arriveTangent=0.4214, interpMode=RCIM_Cubic, time=25.4285, value=1}
          {arriveTangent=0.0224, interpMode=RCIM_Cubic, leaveTangent=0.0224, time=30.5524, value=1.4956}
          {arriveTangent=0.0085, interpMode=RCIM_Cubic, leaveTangent=0.0085, time=70, value=2}
          <at 90 2 RCIM_Cubic>
         ]}}}
      CurveIndex=$User.Speed
     ))) {module=ScaleSpriteSize001}
   &ScaleSpriteSize(Scale Factor=&Vector2DFromFloat(Value=&FloatFromCurve(FloatCurve=DataInterface() {propertyValues {Curve {keys [<at 0 1.2 RCIM_Cubic>,<at 1.0011 3.48 RCIM_Cubic>]}}})))
  ]
  rendererClasses [&NiagaraSpriteRendererProperties]
  renderers [{rendererClass=&NiagaraSpriteRendererProperties}]
 }
}
summary {
 systemName=NS_Hoagie_Scrape_LandingSkids_Chrome
 emitters [
  {emitterName=Smoke, rendererClasses [&NiagaraSpriteRendererProperties]}
  {emitterName=Chrome_Drops, rendererClasses [&NiagaraSpriteRendererProperties]}
  {emitterName=Chrome_Splats_Velocity, rendererClasses [&NiagaraSpriteRendererProperties]}
  {emitterName=Bubbles, rendererClasses [&NiagaraMeshRendererProperties, &NiagaraSpriteRendererProperties]}
 ]
 userVariables [
  User.PhysMatIndex {type=NiagaraInt32, defaultValue=1}
  User.PhysMatInt {type=NiagaraInt32, defaultValue=1}
  User.Speed {type=NiagaraFloat, defaultValue=100, description=""}
  User.ThrustVelocity {type=Vector3f, description="", defaultValue {struct=Vector3f, value=<1 0 0>}}
 ]
}
