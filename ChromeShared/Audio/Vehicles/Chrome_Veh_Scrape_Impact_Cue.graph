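# Sound cue for chrome vehicle scrape impacts, built by reusing close chrome splash grenade explosion samples.
# A mixer layers a random glass sample at half volume over a random goop sample, each through its own pitch-randomizing modulator.
# The whole cue is pitched up and quieted, uses close-detail vehicle attenuation and the exterior vehicle sound class, and overrides concurrency to cap simultaneous instances.
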SoundCue /ChromeShared/Audio/Vehicles/Chrome_Veh_Scrape_Impact_Cue
root=Mixer_1
nodes [
 {id=Mixer_1, children [Modulator_0,Modulator_1], properties {inputVolume [0.5,1]}}
 {id=Modulator_0, children [Random_1], properties {pitchMax=1.5, volumeMax=1, volumeMin=1}}
 {id=Modulator_1, children [Random_2], properties {pitchMax=1.5, volumeMax=1, volumeMin=1}}
 {
  id=Random_1
  children [WavePlayer_4,WavePlayer_5,WavePlayer_6,WavePlayer_8]
  properties {bShouldExcludeFromBranchCulling=true, weights [1,1,1,1]}
 }
 {
  id=Random_2
  children [WavePlayer_9, WavePlayer_10, WavePlayer_11, WavePlayer_12, WavePlayer_13, WavePlayer_14]
  properties {bShouldExcludeFromBranchCulling=true, weights [1,1,1,1,1,1]}
 }
 {id=WavePlayer_10, properties {soundWaveAssetPtr=/ChromeShared/Audio/ChromeSplash/Chrome_Splash_Grenade_Explo_Close_Goop_02}}
 {id=WavePlayer_11, properties {soundWaveAssetPtr=/ChromeShared/Audio/ChromeSplash/Chrome_Splash_Grenade_Explo_Close_Goop_03}}
 {id=WavePlayer_12, properties {soundWaveAssetPtr=/ChromeShared/Audio/ChromeSplash/Chrome_Splash_Grenade_Explo_Close_Goop_04}}
 {id=WavePlayer_13, properties {soundWaveAssetPtr=/ChromeShared/Audio/ChromeSplash/Chrome_Splash_Grenade_Explo_Close_Goop_05}}
 {id=WavePlayer_14, properties {soundWaveAssetPtr=/ChromeShared/Audio/ChromeSplash/Chrome_Splash_Grenade_Explo_Close_Goop_06}}
 {id=WavePlayer_4, properties {soundWaveAssetPtr=/ChromeShared/Audio/ChromeSplash/Chrome_Splash_Grenade_Explo_Close_Glass_01}}
 {id=WavePlayer_5, properties {soundWaveAssetPtr=/ChromeShared/Audio/ChromeSplash/Chrome_Splash_Grenade_Explo_Close_Glass_02}}
 {id=WavePlayer_6, properties {soundWaveAssetPtr=/ChromeShared/Audio/ChromeSplash/Chrome_Splash_Grenade_Explo_Close_Glass_03}}
 {id=WavePlayer_8, properties {soundWaveAssetPtr=/ChromeShared/Audio/ChromeSplash/Chrome_Splash_Grenade_Explo_Close_Glass_04}}
 {id=WavePlayer_9, properties {soundWaveAssetPtr=/ChromeShared/Audio/ChromeSplash/Chrome_Splash_Grenade_Explo_Close_Goop_01}}
]
properties {
 attenuationSettings=/VehicleAudio/Settings/Vehicle_Movement_Close_Detail_Focus_Att
 bOverrideConcurrency=true
 maxDistance=4300
 pitchMultiplier=1.5
 soundClassObject=&Fort_Vehicles_Exterior
 volumeMultiplier=0.5
 attenuationOverrides {bAttenuate=false, bSpatialize=false, falloffDistance=2400, attenuationShapeExtents {x=600}}
 concurrencyOverrides {maxCount=2}
}
