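# Spatialized vehicle-class Sound Cue for vehicle healing, with its own attenuation settings and a small concurrency cap that stops the lowest priority sound.
# A Branch on the FirstIndex bool plays one of four random heal sounds when true or unset, and a quieter random pulse when false.
# Pulses pass through a Quality Level node: the first three slots share a nine-variant pool, while the fourth uses only five.
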SoundCue /ChromeShared/Audio/Vehicles/Chrome_Veh_Heal_Cue
root=Branch_0
nodes [
 {id=Branch_0, children [Mixer_0,Modulator_1,Mixer_0], properties {boolParameterName=FirstIndex}}
 {id=Mixer_0, children [Modulator_0], properties {inputVolume [1]}}
 {id=Mixer_1, children [QualityLevel_0], properties {inputVolume [0.6]}}
 {id=Modulator_0, children [Random_1], properties {volumeMax=1, volumeMin=1}}
 {id=Modulator_1, children [Mixer_1], properties {pitchMax=1, pitchMin=1, volumeMax=1, volumeMin=0.9}}
 {id=QualityLevel_0, children [Random_0,Random_0,Random_0,Random_2]}
 {
  id=Random_0
  children [WavePlayer_1, WavePlayer_2, WavePlayer_7, WavePlayer_8, WavePlayer_9, WavePlayer_10, WavePlayer_11, WavePlayer_12, WavePlayer_13]
  properties {bShouldExcludeFromBranchCulling=true, weights [1,1,1,1,1,1,1,1,1]}
 }
 {
  id=Random_1
  children [WavePlayer_3,WavePlayer_4,WavePlayer_5,WavePlayer_6]
  properties {bShouldExcludeFromBranchCulling=true, weights [1,1,1,1]}
 }
 {
  id=Random_2
  children [WavePlayer_14, WavePlayer_15, WavePlayer_16, WavePlayer_17, WavePlayer_18]
  properties {bShouldExcludeFromBranchCulling=true, weights [1,1,1,1,1]}
 }
 {id=WavePlayer_1, properties {soundWaveAssetPtr=&/Chrome_Veh_Heal_Pulse_01}}
 {id=WavePlayer_10, properties {soundWaveAssetPtr=&/Chrome_Veh_Heal_Pulse_06}}
 {id=WavePlayer_11, properties {soundWaveAssetPtr=&/Chrome_Veh_Heal_Pulse_07}}
 {id=WavePlayer_12, properties {soundWaveAssetPtr=&/Chrome_Veh_Heal_Pulse_08}}
 {id=WavePlayer_13, properties {soundWaveAssetPtr=&/Chrome_Veh_Heal_Pulse_09}}
 {id=WavePlayer_14, properties {soundWaveAssetPtr=&Chrome_Veh_Heal_Pulse_01}}
 {id=WavePlayer_15, properties {soundWaveAssetPtr=&/Chrome_Veh_Heal_Pulse_02}}
 {id=WavePlayer_16, properties {soundWaveAssetPtr=&/Chrome_Veh_Heal_Pulse_03}}
 {id=WavePlayer_17, properties {soundWaveAssetPtr=&/Chrome_Veh_Heal_Pulse_04}}
 {id=WavePlayer_18, properties {soundWaveAssetPtr=&/Chrome_Veh_Heal_Pulse_05}}
 {id=WavePlayer_2, properties {soundWaveAssetPtr=&Chrome_Veh_Heal_Pulse_02}}
 {id=WavePlayer_3, properties {soundWaveAssetPtr=&/Chrome_Veh_Heal_01}}
 {id=WavePlayer_4, properties {soundWaveAssetPtr=&/Chrome_Veh_Heal_02}}
 {id=WavePlayer_5, properties {soundWaveAssetPtr=&/Chrome_Veh_Heal_03}}
 {id=WavePlayer_6, properties {soundWaveAssetPtr=&/Chrome_Veh_Heal_04}}
 {id=WavePlayer_7, properties {soundWaveAssetPtr=&Chrome_Veh_Heal_Pulse_03}}
 {id=WavePlayer_8, properties {soundWaveAssetPtr=&Chrome_Veh_Heal_Pulse_04}}
 {id=WavePlayer_9, properties {soundWaveAssetPtr=&Chrome_Veh_Heal_Pulse_05}}
]
properties {
 attenuationSettings=&/Chrome_Veh_Heal_Att
 bOverrideConcurrency=true
 maxDistance=6000
 soundClassObject=&Fort_Vehicles
 volumeMultiplier=1.35
 attenuationOverrides {falloffDistance=2400, attenuationShapeExtents {x=600}}
 concurrencyOverrides {maxCount=3, resolutionRule=StopLowestPriority}
}
