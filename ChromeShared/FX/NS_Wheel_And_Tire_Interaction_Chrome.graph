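# Niagara system for vehicle wheel and tire effects, including dust, flying debris, tire-blur meshes, peel-out spikes, chrome blobs and exhaust.
# Vehicle-fed User parameters such as per-wheel flags, speed, slide direction, surface and water depth control spawning, size and velocity.
# Spawn groups double as socket indices, with offsets for some emitters, placing particles on the vehicle skeletal mesh.
# Debris simulates on GPU with collision, while a directional spark burst emitter is disabled.
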NiagaraSystem /ChromeShared/FX/NS_Wheel_And_Tire_Interaction_Chrome
dependencies {
 usedDataInterfaces [
  &NiagaraDataInterfaceColorCurve
  &NiagaraDataInterfaceArrayColor
  &NiagaraDataInterfaceArrayInt32
  &NiagaraDataInterfaceParticleRead
 ]
 usedDynamicInputs []
 usedModules [&SystemState]
}
emitters {
 DirectionalBurstSparks002 {
  bEnabled=false
  simTarget=GPUComputeSim
  emitterUpdateScript [
   &EmitterState(Life Cycle Mode=::NewEnumerator1, Loop Delay=0.0025, Loop Duration=1, UseLoopDelay=true)
   &SpawnRate(
    SpawnGroup=1
    SpawnRate=&FloatFromCurve(
     FloatCurve=DataInterface() {propertyValues {Curve {keys [{interpMode=RCIM_Cubic, leaveTangent=-1.0025, tangentMode=RCTM_Break, time=0, value=1}, <at 1 0.55>]}}}
     CurveIndex=$Emitter.Age
     Scale Curve=&RandomRangeFloat(Minimum=65, Maximum=180)
    )
    Use Spawn Probability=false
   )
  ]
  particleSpawnScript [
   /Niagara/Modules/Spawn/Initialization/InitializeParticle(
    Color=&SelectLinearColorFromArray(
     Array Sampling Mode=::NewEnumerator0
     Direct Array Index=?
     Color Selection Array=DataInterface() {propertyValues {ColorData [<rgba 350 80 0 1>, <rgba 25 15 0 1>]}}
     Random Seed=0
    )
    Lifetime=&RandomRangeFloat(Minimum=0.25, Maximum=0.7)
    Mass=&FloatFromCurve(
     FloatCurve=DataInterface() {propertyValues {Curve {keys [<at 0 1>,<at 1 0>]}}}
     Scale Curve=&UniformRangedFloat(Minimum=4, Maximum=250)
    )
    Module.Write Color=true
    Module.Write Lifetime=true
    Module.Write Mass=true
    Module.Write Material Random=true
    Module.Write Position=false
    Module.Write Scale=false
    Module.Write SpriteRotation=false
    Module.Write SpriteSize=true
    Position=&SimulationPosition()
    Sprite Size=&Vector2DFromFloat(Value=&UniformRangedFloat(Minimum=5, Maximum=10))
   )
   /SetVariables(Particles.Position=&self:ScratchDynamicInput_01(New Particle Attribute Reader:NiagaraDataInterfaceParticleRead=DataInterface() {propertyValues {EmitterBinding {bindingMode=Other, emitterName=RandomDirBurst}}})) {isSetParameters=true}
   /Niagara/Modules/Update/Mass/CalculateSizeAndRotationalInertiaByMass(
    Depth=0
    Height=5
    Module.Manually Enter Density=false
    Width:NiagaraFloat=&RandomRangeFloat(Minimum=0.4, Maximum=0.8)
   )
   /SetVariables(Particles.dir:Vector3f=&self:GetVelocityDirection(New Particle Attribute Reader:NiagaraDataInterfaceParticleRead=DataInterface() {propertyValues {EmitterBinding {bindingMode=Other, emitterName=RandomDirBurst}}})) {isSetParameters=true}
   &ShapeLocation(
    Cylinder Height=1
    Cylinder Radius=15
    MaskSpawnGroup=false
    Non Uniform Scale=<1 1 0.25>
    Override Seed=false
    Rotation Axis=$Particles.dir
    Rotation Mode=::NewEnumerator1
    Shape Primitive=::NewEnumerator1
    Surface Only=true
    Surface Only Band Thickness=0.1
    Use Endcaps In Surface Only Mode=false
    UseOwnerScale=true
   )
   &AddVelocity(
    Module.Use Falloff Distance=false
    Origin Offset Coordinate Space=::World
    Override Seed=false
    Velocity Mode=::NewEnumerator1
    Velocity Origin=$Particles.Position
    Velocity Speed=&FloatFromCurve(
     FloatCurve=DataInterface() {propertyValues {Curve {keys [<at 0 0>, {interpMode=RCIM_Cubic, leaveTangent=-1.6226, tangentMode=RCTM_Break, time=0.05, value=1}, <at 1 0.4>]}}}
     CurveIndex=$Emitter.Age
     Scale Curve=&RandomRangeFloat(Minimum=25, Maximum=1500)
    )
   ) {enabled=false}
   &AddVelocity(
    Cone Angle=8
    Cone Axis=&TransformVector(Vector=$Particles.dir, Source Space=::World, Destination Space=::Local)
    Module.UseConeFalloff=false
    Module.UseVelDistribution=false
    Override Seed=false
    Velocity Mode=::NewEnumerator2
    Velocity Speed=&FloatFromCurve(
     FloatCurve=DataInterface() {propertyValues {Curve {keys [<at 0 0.45>, {interpMode=RCIM_Cubic, leaveTangent=-1.6226, tangentMode=RCTM_Break, time=0.05, value=1}, <at 1 0.4>]}}}
     CurveIndex=$Emitter.Age
     Scale Curve=&RandomRangeFloat(Minimum=25, Maximum=1500)
    )
   ) {module=AddVelocity001}
  ]
  particleUpdateScript [
   &ParticleState()
   &GravityForce(Gravity=&Lerp_Vector(
     A=<0 0 0>
     B=&RandomRangeVector(Minimum=<0 0 -500>, Maximum=<1 1 -3500>)
     Alpha=$Particles.NormalizedAge
    ))
   &ScaleColor(
    Scale Alpha=&FloatFromCurve(FloatCurve=DataInterface() {propertyValues {Curve {keys [<at -0.0046 1.0113 RCIM_Cubic>,<at 1.0011 0.002 RCIM_Cubic>]}}})
    ScaleA=true
    ScaleRGB=true
   )
   /Niagara/Modules/Update/Forces/CurlNoiseForce(
    Noise Frequency=0.005
    Noise Quality / Cost=::NewEnumerator0
    Noise Strength=&FloatFromCurve(
     FloatCurve=DataInterface() {propertyValues {Curve {keys [{interpMode=RCIM_Cubic, leaveTangent=2.4969, tangentMode=RCTM_Break, time=0, value=0}, <at 1 1>]}}}
     Scale Curve=150000
    )
    Pan Noise=true
    Pan Noise Field=<25 0 25>
   )
   &Drag(
    Drag=&FloatFromCurve(FloatCurve=DataInterface() {propertyValues {Curve {keys [
         {interpMode=RCIM_Cubic, leaveTangent=-6.9985, tangentMode=RCTM_Break, time=0, value=1}
         {arriveTangent=-0.1691, interpMode=RCIM_Cubic, leaveTangent=-0.1691, tangentMode=RCTM_User, time=0.1, value=0.3}
         <at 1 0.125>
        ]}}})
    Use Linear Drag=true
    Use Rotational Drag=false
   )
   &Collision(
    Advanced Aging Rate=1
    Age Colliding Particles=false
    Collision Enabled=/Niagara/DynamicInputs/Helpers/MaskBoolBySpawnGroup(
     Masked Bool:NiagaraBool=false
     Passthrough Bool:NiagaraBool=&SetBoolByFloatComparison(A=$Particles.Age, B=0.15)
     Spawn Group Mask:NiagaraInt32=0
    )
    Kill Offscreen Particles=true
    Module.Control Rest State Via Surface Normal=false
    Module.EnableMaxCollisionCount=false
    Module.Randomize Collision Normal Vector=true
   )
   &SolveForcesAndVelocity(Clamp Velocity=false, Limit Acceleration=false)
   &ScaleSpriteSizeBySpeed(
    Max Scale Factor=<0.7 0.5>
    Min Scale Factor=<0.4 0.1>
    Sample Scale Factor By Curve=true
    Scale Factor Curve=DataInterface() {propertyValues {Curve {keys [
        {arriveTangent=2.865, interpMode=RCIM_Cubic, time=0.1247, value=-0.0006}
        {arriveTangent=2.5675, interpMode=RCIM_Cubic, leaveTangent=2.5675, tangentMode=RCTM_User, time=0.3948, value=0.4893}
        <at 0.9958 0.9974 RCIM_Cubic>
       ]}}}
    Velocity Threshold=1500
   ) {module=SpriteSizeScaleByVelocity}
   &ScaleSpriteSize(Scale Factor=&Vector2DFromFloat(Value=&FloatFromCurve(FloatCurve=DataInterface() {propertyValues {Curve {keys [
          <at 0 0>
          <at 0.1 1>
          {interpMode=RCIM_Cubic, tangentMode=RCTM_Break, time=0.7172, value=1}
          {arriveTangent=-9.028, interpMode=RCIM_Cubic, leaveTangent=-9.028, tangentMode=RCTM_User, time=1, value=0}
         ]}}})))
  ]
  rendererClasses [&NiagaraSpriteRendererProperties]
  renderers [{rendererClass=&NiagaraSpriteRendererProperties}]
 }
 Dust_Debris {
  simTarget=GPUComputeSim
  emitterUpdateScript [
   &EmitterState(Life Cycle Mode=::NewEnumerator1, Loop Duration=1, UseLoopDelay=false)
   /SetVariables(Emitter.ChunkSpawnRateScale:NiagaraFloat=&Multiply_Float(
     A=&MakeFloatFromInt(Int=$User.Dust_Particles_Per_Wheel)
     B=&Add_Float(A=&self:ScratchAbs(New float:NiagaraFloat=$User.Slide_Direction), B=1.5)
    )) {isSetParameters=true, module=SetVariables_#1}
   &SpawnRate(
    Spawn Probability=&MakeCustomFloatFromBool(
     Bool=$User.PeelOut_Wheel_RL
     True Float=&Multiply_Float(
      A=$Emitter.ChunkSpawnRateScale
      B=&MakeFloatFromInt(Int=&RandomRangeInt(Minimum=0, Maximum=1))
     )
     False Float=0
    )
    SpawnGroup=2
    SpawnRate=$Emitter.ChunkSpawnRateScale
    Use Spawn Probability=true
   ) {module=SpawnRate002}
   &SpawnRate(
    Spawn Probability=&MakeCustomFloatFromBool(
     Bool=$User.PeelOut_Wheel_RR
     True Float=&Multiply_Float(
      A=$Emitter.ChunkSpawnRateScale
      B=&MakeFloatFromInt(Int=&RandomRangeInt(Minimum=0, Maximum=1))
     )
     False Float=0
    )
    SpawnGroup=3
    SpawnRate=$Emitter.ChunkSpawnRateScale
    Use Spawn Probability=true
   ) {module=SpawnRate003}
  ]
  particleSpawnScript [
   /SetVariables(Particles.RandomInt:NiagaraInt32=&UniformRangedInt(Minimum=0, Maximum=1)) {isSetParameters=true, module=SetVariables_#2}
   &InitializeParticle(
    Color=&RandomRangeLinearColor(Minimum=<rgba 0.2448 0.2448 0.2448 1>, Maximum=<rgba 1 1 1 1>)
    Lifetime=&Lerp_Float(A=0.7, B=2, Alpha=$User.Vehicle_Speed)
    Mass=&UniformRangedFloat(Minimum=1, Maximum=1)
    Mesh Scale=&UniformRangedVector(
     Minimum=&VectorFromFloat(Value=0.05)
     Maximum=&VectorFromFloat(Value=0.25)
    )
    Module.Write Color=true
    Module.Write Lifetime=true
    Module.Write Mass=true
    Module.Write Material Random=true
    Module.Write Position=true
    Module.Write Scale=true
    Module.Write SpriteRotation=false
    Module.Write SpriteSize=false
    Position=$Engine.Owner.Position
   )
   &CalculateSizeAndRotationalInertiaByMass(
    Calculate Mesh Scale=true
    Density by Material Type:ENiagaraMassByVolume=::NewEnumerator1
    Initial Model Dimensions:Vector3f=<30 30 30>
    Mass Modulation=4
    Module.Manually Enter Density=false
   )
   &MeshRotationForce(Rotation=&RandomRangeVector(Minimum=<-0.0055 -0.0055 -0.0055>, Maximum=<0.0055 0.0055 0.0055>)) {module=MeshRotationForce001}
   &SkeletalMeshLocation(
    Bone / Socket Index=&SpawnGroup()
    Inherit Velocity=false
    Interpolate Mesh Sampling=false
    MaskSpawnGroup=false
    Mesh Sampling Type=::NewEnumerator5
    Sampled Position Offset=&MakeVector(
     X=0
     Y=&UniformRangedFloat(Minimum=&Multiply_Float(A=$User.Wheel_Width, B=-1), Maximum=$User.Wheel_Width)
     Z=&Add_Float(A=10, B=&MakeCustomFloatFromBool(Bool=$User.Mod_OffroadTires, True Float=-15, False Float=0))
    )
    Sampled Position Offset Coordinate Space:ENiagaraMeshTransforms=::NewEnumerator3
    Skeletal Mesh=$User.Vehicle_Selection_02
    Socket Sampling Mode=::NewEnumerator5
    Write Sampled Mesh Attrributes to Particle Attributes=true
   )
   /Niagara/Modules/Spawn/Orientation/V2/InitialMeshOrientation(
    Rotation=&UniformRangedVector(Minimum=<-3 -3 -3>, Maximum=<3 3 3>)
    Use Orientation Vector=true
    Use Rotation Vector=true
   ) {enabled=false}
   &MeshRotationForce(Rotation=&RandomRangeVector(Minimum=<0 0 0>, Maximum=<1 1 1>)) {enabled=false}
   /Niagara/Modules/Spawn/Orientation/V3/InitialMeshOrientation(
    Rotation=&UniformRangedVector(Minimum=<-3 -3 -3>, Maximum=<3 3 3>)
    Use Orientation Vector=true
    Use Rotation Vector=true
   ) {module=InitialMeshOrientation001}
   &AddVelocityInCone(
    Cone Angle=115
    Cone Axis=<1 0 0.85>
    Module.Use Velocity Falloff On Cone Axis=true
    Velocity Distribution Along Cone Axis=0.25
    Velocity Falloff Away From Cone Axis=0.333
    Velocity Strength=&Lerp_Float(
     A=&UniformRangedFloat(Minimum=0, Maximum=1)
     B=&UniformRangedFloat(Minimum=300, Maximum=800)
     Alpha=$User.Vehicle_Speed
    )
   )
   &AddVelocity(Velocity=&MakeVector(X=0, Y=&Multiply_Float(A=$User.Slide_Direction, B=1800), Z=0))
   &ApplyInitialForces()
  ]
  particleUpdateScript [
   &ParticleState()
   &GravityForce()
   &Drag(Drag=&RandomRangeFloat(Minimum=0.5, Maximum=1.25), Use Linear Drag=true, Use Rotational Drag=false)
   &ScaleColor(
    Scale Alpha=&FloatFromCurve(FloatCurve=DataInterface() {propertyValues {Curve {keys [<at -0.0046 1.0113 RCIM_Cubic>,<at 1.0011 0.002 RCIM_Cubic>]}}})
    ScaleA=true
    ScaleRGB=true
   )
   &Collision(
    Advanced Aging Rate=1
    Age Colliding Particles=false
    Collision Enabled=&Boolean_AndOperation(
     Bool1=&SetBoolByFloatComparison(A=&UniformRangedFloat(Minimum=0, Maximum=1), B=0.25)
     Bool2=&SetBoolByFloatComparison(A=$Particles.Age, B=0.05)
    )
    Control Roll On Collision=true
    Friction=0.4
    Friction During a Bounce=0.2
    GPU Collision Type=::NewEnumerator2
    Module.Control Rest State Via Surface Normal=false
    Module.EnableMaxCollisionCount=false
    Module.Randomize Collision Normal Vector=true
    Restitution=0.7
    ScaleRateOfRotation=false
   )
   &SolveForcesAndVelocity(Clamp Velocity=false, Limit Acceleration=false)
   &SolveRotationalForcesAndVelocity()
   &ScaleMeshSizeBySpeed(Max Scale Factor=<1 1 1>, Min Scale Factor:Vector3f=<0.7 0.7 0.7>)
   &ScaleMeshSize(Scale Factor=&VectorFromFloat(Value=&FloatFromCurve(FloatCurve=DataInterface() {propertyValues {Curve {keys [
          <at 0 0.075>
          <at 0.0756 1>
          {arriveTangent=-0.2, interpMode=RCIM_Cubic, leaveTangent=-0.2, time=0.7956, value=1.0207}
          {arriveTangent=-5.4083, interpMode=RCIM_Cubic, leaveTangent=-5.4083, tangentMode=RCTM_User, time=1, value=0}
         ]}}})))
  ]
  rendererClasses [&NiagaraMeshRendererProperties]
  renderers [
   {rendererClass=&NiagaraMeshRendererProperties, rendererIndex=0}
   {rendererClass=&NiagaraMeshRendererProperties, rendererIndex=1}
  ]
 }
 Dust_Debris001 {
  simTarget=GPUComputeSim
  emitterUpdateScript [
   &EmitterState(Life Cycle Mode=::NewEnumerator1, Loop Duration=1, UseLoopDelay=false)
   /SetVariables(Emitter.ChunkSpawnRateScale:NiagaraFloat=&Multiply_Float(
     A=&Multiply_Float(
      A=&MakeFloatFromInt(Int=$User.Dust_Particles_Per_Wheel)
      B=&Add_Float(A=&self:ScratchAbs(New float:NiagaraFloat=$User.Slide_Direction), B=1.5)
     )
     B=10
    )) {isSetParameters=true, module=SetVariables_#1}
   &SpawnRate(
    Spawn Probability=&MakeCustomFloatFromBool(
     Bool=$User.PeelOut_Wheel_RL
     True Float=&Multiply_Float(
      A=$Emitter.ChunkSpawnRateScale
      B=&MakeFloatFromInt(Int=&RandomRangeInt(Minimum=0, Maximum=1))
     )
     False Float=0
    )
    SpawnGroup=2
    SpawnRate=$Emitter.ChunkSpawnRateScale
    Use Spawn Probability=true
   ) {module=SpawnRate002}
   &SpawnRate(
    Spawn Probability=&MakeCustomFloatFromBool(
     Bool=$User.PeelOut_Wheel_RR
     True Float=&Multiply_Float(
      A=$Emitter.ChunkSpawnRateScale
      B=&MakeFloatFromInt(Int=&RandomRangeInt(Minimum=0, Maximum=1))
     )
     False Float=0
    )
    SpawnGroup=3
    SpawnRate=$Emitter.ChunkSpawnRateScale
    Use Spawn Probability=true
   ) {module=SpawnRate003}
  ]
  particleSpawnScript [
   /SetVariables(Particles.RandomInt:NiagaraInt32=&UniformRangedInt(Minimum=0, Maximum=1)) {isSetParameters=true, module=SetVariables_#2}
   &InitializeParticle(
    Color=<rgba 1.3728 3.0599 5 1>
    Lifetime=&Lerp_Float(A=0.35, B=1, Alpha=$User.Vehicle_Speed)
    Mass=&UniformRangedFloat(Minimum=1, Maximum=1)
    Module.Write Color=true
    Module.Write Lifetime=true
    Module.Write Mass=true
    Module.Write Material Random=true
    Module.Write Position=true
    Module.Write Scale=false
    Module.Write SpriteRotation=false
    Module.Write SpriteSize=true
    Position=$Engine.Owner.Position
    Sprite Size=&Vector2DFromFloat(Value=&Lerp_Float(
      A=4
      B=8
      Alpha=&FloatFromCurve(
       FloatCurve=DataInterface() {propertyValues {Curve {keys [
           {interpMode=RCIM_Cubic, tangentMode=RCTM_User, time=0, value=0}
           {arriveTangent=1.2457, interpMode=RCIM_Cubic, leaveTangent=1.2457, tangentMode=RCTM_User, time=0.7348, value=0.4334}
           {interpMode=RCIM_Cubic, tangentMode=RCTM_User, time=1, value=1}
          ]}}}
       CurveIndex=&RandomRangeFloat(Minimum=0, Maximum=1)
      )
     ))
   )
   &SkeletalMeshLocation(
    Bone / Socket Index=&SpawnGroup()
    Inherit Velocity=false
    Interpolate Mesh Sampling=false
    MaskSpawnGroup=false
    Mesh Sampling Type=::NewEnumerator5
    Sampled Position Offset=&MakeVector(
     X=0
     Y=&UniformRangedFloat(Minimum=&Multiply_Float(A=$User.Wheel_Width, B=-1), Maximum=$User.Wheel_Width)
     Z=&Add_Float(A=10, B=&MakeCustomFloatFromBool(Bool=$User.Mod_OffroadTires, True Float=-15, False Float=0))
    )
    Sampled Position Offset Coordinate Space:ENiagaraMeshTransforms=::NewEnumerator3
    Skeletal Mesh=$User.Vehicle_Selection_02
    Socket Sampling Mode=::NewEnumerator5
    Write Sampled Mesh Attrributes to Particle Attributes=true
   )
   &AddVelocityInCone(
    Cone Angle=75
    Cone Axis=<1 0 0.85>
    Module.Use Velocity Falloff On Cone Axis=true
    Velocity Distribution Along Cone Axis=0.25
    Velocity Falloff Away From Cone Axis=0.333
    Velocity Strength=&Lerp_Float(
     A=&UniformRangedFloat(Minimum=0, Maximum=1)
     B=&UniformRangedFloat(Minimum=300, Maximum=800)
     Alpha=$User.Vehicle_Speed
    )
   )
   &AddVelocity(Velocity=&MakeVector(X=0, Y=&Multiply_Float(A=$User.Slide_Direction, B=1800), Z=0))
  ]
  particleUpdateScript [
   &ParticleState()
   &GravityForce()
   &Drag(Drag=&RandomRangeFloat(Minimum=0.5, Maximum=1.25), Use Linear Drag=true, Use Rotational Drag=false)
   &ScaleColor(
    Scale Alpha=&FloatFromCurve(FloatCurve=DataInterface() {propertyValues {Curve {keys [<at -0.0046 1.0113 RCIM_Cubic>,<at 1.0011 0.002 RCIM_Cubic>]}}})
    Scale RGB=&VectorFromFloat(Value=&FloatFromCurve(FloatCurve=DataInterface() {propertyValues {Curve {keys [{arriveTangent=-4.4221, interpMode=RCIM_Cubic, time=0, value=1}, <at 0.5 0>]}}}))
    ScaleA=true
    ScaleRGB=true
   )
   &Collision(
    Advanced Aging Rate=1
    Age Colliding Particles=false
    Collision Enabled=&Boolean_AndOperation(
     Bool1=&SetBoolByFloatComparison(A=&UniformRangedFloat(Minimum=0, Maximum=1), B=0.25)
     Bool2=&SetBoolByFloatComparison(A=$Particles.Age, B=0.15)
    )
    Friction=0.6
    Friction During a Bounce=0.6
    GPU Collision Type=::NewEnumerator2
    Module.Control Rest State Via Surface Normal=false
    Module.EnableMaxCollisionCount=false
    Module.Randomize Collision Normal Vector=true
    Restitution=0.4
   )
   &SolveForcesAndVelocity(Clamp Velocity=false, Limit Acceleration=false)
   &ScaleSpriteSize(Uniform Curve Sprite Scale=DataInterface() {propertyValues {Curve {keys [
        {arriveTangent=8.2127, interpMode=RCIM_Cubic, leaveTangent=8.2127, tangentMode=RCTM_User, time=0, value=0.5}
        {
         arriveTangent=4.2448
         arriveTangentWeight=0.232
         interpMode=RCIM_Cubic
         leaveTangent=4.2448
         leaveTangentWeight=2.2538
         tangentMode=RCTM_User
         time=0.07
         value=1
        }
        {arriveTangent=-2.5194, interpMode=RCIM_Cubic, leaveTangent=-2.5194, tangentMode=RCTM_User, time=0.75, value=0.8}
        {arriveTangent=-4.4982, interpMode=RCIM_Cubic, leaveTangent=-4.4982, tangentMode=RCTM_User, time=1, value=0}
       ]}}})
   &DynamicMaterialParameters(
    Index 0 Param 1=&FloatFromCurve(
     FloatCurve=DataInterface() {propertyValues {Curve {keys [
         {arriveTangent=-5.8452, interpMode=RCIM_Cubic, time=0, value=1}
         {arriveTangentWeight=0.0434, interpMode=RCIM_Cubic, leaveTangentWeight=0.217, tangentMode=RCTM_User, time=0.2, value=0}
         {arriveTangent=3.4164, interpMode=RCIM_Cubic, leaveTangent=3.4164, tangentMode=RCTM_User, time=0.8, value=0.7}
         <at 1 0 RCIM_Cubic>
        ]}}}
     Scale Curve=1.5
    )
    Index 0 Param 2=10
    Index 0 Param 3=700
    Index 0 Param 4=5
    Module.Param0WriteEnabled=true
    Module.Param1WriteEnabled=true
    Module.Param2WriteEnabled=true
    Module.Param3WriteEnabled=true
   )
  ]
  rendererClasses [&NiagaraSpriteRendererProperties]
  renderers [{rendererClass=&NiagaraSpriteRendererProperties}]
 }
 Dust_Tires_Mesh {
  emitterUpdateScript [
   &EmitterState(Life Cycle Mode=::NewEnumerator1, Loop Duration=0.05, UseLoopDelay=false)
   &SpawnBurst_Instantaneous(
    Spawn Count=&MakeIntFromFloat(Float=&Multiply_Float(
      A=&MakeCustomFloatFromBool(Bool=$User.IsWaterDeep, True Float=0, False Float=1)
      B=&self:ScratchGetFloatFromArray(Index:NiagaraInt32=2, WheelsTouchingArray:NiagaraDataInterfaceArrayFloat=$User.WheelsTouchingArray)
     ))
    Spawn Probability=&self:IsAllTrue(Bool A:NiagaraBool=$User.Dust_Wheel_RL, Bool B:NiagaraBool=$User.TireIntactRL)
    Use Spawn Probability=true
   ) {module=SpawnBurst_Instantaneous002}
   &SpawnBurst_Instantaneous(
    Spawn Count=&MakeIntFromFloat(Float=&Multiply_Float(
      A=&MakeCustomFloatFromBool(Bool=$User.IsWaterDeep, True Float=0, False Float=1)
      B=&self:ScratchGetFloatFromArray(Index:NiagaraInt32=3, WheelsTouchingArray:NiagaraDataInterfaceArrayFloat=$User.WheelsTouchingArray)
     ))
    Spawn Group=1
    Spawn Probability=&self:IsAllTrue(Bool A:NiagaraBool=$User.Dust_Wheel_RR, Bool B:NiagaraBool=$User.TireIntactRR)
    Use Spawn Probability=true
   ) {module=SpawnBurst_Instantaneous003}
  ]
  particleSpawnScript [
   &InitializeParticle(
    Color=&UniformRangedLinearColor(
     Minimum=&MultiplyLinearColor(Linear Color=$User.DustTireBlurColor, Scale Factor=<1 1 1 0.065>)
     Maximum=&MultiplyLinearColor(Linear Color=$User.DustTireBlurColor, Scale Factor=<1 1 1 0.11>)
    )
    Lifetime=&UniformRangedFloat(Minimum=0.225, Maximum=0.3485)
    Mass=&UniformRangedFloat(Minimum=0.75, Maximum=1.75)
    Mesh Scale=$User.Tire_Scale_Back
    Module.Write Color=true
    Module.Write Lifetime=true
    Module.Write Mass=false
    Module.Write Material Random=true
    Module.Write Position=false
    Module.Write Scale=true
    Module.Write SpriteRotation=false
    Module.Write SpriteSize=false
    Position=$Engine.Owner.Position
   ) {enabled=false}
   /Niagara/Modules/Spawn/Initialization/V2/InitializeParticle(
    Color=&UniformRangedLinearColor(
     Minimum=&MultiplyLinearColor(Linear Color=$User.DustTireBlurColor, Scale Factor=<1 1 1 0.25>)
     Maximum=&MultiplyLinearColor(Linear Color=$User.DustTireBlurColor, Scale Factor=<1 1 1 0.45>)
    )
    Lifetime Max=0.3485
    Lifetime Min=0.225
    Lifetime Mode=::NewEnumerator1
    Mesh Scale=&Multiply_VectorByFloat(Vector=$User.Tire_Scale_Back, Float=1.05)
    Mesh Scale Mode=::NewEnumerator4
    UsePositionOffset=false
    Write Material Random=true
   ) {module=InitializeParticle001}
   &SkeletalMeshLocation(
    Apply Sampled Orientation to Mesh Particles=true
    Bone / Socket Index=&Add_Integer(A=&SpawnGroup(), B=8)
    Inherit Velocity=false
    Interpolate Mesh Sampling=false
    MaskSpawnGroup=false
    Mesh Sampling Type=::NewEnumerator5
    Sampled Position Offset Coordinate Space:ENiagaraMeshTransforms=::NewEnumerator3
    Skeletal Mesh=$User.Vehicle_Selection_02
    Socket Sampling Mode=::NewEnumerator5
    Use Sampled Mesh Attribute as Ribbon ID=false
    Write Sampled Mesh Attrributes to Particle Attributes=true
   ) {module=SkeletalMeshLocation001}
  ]
  particleUpdateScript [
   &ParticleState()
   &SkeletalMeshLocation(
    Apply Sampled Orientation to Mesh Particles=true
    Bone / Socket Index=&Add_Integer(A=&SpawnGroup(), B=8)
    Inherit Velocity=false
    Interpolate Mesh Sampling=false
    MaskSpawnGroup=false
    Mesh Sampling Type=::NewEnumerator5
    Sampled Position Offset Coordinate Space:ENiagaraMeshTransforms=::NewEnumerator3
    Skeletal Mesh=$User.Vehicle_Selection_02
    Socket Sampling Mode=::NewEnumerator5
    Write Sampled Mesh Attrributes to Particle Attributes=true
   )
   &DynamicMaterialParameters(
    Index 0 Param 1=1
    Index 0 Param 2=0.125
    Module.Param0WriteEnabled=true
    Module.Param1WriteEnabled=true
    Module.Param2WriteEnabled=true
    Module.Param3WriteEnabled=true
   )
   &ScaleColor(
    Scale Alpha=&FloatFromCurve(FloatCurve=DataInterface() {propertyValues {Curve {keys [
         {arriveTangent=0.002, interpMode=RCIM_Cubic, leaveTangent=0.002, tangentMode=RCTM_Break, time=0.004, value=1.0113}
         <at 1.0011 0.002 RCIM_Cubic>
        ]}}})
    ScaleA=true
    ScaleRGB=true
   )
   &Color(Color=&LerpLinearColors(
     StartColor=<rgba 12 0 10.2952 1>
     EndColor=<rgba 0 12 1.404 1>
     LerpFactor=&MakeCustomFloatFromBool(Bool=$User.TireIntactRL, True Float=0, False Float=1)
    )) {enabled=false}
  ]
  rendererClasses [&NiagaraMeshRendererProperties]
  renderers [{rendererClass=&NiagaraMeshRendererProperties, rendererIndex=0}]
 }
 Dust_Tires_World {
  emitterUpdateScript [
   &EmitterState(Life Cycle Mode=::NewEnumerator1, Loop Duration=1, UseLoopDelay=false)
   &SpawnPerUnit(
    Max Movement Threshold=1000
    Module.Use Max Movement Threshold=false
    Module.Use Movement Tolerance=false
    Spawn Per Unit:NiagaraFloat=&Lerp_Float(A=12, B=800, Alpha=&Power(a=$User.Vehicle_Speed, b=1.4))
    Spawn Probability:NiagaraFloat=&Multiply_Float(
     A=&MakeCustomFloatFromBool(Bool=$User.Dust_Wheel_FL, True Float=1, False Float=0)
     B=&MakeCustomFloatFromBool(
      Bool=$User.IsOnDirt
      True Float=1
      False Float=&FloatFromCurve(FloatCurve=$User.Dust_SpawnCurve, CurveIndex=$User.Vehicle_Speed)
     )
    )
    Use Spawn Probability=true
   )
   &SpawnPerUnit(
    Max Movement Threshold=1000
    Module.Use Max Movement Threshold=false
    Module.Use Movement Tolerance=false
    Spawn Group=1
    Spawn Per Unit:NiagaraFloat=&Lerp_Float(A=12, B=800, Alpha=&Power(a=$User.Vehicle_Speed, b=1.4))
    Spawn Probability:NiagaraFloat=&Multiply_Float(
     A=&MakeCustomFloatFromBool(Bool=$User.Dust_Wheel_FR, True Float=1, False Float=0)
     B=&MakeCustomFloatFromBool(
      Bool=$User.IsOnDirt
      True Float=1
      False Float=&FloatFromCurve(FloatCurve=$User.Dust_SpawnCurve, CurveIndex=$User.Vehicle_Speed)
     )
    )
    Use Spawn Probability=true
   ) {module=SpawnPerUnit001}
   &SpawnPerUnit(
    Max Movement Threshold=1000
    Module.Use Max Movement Threshold=false
    Module.Use Movement Tolerance=false
    Spawn Group=2
    Spawn Per Unit:NiagaraFloat=&Lerp_Float(A=12, B=800, Alpha=&Power(a=$User.Vehicle_Speed, b=1.4))
    Spawn Probability:NiagaraFloat=&Multiply_Float(
     A=&MakeCustomFloatFromBool(Bool=$User.Dust_Wheel_RL, True Float=1, False Float=0)
     B=&MakeCustomFloatFromBool(
      Bool=$User.IsOnDirt
      True Float=1
      False Float=&FloatFromCurve(FloatCurve=$User.Dust_SpawnCurve, CurveIndex=$User.Vehicle_Speed)
     )
    )
    Use Spawn Probability=true
   ) {module=SpawnPerUnit002}
   &SpawnPerUnit(
    Max Movement Threshold=1000
    Module.Use Max Movement Threshold=false
    Module.Use Movement Tolerance=false
    Spawn Group=3
    Spawn Per Unit:NiagaraFloat=&Lerp_Float(A=12, B=800, Alpha=&Power(a=$User.Vehicle_Speed, b=1.4))
    Spawn Probability:NiagaraFloat=&Multiply_Float(
     A=&MakeCustomFloatFromBool(Bool=$User.Dust_Wheel_RR, True Float=1, False Float=0)
     B=&MakeCustomFloatFromBool(
      Bool=$User.IsOnDirt
      True Float=1
      False Float=&FloatFromCurve(FloatCurve=$User.Dust_SpawnCurve, CurveIndex=$User.Vehicle_Speed)
     )
    )
    Use Spawn Probability=true
   ) {module=SpawnPerUnit003}
   &SpawnPerUnit(
    Max Movement Threshold=1000
    Module.Use Max Movement Threshold=false
    Module.Use Movement Tolerance=false
    Spawn Group=4
    Spawn Per Unit:NiagaraFloat=&Lerp_Float(A=12, B=800, Alpha=&Power(a=$User.Vehicle_Speed, b=1.4))
    Spawn Probability:NiagaraFloat=&Multiply_Float(
     A=&MakeCustomFloatFromBool(Bool=$User.Dust_Wheel_ML, True Float=1, False Float=0)
     B=&MakeCustomFloatFromBool(
      Bool=$User.IsOnDirt
      True Float=1
      False Float=&FloatFromCurve(FloatCurve=$User.Dust_SpawnCurve, CurveIndex=$User.Vehicle_Speed)
     )
    )
    Use Spawn Probability=true
   ) {module=SpawnPerUnit004}
   &SpawnPerUnit(
    Max Movement Threshold=1000
    Module.Use Max Movement Threshold=false
    Module.Use Movement Tolerance=false
    Spawn Group=5
    Spawn Per Unit:NiagaraFloat=&Lerp_Float(A=12, B=800, Alpha=&Power(a=$User.Vehicle_Speed, b=1.4))
    Spawn Probability:NiagaraFloat=&Multiply_Float(
     A=&MakeCustomFloatFromBool(Bool=$User.Dust_Wheel_MR, True Float=1, False Float=0)
     B=&MakeCustomFloatFromBool(
      Bool=$User.IsOnDirt
      True Float=1
      False Float=&FloatFromCurve(FloatCurve=$User.Dust_SpawnCurve, CurveIndex=$User.Vehicle_Speed)
     )
    )
    Use Spawn Probability=true
   ) {module=SpawnPerUnit005}
  ]
  particleSpawnScript [
   /Niagara/Modules/Spawn/Initialization/InitializeParticle(
    Color=&MakeLinearColorFromVectorAndFloat(
     "Vector (RGB)"=&Lerp_Vector(A=<1 1 1>, B=<1 0.2 0.1>, Alpha=$User.DustColorLerp)
     "Float (Alpha)"=&ClampFloat(
      Float=&FloatFromCurve(FloatCurve=$User.Dust_AlphaCurve, CurveIndex=$User.Vehicle_Speed, Scale Curve=$User.DustAlphaMultiplier)
      Max=0.1
     )
    )
    Lifetime=&Lerp_Float(
     A=&UniformRangedFloat(Minimum=1.4, Maximum=2)
     B=&UniformRangedFloat(Minimum=0.6, Maximum=0.4)
     Alpha=$User.Vehicle_Speed
    )
    Module.Write Color=true
    Module.Write Lifetime=true
    Module.Write Mass=false
    Module.Write Material Random=true
    Module.Write Position=true
    Module.Write Scale=false
    Module.Write SpriteRotation=true
    Module.Write SpriteSize=true
    Position=$Engine.Owner.Position
    Sprite Rotation=&UniformRangedFloat(Minimum=0, Maximum=360)
    Sprite Size=&Vector2DFromFloat(Value=&Lerp_Float(
      A=&Lerp_Float(
       A=&UniformRangedFloat(Minimum=90, Maximum=150)
       B=&UniformRangedFloat(Minimum=350, Maximum=500)
       Alpha=$User.Vehicle_Speed
      )
      B=700
      Alpha=&self:ScratchAbs(New float:NiagaraFloat=$User.Slide_Direction)
     ))
   )
   &SkeletalMeshLocation(
    Bone / Socket Index=&SpawnGroup()
    Inherit Velocity=false
    Interpolate Mesh Sampling=false
    MaskSpawnGroup=false
    Mesh Sampling Type=::NewEnumerator5
    Sampled Position Offset=&MakeVector(
     X=0
     Y=&UniformRangedFloat(Minimum=&Multiply_Float(A=$User.Wheel_Width, B=-1), Maximum=$User.Wheel_Width)
     Z=&MakeCustomFloatFromBool(Bool=$User.Mod_OffroadTires, True Float=-15, False Float=0)
    )
    Skeletal Mesh=$User.Vehicle_Selection_02
    Socket Sampling Mode=::NewEnumerator5
    Write Sampled Mesh Attrributes to Particle Attributes=true
   )
   &AddVelocityInCone(
    Cone Angle=12
    Cone Axis=&MakeVector(X=1, Y=$User.Slide_Direction, Z=0)
    Module.Use Velocity Falloff On Cone Axis=true
    Velocity Distribution Along Cone Axis=0.6
    Velocity Falloff Away From Cone Axis=0.333
    Velocity Strength=&Lerp_Float(A=0, B=1, Alpha=$User.Vehicle_Speed)
   )
   &AddVelocity(Velocity=&MakeVector(X=0, Y=&Multiply_Float(A=$User.Slide_Direction, B=300), Z=0))
  ]
  particleUpdateScript [
   &ParticleState()
   &GravityForce(Gravity=<0 0 50>)
   &Drag(Drag=0.5, Use Linear Drag=true, Use Rotational Drag=false)
   &ScaleColor(
    Scale Alpha=&FloatFromCurve(FloatCurve=DataInterface() {propertyValues {Curve {keys [
         {interpMode=RCIM_Cubic, tangentMode=RCTM_Break, time=-0.0054, value=0.0052}
         {arriveTangent=10.4414, interpMode=RCIM_Cubic, leaveTangent=-2.1615, tangentMode=RCTM_Break, time=0.0811, value=1.0113}
         {arriveTangent=-0.7171, interpMode=RCIM_Cubic, leaveTangent=-0.7171, tangentMode=RCTM_User, time=0.1932, value=0.3197}
         <at 1.0011 0.002 RCIM_Cubic>
        ]}}})
    ScaleA=true
    ScaleRGB=true
   )
   &SolveForcesAndVelocity(Clamp Velocity=false, Limit Acceleration=false)
   /Niagara/Modules/Update/SubUV/SubUVAnimation(Number Of Frames=&FloatFromCurve(FloatCurve=DataInterface() {propertyValues {Curve {keys [<at 0 0>,<at 0.1797 0.7182>,<at 1 1>]}}}, Scale Curve=63))
   &ScaleSpriteSize(Scale Factor=&Vector2DFromFloat(Value=&FloatFromCurve(FloatCurve=DataInterface() {propertyValues {Curve {keys [{arriveTangent=2.0487, interpMode=RCIM_Cubic, leaveTangent=2.0487, tangentMode=RCTM_User, time=0, value=0.15}, <at 1 1>]}}})))
  ]
  rendererClasses [&NiagaraSpriteRendererProperties]
  renderers [{rendererClass=&NiagaraSpriteRendererProperties}]
 }
 Exhaust_Local {
  emitterUpdateScript [
   &EmitterState(Life Cycle Mode=::NewEnumerator1, Loop Duration=1, UseLoopDelay=false)
   &SpawnRate(
    Spawn Probability=&MakeCustomFloatFromBool(Bool=&Boolean_AndOperation(Bool1=$User.VehicleHasDriver, Bool2=$User.VehicleHasFuel), True Float=1, False Float=0)
    SpawnRate=&Multiply_Float(A=24, B=&MakeCustomFloatFromBool(Bool=$User.IsWaterDeep, True Float=0, False Float=1))
    Use Spawn Probability=true
   )
   &SpawnRate(
    Spawn Probability=&MakeCustomFloatFromBool(Bool=&Boolean_AndOperation(Bool1=$User.VehicleHasDriver, Bool2=$User.VehicleHasFuel), True Float=1, False Float=0)
    SpawnGroup=1
    SpawnRate=&Multiply_Float(A=24, B=&MakeCustomFloatFromBool(Bool=$User.IsWaterDeep, True Float=0, False Float=1))
    Use Spawn Probability=true
   ) {module=SpawnRate001}
  ]
  particleSpawnScript [
   /SetVariables(Particles.RandomFloat=&UniformAOrBFloat(A=-1)) {isSetParameters=true, module=SetVariables_#3}
   &InitializeParticle(
    Color=$User.ExhaustColor
    Lifetime=&Lerp_Float(A=&UniformRangedFloat(Minimum=0.225, Maximum=0.8), B=1, Alpha=$User.Vehicle_Speed)
    Module.Write Color=true
    Module.Write Lifetime=true
    Module.Write Mass=false
    Module.Write Material Random=true
    Module.Write Position=true
    Module.Write Scale=false
    Module.Write SpriteRotation=true
    Module.Write SpriteSize=true
    Position=$Engine.Owner.Position
    Sprite Rotation=&UniformRangedFloat(Minimum=0, Maximum=360)
    Sprite Size=&Vector2DFromFloat(Value=&Lerp_Float(
      A=&UniformRangedFloat(Minimum=40, Maximum=60)
      B=&UniformRangedFloat(Minimum=65, Maximum=80)
      Alpha=$User.Vehicle_Speed
     ))
   )
  ]
  particleUpdateScript [
   &ParticleState()
   &SkeletalMeshLocation(
    Bone / Socket Index=&Add_Integer(A=&SpawnGroup(), B=12)
    Inherit Velocity=true
    "Inherit Velocity (Scale)"=-0.05
    Interpolate Mesh Sampling=false
    MaskSpawnGroup=false
    Mesh Sampling Type=::NewEnumerator5
    Sampled Position Offset=&MakeVector(X=0, Y=0, Z=0)
    Sampled Position Offset Coordinate Space:ENiagaraMeshTransforms=::NewEnumerator3
    Skeletal Mesh=$User.Vehicle_Selection_02
    Socket Sampling Mode=::NewEnumerator5
    VelocitySpeedLimit=false
    Write Sampled Mesh Attrributes to Particle Attributes=true
   ) {module=SkeletalMeshLocation001}
   /SetVariables(Particles.Position=&Add_Vector(
     A=$Particles.SkeletalMeshLocation001.SampledPosition
     B=&Multiply_VectorByFloat(
      Vector=&RotateVectorByQuaternion(
       Vector To Rotate=&MakeVector(X=1, Y=&Lerp_Float(A=0, B=$User.Slide_Direction, Alpha=$Particles.NormalizedAge), Z=0)
       Quaternion=$Particles.SkeletalMeshLocation001.SampledRotation
      )
      Float=&Multiply_Float(A=$Particles.NormalizedAge, B=&Lerp_Float(A=125, B=100, Alpha=$User.Vehicle_Speed))
     )
    )) {isSetParameters=true, module=SetVariables_#4}
   &ScaleColor(
    Scale Alpha=&FloatFromCurve(FloatCurve=DataInterface() {propertyValues {Curve {keys [
         <at 0 0>
         {arriveTangent=0.002, interpMode=RCIM_Cubic, leaveTangent=-2.5776, tangentMode=RCTM_Break, time=0.1559, value=1.0113}
         <at 1.0011 0.002 RCIM_Cubic>
        ]}}})
    ScaleA=true
    ScaleRGB=true
   )
   &SubUVAnimation(Number Of Frames=&FloatFromCurve(
     FloatCurve=DataInterface() {propertyValues {Curve {keys [
         {interpMode=RCIM_Cubic, leaveTangent=9.8981, tangentMode=RCTM_Break, time=0, value=0}
         {arriveTangent=1, interpMode=RCIM_Cubic, leaveTangent=1, tangentMode=RCTM_Break, time=0.075, value=0.5807}
         <at 1 1>
        ]}}}
     Scale Curve=&Multiply_Float(A=$Particles.RandomFloat, B=63)
    ))
   &ScaleSpriteSize(Scale Factor=&Vector2DFromFloat(Value=&FloatFromCurve(
      FloatCurve=DataInterface() {propertyValues {Curve {keys [{interpMode=RCIM_Cubic, tangentMode=RCTM_Break, time=0, value=0.179}, <at 0.8329 1>]}}}
      Scale Curve=$User.ExhaustScale
     )))
   &SpriteRotationRate(Rotation Rate=&Multiply_Float(A=$Particles.Initial.MaterialRandom, B=1))
  ]
  rendererClasses [&NiagaraSpriteRendererProperties]
  renderers [{rendererClass=&NiagaraSpriteRendererProperties}]
 }
 Peel_Tires_Spikes {
  emitterUpdateScript [
   &EmitterState(Life Cycle Mode=::NewEnumerator1, Loop Duration=1, UseLoopDelay=false)
   &SpawnRate(
    Spawn Probability=&Multiply_Float(
     A=&MakeCustomFloatFromBool(Bool=$User.IsWaterDeep, True Float=0, False Float=1)
     B=&self:ScratchGetFloatFromArray(Index:NiagaraInt32=2, WheelsTouchingArray:NiagaraDataInterfaceArrayFloat=DataInterface() {propertyValues {FloatData []}})
    )
    SpawnRate=&MakeFloatFromInt(Int=&MakeCustomIntFromBool(Bool=$User.PeelOut_Wheel_RL, True INT=&Multiply_Int(A=$User.Dust_Particles_Per_Wheel, B=4), False INT=0))
    Use Spawn Probability=true
   )
   &SpawnRate(
    Spawn Probability=&Multiply_Float(
     A=&MakeCustomFloatFromBool(Bool=$User.IsWaterDeep, True Float=0, False Float=1)
     B=&self:ScratchGetFloatFromArray(Index:NiagaraInt32=3, WheelsTouchingArray:NiagaraDataInterfaceArrayFloat=DataInterface() {propertyValues {FloatData []}})
    )
    SpawnGroup=1
    SpawnRate=&MakeFloatFromInt(Int=&MakeCustomIntFromBool(Bool=$User.PeelOut_Wheel_RR, True INT=&Multiply_Int(A=$User.Dust_Particles_Per_Wheel, B=4), False INT=0))
    Use Spawn Probability=true
   ) {module=SpawnRate001}
  ]
  particleSpawnScript [
   &InitializeParticle(
    Color=<rgba 1 0.9167 0.7917 0.35>
    Lifetime=&UniformRangedFloat(Minimum=0.11, Maximum=0.14)
    Module.Write Color=true
    Module.Write Lifetime=true
    Module.Write Mass=true
    Module.Write Material Random=true
    Module.Write Position=false
    Module.Write Scale=false
    Module.Write SpriteRotation=true
    Module.Write SpriteSize=true
    Position=&SimulationPosition()
    Sprite Rotation=&UniformRangedFloat(Minimum=0, Maximum=0)
    Sprite Size=&MakeVector2D(
     X=&UniformRangedFloat(Minimum=15, Maximum=25)
     Y=&Lerp_Float(
      A=&UniformRangedFloat(Minimum=35, Maximum=70)
      B=&UniformRangedFloat(Minimum=100, Maximum=180)
      Alpha=$User.Vehicle_Speed
     )
    )
   )
   &AddVelocityInCone(
    Cone Angle=27
    Cone Axis=&MakeVector(X=1, Y=$User.Slide_Direction, Z=0.5)
    Module.Use Velocity Falloff On Cone Axis=true
    Velocity Distribution Along Cone Axis=0.6
    Velocity Falloff Away From Cone Axis=0.333
    Velocity Strength=1900
   )
   &SubUVAnimation(Number Of Frames=3, SubUV Animation Mode=::NewEnumerator1)
  ]
  particleUpdateScript [
   &ParticleState()
   &SkeletalMeshLocation(
    Bone / Socket Index=&Add_Integer(A=&SpawnGroup(), B=2)
    Inherit Velocity=false
    MaskSpawnGroup=false
    Mesh Sampling Type=::NewEnumerator5
    Sampled Position Offset=&MakeVector(
     X=0
     Y=&UniformRangedFloat(Minimum=&Multiply_Float(A=$User.Wheel_Width_Back, B=-1), Maximum=$User.Wheel_Width_Back)
     Z=&Multiply_Float(A=-15, B=&MakeCustomFloatFromBool(Bool=$User.Mod_OffroadTires, True Float=1, False Float=0))
    )
    Skeletal Mesh=$User.Vehicle_Selection_02
    Socket Sampling Mode=::NewEnumerator5
    Write Sampled Mesh Attrributes to Particle Attributes=true
   )
   &SolveForcesAndVelocity(Clamp Velocity=false, Limit Acceleration=false)
   &DynamicMaterialParameters(
    Index 0 Param 1=&FloatFromCurve(FloatCurve=DataInterface() {propertyValues {Curve {keys [<at -0.0046 1.0113>,<at 1.0011 -1>]}}})
    Index 0 Param 4=0.5
    Module.Param0WriteEnabled=true
    Module.Param1WriteEnabled=true
    Module.Param2WriteEnabled=true
    Module.Param3WriteEnabled=true
   )
   &ScaleColor(
    Scale Alpha=&FloatFromCurve(FloatCurve=DataInterface() {propertyValues {Curve {keys [<at 0 0>, {arriveTangent=-3.1724, interpMode=RCIM_Cubic, leaveTangent=-3.1724, tangentMode=RCTM_User, time=0.0788, value=1.0113}, <at 1 0>]}}})
    ScaleA=true
    ScaleRGB=true
   )
   &ScaleSpriteSize(Scale Factor=&Vector2DFromFloat(Value=&FloatFromCurve(FloatCurve=DataInterface() {propertyValues {Curve {keys [{arriveTangent=2.3953, interpMode=RCIM_Cubic, leaveTangent=2.3953, tangentMode=RCTM_User, time=0, value=0.2}, <at 1 1>]}}})))
  ]
  rendererClasses [&NiagaraSpriteRendererProperties]
  renderers [{rendererClass=&NiagaraSpriteRendererProperties}]
 }
 Peel_Tires_Spikes_ChromeBlobs {
  simTarget=GPUComputeSim
  emitterUpdateScript [
   &EmitterState(Life Cycle Mode=::NewEnumerator1, Loop Duration=1, UseLoopDelay=false)
   &SpawnRate(
    Spawn Probability=&Multiply_Float(
     A=&MakeCustomFloatFromBool(Bool=$User.IsWaterDeep, True Float=0, False Float=1)
     B=&self:ScratchGetFloatFromArray(Index:NiagaraInt32=2, WheelsTouchingArray:NiagaraDataInterfaceArrayFloat=DataInterface() {propertyValues {FloatData []}})
    )
    SpawnRate=&MakeFloatFromInt(Int=&MakeCustomIntFromBool(Bool=$User.PeelOut_Wheel_RL, True INT=&Multiply_Int(A=$User.Dust_Particles_Per_Wheel, B=10), False INT=0))
    Use Spawn Probability=true
   )
   &SpawnRate(
    Spawn Probability=&Multiply_Float(
     A=&MakeCustomFloatFromBool(Bool=$User.IsWaterDeep, True Float=0, False Float=1)
     B=&self:ScratchGetFloatFromArray(Index:NiagaraInt32=3, WheelsTouchingArray:NiagaraDataInterfaceArrayFloat=DataInterface() {propertyValues {FloatData []}})
    )
    SpawnGroup=1
    SpawnRate=&MakeFloatFromInt(Int=&MakeCustomIntFromBool(Bool=$User.PeelOut_Wheel_RR, True INT=&Multiply_Int(A=$User.Dust_Particles_Per_Wheel, B=10), False INT=0))
    Use Spawn Probability=true
   ) {module=SpawnRate001}
  ]
  particleSpawnScript [
   &InitializeParticle(
    Color=<rgba 1.3728 3.0599 5 1>
    Lifetime=&UniformRangedFloat(Minimum=0.11, Maximum=0.3)
    Module.Write Color=true
    Module.Write Lifetime=true
    Module.Write Mass=true
    Module.Write Material Random=true
    Module.Write Position=false
    Module.Write Scale=false
    Module.Write SpriteRotation=true
    Module.Write SpriteSize=true
    Position=&SimulationPosition()
    Sprite Rotation=&RandomRangeFloat(Minimum=0, Maximum=360)
    Sprite Size=&RandomRangeVector2D(Minimum=<30 40>, Maximum=<45 60>)
   )
   &SkeletalMeshLocation(
    Bone / Socket Index=&Add_Integer(A=&SpawnGroup(), B=2)
    Inherit Velocity=false
    MaskSpawnGroup=false
    Mesh Sampling Type=::NewEnumerator5
    Sampled Position Offset=&MakeVector(
     X=0
     Y=&UniformRangedFloat(Minimum=&Multiply_Float(A=$User.Wheel_Width_Back, B=-1), Maximum=$User.Wheel_Width_Back)
     Z=&Multiply_Float(A=-15, B=&MakeCustomFloatFromBool(Bool=$User.Mod_OffroadTires, True Float=1, False Float=0))
    )
    Skeletal Mesh=$User.Vehicle_Selection_02
    Socket Sampling Mode=::NewEnumerator5
    Write Sampled Mesh Attrributes to Particle Attributes=true
   )
   &AddVelocityInCone(
    Cone Angle=27
    Cone Axis=&MakeVector(X=1, Y=$User.Slide_Direction, Z=0.5)
    Module.Use Velocity Falloff On Cone Axis=true
    Velocity Distribution Along Cone Axis=0.6
    Velocity Falloff Away From Cone Axis=0.333
    Velocity Strength=1900
   )
   /Niagara/Modules/Update/SubUV/V2/SubUVAnimation(End Frame=3, SubUV Animation Mode:ENiagara_SubUVLookupMode=::NewEnumerator2)
  ]
  particleUpdateScript [
   &ParticleState()
   &SolveForcesAndVelocity(Clamp Velocity=false, Limit Acceleration=false)
   &ScaleColor(
    Scale Alpha=&FloatFromCurve(FloatCurve=DataInterface() {propertyValues {Curve {keys [<at 0 0>, {arriveTangent=-3.1724, interpMode=RCIM_Cubic, leaveTangent=-3.1724, tangentMode=RCTM_User, time=0.0788, value=1.0113}, <at 1 0>]}}})
    Scale RGB=&VectorFromFloat(Value=&FloatFromCurve(FloatCurve=DataInterface() {propertyValues {Curve {keys [{arriveTangent=-6.5524, interpMode=RCIM_Cubic, leaveTangent=-6.5524, tangentMode=RCTM_User, time=0, value=1}, <at 0.5 0>]}}}))
    ScaleA=true
    ScaleRGB=true
   )
  ]
  rendererClasses [&NiagaraSpriteRendererProperties]
  renderers [{rendererClass=&NiagaraSpriteRendererProperties}]
 }
 RandomDirBurst {
  bEnabled=false
  simTarget=GPUComputeSim
  emitterUpdateScript [
   &EmitterState(
    Life Cycle Mode=::NewEnumerator1
    Loop Duration=&RandomRangeFloat(Minimum=1.25, Maximum=4, Evaluation Type=::NewEnumerator1)
    UseLoopDelay=false
   )
   &SpawnBurst_Instantaneous(Use Spawn Probability=true, UseLoopCountLimit=false)
  ]
  particleSpawnScript [
   &InitializeParticle(
    Color=&SelectLinearColorFromArray(
     Array Sampling Mode=::NewEnumerator0
     Direct Array Index=?
     Color Selection Array=DataInterface() {propertyValues {ColorData [<rgba 350 80 0 1>, <rgba 25 15 0 1>]}}
     Random Seed=0
    )
    Lifetime=&RandomRangeFloat(Minimum=1, Maximum=1)
    Mass=&UniformRangedFloat(Minimum=80, Maximum=350)
    Module.Write Color=true
    Module.Write Lifetime=true
    Module.Write Mass=true
    Module.Write Material Random=true
    Module.Write Position=false
    Module.Write Scale=false
    Module.Write SpriteRotation=false
    Module.Write SpriteSize=true
    Position=&SimulationPosition()
    Sprite Size=&Vector2DFromFloat(Value=&UniformRangedFloat(Minimum=5, Maximum=10))
   )
   &SkeletalMeshLocation(
    Bone / Socket Index=&SelectIntFromArray(
     Array Sampling Mode=::NewEnumerator1
     Direct Array Index=?
     Int Selection Array=DataInterface() {propertyValues {IntData [0,1,2,3]}}
     Random Seed=0
    )
    MaskSpawnGroup=false
    Mesh Sampling Type=::NewEnumerator5
    Skeletal Mesh=$User.Vehicle_Selection_02
    Socket Sampling Mode=::NewEnumerator5
   )
   &CalculateSizeAndRotationalInertiaByMass(Depth=0, Height=0.5, Mass Modulation=3, Module.Manually Enter Density=false, Width=0.033)
   &AddVelocity(
    Cone Angle=5
    Cone Axis=&MakeVector(
     X=&RandomRangeFloat(Minimum=-1, Maximum=-0.5)
     Y=&self:ChooseVehicleSideVelocity(ParticlePosition:Vector3f=Linked() {linkedVariable {name=Particles.Position, type=NiagaraPosition}})
     Z=&RandomRangeFloat(Minimum=0.35, Maximum=0.7)
    )
    Module.UseConeFalloff=false
    Module.UseVelDistribution=false
    Override Seed=false
    Velocity Mode=::NewEnumerator2
    Velocity Speed=&RandomRangeFloat(Minimum=900, Maximum=1650)
   )
  ]
  particleUpdateScript [
   &ParticleState()
   &GravityForce(Gravity=&UniformRangedVector(Minimum=<0 0 -980>, Maximum=<1 1 -1800>))
   &Drag(Drag=&UniformRangedFloat(Minimum=0.5, Maximum=1), Use Linear Drag=true, Use Rotational Drag=false)
   &ScaleColor(
    Scale Alpha=&FloatFromCurve(FloatCurve=DataInterface() {propertyValues {Curve {keys [<at -0.0046 1.0113 RCIM_Cubic>,<at 1.0011 0.002 RCIM_Cubic>]}}})
    ScaleA=true
    ScaleRGB=true
   )
   &Collision(
    Advanced Aging Rate=1
    Age Colliding Particles=false
    Collision Enabled=false
    Kill Offscreen Particles=true
    Module.Control Rest State Via Surface Normal=false
    Module.EnableMaxCollisionCount=false
    Module.Randomize Collision Normal Vector=true
   ) {enabled=false}
   &CurlNoiseForce(Noise Frequency=0.1, Noise Quality / Cost=::NewEnumerator0, Noise Strength=50000, Pan Noise=true, Pan Noise Field=<25 0 0>)
   /SetVariables(Particles.PositionOld:NiagaraPosition=$Particles.Position) {isSetParameters=true}
   &SolveForcesAndVelocity(Clamp Velocity=false, Limit Acceleration=false)
   /SetVariables(Particles.dir:Vector3f=&Subtract_Vector(A=Linked() {linkedVariable {name=Particles.Position, type=NiagaraPosition}}, B=$Particles.PositionOld)) {isSetParameters=true}
   &ScaleSpriteSizeBySpeed(
    Max Scale Factor=<0.7 0.5>
    Min Scale Factor=<0.4 0.1>
    Sample Scale Factor By Curve=true
    Scale Factor Curve=DataInterface() {propertyValues {Curve {keys [
        {arriveTangent=2.865, interpMode=RCIM_Cubic, time=0.1247, value=-0.0006}
        {arriveTangent=2.5675, interpMode=RCIM_Cubic, leaveTangent=2.5675, tangentMode=RCTM_User, time=0.3948, value=0.4893}
        <at 0.9958 0.9974 RCIM_Cubic>
       ]}}}
    Velocity Threshold=1500
   ) {module=SpriteSizeScaleByVelocity}
   &ScaleSpriteSize(Scale Factor=&Vector2DFromFloat(Value=&FloatFromCurve(
      FloatCurve=DataInterface() {propertyValues {Curve {keys [
          {interpMode=RCIM_Cubic, tangentMode=RCTM_Break, time=0.7172, value=1}
          {arriveTangent=-9.028, interpMode=RCIM_Cubic, leaveTangent=-9.028, tangentMode=RCTM_User, time=1, value=0}
         ]}}}
      Scale Curve=1.25
     )))
  ]
  rendererClasses [&NiagaraSpriteRendererProperties]
  renderers [{rendererClass=&NiagaraSpriteRendererProperties}]
 }
 ShiftBlip {
  emitterUpdateScript [
   &EmitterState(Life Cycle Mode=::NewEnumerator1, Loop Duration=0.1, UseLoopDelay=false)
   &SpawnBurst_Instantaneous(
    Spawn Count=&Multiply_Int(
     A=&RandomRangeInt(Minimum=1, Maximum=2)
     B=&MakeCustomIntFromBool(Bool=$User.IsWaterDeep, True INT=0, False INT=1)
    )
    Spawn Probability=&MakeCustomFloatFromBool(
     Bool=&Boolean_AndOperation(Bool1=$User.VehicleHasDriver, Bool2=&Boolean_AndOperation(Bool1=$User.VehicleHasFuel, Bool2=$User.IsShifting))
     True Float=1
     False Float=0
    )
    Use Spawn Probability=true
   )
   &SpawnBurst_Instantaneous(
    Spawn Count=&Multiply_Int(
     A=&RandomRangeInt(Minimum=1, Maximum=2)
     B=&MakeCustomIntFromBool(Bool=$User.IsWaterDeep, True INT=0, False INT=1)
    )
    Spawn Group=1
    Spawn Probability=&MakeCustomFloatFromBool(
     Bool=&Boolean_AndOperation(Bool1=$User.VehicleHasDriver, Bool2=&Boolean_AndOperation(Bool1=$User.VehicleHasFuel, Bool2=$User.IsShifting))
     True Float=1
     False Float=0
    )
    Use Spawn Probability=true
   ) {module=SpawnBurst_Instantaneous001}
  ]
  particleSpawnScript [
   /SetVariables(Particles.RandomFloat=&UniformAOrBFloat(A=-1)) {isSetParameters=true, module=SetVariables_#3}
   &InitializeParticle(
    Color=<rgba 0.6062 0.754 1 0.025>
    Lifetime=&MakeCustomFloatFromBool(
     Bool=$User.Downshifting
     True Float=&RandomRangeFloat(Minimum=0.1, Maximum=0.14)
     False Float=&RandomRangeFloat(Minimum=0.225, Maximum=0.375)
    )
    Module.Write Color=true
    Module.Write Lifetime=true
    Module.Write Mass=false
    Module.Write Material Random=true
    Module.Write Position=true
    Module.Write Scale=false
    Module.Write SpriteRotation=true
    Module.Write SpriteSize=true
    Position=$Engine.Owner.Position
    Sprite Rotation=&UniformRangedFloat(Minimum=0, Maximum=360)
    Sprite Size=&Vector2DFromFloat(Value=&MakeCustomFloatFromBool(
      Bool=$User.Downshifting
      True Float=&RandomRangeFloat(Minimum=40, Maximum=60)
      False Float=&RandomRangeFloat(Minimum=55, Maximum=80)
     ))
   )
   &DynamicMaterialParameters(
    Index 0 Param 1=&FloatFromCurve(FloatCurve=DataInterface() {propertyValues {Curve {keys [
         {arriveTangent=8.397, interpMode=RCIM_Cubic, leaveTangent=4.6996, tangentMode=RCTM_User, time=-0.0012, value=-0.0261}
         {arriveTangent=1.3472, interpMode=RCIM_Cubic, leaveTangent=1.3472, tangentMode=RCTM_User, time=0.1801, value=0.7926}
         <at 0.5 1 RCIM_Cubic>
        ]}}})
    Index 0 Param 2=&UniformRangedFloat(Minimum=0.333, Maximum=1)
    Index 0 Param 3=&FloatFromCurve(FloatCurve=DataInterface() {propertyValues {Curve {keys [{arriveTangent=1.4224, interpMode=RCIM_Cubic, time=0.5, value=1}, <at 1 2 RCIM_Cubic>]}}})
    Index 0 Param 4=&UniformRangedFloat(Minimum=1.333, Maximum=2)
    Module.Param0WriteEnabled=true
    Module.Param1WriteEnabled=true
    Module.Param2WriteEnabled=true
    Module.Param3WriteEnabled=true
   ) {module=DynamicMaterialParameters001}
  ]
  particleUpdateScript [
   &ParticleState()
   &SkeletalMeshLocation(
    Bone / Socket Index=&Add_Integer(A=&SpawnGroup(), B=12)
    Inherit Velocity=true
    "Inherit Velocity (Scale)"=-0.05
    Interpolate Mesh Sampling=false
    MaskSpawnGroup=false
    Mesh Sampling Type=::NewEnumerator5
    Sampled Position Offset=&MakeVector(X=0, Y=0, Z=0)
    Sampled Position Offset Coordinate Space:ENiagaraMeshTransforms=::NewEnumerator3
    Skeletal Mesh=$User.Vehicle_Selection_02
    Socket Sampling Mode=::NewEnumerator5
    VelocitySpeedLimit=false
    Write Sampled Mesh Attrributes to Particle Attributes=true
   ) {module=SkeletalMeshLocation001}
   /SetVariables(Particles.Position=&Add_Vector(
     A=$Particles.SkeletalMeshLocation001.SampledPosition
     B=&Multiply_VectorByFloat(
      Vector=&RotateVectorByQuaternion(
       Vector To Rotate=&MakeVector(X=1, Y=0, Z=0)
       Quaternion=$Particles.SkeletalMeshLocation001.SampledRotation
      )
      Float=&Multiply_Float(A=$Particles.NormalizedAge, B=&Lerp_Float(A=90, B=60, Alpha=$User.Vehicle_Speed))
     )
    )) {isSetParameters=true, module=SetVariables_#4}
   /SetVariables(Particles.Color2=&ColorFromCurve(ColorCurve=DataInterface() {propertyValues {
       AlphaCurve {keys [<at 0.4847 1 RCIM_Cubic>,<at 1 0 RCIM_Cubic>]}
       BlueCurve {keys [<at 0.0194 50 RCIM_Cubic>,<at 0.1503 150>,<at 0.1866 50>,<at 0.3975 0 RCIM_Cubic>]}
       GreenCurve {keys [<at 0.0194 30.8898 RCIM_Cubic>,<at 0.1503 92.6695>,<at 0.1866 0>,<at 0.3975 5.5901 RCIM_Cubic>]}
       RedCurve {keys [<at 0.0194 0.1 RCIM_Cubic>,<at 0.1503 0>,<at 0.1866 10.781>,<at 0.3975 25 RCIM_Cubic>]}
      }})) {isSetParameters=true}
   /SetVariables(Particles.SubUVIndex2:NiagaraFloat=&Add_Float(
     A=&RandomRangeFloat(Minimum=0, Maximum=32)
     B=&Lerp_Float(A=0, B=60, Alpha=$Particles.NormalizedAge)
    )) {isSetParameters=true}
   &ScaleSpriteSize(Scale Factor=&Vector2DFromFloat(Value=&FloatFromCurve(FloatCurve=DataInterface() {propertyValues {Curve {keys [{interpMode=RCIM_Cubic, leaveTangent=2.377, tangentMode=RCTM_Break, time=0, value=0.179}, <at 0.8329 1>]}}})))
   &SpriteRotationRate(Rotation Rate=&Multiply_Float(A=$Particles.Initial.MaterialRandom, B=1))
   &DynamicMaterialParameters(
    Index 0 Param 1=&FloatFromCurve(FloatCurve=DataInterface() {propertyValues {Curve {keys [{arriveTangent=2.3188, interpMode=RCIM_Cubic, time=0, value=0}, <at 0.5 1 RCIM_Cubic>]}}})
    Index 0 Param 3=&FloatFromCurve(FloatCurve=DataInterface() {propertyValues {Curve {keys [{arriveTangent=1.4224, interpMode=RCIM_Cubic, time=0.5, value=1}, <at 1 2 RCIM_Cubic>]}}})
    Module.Param0WriteEnabled=true
    Module.Param1WriteEnabled=false
    Module.Param2WriteEnabled=true
    Module.Param3WriteEnabled=false
   )
  ]
  rendererClasses [&NiagaraSpriteRendererProperties]
  renderers [{rendererClass=&NiagaraSpriteRendererProperties}]
 }
 ShiftBlipLight {
  emitterUpdateScript [
   &EmitterState(Life Cycle Mode=::NewEnumerator1, Loop Duration=0.1, UseLoopDelay=false)
   &SpawnBurst_Instantaneous(
    Spawn Count=&Multiply_Int(A=1, B=&MakeCustomIntFromBool(Bool=$User.IsWaterDeep, True INT=0, False INT=1))
    Spawn Probability=&MakeCustomFloatFromBool(
     Bool=&Boolean_AndOperation(Bool1=$User.VehicleHasDriver, Bool2=&Boolean_AndOperation(Bool1=$User.VehicleHasFuel, Bool2=$User.IsShifting))
     True Float=1
     False Float=0
    )
    Use Spawn Probability=true
   )
   &SpawnBurst_Instantaneous(
    Spawn Count=&Multiply_Int(A=1, B=&MakeCustomIntFromBool(Bool=$User.IsWaterDeep, True INT=0, False INT=1))
    Spawn Group=1
    Spawn Probability=&MakeCustomFloatFromBool(
     Bool=&Boolean_AndOperation(Bool1=$User.VehicleHasDriver, Bool2=&Boolean_AndOperation(Bool1=$User.VehicleHasFuel, Bool2=$User.IsShifting))
     True Float=1
     False Float=0
    )
    Use Spawn Probability=true
   ) {module=SpawnBurst_Instantaneous001}
  ]
  particleSpawnScript [&InitializeParticle(
    Lifetime=&RandomRangeFloat(Minimum=0.085, Maximum=0.125)
    Module.Write Color=true
    Module.Write Lifetime=true
    Module.Write Mass=false
    Module.Write Material Random=true
    Module.Write Position=true
    Module.Write Scale=false
    Module.Write SpriteRotation=true
    Module.Write SpriteSize=true
    Position=$Engine.Owner.Position
   )]
  particleUpdateScript [
   &ParticleState()
   &SkeletalMeshLocation(
    Bone / Socket Index=&Add_Integer(A=&SpawnGroup(), B=12)
    Inherit Velocity=false
    "Inherit Velocity (Scale)"=-0.05
    Interpolate Mesh Sampling=false
    MaskSpawnGroup=false
    Mesh Sampling Type=::NewEnumerator5
    Sampled Position Offset=&MakeVector(X=15, Y=0, Z=0)
    Sampled Position Offset Coordinate Space:ENiagaraMeshTransforms=::NewEnumerator3
    Skeletal Mesh=$User.Vehicle_Selection_02
    Socket Sampling Mode=::NewEnumerator5
    Write Sampled Mesh Attrributes to Particle Attributes=true
   ) {module=SkeletalMeshLocation001}
   &FN_Light_Attributes(
    Color=&ColorFromCurve(ColorCurve=DataInterface() {propertyValues {
       AlphaCurve {keys [<at 0.0118 0>,<at 0.3542 1 RCIM_Cubic>,<at 1 0 RCIM_Cubic>]}
       BlueCurve {keys [<at 0.054 0>,<at 0.4426 0 RCIM_Cubic>]}
       GreenCurve {keys [<at 0.054 4.7963>,<at 0.4426 2.6832 RCIM_Cubic>]}
       RedCurve {keys [<at 0.054 24>,<at 0.4426 12 RCIM_Cubic>]}
      }})
    Create Color Attribute=true
    Create Exponent Attribute=true
    Create Radius Attribute=true
    Exponent=8
    Radius=120
    Scale Color=&MakeCustomFloatFromBool(Bool=$User.Downshifting, True Float=0.15, False Float=1)
   )
  ]
  rendererClasses [&NiagaraLightRendererProperties]
  renderers [{rendererClass=&NiagaraLightRendererProperties, rendererIndex=0}]
 }
 Sparks_Tires_Blown_Local {
  emitterUpdateScript [
   &EmitterState(Life Cycle Mode=::NewEnumerator1, Loop Duration=1, UseLoopDelay=false)
   &SpawnPerUnit(
    Module.Use Max Movement Threshold=true
    Module.Use Movement Tolerance=false
    Spawn Per Unit:NiagaraFloat=&Lerp_Float(A=$User.SparkSPU_MinSpeed, B=$User.SparkSPU_MaxSpeed, Alpha=&Power(a=$User.Vehicle_Speed, b=1.25))
    Spawn Probability:NiagaraFloat=&MakeCustomFloatFromBool(Bool=$User.TireIntactFL, True Float=&UniformRangedFloat(Minimum=0, Maximum=0), False Float=1)
    Use Spawn Probability=true
   )
   &SpawnPerUnit(
    Module.Use Max Movement Threshold=true
    Module.Use Movement Tolerance=false
    Spawn Group=1
    Spawn Per Unit:NiagaraFloat=&Lerp_Float(A=$User.SparkSPU_MinSpeed, B=$User.SparkSPU_MaxSpeed, Alpha=&Power(a=$User.Vehicle_Speed, b=1.25))
    Spawn Probability:NiagaraFloat=&MakeCustomFloatFromBool(Bool=$User.TireIntactFR, True Float=&UniformRangedFloat(Minimum=0, Maximum=0), False Float=1)
    Use Spawn Probability=true
   ) {module=SpawnPerUnit003}
   &SpawnPerUnit(
    Module.Use Max Movement Threshold=true
    Module.Use Movement Tolerance=false
    Spawn Group=2
    Spawn Per Unit:NiagaraFloat=&Lerp_Float(A=$User.SparkSPU_MinSpeed, B=$User.SparkSPU_MaxSpeed, Alpha=&Power(a=$User.Vehicle_Speed, b=1.25))
    Spawn Probability:NiagaraFloat=&MakeCustomFloatFromBool(Bool=$User.TireIntactRL, True Float=&UniformRangedFloat(Minimum=0, Maximum=0), False Float=1)
    Use Spawn Probability=true
   ) {module=SpawnPerUnit001}
   &SpawnPerUnit(
    Module.Use Max Movement Threshold=true
    Module.Use Movement Tolerance=false
    Spawn Group=3
    Spawn Per Unit:NiagaraFloat=&Lerp_Float(A=$User.SparkSPU_MinSpeed, B=$User.SparkSPU_MaxSpeed, Alpha=&Power(a=$User.Vehicle_Speed, b=1.25))
    Spawn Probability:NiagaraFloat=&MakeCustomFloatFromBool(Bool=$User.TireIntactRR, True Float=&UniformRangedFloat(Minimum=0, Maximum=0), False Float=1)
    Use Spawn Probability=true
   ) {module=SpawnPerUnit002}
   &SpawnPerUnit(
    Module.Use Max Movement Threshold=true
    Module.Use Movement Tolerance=false
    Spawn Group=4
    Spawn Per Unit:NiagaraFloat=&Lerp_Float(A=$User.SparkSPU_MinSpeed, B=$User.SparkSPU_MaxSpeed, Alpha=&Power(a=$User.Vehicle_Speed, b=1.25))
    Spawn Probability:NiagaraFloat=&MakeCustomFloatFromBool(Bool=$User.Sparks_Wheel_ML, True Float=&UniformRangedFloat(Minimum=0, Maximum=1), False Float=0)
    Use Spawn Probability=true
   ) {module=SpawnPerUnit004}
   &SpawnPerUnit(
    Module.Use Max Movement Threshold=true
    Module.Use Movement Tolerance=false
    Spawn Group=5
    Spawn Per Unit:NiagaraFloat=&Lerp_Float(A=$User.SparkSPU_MinSpeed, B=$User.SparkSPU_MaxSpeed, Alpha=&Power(a=$User.Vehicle_Speed, b=1.25))
    Spawn Probability:NiagaraFloat=&MakeCustomFloatFromBool(Bool=$User.Sparks_Wheel_MR, True Float=&UniformRangedFloat(Minimum=0, Maximum=1), False Float=0)
    Use Spawn Probability=true
   ) {module=SpawnPerUnit005}
   &SpawnRate(SpawnRate=1000, Use Spawn Probability=false) {enabled=false}
  ]
  particleSpawnScript [
   &InitializeParticle(
    Color=&UniformRangedLinearColor(Minimum=<rgba 80 15.9878 0 1>, Maximum=<rgba 40 11.0654 0 1>)
    Lifetime=10
    Mass=&UniformRangedFloat(Minimum=1.25, Maximum=2.25)
    Module.Write Color=true
    Module.Write Lifetime=true
    Module.Write Mass=true
    Module.Write Material Random=true
    Module.Write Position=true
    Module.Write Scale=false
    Module.Write SpriteRotation=false
    Module.Write SpriteSize=true
    Position=$Engine.Owner.Position
    Sprite Size=<3 20>
   ) {enabled=false}
   /Niagara/Modules/Spawn/Initialization/V2/InitializeParticle(
    Color Maximum=<rgba 40 0.5 0 1>
    Color Minimum=<rgba 25 15.7812 0 1>
    Color Mode=::NewEnumerator2
    Lifetime Max=0.2
    Lifetime Min=0.05
    Lifetime Mode=::NewEnumerator1
    Sprite Size Max=<2.5 30>
    Sprite Size Min=<1.25 20>
    Sprite Size Mode=::NewEnumerator2
    UsePositionOffset=false
    Write Material Random=true
   ) {module=InitializeParticle001}
   &SkeletalMeshLocation(
    Bone / Socket Index=&SpawnGroup()
    Inherit Velocity=true
    "Inherit Velocity (Scale)"=-0.05
    Interpolate Mesh Sampling=false
    MaskSpawnGroup=false
    Mesh Sampling Type=::NewEnumerator5
    Sampled Position Offset=&MakeVector(
     X=0
     Y=&UniformRangedFloat(Minimum=&Multiply_Float(A=$User.Wheel_Width, B=-1), Maximum=$User.Wheel_Width)
     Z=13
    )
    Sampled Position Offset Coordinate Space:ENiagaraMeshTransforms=::NewEnumerator3
    Skeletal Mesh=DataInterface() {propertyValues {
      DefaultMesh=null
      ExcludeBoneName=root
      OptionalComponentClass=null
      PreviewMesh=/Valet/Sedan/Meshes/Sedan_Veh
      SkinningMode=SkinOnTheFly
      SoftSourceActor=null
      SourceMode=Default
      UvSetIndex=0
      WholeMeshLOD=-1
      bExcludeBone=false
      bReadDeformedGeometry=true
      bRequireCurrentFrameData=true
      ComponentTags []
      FilteredBones []
      FilteredSockets [
       FX_Tire_fr_l
       FX_Tire_fr_r
       FX_Tire_bk_l
       FX_Tire_bk_r
       FX_Tire_Center_fr_l
       FX_Tire_Center_fr_r
       FX_Tire_Center_bk_l
       FX_Tire_Center_bk_r
       FX_Exhaust_l
       FX_Exhaust_r
      ]
      MeshUserParameter {parameter {name=None, typeDefHandle {registeredTypeIndex=1}, varData []}}
      SamplingRegions []
     }}
    Socket Sampling Mode=::NewEnumerator5
    VelocitySpeedLimit=false
    Write Sampled Mesh Attrributes to Particle Attributes=true
   ) {enabled=false, module=SkeletalMeshLocation001}
   &SkeletalMeshLocation(
    Bone / Socket Index=&SpawnGroup()
    Inherit Velocity=true
    Interpolate Mesh Sampling=false
    MaskSpawnGroup=false
    Mesh Sampling Type=::NewEnumerator5
    Sampled Position Offset=&MakeVector(
     X=&Multiply_Float(A=$User.Vehicle_Speed, B=-20)
     Y=&UniformRangedFloat(Minimum=&Multiply_Float(A=$User.Wheel_Width, B=-1), Maximum=$User.Wheel_Width)
     Z=13
    )
    Sampled Position Offset Coordinate Space:ENiagaraMeshTransforms=::NewEnumerator3
    Skeletal Mesh=$User.Vehicle_Selection_02
    Socket Sampling Mode=::NewEnumerator5
    VelocitySpeedLimit=false
    Write Sampled Mesh Attrributes to Particle Attributes=true
   )
   /SetVariables(Particles.ParticleVelocityAdd:Vector3f=&RotateVectorByQuaternion(
     Vector To Rotate=&MakeVector(
      X=&Multiply_Float(
       A=&RandomRangeFloat(Minimum=-800, Maximum=-1300)
       B=HlslExpression() {hlslExpression="sign(User.Vehicle_Speed) * lerp(0.5,1,abs(User.Vehicle_Speed))"}
      )
      Y=&Add_Float(
       A=&Lerp_Float(A=-170, B=170, Alpha=HlslExpression() {hlslExpression="fmod(Particles.SpawnGroup,2)"})
       B=&RandomRangeFloat(Minimum=-40, Maximum=40)
      )
      Z=&Multiply_Float(
       A=&RandomRangeFloat(Minimum=200, Maximum=500)
       B=&Abs_Float(Float:NiagaraFloat=$User.Vehicle_Speed)
      )
     )
     Quaternion=$Particles.SkeletalMeshLocation.SampledRotation
    )) {isSetParameters=true, module=SetVariables_#5}
   /SetVariables(
    Particles.SpriteAlignment=&NormalizeVector(Vector To Normalize=$Particles.ParticleVelocityAdd)
    Particles.Velocity=&Add_Vector(A=$Particles.Velocity, B=$Particles.ParticleVelocityAdd)
   ) {isSetParameters=true, module=SetVariables_#6}
  ]
  particleUpdateScript [
   &ParticleState()
   &Drag(Use Linear Drag=true, Use Rotational Drag=false) {enabled=false}
   &ScaleColor(
    Scale Alpha=&FloatFromCurve(FloatCurve=DataInterface() {propertyValues {Curve {keys [<at -0.0046 1.0113 RCIM_Cubic>,<at 1.0011 0.002 RCIM_Cubic>]}}})
    ScaleA=true
    ScaleRGB=true
   )
   &SolveForcesAndVelocity(Clamp Velocity=false, Limit Acceleration=false)
   &ScaleSpriteSizeBySpeed(
    Max Scale Factor=<0.5 2.5>
    Min Scale Factor=<0.35 0.3>
    Sample Scale Factor By Curve=true
    Scale Factor Curve=DataInterface() {propertyValues {Curve {keys [
        {arriveTangent=2.865, interpMode=RCIM_Cubic, time=0.1247, value=-0.0006}
        {arriveTangent=4.1522, interpMode=RCIM_Cubic, leaveTangent=4.1522, tangentMode=RCTM_User, time=0.3948, value=0.4893}
        <at 0.9958 0.9974 RCIM_Cubic>
       ]}}}
    Velocity Threshold=1500
   ) {enabled=false, module=SpriteSizeScaleByVelocity}
   &ScaleSpriteSize(Scale Factor=&Vector2DFromFloat(Value=&FloatFromCurve(
      FloatCurve=DataInterface() {propertyValues {Curve {keys [
          {arriveTangent=3.0697, interpMode=RCIM_Cubic, leaveTangent=3.0697, tangentMode=RCTM_User, time=0, value=0}
          {arriveTangent=-1.0335, interpMode=RCIM_Cubic, leaveTangent=-1.0335, tangentMode=RCTM_User, time=0.1674, value=2.5}
          <at 1 0>
         ]}}}
      Scale Curve=&Lerp_Float(A=0.2, B=1, Alpha=$User.Vehicle_Speed)
     )))
  ]
  rendererClasses [&NiagaraSpriteRendererProperties]
  renderers [{rendererClass=&NiagaraSpriteRendererProperties}]
 }
 Sparks_Tires_Blown_Local_Mesh {
  emitterUpdateScript [
   &EmitterState(Life Cycle Mode=::NewEnumerator1, Loop Duration=0.1, UseLoopDelay=false)
   &SpawnBurst_Instantaneous(
    Spawn Count=1
    Spawn Probability=&MakeCustomFloatFromBool(Bool=$User.TireIntactFL, True Float=0, False Float=1)
    Use Spawn Probability=true
   )
   &SpawnBurst_Instantaneous(
    Spawn Count=1
    Spawn Group=1
    Spawn Probability=&MakeCustomFloatFromBool(Bool=$User.TireIntactFR, True Float=0, False Float=1)
    Use Spawn Probability=true
   ) {module=SpawnBurst_Instantaneous001}
   &SpawnBurst_Instantaneous(
    Spawn Count=1
    Spawn Group=2
    Spawn Probability=&MakeCustomFloatFromBool(Bool=$User.TireIntactRL, True Float=0, False Float=1)
    Use Spawn Probability=true
   ) {module=SpawnBurst_Instantaneous002}
   &SpawnBurst_Instantaneous(
    Spawn Count=1
    Spawn Group=3
    Spawn Probability=&MakeCustomFloatFromBool(Bool=$User.TireIntactRR, True Float=0, False Float=1)
    Use Spawn Probability=true
   ) {module=SpawnBurst_Instantaneous003}
   &SpawnBurst_Instantaneous(
    Spawn Count=1
    Spawn Group=4
    Spawn Probability=&MakeCustomFloatFromBool(Bool=$User.Sparks_Wheel_ML, True Float=1, False Float=0)
    Use Spawn Probability=true
   ) {module=SpawnBurst_Instantaneous004}
   &SpawnBurst_Instantaneous(
    Spawn Count=1
    Spawn Group=5
    Spawn Probability=&MakeCustomFloatFromBool(Bool=$User.Sparks_Wheel_MR, True Float=1, False Float=0)
    Use Spawn Probability=true
   ) {module=SpawnBurst_Instantaneous005}
  ]
  particleSpawnScript [
   /Niagara/Modules/Spawn/Initialization/InitializeParticle(
    Color=&UniformRangedLinearColor(Minimum=<rgba 50 9.9924 0 1>, Maximum=<rgba 150 41.4953 0 1>)
    Lifetime=&UniformRangedFloat(Minimum=0.05, Maximum=0.1)
    Mass=&UniformRangedFloat(Minimum=0.75, Maximum=1.75)
    Mesh Scale=&MakeCustomVectorFromBool(
     Bool=HlslExpression() {hlslExpression="Emitter.SpawnGroup == 2 || Emitter.SpawnGroup == 3 || Emitter.SpawnGroup == 4 || Emitter.SpawnGroup == 5"}
     True Vector=$User.Wheel_Scale_Back
     False Vector=$User.Wheel_Scale_Front
    )
    Module.Write Color=true
    Module.Write Lifetime=true
    Module.Write Mass=false
    Module.Write Material Random=true
    Module.Write Position=true
    Module.Write Scale=true
    Module.Write SpriteRotation=false
    Module.Write SpriteSize=false
    Position=$Engine.Owner.Position
   )
   /Niagara/Modules/Spawn/Orientation/V2/InitialMeshOrientation(
    Orientation Coordinate Space:ENiagaraCoordinateSpace=::Local
    Orientation Vector=<0 1 0>
    Rotation=&UniformRangedVector(Minimum=<0 0 0>, Maximum=<0 1 0>)
    Use Orientation Vector=true
    Use Rotation Vector=true
   ) {enabled=false}
   /Niagara/Modules/Spawn/Orientation/V3/InitialMeshOrientation(
    Orientation Coordinate Space=::Local
    Orientation Vector=<0 1 0>
    Rotation=&UniformRangedVector(Minimum=<0 0 0>, Maximum=<0 1 0>)
    Use Orientation Vector=true
    Use Rotation Vector=true
   ) {module=InitialMeshOrientation001}
  ]
  particleUpdateScript [
   &ParticleState()
   &SkeletalMeshLocation(
    Bone / Socket Index=&Add_Integer(A=&SpawnGroup(), B=6)
    Inherit Velocity=false
    Interpolate Mesh Sampling=false
    MaskSpawnGroup=false
    Mesh Sampling Type=::NewEnumerator5
    Skeletal Mesh=$User.Vehicle_Selection_02
    Socket Sampling Mode=::NewEnumerator5
    Write Sampled Mesh Attrributes to Particle Attributes=true
   )
   /SetVariables(Particles.MeshOrientation=&MultiplyQuaternion(
     Quaternion A=$Particles.SkeletalMeshLocation.SampledRotation
     Quaternion B=&OrientMeshToVector(Look At Direction=&UniformRangedVector(Minimum=<0 0 0>, Maximum=<0 1 0>), Reference Axis Vector=<1 0 0>)
    )) {isSetParameters=true}
   &DynamicMaterialParameters(
    Index 0 Param 1=5
    Index 0 Param 2=2
    Module.Param0WriteEnabled=true
    Module.Param1WriteEnabled=true
    Module.Param2WriteEnabled=true
    Module.Param3WriteEnabled=true
   )
   &ScaleColor(
    Scale Alpha=&FloatFromCurve(
     FloatCurve=DataInterface() {propertyValues {Curve {keys [<at -0.0046 1.0113 RCIM_Cubic>,<at 1.0011 0.002 RCIM_Cubic>]}}}
     Scale Curve=&ClampFloat(Float=&Multiply_Float(A=2, B=&Abs_Float(Float:NiagaraFloat=&Power(a=$User.Vehicle_Speed, b=2.5))))
    )
    Scale RGB=<100 100 100>
    ScaleA=true
    ScaleRGB=true
   )
  ]
  rendererClasses [&NiagaraMeshRendererProperties]
  renderers [{rendererClass=&NiagaraMeshRendererProperties, rendererIndex=0}]
 }
 Sparks_Tires_Blown_World {
  emitterUpdateScript [
   &EmitterState(Life Cycle Mode=::NewEnumerator1, Loop Duration=0.0125, UseLoopDelay=false)
   &SpawnPerUnit(
    Module.Use Max Movement Threshold=true
    Module.Use Movement Tolerance=false
    Spawn Per Unit:NiagaraFloat=&Lerp_Float(A=$User.SparkSPU_MinSpeed, B=$User.SparkSPU_MaxSpeed, Alpha=&Power(a=$User.Vehicle_Speed, b=1.25))
    Spawn Probability:NiagaraFloat=&MakeCustomFloatFromBool(Bool=$User.TireIntactFL, True Float=&UniformRangedFloat(Minimum=0, Maximum=0), False Float=1)
    Use Spawn Probability=true
   )
   &SpawnPerUnit(
    Module.Use Max Movement Threshold=true
    Module.Use Movement Tolerance=false
    Spawn Group=1
    Spawn Per Unit:NiagaraFloat=&Lerp_Float(A=$User.SparkSPU_MinSpeed, B=$User.SparkSPU_MaxSpeed, Alpha=&Power(a=$User.Vehicle_Speed, b=1.25))
    Spawn Probability:NiagaraFloat=&MakeCustomFloatFromBool(Bool=$User.TireIntactFR, True Float=&UniformRangedFloat(Minimum=0, Maximum=0), False Float=1)
    Use Spawn Probability=true
   ) {module=SpawnPerUnit003}
   &SpawnPerUnit(
    Module.Use Max Movement Threshold=true
    Module.Use Movement Tolerance=false
    Spawn Group=2
    Spawn Per Unit:NiagaraFloat=&Lerp_Float(A=$User.SparkSPU_MinSpeed, B=$User.SparkSPU_MaxSpeed, Alpha=&Power(a=$User.Vehicle_Speed, b=1.25))
    Spawn Probability:NiagaraFloat=&MakeCustomFloatFromBool(Bool=$User.TireIntactRL, True Float=&UniformRangedFloat(Minimum=0, Maximum=0), False Float=1)
    Use Spawn Probability=true
   ) {module=SpawnPerUnit001}
   &SpawnPerUnit(
    Module.Use Max Movement Threshold=true
    Module.Use Movement Tolerance=false
    Spawn Group=3
    Spawn Per Unit:NiagaraFloat=&Lerp_Float(A=$User.SparkSPU_MinSpeed, B=$User.SparkSPU_MaxSpeed, Alpha=&Power(a=$User.Vehicle_Speed, b=1.25))
    Spawn Probability:NiagaraFloat=&MakeCustomFloatFromBool(Bool=$User.TireIntactRR, True Float=&UniformRangedFloat(Minimum=0, Maximum=0), False Float=1)
    Use Spawn Probability=true
   ) {module=SpawnPerUnit002}
   &SpawnPerUnit(
    Module.Use Max Movement Threshold=true
    Module.Use Movement Tolerance=false
    Spawn Group=4
    Spawn Per Unit:NiagaraFloat=&Lerp_Float(A=$User.SparkSPU_MinSpeed, B=$User.SparkSPU_MaxSpeed, Alpha=&Power(a=$User.Vehicle_Speed, b=1.25))
    Spawn Probability:NiagaraFloat=&MakeCustomFloatFromBool(Bool=$User.Sparks_Wheel_ML, True Float=&UniformRangedFloat(Minimum=0, Maximum=1), False Float=0)
    Use Spawn Probability=true
   ) {module=SpawnPerUnit004}
   &SpawnPerUnit(
    Module.Use Max Movement Threshold=true
    Module.Use Movement Tolerance=false
    Spawn Group=5
    Spawn Per Unit:NiagaraFloat=&Lerp_Float(A=$User.SparkSPU_MinSpeed, B=$User.SparkSPU_MaxSpeed, Alpha=&Power(a=$User.Vehicle_Speed, b=1.25))
    Spawn Probability:NiagaraFloat=&MakeCustomFloatFromBool(Bool=$User.Sparks_Wheel_MR, True Float=&UniformRangedFloat(Minimum=0, Maximum=1), False Float=0)
    Use Spawn Probability=true
   ) {module=SpawnPerUnit005}
   &SpawnRate(
    SpawnGroup:NiagaraInt32=&RandomRangeInt(Minimum=0, Maximum=5)
    SpawnRate=250
    Use Spawn Probability=false
   ) {enabled=false}
  ]
  particleSpawnScript [
   &InitializeParticle(
    Color=&UniformRangedLinearColor(Minimum=<rgba 80 15.9878 0 1>, Maximum=<rgba 40 11.0654 0 1>)
    Lifetime=10
    Mass=&UniformRangedFloat(Minimum=1.25, Maximum=2.25)
    Module.Write Color=true
    Module.Write Lifetime=true
    Module.Write Mass=true
    Module.Write Material Random=true
    Module.Write Position=true
    Module.Write Scale=false
    Module.Write SpriteRotation=false
    Module.Write SpriteSize=true
    Position=$Engine.Owner.Position
    Sprite Size=<3 20>
   ) {enabled=false}
   /Niagara/Modules/Spawn/Initialization/V2/InitializeParticle(
    Color Maximum=<rgba 40 11.0654 0 1>
    Color Minimum=<rgba 80 15.9878 0 1>
    Color Mode=::NewEnumerator2
    Lifetime Max=0.8
    Lifetime Min=0.25
    Lifetime Mode=::NewEnumerator1
    Sprite Size Max=<2.5 30>
    Sprite Size Min=<1.25 20>
    Sprite Size Mode=::NewEnumerator2
    UsePositionOffset=false
    Write Material Random=true
   ) {module=InitializeParticle001}
   &SkeletalMeshLocation(
    Bone / Socket Index=&SpawnGroup()
    Inherit Velocity=true
    "Inherit Velocity (Scale)"=-0.05
    Interpolate Mesh Sampling=false
    MaskSpawnGroup=false
    Mesh Sampling Type=::NewEnumerator5
    Sampled Position Offset=&MakeVector(
     X=0
     Y=&UniformRangedFloat(Minimum=&Multiply_Float(A=$User.Wheel_Width, B=-1), Maximum=$User.Wheel_Width)
     Z=13
    )
    Sampled Position Offset Coordinate Space:ENiagaraMeshTransforms=::NewEnumerator3
    Skeletal Mesh=DataInterface() {propertyValues {
      DefaultMesh=null
      ExcludeBoneName=root
      OptionalComponentClass=null
      PreviewMesh=&Sedan_Veh
      SkinningMode=SkinOnTheFly
      SoftSourceActor=null
      SourceMode=Default
      UvSetIndex=0
      WholeMeshLOD=-1
      bExcludeBone=false
      bReadDeformedGeometry=true
      bRequireCurrentFrameData=true
      ComponentTags []
      FilteredBones []
      FilteredSockets [
       FX_Tire_fr_l
       FX_Tire_fr_r
       FX_Tire_bk_l
       FX_Tire_bk_r
       FX_Tire_Center_fr_l
       FX_Tire_Center_fr_r
       FX_Tire_Center_bk_l
       FX_Tire_Center_bk_r
       FX_Exhaust_l
       FX_Exhaust_r
      ]
      MeshUserParameter {parameter {name=None, typeDefHandle {registeredTypeIndex=1}, varData []}}
      SamplingRegions []
     }}
    Socket Sampling Mode=::NewEnumerator5
    VelocitySpeedLimit=false
    Write Sampled Mesh Attrributes to Particle Attributes=true
   ) {enabled=false, module=SkeletalMeshLocation001}
   &SkeletalMeshLocation(
    Bone / Socket Index=&SpawnGroup()
    Inherit Velocity=true
    "Inherit Velocity (Scale)"=-0.05
    Interpolate Mesh Sampling=false
    MaskSpawnGroup=false
    Mesh Sampling Type=::NewEnumerator5
    Sampled Position Offset=&MakeVector(
     X=&Multiply_Float(A=$User.Vehicle_Speed, B=-20)
     Y=&UniformRangedFloat(Minimum=&Multiply_Float(A=$User.Wheel_Width, B=-1), Maximum=$User.Wheel_Width)
     Z=13
    )
    Sampled Position Offset Coordinate Space:ENiagaraMeshTransforms=::NewEnumerator3
    Skeletal Mesh=$User.Vehicle_Selection_02
    Socket Sampling Mode=::NewEnumerator5
    VelocitySpeedLimit=false
    Write Sampled Mesh Attrributes to Particle Attributes=true
   )
   /SetVariables(Particles.ParticleVelocityAdd:Vector3f=&RotateVectorByQuaternion(
     Vector To Rotate=&MakeVector(
      X=&Multiply_Float(
       A=&RandomRangeFloat(Minimum=-250, Maximum=-800)
       B=HlslExpression() {hlslExpression="sign(User.Vehicle_Speed) * lerp(0.5,1,abs(User.Vehicle_Speed))"}
      )
      Y=&Add_Float(
       A=&Lerp_Float(A=-170, B=170, Alpha=HlslExpression() {hlslExpression="fmod(Particles.SpawnGroup,2)"})
       B=&RandomRangeFloat(Minimum=-40, Maximum=40)
      )
      Z=&Multiply_Float(
       A=&RandomRangeFloat(Minimum=200, Maximum=500)
       B=&Abs_Float(Float:NiagaraFloat=$User.Vehicle_Speed)
      )
     )
     Quaternion=$Particles.SkeletalMeshLocation.SampledRotation
    )) {isSetParameters=true, module=SetVariables_#5}
   /SetVariables(
    Particles.SpriteAlignment=&NormalizeVector(Vector To Normalize=$Particles.ParticleVelocityAdd)
    Particles.Velocity=&Add_Vector(A=$Particles.Velocity, B=$Particles.ParticleVelocityAdd)
   ) {isSetParameters=true, module=SetVariables_#6}
  ]
  particleUpdateScript [
   &ParticleState()
   &Drag(Use Linear Drag=true, Use Rotational Drag=false)
   &ScaleColor(
    Scale Alpha=&FloatFromCurve(FloatCurve=DataInterface() {propertyValues {Curve {keys [<at -0.0046 1.0113 RCIM_Cubic>,<at 1.0011 0.002 RCIM_Cubic>]}}})
    ScaleA=true
    ScaleRGB=true
   )
   &GravityForce(Gravity=<0 0 -1800>)
   &SolveForcesAndVelocity(Clamp Velocity=false, Limit Acceleration=false)
   &ScaleSpriteSizeBySpeed(
    Max Scale Factor=<0.5 2.5>
    Min Scale Factor=<0.35 0.3>
    Sample Scale Factor By Curve=true
    Scale Factor Curve=DataInterface() {propertyValues {Curve {keys [
        {arriveTangent=2.865, interpMode=RCIM_Cubic, time=0.1247, value=-0.0006}
        {arriveTangent=4.1522, interpMode=RCIM_Cubic, leaveTangent=4.1522, tangentMode=RCTM_User, time=0.3948, value=0.4893}
        <at 0.9958 0.9974 RCIM_Cubic>
       ]}}}
    Velocity Threshold=1500
   ) {enabled=false, module=SpriteSizeScaleByVelocity}
   &ScaleSpriteSize(Scale Factor=&Vector2DFromFloat(Value=&FloatFromCurve(
      FloatCurve=DataInterface() {propertyValues {Curve {keys [
          {arriveTangent=56.0592, interpMode=RCIM_Cubic, leaveTangent=56.0592, tangentMode=RCTM_User, time=0, value=0}
          {arriveTangent=-1.0335, interpMode=RCIM_Cubic, leaveTangent=-1.0335, tangentMode=RCTM_User, time=0.1674, value=2.5}
          <at 1 0>
         ]}}}
      Scale Curve=&Lerp_Float(A=0.2, B=1, Alpha=$User.Vehicle_Speed)
     )))
   &ScaleSpriteSizeBySpeed(Max Scale Factor=<0.5 0.6>, Min Scale Factor=<1 0.05>)
  ]
  rendererClasses [&NiagaraSpriteRendererProperties]
  renderers [{rendererClass=&NiagaraSpriteRendererProperties}]
 }
 Streaks_Chrome {
  emitterUpdateScript [
   &EmitterState(Life Cycle Mode=::NewEnumerator1, Loop Duration=1, UseLoopDelay=false)
   &SpawnRate(
    Spawn Probability=&Multiply_Float(
     A=&MakeCustomFloatFromBool(Bool=$User.IsWaterDeep, True Float=0, False Float=1)
     B=&self:ScratchGetFloatFromArray(Index:NiagaraInt32=3, WheelsTouchingArray:NiagaraDataInterfaceArrayFloat=DataInterface() {propertyValues {FloatData []}})
    )
    SpawnGroup=1
    SpawnRate=&MakeFloatFromInt(Int=&MakeCustomIntFromBool(Bool=$User.PeelOut_Wheel_RR, True INT=&Multiply_Int(A=$User.Dust_Particles_Per_Wheel, B=5), False INT=0))
    Use Spawn Probability=true
   ) {module=SpawnRate001}
   &SpawnRate(
    Spawn Probability=&Multiply_Float(
     A=&MakeCustomFloatFromBool(Bool=$User.IsWaterDeep, True Float=0, False Float=1)
     B=&self:ScratchGetFloatFromArray(Index:NiagaraInt32=2, WheelsTouchingArray:NiagaraDataInterfaceArrayFloat=DataInterface() {propertyValues {FloatData []}})
    )
    SpawnRate=&MakeFloatFromInt(Int=&MakeCustomIntFromBool(Bool=$User.PeelOut_Wheel_RL, True INT=&Multiply_Int(A=$User.Dust_Particles_Per_Wheel, B=5), False INT=0))
    Use Spawn Probability=true
   )
  ]
  particleSpawnScript [
   &InitializeParticle(
    Color=<rgba 1.3728 3.0599 5 1>
    Lifetime Max=0.3
    Lifetime Min=0.1
    Lifetime Mode=::NewEnumerator1
    Mesh Scale Mode=::NewEnumerator1
    Mesh Uniform Scale Max=3
    UsePositionOffset=false
    Write Material Random=true
   )
   &SubUVAnimation(End Frame=7, SubUV Animation Mode:ENiagara_SubUVLookupMode=::NewEnumerator2)
   &ShapeLocation(
    Hemisphere Distribution=<0.5 1>
    MaskSpawnGroup=false
    Override Seed=false
    Rotation Angle=-0.25
    Rotation Axis=<0 0 1>
    Rotation Mode=::NewEnumerator1
    Sphere Radius=3
    UseOwnerScale=true
   )
   &AddVelocity(
    Cone Angle=5
    Module.UseConeFalloff=false
    Module.UseVelDistribution=false
    Override Seed=false
    Velocity Mode=::NewEnumerator2
    Velocity Speed=&RandomRangeFloat(Minimum=250, Maximum=750)
   )
   &CameraOffset(Camera Offset Amount=10)
   &InitialMeshOrientation(Mesh Orientation Mode=::NewEnumerator2, Rotation=<0.5 0 0.25>, Use Rotation Vector=true)
  ]
  particleUpdateScript [
   &ParticleState()
   &SkeletalMeshLocation(
    Bone / Socket Index=&Add_Integer(A=&SpawnGroup(), B=2)
    Inherit Velocity=false
    MaskSpawnGroup=false
    Mesh Sampling Type=::NewEnumerator5
    Sampled Position Offset=&MakeVector(
     X=0
     Y=&UniformRangedFloat(Minimum=&Multiply_Float(A=$User.Wheel_Width_Back, B=-1), Maximum=$User.Wheel_Width_Back)
     Z=&Multiply_Float(A=-15, B=&MakeCustomFloatFromBool(Bool=$User.Mod_OffroadTires, True Float=1, False Float=0))
    )
    Skeletal Mesh=$User.Vehicle_Selection_02
    Socket Sampling Mode=::NewEnumerator5
    Write Sampled Mesh Attrributes to Particle Attributes=true
   )
   &Drag(Drag=4, Use Linear Drag=true, Use Rotational Drag=false)
   &GravityForce(Gravity=<0 0 -1600>)
   &SolveForcesAndVelocity(Clamp Velocity=false, Limit Acceleration=false)
   &ScaleColor(
    Scale RGB=&VectorFromFloat(Value=&FloatFromCurve(FloatCurve=DataInterface() {propertyValues {Curve {keys [{arriveTangent=-4.4221, interpMode=RCIM_Cubic, time=0, value=1}, <at 0.3 0>]}}}))
    ScaleA=false
    ScaleRGB=true
   )
   &ScaleMeshSize(Scale Factor=&VectorFromFloat(Value=&FloatFromCurve(FloatCurve=DataInterface() {propertyValues {Curve {keys [
          {arriveTangent=3.8261, interpMode=RCIM_Cubic, leaveTangent=3.8261, tangentMode=RCTM_User, time=0, value=0}
          {arriveTangent=0.7815, interpMode=RCIM_Cubic, leaveTangent=0.7815, tangentMode=RCTM_User, time=0.25, value=0.6}
          {arriveTangent=0.2614, interpMode=RCIM_Cubic, leaveTangent=0.2614, tangentMode=RCTM_User, time=1, value=1}
         ]}}})))
   &DynamicMaterialParameters(
    Index 0 Param 1=&FloatFromCurve(FloatCurve=DataInterface() {propertyValues {Curve {keys [
         <at 0 0 RCIM_Cubic>
         {arriveTangent=2.728, arriveTangentWeight=0.9068, interpMode=RCIM_Cubic, leaveTangent=2.728, tangentMode=RCTM_User, time=1, value=1}
        ]}}})
    Index 0 Param 2=&FloatFromCurve(
     FloatCurve=DataInterface() {propertyValues {Curve {keys [
         {arriveTangent=1.3993, interpMode=RCIM_Cubic, leaveTangent=1.3993, tangentMode=RCTM_User, time=0, value=0.5}
         {interpMode=RCIM_Cubic, tangentMode=RCTM_User, time=1, value=1}
        ]}}}
     Scale Curve=1.5
    )
    Index 0 Param 3=&Multiply_Float(
     A=&MakeFloatFromVector(Vector=$Particles.Scale, Channel=::NewEnumerator0)
     B=&RandomRangeFloat(Minimum=3, Maximum=5)
    )
    Index 0 Param 4=&FloatFromCurve(
     FloatCurve=DataInterface() {propertyValues {Curve {keys [
         {interpMode=RCIM_Cubic, leaveTangentWeight=0.7017, tangentMode=RCTM_User, time=0, value=0}
         {arriveTangentWeight=1.2352, interpMode=RCIM_Cubic, tangentMode=RCTM_User, time=1, value=1}
        ]}}}
     Scale Curve=&RandomRangeFloat(Minimum=1, Maximum=2)
    )
    Index 1 Param 1=&FloatFromCurve(
     FloatCurve=DataInterface() {propertyValues {Curve {keys [
         {arriveTangent=2.4162, interpMode=RCIM_Cubic, leaveTangent=2.4162, tangentMode=RCTM_User, time=0, value=0}
         {interpMode=RCIM_Cubic, tangentMode=RCTM_User, time=1, value=1}
        ]}}}
     Scale Curve=-3
    )
    Index 1 Param 2=&FloatFromCurve(
     FloatCurve=DataInterface() {propertyValues {Curve {keys [{interpMode=RCIM_Cubic, tangentMode=RCTM_User, time=0.6, value=1}, {interpMode=RCIM_Cubic, tangentMode=RCTM_User, time=1, value=2}]}}}
     Scale Curve=&RandomRangeFloat(Minimum=0.333, Maximum=1)
    )
    Module.Param0WriteEnabled=true
    Module.Param0WriteEnabled1=true
    Module.Param1WriteEnabled=true
    Module.Param1WriteEnabled1=true
    Module.Param2WriteEnabled=true
    Module.Param2WriteEnabled1=false
    Module.Param3WriteEnabled=true
    Module.Param3WriteEnabled1=false
    Write Parameter Index 1=true
   )
   &UpdateMeshOrientation(
    Rotation Rate=&FloatFromCurve(
     FloatCurve=DataInterface() {propertyValues {Curve {keys [
         {arriveTangent=-2.1925, interpMode=RCIM_Cubic, leaveTangent=-2.1925, tangentMode=RCTM_User, time=0, value=1}
         {arriveTangent=-0.2845, interpMode=RCIM_Cubic, leaveTangent=-0.2845, tangentMode=RCTM_User, time=1, value=0.05}
        ]}}}
     Scale Curve=&RandomRangeFloat(Minimum=0.3, Maximum=0.4)
    )
    Rotation Vector=<0 1 -1>
   )
  ]
  rendererClasses [&NiagaraMeshRendererProperties]
  renderers [{rendererClass=&NiagaraMeshRendererProperties, rendererIndex=0}]
 }
}
summary {
 systemName=NS_Wheel_And_Tire_Interaction_Chrome
 emitters [
  {emitterName=Sparks_Tires_Blown_Local, rendererClasses [&NiagaraSpriteRendererProperties]}
  {emitterName=Sparks_Tires_Blown_Local_Mesh, rendererClasses [&NiagaraMeshRendererProperties]}
  {emitterName=Dust_Tires_World, rendererClasses [&NiagaraSpriteRendererProperties]}
  {emitterName=Peel_Tires_Spikes, rendererClasses [&NiagaraSpriteRendererProperties]}
  {emitterName=Dust_Tires_Mesh, rendererClasses [&NiagaraMeshRendererProperties]}
  {emitterName=Dust_Debris, simTarget=GPUComputeSim, rendererClasses [&NiagaraMeshRendererProperties]}
  {emitterName=Exhaust_Local, rendererClasses [&NiagaraSpriteRendererProperties]}
  {emitterName=Sparks_Tires_Blown_World, rendererClasses [&NiagaraSpriteRendererProperties]}
  {emitterName=ShiftBlip, rendererClasses [&NiagaraSpriteRendererProperties]}
  {emitterName=ShiftBlipLight, rendererClasses [&NiagaraLightRendererProperties]}
  {bEnabled=false, emitterName=RandomDirBurst, simTarget=GPUComputeSim, rendererClasses [&NiagaraSpriteRendererProperties]}
  {
   bEnabled=false
   emitterName=DirectionalBurstSparks002
   simTarget=GPUComputeSim
   rendererClasses [&NiagaraSpriteRendererProperties]
  }
  {emitterName=Dust_Debris001, simTarget=GPUComputeSim, rendererClasses [&NiagaraSpriteRendererProperties]}
  {emitterName=Peel_Tires_Spikes_ChromeBlobs, simTarget=GPUComputeSim, rendererClasses [&NiagaraSpriteRendererProperties]}
  {emitterName=Streaks_Chrome, rendererClasses [&NiagaraMeshRendererProperties]}
 ]
 userVariables [
  User.Downshifting {type=NiagaraBool}
  User.Dust_AlphaCurve {
   type=NiagaraDataInterfaceCurve
   description=""
   defaultValue {
    struct=NiagaraExt_VariableValue_DataInterface
    value {
     dataInterface=&self:NiagaraDataInterfaceCurve_0
     dataInterfaceClass=&NiagaraDataInterfaceCurve
    }
   }
  }
  User.Dust_Particles_Per_Wheel {type=NiagaraInt32, defaultValue=18}
  User.Dust_SpawnCurve {
   type=NiagaraDataInterfaceCurve
   description=""
   defaultValue {
    struct=NiagaraExt_VariableValue_DataInterface
    value {
     dataInterface=&self:NiagaraDataInterfaceCurve_1
     dataInterfaceClass=&NiagaraDataInterfaceCurve
    }
   }
  }
  User.Dust_Wheel_FL {type=NiagaraBool}
  User.Dust_Wheel_FR {type=NiagaraBool}
  User.Dust_Wheel_ML {type=NiagaraBool}
  User.Dust_Wheel_MR {type=NiagaraBool}
  User.Dust_Wheel_RL {type=NiagaraBool}
  User.Dust_Wheel_RR {type=NiagaraBool}
  User.DustAlphaMultiplier {type=NiagaraFloat, defaultValue=0.25, description=""}
  User.DustColorLerp {type=NiagaraFloat, defaultValue=0, description=""}
  User.DustSpawnPerUnit {type=NiagaraFloat, defaultValue=150, description=""}
  User.DustTireBlurColor {type=LinearColor, description="", defaultValue {struct=LinearColor, value=<rgba 1 1 1 1>}}
  User.ExhaustColor {type=LinearColor, description="", defaultValue {struct=LinearColor, value=<rgba 0.6062 0.754 1 0.025>}}
  User.ExhaustScale {type=NiagaraFloat, defaultValue=1, description=""}
  User.IsOnDirt {type=NiagaraBool}
  User.IsShifting {type=NiagaraBool}
  User.IsWaterDeep {type=NiagaraBool}
  User.Mod_OffroadTires {type=NiagaraBool}
  User.ParticleCountPerWheel {type=NiagaraInt32, defaultValue=150}
  User.PeelOut_Wheel_RL {type=NiagaraBool}
  User.PeelOut_Wheel_RR {type=NiagaraBool}
  User.Skid_Wheels {type=NiagaraBool}
  User.Slide_Direction {type=NiagaraFloat, defaultValue=0, description=""}
  User.Sparks_Wheel_FL {type=NiagaraBool}
  User.Sparks_Wheel_FR {type=NiagaraBool}
  User.Sparks_Wheel_ML {type=NiagaraBool}
  User.Sparks_Wheel_MR {type=NiagaraBool}
  User.Sparks_Wheel_RL {type=NiagaraBool}
  User.Sparks_Wheel_RR {type=NiagaraBool}
  User.SparkSPU_MaxSpeed {type=NiagaraFloat, defaultValue=3, description=""}
  User.SparkSPU_MinSpeed {type=NiagaraFloat, defaultValue=4, description=""}
  User.Tire_Scale_Back {type=Vector3f, description="", defaultValue {struct=Vector3f, value=<0.39 0.4325 0.39>}}
  User.Tire_Scale_Front {type=Vector3f, description="", defaultValue {struct=Vector3f, value=<0.3 0.4 0.3>}}
  User.TireIntactFL {type=NiagaraBool}
  User.TireIntactFR {type=NiagaraBool}
  User.TireIntactRL {type=NiagaraBool}
  User.TireIntactRR {type=NiagaraBool}
  User.Vehicle_Selection_02 {
   type=NiagaraDataInterfaceSkeletalMesh
   defaultValue {
    struct=NiagaraExt_VariableValue_DataInterface
    value {dataInterface=&self:NiagaraDataInterfaceSkeletalMesh_1}
   }
  }
  User.Vehicle_Speed {type=NiagaraFloat, defaultValue=0, description=""}
  User.VehicleHasDriver {type=NiagaraBool}
  User.VehicleHasFuel {type=NiagaraBool}
  User.Wheel_Scale_Back {type=Vector3f, description="", defaultValue {struct=Vector3f, value=<0.3 0.3 0.3>}}
  User.Wheel_Scale_Front {type=Vector3f, description="", defaultValue {struct=Vector3f, value=<0.3 0.3 0.3>}}
  User.Wheel_Width {type=NiagaraFloat, defaultValue=16, description=""}
  User.Wheel_Width_Back {type=NiagaraFloat, defaultValue=18, description=""}
  User.WheelsTouchingArray {
   type=NiagaraDataInterfaceArrayFloat
   defaultValue {
    struct=NiagaraExt_VariableValue_DataInterface
    value {
     dataInterface=&self:NiagaraDataInterfaceArrayFloat_0
     dataInterfaceClass=&NiagaraDataInterfaceArrayFloat
    }
   }
  }
 ]
}
