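# Placeholder chrome material for skeletal and mesh-deformer meshes, using Default Lit shading and a Chrome physical material.
# A Fresnel term, scaled and optionally animated over time through a static switch, wraps through Frac to hue-shift a tint color.
# That tint multiplies the base color, then passes through Desaturation, whose negative default fraction boosts saturation.
# Scalar parameters supply high metallic and low roughness values for the chrome look.
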Material /ChromeShared/Materials/M_Chrome_Placeholder
graph {
 properties {
  bUsedWithMeshDeformer=true
  bUsedWithSkeletalMesh=true
  physMaterial=/Game/TechArt/Seasonal/LackLuster/Physics/Chrome_PhysicalMaterial
  pixelDepthOffsetMode=PDOM_Legacy
  usedShadingModels=Default Lit
 }
 nodes [
  Constant()
  n1 = Multiply(
   A=MaterialFunctionCall(Exponent=ScalarParameter() {defaultValue=1.5, group=HueShift, parameterName=FresnelHueShiftExponent}) {materialFunction=&MF_Fresnel_FX}.Fresnel
   B=ScalarParameter() {defaultValue=1, group=HueShift, parameterName=HueShiftScale}
  )
  n2 = Multiply(A=Time(), B=ScalarParameter() {defaultValue=1, parameterName=AnimatedHueShiftSpeed})
  n3 = Frac(None=StaticSwitchParameter(True=Add(A=@n1, B=@n2), False=@n1) {parameterName=AnimateHueShift})
  n4 = VectorParameter() {group=HueShift, parameterName=HueShiftColor, defaultValue=<rgba 0.8212 0.9028 1 0>}
  Desaturation(
   None=Multiply(
    A=VectorParameter() {group=BaseColor, parameterName=BaseColor, defaultValue=<rgba 0.8385 0.8193 0.6886 0>}.RGB
    B=MaterialFunctionCall(Hue Shift Percentage=@n3, Texture=@n4.RGB) {materialFunction=&HueShift}.Result
   )
   Fraction=ScalarParameter() {defaultValue=-0.5, parameterName=BaseColorDesaturationFraction}
  )
  ScalarParameter() {defaultValue=0.22, parameterName=Roughness}
  ScalarParameter() {defaultValue=0.965, parameterName=Metallic}
 ]
}
