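# Master material for the Lackluster chrome Valet vehicle, a masked skeletal-mesh surface using Default Lit or Clear Coat shading.
# MF_LackLuster_Base feeds an angle-corrected normal blend, headlights, turn signals and damage, with UseValetFunctions switching back to base only.
# Pre-skinned bounds split the light masks; two panning pre-skinned projected swirl normals are summed and Fresnel-flattened.
# Vertex-color hiding sets world position offset; reverse lights are disconnected pending a parameter bug fix.
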Material /ChromeShared/Materials/M_Lackluster_Valet_Master
graph {
 properties {
  bUseMaterialAttributes=true
  bUsedWithMeshDeformer=true
  bUsedWithSkeletalMesh=true
  ditherOpacityMask=true
  opacityMaskClipValue=0.5
  physMaterial=/Game/TechArt/Seasonal/LackLuster/Physics/Chrome_PhysicalMaterial
  pixelDepthOffsetMode=PDOM_Legacy
  shadingModel=MSM_FromMaterialExpression
  usedShadingModels=Default Lit | Clear Coat
 }
 nodes [
  n1 = VertexInterpolator(VS=PreSkinnedPosition())
  n2 = PreSkinnedLocalBounds()
  n3 = Divide(
   A=Subtract(A=NamedRerouteDeclaration(None=@n1.PS) {name=Preskinned Position, nodeColor=<rgba 0 1 0.4375 1>}, B=@n2.Min)
   B=Subtract(A=@n2.Max, B=@n2.Max)
  )
  n4 = Round(None=ComponentMask(None=@n3) {r=true})
  n5 = MaterialFunctionCall() {materialFunction=/Valet/Materials/MF_Valet_BaseFunctionality}
  n6 = MaterialFunctionCall!(
   bBlendMaterialAttributes
   MA_In
   UVsOverride
   Override Normals
   BlendMaterialAttributesAlpha
   BaseColorOverride
   WorldPositionOverride=NamedRerouteUsage()
   VertexNormalOverride
   SourceNormalsStrengthOverride
  ) {materialFunction=/Game/TechArt/Seasonal/LackLuster/MaterialFunctions/MF_LackLuster_Base}
  n7 = NamedRerouteUsage()
  n8 = MaterialFunctionCall!(
   In=MaterialFunctionCall!(
    Mask=Multiply(A=@n4, B=@n5.Lights)
    MA=MaterialFunctionCall!(Material=@n6.MA Out, Additional Normal=@n7, Mask) {materialFunction=/Engine/Functions/MaterialLayerFunctions/MatLayerBlend_BlendAngleCorrectedNormals}.Blended Material
   ) {materialFunction=/Valet/Materials/MF_Valet_Headlights}.Result
   CenterLine=ComponentMask(None=@n3) {g=true}
   Alpha=Multiply(A=@n5.Lights, B=OneMinus(None=@n4))
  ) {materialFunction=/Game/Weapons/WeaponSkins/MaterialFunction/TurnSignals}
  MaterialFunctionCall!(MA=@n8.Result) {desc=Reconnect when param bug is fixed, materialFunction=&/MF_ChromeValet_ReverseLights}
  n9 = TextureObject() {samplerType=SAMPLERTYPE_Normal, texture=/Game/TechArt/Seasonal/LackLuster/Textures/T_Lackluster_Swirl_N}
  n10 = ScalarParameter() {defaultValue=1200, parameterName=Swirl Normal Size 1}
  n11 = NamedRerouteUsage()
  n12 = Time()
  n13 = Add(A=@n11, B=Multiply(A=@n12, B=VectorParameter() {parameterName=Swirl Panning 1, defaultValue=<rgba 9 4.6598 -0.8654 0>}.RGB))
  n14 = ScalarParameter() {defaultValue=800, parameterName=Swirl Normal Size 2}
  n15 = Add(A=@n11, B=Multiply(A=@n12, B=VectorParameter() {parameterName=Swirl Panning 2, defaultValue=<rgba 3.3 -2.1747 0.3542 0>}.RGB))
  n16 = ShadingPathSwitch(Default=Add(
    A=MaterialFunctionCall!(TextureObject=@n9, TextureSize=@n10, RotYAxis, RotZAxis, RotXAxis, PreSkinNormal, ProjectionTransitionContrast, WorldPosition=@n13) {materialFunction=/Game/Packages/Fortress_SharedMaterials/Base_Material_Functions/BaseCustomizationFunctions/PreSkinnedAlignedNormal}.XYZ Texture
    B=MaterialFunctionCall(TextureObject=@n9, TextureSize=@n14, WorldPosition=@n15) {materialFunction=&PreSkinnedAlignedNormal}.XYZ Texture
   ))
  n17 = ScalarParameter() {defaultValue=0.97, parameterName=Flatten Swirl Facing}
  n18 = ScalarParameter() {defaultValue=0.95, parameterName=Flatten Swirl Side}
  n19 = Fresnel(Normal=VertexNormalWS()) {baseReflectFraction=0, exponent=3}
  NamedRerouteDeclaration(None=MaterialFunctionCall(Normal=Normalize(VectorInput=@n16), Flatness=LinearInterpolate(A=@n17, B=@n18, Alpha=@n19)) {materialFunction=&FlattenNormal}.Result) {name=Swirl Normal, nodeColor=<rgba 1 0.0469 0 1>}
  SetMaterialAttributes!(
   MaterialAttributes=StaticSwitchParameter(
    True=MaterialFunctionCall!(In=@n8.Result, Metal Tint, Local Bounds Split=@n3) {materialFunction=/Game/Vehicles/Valet/MF_Valet_Damage}.Result
    False=@n6.MA Out
   ) {parameterName=UseValetFunctions}
   World Position Offset=MaterialFunctionCall!(WPO) {materialFunction=/Game/Packages/Fortress_SharedMaterials/Base_Material_Functions/Fortnite_HideElementsWithVertexColor}.Result
  ) {attributeSetTypes [#1]}
 ]
}
