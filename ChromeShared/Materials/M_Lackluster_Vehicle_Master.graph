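# Master material for vehicles with the seasonal Lackluster chrome look, built on the MF_LackLuster_Base layer.
# Two time-panned swirl normal projections in pre-skinned local space are summed, then flattened by a Fresnel-driven amount.
# The swirl is angle-corrected blended onto the base, then passes through vehicle damage (masked by the shared vehicle function) and camera dither.
# World position offset comes from the vertex-color element-hiding function.
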Material /ChromeShared/Materials/M_Lackluster_Vehicle_Master
graph {
 properties {
  bAutomaticallySetUsageInEditor=false
  bUseMaterialAttributes=true
  bUsedWithClothing=true
  bUsedWithMeshDeformer=true
  bUsedWithMorphTargets=true
  bUsedWithSkeletalMesh=true
  physMaterial=/Game/TechArt/Seasonal/LackLuster/Physics/Chrome_PhysicalMaterial
  pixelDepthOffsetMode=PDOM_Legacy
  usedShadingModels=Default Lit
 }
 nodes [
  n1 = MaterialFunctionCall() {materialFunction=/Game/Packages/Fortress_SharedMaterials/Base_Material_Functions/BaseCustomizationFunctions/MF_FN_Vehicle_Master.MF_FN_Vehicle_MASTER}
  NamedRerouteDeclaration(None=@n1.Pre-Skinned Local Position Pixel Shader) {name=Preskinned Position, nodeColor=<rgba 1 0 0.4219 1>}
  n2 = TextureObject() {samplerType=SAMPLERTYPE_Normal, texture=/Game/TechArt/Seasonal/LackLuster/Textures/T_Lackluster_Swirl_N}
  n3 = ScalarParameter() {defaultValue=1200, parameterName=Swirl Normal Size 1}
  n4 = NamedRerouteUsage()
  n5 = Time()
  n6 = Add(A=@n4, B=Multiply(A=@n5, B=VectorParameter() {parameterName=Swirl Panning 1, defaultValue=<rgba 9 4.6598 -0.8654 0>}.RGB))
  n7 = ScalarParameter() {defaultValue=800, parameterName=Swirl Normal Size 2}
  n8 = Add(A=@n4, B=Multiply(A=@n5, B=VectorParameter() {parameterName=Swirl Panning 2, defaultValue=<rgba 3.3 -2.1747 0.3542 0>}.RGB))
  n9 = ShadingPathSwitch(Default=Add(
    A=MaterialFunctionCall!(TextureObject=@n2, TextureSize=@n3, RotYAxis, RotZAxis, RotXAxis, PreSkinNormal, ProjectionTransitionContrast, WorldPosition=@n6) {materialFunction=/Game/Packages/Fortress_SharedMaterials/Base_Material_Functions/BaseCustomizationFunctions/PreSkinnedAlignedNormal}.XYZ Texture
    B=MaterialFunctionCall(TextureObject=@n2, TextureSize=@n7, WorldPosition=@n8) {materialFunction=&PreSkinnedAlignedNormal}.XYZ Texture
   ))
  n10 = ScalarParameter() {defaultValue=0.97, parameterName=Flatten Swirl Facing}
  n11 = ScalarParameter() {defaultValue=0.95, parameterName=Flatten Swirl Side}
  n12 = Fresnel(Normal=VertexNormalWS()) {baseReflectFraction=0, exponent=3}
  NamedRerouteDeclaration(None=MaterialFunctionCall(Normal=Normalize(VectorInput=@n9), Flatness=LinearInterpolate(A=@n10, B=@n11, Alpha=@n12)) {materialFunction=&FlattenNormal}.Result) {name=Swirl Normal, nodeColor=<rgba 1 0.0469 0 1>}
  n13 = MaterialFunctionCall!(
   Material=MaterialFunctionCall!(
    bBlendMaterialAttributes
    MA_In
    UVsOverride
    Override Normals
    BlendMaterialAttributesAlpha
    BaseColorOverride
    WorldPositionOverride
    VertexNormalOverride
    SourceNormalsStrengthOverride=Constant() {r=1}
   ) {materialFunction=/Game/TechArt/Seasonal/LackLuster/MaterialFunctions/MF_LackLuster_Base}.MA Out
   Additional Normal=NamedRerouteUsage()
   Mask
  ) {materialFunction=/Engine/Functions/MaterialLayerFunctions/MatLayerBlend_BlendAngleCorrectedNormals}
  SetMaterialAttributes!(
   MaterialAttributes=MaterialFunctionCall!(In=MaterialFunctionCall(Material=@n13.Blended Material, "Clear Coat/Damage Mask (_M B)"=@n1."CC/Damage Mask (_M B)") {materialFunction=&MF_BaseVehicle_Damage}.Result) {materialFunction=/Game/Packages/Fortress_SharedMaterials/Base_Material_Functions/BaseCustomizationFunctions/MF_FN_Vehicle_CameraDither}.Result
   World Position Offset=MaterialFunctionCall!(WPO) {materialFunction=/Game/Packages/Fortress_SharedMaterials/Base_Material_Functions/Fortnite_HideElementsWithVertexColor}.Result
  ) {attributeSetTypes [#1]}
 ]
}
