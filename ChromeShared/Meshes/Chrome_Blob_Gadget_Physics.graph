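# Physics asset for the Chrome Blob Gadget mesh, giving collision bodies to a center bone and six surrounding blob bones.
# Each body is a single zero-length capsule, effectively a sphere, with the center one slightly larger than the rest.
# The six outer spheres (top, bottom, front, back, left, right) are offset from their bone origins along a local axis.
# No constraints join the bodies, so the asset itself defines no joint limits between them.
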PhysicsAsset /ChromeShared/Meshes/Chrome_Blob_Gadget_Physics
bodies {
 blob_bk [{shapeType=Capsule, length=0, radius=30, shapeName=blob_bk_capsule, center=<-28 0 0>}]
 blob_bottom [{shapeType=Capsule, length=0, radius=30, shapeName=blob_bottom_capsule, center=<0 0 28>}]
 blob_center_physics [{shapeType=Capsule, length=0, radius=35, shapeName=blob_center_physics_capsule, rotation=<rot 0 -90 0>}]
 blob_fr [{shapeType=Capsule, length=0, radius=30, shapeName=blob_fr_capsule, center=<-28 0 0>}]
 blob_l [{shapeType=Capsule, length=0, radius=30, shapeName=blob_l_capsule, center=<-28 0 0>}]
 blob_r [{shapeType=Capsule, length=0, radius=30, shapeName=blob_r_capsule, center=<-28 0 0>}]
 blob_top [{shapeType=Capsule, length=0, radius=30, shapeName=blob_top_capsule, center=<0 0 -28>}]
}
constraints []
